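; BP_proctest is a procedural test actor in the Foundation tools folder that builds two stacked static mesh components in its construction script.
; The first mesh uses a hard-coded, rotated, non-uniformly scaled transform.
; The second derives from it, with X and Y scale shrunk by a fixed margin on a 512-unit basis, shifted by a rotated half-margin offset and raised locally.
; BeginPlay, overlap and Tick events are empty.
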(event EventBeginPlay)

(event Collision|EventActorBeginOverlap
  (params (OtherActor :type Actor%20Object%20Reference)))

(event EventTick
  (params (DeltaSeconds :type Float%20%28single-precision%29)))

(fn ConstructionScript
  (bind _returnvalue (Math|Transform|MakeTransform "512.000000,1024.000000, 0" "0.000000,-90.000000,0.000000" "2.000000,3.000000,1.000000"))
  (AddComponent|NODE_AddStaticMeshComponent-0 :RelativeTransform _returnvalue)
  (bind _returnvalue_2 (AddComponent|NODE_AddStaticMeshComponent-1 :RelativeTransform (Math|Transform|ComposeTransforms (Math|Transform|MakeTransform (* (.location _returnvalue) "1.000000,1.000000, 0") (.rotation _returnvalue) (Math|Vector|MakeVector (/ (- (* (.x (.scale _returnvalue)) 512.0) 4e+01) 512.0) (/ (- (* (.y (.scale _returnvalue)) 512.0) 4e+01) 512.0) (.z (.scale _returnvalue)))) (Math|Transform|MakeTransform (Math|Vector|RotateVector "20.000000,-20.000000, 0" (.rotation _returnvalue))))))
  (Transformation|AddLocalOffset _returnvalue_2 "0, 0,16.000000" false false))
