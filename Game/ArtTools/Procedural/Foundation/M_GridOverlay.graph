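# M_GridOverlay is an unlit translucent material for a world-aligned grid overlay.
# WorldAlignedTexture projects the engine's white grid using Size and pos parameters, feeding several tint variants from Color and Color2.
# A directional mask scales Opacity by the saturated dot product of the local-space object-to-pixel direction and Direction.
# Further branches include a world-height hue shift and a grid-driven opacity blend, with no material output links shown.
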Material /Game/ArtTools/Procedural/Foundation/M_GridOverlay
graph {
 properties {blendMode=BLEND_Translucent, pixelDepthOffsetMode=PDOM_Legacy, shadingModel=MSM_Unlit, usedShadingModels=Unlit}
 nodes [
  n1 = MaterialFunctionCall(
   TextureObject=TextureObject() {texture=/Engine/EngineMaterials/DefaultWhiteGrid}
   TextureSize=VectorParameter() {parameterName=Size, defaultValue=<rgba 1024 1024 768 1>}.RGB
   WorldPosition=Subtract(A=WorldPosition().XYZ, B=VectorParameter() {parameterName=pos, defaultValue=<rgba 256 256 768 1>}.RGB)
  ) {materialFunction=&WorldAlignedTexture}
  n2 = VectorParameter() {parameterName=Color, defaultValue=<rgba 1 0.0271 0 1>}
  Add(A=Desaturation(None=@n1.XYZ Texture), B=@n2.RGB)
  LinearInterpolate(A=@n2.RGB, B=VectorParameter() {parameterName=Color2, defaultValue=<rgba 0 0 0 1>}.RGB, Alpha=@n1.XYZ Texture)
  n3 = ScalarParameter() {defaultValue=0.0657, parameterName=Opacity}
  LinearInterpolate(A=@n3, B=ScalarParameter() {defaultValue=0.0657, parameterName=Opacity}, Alpha=@n1.XYZ Texture)
  n4 = WorldPosition()
  MaterialFunctionCall(
   Hue Shift Percentage=Divide(A=ComponentMask(None=@n4.XYZ) {b=true}, B=ScalarParameter() {defaultValue=1.7408, parameterName=DivideHueShift})
   Texture=@n2.RGB
  ) {materialFunction=&HueShift}
  n5 = WorldPosition()
  n6 = ObjectPositionWS()
  n7 = Normalize(VectorInput=Transform(None=Subtract(A=@n5.XYZ, B=@n6)) {transformSourceType=TRANSFORMSOURCE_World, transformType=TRANSFORM_Local})
  n8 = VectorParameter() {parameterName=Direction, defaultValue=<rgba 0 1 0 1>}
  Multiply(A=@n3, B=Saturate(None=DotProduct(A=@n7, B=@n8.RGB)))
  Multiply(A=@n1.XYZ Texture, B=@n2.RGB)
 ]
}
