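; Procedural foundation POI override actor that snaps itself to a building grid in its construction script; its event nodes are empty.
; Location snaps to centers of 512 x 512 x 384 cells and yaw to 90-degree steps.
; Width and Height are multiplied by X and Z scale, clamped to at least 1, then scale X/Y is rounded and Z reset.
; A valid ConnectedActor then has UpdateAllComponents called.
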(event EventBeginPlay)

(event Collision|EventActorBeginOverlap
  (params (OtherActor :type Actor%20Object%20Reference)))

(event EventTick
  (params (DeltaSeconds :type Float%20%28single-precision%29)))

(fn ConstructionScript
  (bind _returnvalue (Transformation|GetActorScale3D))
  (bind _connectedactor (Variables|DoNOTEdit|GetConnectedActor))
  (Transformation|SetActorLocation (Math|Vector|MakeVector (* (+ (Math|Conversions|ToFloat(Integer) (Math|Float|Floor (/ (.x (Transformation|GetActorLocation)) 512.0))) 0.5) 512.0) (* (+ (Math|Conversions|ToFloat(Integer) (Math|Float|Floor (/ (.y (Transformation|GetActorLocation)) 512.0))) 0.5) 512.0) (* (+ (Math|Conversions|ToFloat(Integer) (Math|Float|Floor (/ (.z (Transformation|GetActorLocation)) 384.0))) 0.5) 384.0)))
  (Transformation|SetActorRotation (Math|Rotator|MakeRotator 0.0 0.0 (* (Math|Float|Round (/ (.yaw (Transformation|GetActorRotation)) 9e+01)) 9e+01)))
  (Variables|Default|SetWidth (Math|Integer|Max(Integer) (Math|Float|Round (* (Variables|Default|GetWidth) (.x _returnvalue))) 1))
  (Variables|Default|SetHeight (Math|Integer|Max(Integer) (Math|Float|Round (* (Variables|Default|GetHeight) (.z _returnvalue))) 1))
  (Transformation|SetActorScale3D (Math|Vector|MakeVector (Math|Conversions|ToFloat(Integer) (Math|Float|Round (.x (Transformation|GetActorScale3D)))) (Math|Conversions|ToFloat(Integer) (Math|Float|Round (.y (Transformation|GetActorScale3D)))) 1.0))
  (Utilities|IsValid _connectedactor
    (:Is%20Valid
      (Tools|UpdateAllComponents _connectedactor))))
