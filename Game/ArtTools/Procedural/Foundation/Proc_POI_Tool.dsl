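; Spline-driven editor actor that procedurally builds POI foundations such as basements, sidewalks, cliffs, ramps and caves.
; A mode switch picks builders that choose Proc_AssetList meshes and record transforms, which become preview components or baked FortStaticMeshActors.
; Expanded and offset spline copies deform and blend the landscape on named edit layers.
; Lower tool actors inside the shape, found by traces against temporary fill collision, link as connected sub-basements updated recursively.
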(event Custom|Start%20Live%20Edit%20Mode
  (Variables|Default|SetLiveUpdate true)
  (CreateComponents|CreateAllComponents)
  (Tools|SetConnectedBasementsUPonChilds)
  (Tools|UpdateAllConnectedUP))

(event EventBeginPlay)

(event Collision|EventActorBeginOverlap
  (params (OtherActor :type Actor%20Object%20Reference)))

(event EventTick
  (params (DeltaSeconds :type Float%20%28single-precision%29)))

(event Custom|Paint%20Street%20VertexColor
  (for _array_element (Variables|MeshLists|GetFloorMeshList)
    (VertexPaint|PaintVerticesSingleColor _array_element (Math|Color|MakeColor 1.0 1.0))))

(fn AddComponentAndSetFromData
  (params (RelativeTransform :type Transform) (Proc%20Asset%20List :type Proc%20Asset%20List%20Object%20Reference) (Asset%20Index :type Integer))
  (bind _returnvalue (AddComponent|NODE_AddStaticMeshComponent-39 :RelativeTransform RelativeTransform))
  (Tools|SetStaticmeshfromAssetData _returnvalue Proc%20Asset%20List Asset%20Index)
  (Utilities|Array|Add (Variables|MeshLists|GetComponentMeshList) _returnvalue)
  (Utilities|Array|Add (Variables|X.%29DoNOTEdit|GetComponent%20Transform%20List) (Utilities|Struct|MakeProcComponenttoSpawn RelativeTransform Proc%20Asset%20List Asset%20Index)))

(fn AddPointToArrayAndFuse
  (params (Array :type Array%20of%20Vectors) (PointToAdd :type Vector))
  (LocalVariables|SetExists false)
  (bind _array_element (Utilities|Array|ForEachLoopwithBreak Array)
    (:LoopBody
      (if (Math|Vector|Equal(Vector) _array_element PointToAdd 0.1)
        (LocalVariables|SetExists true)))
    (:Completed
      _))
  (if (not (LocalVariables|GetExists))
    (Utilities|Array|Add Array PointToAdd)))

(fn CheckIfDuplicated
  (bind _self self)
  (bind _returnvalue (Utilities|GetDisplayName _self))
  (if (not (Utilities|String|EqualExactly(String) _returnvalue (Variables|X.%29DoNOTEdit|GetName)))
    (Variables|X.%29DoNOTEdit|SetName _returnvalue)
    (C.%29ConnectingComplexPOIs|Reset&ClearAllConnections)))

(fn CheckifSplinewasEdited
  (bind _splinepointlistcheck (Variables|X.%29DoNOTEdit|GetSplinePointListCheck))
  (bind _spline (Variables|Default|GetSpline))
  (LocalVariables|SetNeedLiveUpdate false)
  (Utilities|FlowControl|Sequence
    (:then_0
      (if (not (Utilities|Operators|Equal(==) (Transformation|GetActorLocation) (Variables|X.%29DoNOTEdit|GetCheckLocation) 0))
        (Variables|Default|SetLiveUpdate true)))
    (:then_1
      _))
  (bind _index (Utilities|FlowControl|ForLoopwithBreak 0 (- (Spline|GetNumberOfSplinePoints _spline) 1))
    (:LoopBody
      (if (< _index (Utilities|Array|Length _splinepointlistcheck))
        (if (Utilities|Operators|Equal(==) (Utilities|Array|Get(acopy) _splinepointlistcheck _index) (Spline|GetLocationatSplinePoint _spline _index "Local") 0.1)
          _
          (else
            _)))
      (LocalVariables|SetNeedLiveUpdate true))
    (:Completed
      _))
  (if (LocalVariables|GetNeedLiveUpdate)
    (Variables|Default|SetLiveUpdate true)))

(fn CheckIsPointInsideofSpline
  (params (Location%20WorldSpace :type Vector) (Spline :type Spline%20Component%20Object%20Reference) (CoordinateSpace :type ESplineCoordinateSpace%20Enum :default "Local") (CreateCollision :type Boolean))
  (if CreateCollision
    (CreateCollisionComponents|MakeFillCollisionfromSpline Spline CoordinateSpace))
  (bind (_outhit _returnvalue) (Collision|LineTraceForObjects :Start (+ (Math|Vector|MakeVector (.x Location%20WorldSpace) (.y Location%20WorldSpace) (.z (Transformation|GetActorLocation))) "0, 0,256.000000") :End (Math|Vector|MakeVector (.x Location%20WorldSpace) (.y Location%20WorldSpace) (.z (Transformation|GetActorLocation))) :ObjectTypes (Utilities|Array|MakeArray "ObjectTypeQuery3") :bTraceComplex true :DrawDebugType "ForDuration" :bIgnoreSelf false :TraceColor "(R=1.000000,G=0.000000,B=0.902083,A=1.000000)" :TraceHitColor "(R=0.000000,G=1.000000,B=0.000000,A=1.000000)" :DrawTime 5.0))
  (if CreateCollision
    (Cleanup|CleanUpFillCollisionList)
    (Cleanup|CleanUpBorderCollisionList))
  (return _returnvalue))

(fn CleanUpBorderCollisionList
  (bind _bordercollisionlist (Variables|MeshLists|GetBorderCollisionList))
  (for _array_element _bordercollisionlist
    (Components|DestroyComponent _array_element))
  (Utilities|Array|Clear _bordercollisionlist))

(fn CleanUpFillCollisionList
  (bind _fillcollisionlist (Variables|MeshLists|GetFillCollisionList))
  (for _array_element _fillcollisionlist
    (Components|DestroyComponent _array_element))
  (Utilities|Array|Clear _fillcollisionlist))

(fn CloseSplineForTerrainConform
  (params (Spline :type Spline%20Component%20Object%20Reference))
  (bind _returnvalue (Spline|GetNumberOfSplinePoints Spline))
  (bind _returnvalue_2 (- _returnvalue 2))
  (bind _returnvalue_3 (- _returnvalue 1))
  (Spline|AddSplinePoint Spline (Spline|GetLocationatSplinePoint Spline 0 "Local") "Local" true)
  (Spline|AddSplinePoint Spline (Spline|GetLocationatSplinePoint Spline 1 "Local") "Local" true)
  (Spline|SetSplinePointType Spline _returnvalue_3 "Linear" true)
  (Spline|SetSplinePointType Spline _returnvalue_2 "Linear" true)
  (Spline|SetRotationatSplinePoint Spline _returnvalue_2 (Spline|GetRotationatSplinePoint Spline 0 "Local") "Local" true)
  (Spline|SetRotationatSplinePoint Spline _returnvalue_3 (Spline|GetRotationatSplinePoint Spline 1 "Local") "Local" true)
  (return Spline))

(fn ClosestHitResult
  (params (HitResult :type Array%20of%20Hit%20Result%20Structures) (WorldSpaceLocation :type Vector) (Debug :type Boolean))
  (bind _self self)
  (bind _returnvalue (Utilities|GetDisplayName _self))
  (bind _returnvalue_2 (Utilities|String|Append _returnvalue "   _HIT_    "))
  (if (> (Utilities|Array|Length HitResult) 0)
    (LocalVariables|SetClosestHitIndex -1)
    (bind (_array_element _array_index) (Utilities|Array|ForEachLoopwithBreak HitResult)
      (:LoopBody
        (if (Utilities|Equal(Object) (.HitActor (Collision|BreakHitResult _array_element)) _self)
          (Development|PrintString (Utilities|String|Append (Utilities|String|Append (Utilities|String|Append _returnvalue_2 (Utilities|GetDisplayName (.HitActor (Collision|BreakHitResult _array_element)))) "   _Bool_    ") (Utilities|String|ToString(Boolean) (Utilities|Equal(Object) (.HitActor (Collision|BreakHitResult _array_element)) _self))) Debug Debug "(R=0.142391,G=1.000000,B=0.000000,A=1.000000)")
          (if (== (LocalVariables|GetClosestHitIndex) -1)
            (LocalVariables|SetMinLength (Math|Vector|VectorLength (- (.Location (Collision|BreakHitResult _array_element)) WorldSpaceLocation)))
            (LocalVariables|SetClosestHitIndex 0)
            (elif (< (Math|Vector|VectorLength (- (.Location (Collision|BreakHitResult _array_element)) WorldSpaceLocation)) (LocalVariables|GetMinLength))
              (LocalVariables|SetMinLength (Math|Vector|VectorLength (- (.Location (Collision|BreakHitResult _array_element)) WorldSpaceLocation)))
              (LocalVariables|SetClosestHitIndex _array_index)))
          (else
            (Development|PrintString (Utilities|String|Append (Utilities|String|Append (Utilities|String|Append _returnvalue_2 (Utilities|GetDisplayName (.HitActor (Collision|BreakHitResult _array_element)))) "   _Bool_    ") (Utilities|String|ToString(Boolean) (Utilities|Equal(Object) (.HitActor (Collision|BreakHitResult _array_element)) _self))) Debug Debug "(R=1.000000,G=0.000000,B=0.000000,A=1.000000)"))))
      (:Completed
        (return (Utilities|Array|Get(aref) HitResult (LocalVariables|GetClosestHitIndex)))))
    (else
      (return (Collision|MakeHitResult false)))))

(fn ConformLandscape
  (Utilities|FlowControl|Sequence
    (:then_0
      (switch Utilities|FlowControl|Switch|SwitchonProc_FoundationSelection (Variables|Default|GetMode)
        (:NewEnumerator9
          (Utilities|FlowControl|Sequence
            (:then_0
              _)
            (:then_1
              (Cleanup|DeleteAllTempSplines))))
        (:NewEnumerator0
          (Utilities|FlowControl|Sequence
            (:then_0
              (LandscapeDeformation|ConformLandscapewithSplineDeform))
            (:then_1
              (LandscapeDeformation|ConformLandscapeBlend))
            (:then_2
              _)
            (:then_3
              _)))
        (:NewEnumerator1
          _)
        (:NewEnumerator2
          (Utilities|FlowControl|Sequence
            (:then_0
              _)
            (:then_1
              _)
            (:then_2
              _)))
        (:NewEnumerator4
          _)
        (:NewEnumerator3
          _))
      (LandscapeDeformation|FillArea 0))
    (:then_1
      (for _array_element (Variables|C.%29ConnectingComplexPOIs|GetConnectedActors)
        (B.%29LandscapeEdtiting|ConformLandscape _array_element)))))

(fn ConformLandscapeBlend
  (bind _splinecopy (SplineOperations|DuplicateSpline (Variables|Default|GetSpline)))
  (bind _splineoutput (SplineOperations|ExpandSpline _splinecopy (* (Variables|Default|GetTerrainBlendDistance) 2.0)))
  (bind _splineoutput_2 (SplineOperations|ConformSplinetoLandscape _splineoutput))
  (bind _splinecopy_2 (SplineOperations|DuplicateSpline _splineoutput_2))
  (bind _splinecopy_3 (SplineOperations|DuplicateSpline (Variables|Default|GetSpline)))
  (bind _splineoutput_3 (SplineOperations|ExpandSpline _splinecopy_3 (Variables|Default|GetTerrainBlendDistance)))
  (Utilities|FlowControl|Sequence
    (:then_0
      (bind _spline_output (LandscapeDeformation|MakeBlendSpline _splineoutput_3 _splineoutput_2 2.0 0.2))
      (bind _outactors (Actor|GetAllActorsOfClass "/Script/Landscape.LandscapeProxy"))
      ;[1 1 1] Add For each Segment and Roll here
      (for _array_element _outactors
        (for _index (range (Spline|GetNumberOfSplineSegments _spline_output))
          (bind _returnvalue (AddComponent|NODE_AddSplineComponent-7))
          (Utilities|Array|Add (Variables|SplineList|GetTempSplineList) _returnvalue)
          (Spline|SetLocationatSplinePoint _returnvalue 0 (Spline|GetLocationatSplinePoint _spline_output _index "Local") "Local" true)
          (Spline|SetLocationatSplinePoint _returnvalue 1 (Spline|GetLocationatSplinePoint _spline_output (Math|Integer|SelectInt 0 (+ _index 1) (== _index (- (Spline|GetNumberOfSplineSegments _spline_output) 1))) "Local") "Local" true)
          (Spline|SetRotationatSplinePoint _returnvalue 0 (Spline|GetRotationatSplinePoint _spline_output _index "Local") "Local" true)
          (Spline|SetRotationatSplinePoint _returnvalue 1 (Spline|GetRotationatSplinePoint _spline_output (Math|Integer|SelectInt 0 (+ _index 1) (== _index (- (Spline|GetNumberOfSplineSegments _spline_output) 1))) "Local") "Local" true)
          (for _array_element_2 (Variables|PaintLayer|GetEdit%20Layer%20Name)
            (Landscape|Editor|EditorApplySpline _array_element _returnvalue 0.0 0.0 (* (Variables|Default|GetTerrainBlendDistance) 1.0) (* (Variables|Default|GetTerrainBlendDistance) 1.0) (Spline|GetRollatSplinePoint _returnvalue 0 "Local") (Spline|GetRollatSplinePoint _returnvalue 1 "Local") 20 true true 0 _array_element_2)))))
    (:then_1
      (bind _spline_output_2 (LandscapeDeformation|MakeBlendSpline _splineoutput_3 _splinecopy_2 1.5 0.2))
      (bind _outactors_2 (Actor|GetAllActorsOfClass "/Script/Landscape.LandscapeProxy"))
      ;[1 1 1] Add For each Segment and Roll here
      (for _array_element_3 _outactors_2
        (for _index_2 (range (Spline|GetNumberOfSplineSegments _spline_output_2))
          (bind _returnvalue_2 (AddComponent|NODE_AddSplineComponent-8))
          (Utilities|Array|Add (Variables|SplineList|GetTempSplineList) _returnvalue_2)
          (Spline|SetLocationatSplinePoint _returnvalue_2 0 (Spline|GetLocationatSplinePoint _spline_output_2 _index_2 "Local") "Local" true)
          (Spline|SetLocationatSplinePoint _returnvalue_2 1 (Spline|GetLocationatSplinePoint _spline_output_2 (Math|Integer|SelectInt 0 (+ _index_2 1) (== _index_2 (- (Spline|GetNumberOfSplineSegments _spline_output_2) 1))) "Local") "Local" true)
          (Spline|SetRotationatSplinePoint _returnvalue_2 0 (Spline|GetRotationatSplinePoint _spline_output_2 _index_2 "Local") "Local" true)
          (Spline|SetRotationatSplinePoint _returnvalue_2 1 (Spline|GetRotationatSplinePoint _spline_output_2 (Math|Integer|SelectInt 0 (+ _index_2 1) (== _index_2 (- (Spline|GetNumberOfSplineSegments _spline_output_2) 1))) "Local") "Local" true)
          (for _array_element_4 (Variables|PaintLayer|GetEdit%20Layer%20Name)
            (Landscape|Editor|EditorApplySpline _array_element_3 _returnvalue_2 1.0 1.0 (* (Variables|Default|GetTerrainBlendDistance) 0.25) (* (Variables|Default|GetTerrainBlendDistance) 0.25) (Spline|GetRollatSplinePoint _returnvalue_2 0 "Local") (Spline|GetRollatSplinePoint _returnvalue_2 1 "Local") 20 true true 0 _array_element_4)))))
    (:then_2
      (bind _splineout (SplineOperations|OffsetSpline _splineoutput_3 1e+01))
      (bind _outactors_3 (Actor|GetAllActorsOfClass "/Script/Landscape.LandscapeProxy"))
      ;[1 1 1] Add For each Segment and Roll here
      (for _array_element_5 _outactors_3
        (for _index_3 (range (Spline|GetNumberOfSplineSegments _splineout))
          (bind _returnvalue_3 (AddComponent|NODE_AddSplineComponent-9))
          (Utilities|Array|Add (Variables|SplineList|GetTempSplineList) _returnvalue_3)
          (Spline|SetLocationatSplinePoint _returnvalue_3 0 (Spline|GetLocationatSplinePoint _splineout _index_3 "Local") "Local" true)
          (Spline|SetLocationatSplinePoint _returnvalue_3 1 (Spline|GetLocationatSplinePoint _splineout (Math|Integer|SelectInt 0 (+ _index_3 1) (== _index_3 (- (Spline|GetNumberOfSplineSegments _splineout) 1))) "Local") "Local" true)
          (Spline|SetRotationatSplinePoint _returnvalue_3 0 (Spline|GetRotationatSplinePoint _splineout _index_3 "Local") "Local" true)
          (Spline|SetRotationatSplinePoint _returnvalue_3 1 (Spline|GetRotationatSplinePoint _splineout (Math|Integer|SelectInt 0 (+ _index_3 1) (== _index_3 (- (Spline|GetNumberOfSplineSegments _splineout) 1))) "Local") "Local" true)
          (for _array_element_6 (Variables|PaintLayer|GetEdit%20Layer%20Name)
            (Landscape|Editor|EditorApplySpline _array_element_5 _returnvalue_3 1e+02 1e+02 1e+02 1e+02 (Spline|GetRollatSplinePoint _returnvalue_3 0 "Local") (Spline|GetRollatSplinePoint _returnvalue_3 1 "Local") 20 true true 0 _array_element_6)))))))

(fn ConformLandscapeWithDupSpline
  (params (PaintLayer :type Landscape%20Layer%20Info%20Object%20Object%20Reference) (Width :type Float%20%28double-precision%29) (Sidefalloff :type Float%20%28double-precision%29) (ZOffset :type Float%20%28double-precision%29) (ExpandDistance :type Float%20%28double-precision%29))
  (bind _splinecopy (SplineOperations|DuplicateSpline (Variables|Default|GetSpline)))
  (bind _splineoutput (SplineOperations|ExpandSpline _splinecopy ExpandDistance))
  (bind _splineout (SplineOperations|OffsetSpline _splineoutput ZOffset))
  (bind _splineoutput_2 (SplineOperations|CloseSplineforTerrainConform _splineout))
  (bind _outactors (Actor|GetAllActorsOfClass "/Script/Landscape.LandscapeProxy"))
  (for _array_element _outactors
    (for _array_element_2 (Variables|PaintLayer|GetEdit%20Layer%20Name)
      (Landscape|Editor|EditorApplySpline _array_element _splineoutput_2 Width Width Sidefalloff Sidefalloff 0.0 0.0 20 true true PaintLayer _array_element_2))))

(fn ConformLandscapeWithSplineDeform
  (bind _terrainblenddistance (Variables|Default|GetTerrainBlendDistance))
  (bind _returnvalue (/ _terrainblenddistance 8.0))
  (bind _returnvalue_2 (- _terrainblenddistance _returnvalue))
  (bind _returnvalue_3 (* _returnvalue 4.0))
  (bind _terrainconformheight (Variables|Default|GetTerrainConformHeight))
  (bind _returnvalue_4 (* _terrainconformheight -1.0))
  (LandscapeDeformation|ConformLandscapewithDupSpline 0 (/ (Variables|Default|GetTerrainBlendDistance) 2.0) 0.0 (* (+ (* (Variables|Default|GetNumberOfFloors) 384.0) (Variables|Default|GetTerrainCellarMargin)) -1.0))
  (LandscapeDeformation|ConformLandscapewithDupSpline 0 (/ (/ (Variables|Default|GetTerrainBlendDistance) 2.0) 1.5) 0.0 (* (Variables|Default|GetTerrainConformHeight) -1.0) (/ (Variables|Default|GetTerrainBlendDistance) 2.0))
  (LandscapeDeformation|ConformLandscapewithDupSpline 0 (+ _returnvalue _returnvalue_3) 0.0 _returnvalue_4 _returnvalue_2)
  (LandscapeDeformation|ConformLandscapewithDupSpline 0 _returnvalue _returnvalue_3 (+ _returnvalue_4 (* (Variables|Default|GetTerrainConformHeight) 2.0)) _returnvalue_2))

(fn ConformSplineToLandscape
  (params (Spline :type Spline%20Component%20Object%20Reference))
  (for _index (range (Spline|GetNumberOfSplinePoints Spline))
    (bind _outhits (Collision|MultiLineTraceForObjects (+ (Spline|GetLocationatSplinePoint Spline _index "World") "0, 0,5000.000000") (+ (Spline|GetLocationatSplinePoint Spline _index "World") "0, 0,-5000.000000") (Utilities|Array|MakeArray "ObjectTypeQuery1")))
    (bind _array_element (Utilities|Array|ForEachLoopwithBreak _outhits)
      (:LoopBody
        (Utilities|Casting|CastToLandscape (.HitActor (Collision|BreakHitResult _array_element)))
        (Spline|SetLocationatSplinePoint Spline _index (.Location (Collision|BreakHitResult _array_element)) "World" true))))
  (return Spline))

(fn ConnectSubActors
  (C.%29ConnectingComplexPOIs|ConnectSubBasementsFunction false))

(fn ConnectSubActorsRecursive
  (C.%29ConnectingComplexPOIs|ConnectSubBasementsFunction true))

(fn ConnectSubBasementsFunction
  (params (Recursive :type Boolean))
  (bind (_origin _boxextent) (Collision|GetActorBounds false))
  (bind _numberoffloors (Variables|Default|GetNumberOfFloors))
  (bind _returnvalue (* _numberoffloors 384.0))
  (bind _returnvalue_2 (* _boxextent "1.000000,1.000000, 0"))
  (bind _returnvalue_3 (Math|Vector|VectorLength _returnvalue_2))
  (bind _returnvalue_4 (Transformation|GetActorLocation))
  (bind _returnvalue_5 (* _returnvalue_4 "0.000000,0.000000,1.000000"))
  (bind _returnvalue_6 (* _origin "1.000000,1.000000, 0"))
  (bind _connectedactors (Variables|C.%29ConnectingComplexPOIs|GetConnectedActors))
  (Utilities|FlowControl|Sequence
    (:then_0
      (CreateCollisionComponents|MakeFillCollisionfromSpline (Variables|Default|GetSpline)))
    (:then_1
      (for _array_element _connectedactors
        (Utilities|IsValid _array_element
          (:Is%20Not%20Valid
            (Utilities|Array|RemoveIndex _connectedactors))))
      (Utilities|FlowControl|Sequence
        (:then_0
          (bind _outactors (Actor|GetAllActorsOfClass "/Game/ArtTools/Procedural/Foundation/Proc_POI_Tool.Proc_POI_Tool_C"))
          (for _array_element_2 _outactors
            (if (Utilities|NotEqual(Object) _array_element_2 self)
              (Development|PrintString (Utilities|String|Append (Utilities|String|Append (Utilities|String|Append (Utilities|String|Append (Utilities|String|Append (Utilities|String|Append (Variables|X.%29DoNOTEdit|GetName _array_element_2) "IsInDistance = ") (Utilities|String|ToString(Boolean) (< (Math|Vector|VectorLength (- (* (Transformation|GetActorLocation _array_element_2) "1.000000,1.000000, 0") _returnvalue_6)) _returnvalue_3))) "   Is below Actor =  ") (Utilities|String|ToString(Boolean) (<= (Math|Vector|VectorLength (- (* (Transformation|GetActorLocation _array_element_2) "0.000000,0.000000,1.000000") _returnvalue_5)) _returnvalue))) "Is at same height as bottom Floor = ") (Utilities|String|ToString(Boolean) (Math|Float|NearlyEqual(Float) _returnvalue (Math|Vector|VectorLength (- (* (Transformation|GetActorLocation _array_element_2) "0.000000,0.000000,1.000000") _returnvalue_5)) 0.1))) false)
              (if (and (< (Math|Vector|VectorLength (- (* (Transformation|GetActorLocation _array_element_2) "1.000000,1.000000, 0") _returnvalue_6)) _returnvalue_3) (and (<= (.z (Transformation|GetActorLocation _array_element_2)) (.z _returnvalue_4)) (Math|Float|NearlyEqual(Float) _returnvalue (Math|Vector|VectorLength (- (* (Transformation|GetActorLocation _array_element_2) "0.000000,0.000000,1.000000") _returnvalue_5)) 0.1)))
                (Development|PrintString (Utilities|String|Append "Has more than 2 Points" (Utilities|String|ToString(Boolean) (> (Spline|GetNumberOfSplinePoints (Variables|Default|GetSpline _array_element_2)) 2))) false)
                (if (> (Spline|GetNumberOfSplinePoints (Variables|Default|GetSpline _array_element_2)) 2)
                  (bind _returnvalue_7 (Tools|CheckIsPointInsideofSpline (+ (* (+ (Math|Vector|GetUnitDirection(Vector) (Spline|GetLocationatSplinePoint (Variables|Default|GetSpline _array_element_2) 0 "World") (Spline|GetLocationatSplinePoint (Variables|Default|GetSpline _array_element_2) 1 "World")) (Math|Vector|GetUnitDirection(Vector) (Spline|GetLocationatSplinePoint (Variables|Default|GetSpline _array_element_2) 0 "World") (Spline|GetLocationatSplinePoint (Variables|Default|GetSpline _array_element_2) (- (Spline|GetNumberOfSplinePoints (Variables|Default|GetSpline _array_element_2)) 1) "World"))) 11.0) (Spline|GetLocationatSplinePoint (Variables|Default|GetSpline _array_element_2) 0 "World")) (Variables|Default|GetSpline)))
                  (Development|PrintString (Utilities|String|Append "Point is within Shape = " (Utilities|String|ToString(Boolean) _returnvalue_7)) false)
                  (if _returnvalue_7
                    (Utilities|Array|AddUnique (Variables|C.%29ConnectingComplexPOIs|GetConnectedActors) _array_element_2)))))))))
    (:then_2
      (Cleanup|CleanUpFillCollisionList)
      (Cleanup|CleanUpBorderCollisionList))
    (:then_3
      (Tools|SetConnectedBasementsUPonChilds))
    (:then_4
      (if (Variables|GetRecursive)
        (Utilities|FlowControl|Sequence
          (:then_0
            (for _array_element_3 (Variables|C.%29ConnectingComplexPOIs|GetConnectedActors)
              (C.%29ConnectingComplexPOIs|ConnectSubBasementsFunction _array_element_3 true))))
        (else
          (Utilities|FlowControl|Sequence
            (:then_0
              (Tools|ForceSubBasmentMode))
            (:then_1
              (Tools|UpdateAllComponents)
              (Tools|UpdateAllConnectedDown))))))
    (:then_5
      (Tools|UpdateAllComponents))))

(fn ConvertToActors
  (if (Variables|Default|GetEnabled)
    (Variables|X.%29DoNOTEdit|SetRandomID (Utilities|String|StringToName (Utilities|String|Append (Variables|X.%29DoNOTEdit|GetName) (Utilities|String|ToString(Integer) (Math|Random|RandomInteger 2147483647)))))
    (Utilities|FlowControl|Sequence
      (:then_0
        (for (_array_element _array_index) (Variables|X.%29DoNOTEdit|GetComponent%20Transform%20List)
          (Development|PrintString (Utilities|String|ToString(Integer) _array_index))
          (bind _returnvalue (Game|SpawnActorfromClass "/Script/FortniteGame.FortStaticMeshActor" (Utilities|Struct|BreakProcComponenttoSpawn _array_element) "Undefined" "MultiplyWithRoot" 0 0))
          (Transformation|AddActorWorldTransform _returnvalue (Transformation|GetActorTransform) false false)
          (Transformation|SetActorScale3D _returnvalue (.scale (Utilities|Struct|BreakProcComponenttoSpawn _array_element)))
          (Tools|SetStaticmeshfromAssetData (Variables|/Script/Engine.StaticMeshActor|GetStaticMeshComponent _returnvalue) (.ProAssetList_8_82A9873447ED3A9EA990E390BE75AB8D (Utilities|Struct|BreakProcComponenttoSpawn _array_element)) (.Index_6_9AC1D1DE4901E9AF2516CA98A2CDC90C (Utilities|Struct|BreakProcComponenttoSpawn _array_element)))
          (Variables|/Script/Engine.Actor|SetTags (Utilities|Array|MakeArray (Variables|X.%29DoNOTEdit|GetRandomID)) _returnvalue)))
      (:then_1
        (Variables|Default|SetEnabled false)
        (Cleanup|DestroyAllComponents)))))

(fn CreateLandscapeBlendActor
  (bind _spline (Variables|Default|GetSpline))
  ;[1 1 1] Doesnt WOrk
  (bind _returnvalue (Game|SpawnActorfromClass "/Game/ArtTools/Procedural/Foundation/Proc_POI_Tool.Proc_POI_Tool_C" (Transformation|GetActorTransform) "Undefined" "OverrideRootScale" 0 0))
  (Utilities|FlowControl|Sequence
    (:then_1
      (for _index (range (+ (- (Spline|GetNumberOfSplinePoints _spline)) 1))
        (Spline|SetLocationatSplinePoint (Variables|Default|GetSpline _returnvalue) _index (Spline|GetLocationatSplinePoint _spline _index "Local") "Local" true)
        (Spline|SetSplinePointType (Variables|Default|GetSpline _returnvalue) _index (Spline|GetSplinePointType _spline _index) true)
        (Development|PrintString (Utilities|String|ToString(Integer) _index))))
    (:then_2
      (Variables|Default|SetMode "NewEnumerator2" _returnvalue)
      (Variables|Default|SetMode "NewEnumerator1" self)
      (Tools|UpdateAllComponents)
      (Tools|UpdateAllComponents _returnvalue))))

(fn CreateAllComponents
  (Development|PrintString (Utilities|String|Append "CreateAllComponents -count- " (Utilities|String|ToString(Integer) (Utilities|Array|Length (Variables|X.%29DoNOTEdit|GetComponent%20Transform%20List)))) false false)
  (Utilities|Array|Clear (Variables|X.%29DoNOTEdit|GetComponent%20Transform%20List))
  (Utilities|FlowControl|Sequence
    (:then_0
      (Variables|Default|SetisClosed true)
      (Spline|SetClosedLoop (Variables|Default|GetSpline) (Variables|Default|GetisClosed) true)
      (Utilities|FlowControl|Sequence
        (:then_0
          (switch Utilities|FlowControl|Switch|SwitchonProc_FoundationSelection (Variables|Default|GetMode)
            (:NewEnumerator9
              (CallFunction|Force4SplinePoints)
              (CreateComponents|MakeBasementFloorOptimized (Variables|E.%29Meshes|GetCliff_Asset_List_) false (Variables|Default|GetSpline)))
            (:NewEnumerator0
              (CallFunction|Force4SplinePoints)
              (Utilities|FlowControl|Sequence
                (:then_1
                  _)
                (:then_2
                  (CreateComponents|MakeLandscapeMeshBlend))
                (:then_3
                  (CreateComponents|MakeBasementFloorOptimized (Variables|E.%29Meshes|GetBasement_Asset_List) false (Variables|Default|GetSpline)))))
            (:NewEnumerator1
              (CallFunction|Force4SplinePoints)
              (Utilities|FlowControl|Sequence
                (:then_0
                  _)
                (:then_1
                  _)))
            (:NewEnumerator2
              (CallFunction|Force4SplinePoints))
            (:NewEnumerator4
              (CallFunction|Force4SplinePoints)
              (Utilities|FlowControl|Sequence
                (:then_0
                  (Variables|Default|SetNumberOfFloors 0))
                (:then_1
                  (CreateComponents|MakeSidewalkswithMergedSplines))
                (:then_3
                  (CreateComponents|MakeBasementFloorOptimized (Variables|E.%29Meshes|GetSidewalk_Asset_List) false (Variables|Default|GetSpline)))))
            (:NewEnumerator3
              (CallFunction|Force4SplinePoints)
              (Utilities|FlowControl|Sequence
                (:then_1
                  (CreateComponents|MakeMeshFillArea (Variables|E.%29Meshes|GetSidewalk_Asset_List)))))
            (:NewEnumerator5
              (CallFunction|Force4SplinePoints)
              (Utilities|FlowControl|Sequence
                (:then_0
                  (CreateComponents|CreateCliffWalls)
                  (bind _splinecopy (SplineOperations|DuplicateSpline (Variables|Default|GetSpline)))
                  (bind _splineoutput (SplineOperations|ExpandSpline _splinecopy -256.0))
                  (CreateComponents|MakeBasementFloorOptimized (Variables|E.%29Meshes|GetCliff_Asset_List_) true _splineoutput))
                (:then_1
                  (CallFunction|DoNotUseText))))
            (:NewEnumerator6
              (Utilities|FlowControl|Sequence
                (:then_0
                  (Variables|Default|SetisClosed false)
                  (Spline|SetClosedLoop (Variables|Default|GetSpline) (Variables|Default|GetisClosed) true)
                  (CreateComponents|CreateCliffWalls))
                (:then_1
                  _)))
            (:NewEnumerator7
              (Utilities|FlowControl|Sequence
                (:then_0
                  (CreateComponents|CreateCliffRamps))
                (:then_1
                  _)))
            (:NewEnumerator10
              (Utilities|FlowControl|Sequence
                (:then_0
                  (CreateComponents|CreateCaveComponents))
                (:then_1
                  _)))
            (:NewEnumerator8
              (CreateComponents|CreateVisualDebugComponents)))
          (CreateComponents|MakeBasementWalls))
        (:then_1
          (CreateComponents|CreateComponentsfromSavedList))))
    (:then_1
      (CallFunction|MakeSplinePointChecklist))))

(fn CreateCaveComponents
  (bind _spline (Variables|Default|GetSpline))
  (if (<= (Variables|Default|GetNumberOfFloors) 0)
    (Variables|Default|SetNumberOfFloors 1))
  (Utilities|FlowControl|Sequence
    (:then_0
      (bind _assetindexlist (Tools|GetAssetsOfType (Variables|/Game/ArtTools/Procedural/Foundation/Proc_AssetList.Proc_AssetList_C|GetAssetList (Variables|E.%29Meshes|GetCliff_Asset_List_)) "NewEnumerator10" "NewEnumerator2"))
      (LocalVariables|SetAssetIndexList_CaveWalls _assetindexlist)
      (bind _assetindexlist_2 (Tools|GetAssetsOfType (Variables|/Game/ArtTools/Procedural/Foundation/Proc_AssetList.Proc_AssetList_C|GetAssetList (Variables|E.%29Meshes|GetCliff_Asset_List_)) "NewEnumerator9"))
      (LocalVariables|SetAsset%20Index%20List%20Cave%20Ceiling _assetindexlist_2)
      (bind _assetindexlist_3 (Tools|GetAssetsOfType (Variables|/Game/ArtTools/Procedural/Foundation/Proc_AssetList.Proc_AssetList_C|GetAssetList (Variables|E.%29Meshes|GetCliff_Asset_List_)) "NewEnumerator11"))
      (LocalVariables|SetAsset%20Index%20List%20CeilingSupport _assetindexlist_3))
    (:then_1
      (CreateCollisionComponents|MakeBorderCollisionFromSpline (Variables|Default|GetSpline)))
    (:then_2
      (for _index (range (+ (- (Spline|GetNumberOfSplinePoints _spline) (Math|Integer|SelectInt 1 2 (Spline|IsClosedLoop _spline))) 1))
        (bind (_next__1 _next _current _previous) (Tools|GetCurrentNextPreviousSplinePoint _spline _index))
        (Utilities|FlowControl|Sequence)))))

(fn CreateCliff_Ramps
  (bind _spline (Variables|Default|GetSpline))
  (bind _returnvalue (Spline|GetNumberOfSplinePoints _spline))
  (bind _spline_2 (Variables|Default|GetSpline))
  (bind _returnvalue_2 (Spline|GetLocationatSplinePoint _spline_2 1 "Local"))
  (bind _returnvalue_3 (Spline|GetLocationatSplinePoint _spline_2 0 "Local"))
  (bind _returnvalue_4 (Math|Vector|GetUnitDirection(Vector) _returnvalue_3 _returnvalue_2))
  (bind _returnvalue_5 (Math|Vector|CrossProduct _returnvalue_4 "0, 0,1.000000"))
  (bind _returnvalue_6 (Math|Rotator|MakeRotationfromAxes _returnvalue_5 _returnvalue_4 "0, 0,1.000000"))
  (bind _returnvalue_7 (Math|Vector|Normalize _returnvalue_5))
  (bind _returnvalue_8 (* _returnvalue_7 -256.0))
  (bind _cliff_asset_list (Variables|E.%29Meshes|GetCliff_Asset_List_))
  (bind _assetlist (Variables|/Game/ArtTools/Procedural/Foundation/Proc_AssetList.Proc_AssetList_C|GetAssetList _cliff_asset_list))
  (bind _cliff_asset_list_2 (Variables|E.%29Meshes|GetCliff_Asset_List_))
  (bind _assetlist_2 (Variables|/Game/ArtTools/Procedural/Foundation/Proc_AssetList.Proc_AssetList_C|GetAssetList _cliff_asset_list_2))
  (Utilities|FlowControl|Sequence
    (:then_0
      (if (or (< _returnvalue 2) (> _returnvalue 2))
        (Utilities|FlowControl|Sequence
          (:then_0
            (for _index (range _returnvalue)
              (Spline|RemoveSplinePoint _spline _index true)))
          (:then_1
            (Spline|SetSplinePoints _spline (select (== (Utilities|Array|Length (Variables|X.%29DoNOTEdit|GetSplinePointListCheck)) 2) (Variables|X.%29DoNOTEdit|GetSplinePointListCheck) (Utilities|Array|MakeArray "0, 0, 0" "-512.000000, 0, 0")) "Local" true)
            (if (< (Variables|Default|GetNumberOfFloors) 1)
              (Variables|Default|SetNumberOfFloors 1))))))
    (:then_1
      (bind _assetindexlist (Tools|GetAssetsOfType (Variables|/Game/ArtTools/Procedural/Foundation/Proc_AssetList.Proc_AssetList_C|GetAssetList (Variables|E.%29Meshes|GetCliff_Asset_List_)) "NewEnumerator7" "NewEnumerator2"))
      (LocalVariables|SetRamp%20Index%20List _assetindexlist)
      (bind _assetindexlist_2 (Tools|GetAssetsOfType (Variables|/Game/ArtTools/Procedural/Foundation/Proc_AssetList.Proc_AssetList_C|GetAssetList (Variables|E.%29Meshes|GetCliff_Asset_List_)) "NewEnumerator8" "NewEnumerator2"))
      (LocalVariables|SetRamp%20Cap%20Index%20List _assetindexlist_2))
    (:then_2
      (bind _array_element (Utilities|Array|ForEachLoopwithBreak (LocalVariables|GetRamp%20Cap%20Index%20List))
        (:LoopBody
          (LocalVariables|SetWidth (Math|Vector|Distance(Vector) _returnvalue_3 _returnvalue_2))
          (if (> (Math|Float|Floor (/ (* 384.0 (Variables|Default|GetNumberOfFloors)) (.SizeZ_15_C53BDD28478B1F2003457E995BF7FDC6 (Utilities|Struct|BreakProcAssetInfo (Utilities|Array|Get(acopy) _assetlist _array_element))))) 0)
            (Utilities|Array|Add (Variables|X.%29DoNOTEdit|GetComponent%20Transform%20List) (Utilities|Struct|MakeProcComponenttoSpawn (Math|Transform|MakeTransform (+ _returnvalue_3 _returnvalue_8) _returnvalue_6 "1,-1.000000,1") (Variables|E.%29Meshes|GetCliff_Asset_List_) _array_element))
            (LocalVariables|SetWidth (- (LocalVariables|GetWidth) (.SizeY_14_2F84FFCF464B2148EFD89E96265F5278 (Utilities|Struct|BreakProcAssetInfo (Utilities|Array|Get(acopy) _assetlist _array_element)))))
            (Utilities|Array|Add (Variables|X.%29DoNOTEdit|GetComponent%20Transform%20List) (Utilities|Struct|MakeProcComponenttoSpawn (Math|Transform|MakeTransform (+ _returnvalue_2 _returnvalue_8) _returnvalue_6) (Variables|E.%29Meshes|GetCliff_Asset_List_) _array_element))
            (LocalVariables|SetWidth (- (LocalVariables|GetWidth) (.SizeY_14_2F84FFCF464B2148EFD89E96265F5278 (Utilities|Struct|BreakProcAssetInfo (Utilities|Array|Get(acopy) _assetlist _array_element)))))))))
    (:then_3
      (if (> (LocalVariables|GetWidth) 0)
        (bind _array_element_2 (Utilities|Array|ForEachLoopwithBreak (LocalVariables|GetRamp%20Index%20List))
          (:LoopBody
            (if (> (Math|Float|Floor (/ (* 384.0 (Variables|Default|GetNumberOfFloors)) (.SizeZ_15_C53BDD28478B1F2003457E995BF7FDC6 (Utilities|Struct|BreakProcAssetInfo (Utilities|Array|Get(acopy) _assetlist_2 _array_element_2))))) 0)
              (Utilities|Array|Add (Variables|X.%29DoNOTEdit|GetComponent%20Transform%20List) (Utilities|Struct|MakeProcComponenttoSpawn (Math|Transform|MakeTransform (Utilities|Operators|Add _returnvalue_3 (* _returnvalue_4 256.0) _returnvalue_8) _returnvalue_6 (Math|Vector|MakeVector 1.0 (* (/ (LocalVariables|GetWidth) (.SizeY_14_2F84FFCF464B2148EFD89E96265F5278 (Utilities|Struct|BreakProcAssetInfo (Utilities|Array|Get(acopy) _assetlist_2 _array_element_2)))) -1.0) 1.0)) (Variables|E.%29Meshes|GetCliff_Asset_List_) _array_element_2)))))))))

(fn CreateCliffWalls
  (bind _currentdistance (LocalVariables|GetCurrentDistance))
  (bind _returnvalue (Utilities|String|ToString(Float) _currentdistance))
  (bind _currentheight (LocalVariables|GetCurrentHeight))
  (bind _returnvalue_2 (Utilities|String|ToString(Float) _currentheight))
  (bind _returnvalue_3 224.0)
  (bind _returnvalue_4 -288.0)
  (bind _returnvalue_5 256.0)
  (bind _size_z (LocalVariables|GetSize%20Z))
  (bind _originalheight (LocalVariables|GetOriginalheight))
  (bind _returnvalue_6 (/ _originalheight _size_z))
  (bind _returnvalue_7 (Math|Float|Floor _returnvalue_6))
  (bind _size_x (LocalVariables|GetSize%20X))
  (bind _originaldistance (LocalVariables|GetOriginalDistance))
  (bind _returnvalue_8 (/ _originaldistance _size_x))
  (bind _returnvalue_9 (Math|Float|Floor _returnvalue_8))
  (bind _returnvalue_10 (Utilities|String|Append "CurrenDistance = " _returnvalue))
  (bind _returnvalue_11 (Utilities|String|Append _returnvalue_10 "  ___  CurrentHeight = "))
  (bind _returnvalue_12 (Utilities|String|Append _returnvalue_11 _returnvalue_2))
  (bind _size_z_2 (LocalVariables|GetSize%20Z))
  (bind _currentheight_2 (LocalVariables|GetCurrentHeight))
  (bind _returnvalue_13 (/ _currentheight_2 _size_z_2))
  (bind _size_x_2 (LocalVariables|GetSize%20X))
  (bind _currentdistance_2 (LocalVariables|GetCurrentDistance))
  (bind _returnvalue_14 (/ _currentdistance_2 _size_x_2))
  (bind _returnvalue_15 (* _returnvalue_5 -1.0))
  (bind _cliff_asset_list (Variables|E.%29Meshes|GetCliff_Asset_List_))
  (bind _assetlist (Variables|/Game/ArtTools/Procedural/Foundation/Proc_AssetList.Proc_AssetList_C|GetAssetList _cliff_asset_list))
  (bind _spline (Variables|Default|GetSpline))
  (bind _size_x_3 (LocalVariables|GetSize%20X))
  (bind _currentdistance_3 (LocalVariables|GetCurrentDistance))
  (bind _size_x_4 (LocalVariables|GetSize%20X))
  (bind _size_x_5 (LocalVariables|GetSize%20X))
  (bind _size_x_6 (LocalVariables|GetSize%20X))
  (if (<= (Variables|Default|GetNumberOfFloors) 0)
    (Variables|Default|SetNumberOfFloors 1))
  (Utilities|FlowControl|Sequence
    (:then_0
      (bind _assetindexlist (Tools|GetAssetsOfType (Variables|/Game/ArtTools/Procedural/Foundation/Proc_AssetList.Proc_AssetList_C|GetAssetList (Variables|E.%29Meshes|GetCliff_Asset_List_)) "NewEnumerator0" "NewEnumerator2"))
      (LocalVariables|SetAssetIndexList_CliffWalls _assetindexlist))
    (:then_1
      (for _index (range (+ (- (Spline|GetNumberOfSplinePoints _spline) (Math|Integer|SelectInt 1 2 (Spline|IsClosedLoop _spline))) 1))
        (bind (_next__1 _next _current _previous) (Tools|GetCurrentNextPreviousSplinePoint _spline _index))
        (Utilities|FlowControl|Sequence
          (:then_0
            ;[1 1 1] debug
            (if false
              (bind _returnvalue_16 (AddComponent|NODE_AddTextRenderComponent-17 :RelativeTransform (Math|Transform|MakeTransform _current)))
              (Rendering|Components|TextRender|SetText _returnvalue_16 (Utilities|Text|ToText(String) (Utilities|String|Append (Utilities|String|Append (Utilities|String|ToString(Boolean) (Utilities|IsOuterCorner _next _current _previous)) "  next is ") (Utilities|String|ToString(Boolean) (Utilities|IsOuterCorner _next__1 _next _current)))))))
          (:then_1
            (LocalVariables|SetCurrentDistance (Utilities|Operators|Subtract (Math|Vector|Distance(Vector) _current _next) (Math|Float|SelectFloat _returnvalue_4 _returnvalue_3 (Utilities|IsOuterCorner _next__1 _next _current)) (Math|Float|SelectFloat _returnvalue_15 _returnvalue_3 (Utilities|IsOuterCorner _next _current _previous))))
            (LocalVariables|SetOriginalDistance (Utilities|Operators|Subtract (Math|Vector|Distance(Vector) _current _next) (Math|Float|SelectFloat _returnvalue_4 _returnvalue_3 (Utilities|IsOuterCorner _next__1 _next _current)) (Math|Float|SelectFloat _returnvalue_15 _returnvalue_3 (Utilities|IsOuterCorner _next _current _previous))))
            (LocalVariables|SetOriginalheight (* 384.0 (Variables|Default|GetNumberOfFloors)))
            (bind _array_element (Utilities|Array|ForEachLoopwithBreak (LocalVariables|GetAssetIndexList_CliffWalls))
              (:LoopBody
                (if (>= (LocalVariables|GetCurrentDistance) (.SizeX_11_DFCCFEB9407AC738B8EC5AAF3A48E13C (Utilities|Struct|BreakProcAssetInfo (Utilities|Array|Get(aref) _assetlist _array_element))))
                  (LocalVariables|SetSize%20X (.SizeX_11_DFCCFEB9407AC738B8EC5AAF3A48E13C (Utilities|Struct|BreakProcAssetInfo (Utilities|Array|Get(aref) _assetlist _array_element))))
                  (LocalVariables|SetSize%20Z (.SizeZ_15_C53BDD28478B1F2003457E995BF7FDC6 (Utilities|Struct|BreakProcAssetInfo (Utilities|Array|Get(aref) _assetlist _array_element))))
                  (Development|PrintString (Utilities|String|Append (Utilities|String|Append (Utilities|String|Append "NumberOfloop = " (Utilities|String|ToString(Integer) _returnvalue_9)) "  /  ") (Utilities|String|ToString(Integer) _returnvalue_7)))
                  (if (> _returnvalue_9 0)
                    (bind _index_2 (Utilities|FlowControl|ForLoopwithBreak 0 (- _returnvalue_9 1))
                      (:LoopBody
                        (Development|PrintString (Utilities|String|ToString(Integer) _index_2))
                        (LocalVariables|SetCurrentHeight (LocalVariables|GetOriginalheight))
                        (if (> _returnvalue_7 0)
                          (LocalVariables|SetCurrentDistance_toSubtract (Math|Float|SelectFloat (* _size_x_3 (Math|Float|SelectFloat (- (Math|Float|Min(Float) (Math|Float|Min(Float) _returnvalue_14 (.x (.MaxScale_35_FCE6C7E041374C609DB9D58E2E74BBA8 (Utilities|Struct|BreakProcAssetInfo (Utilities|Array|Get(aref) _assetlist _array_element))))) (* (Math|Float|Min(Float) _returnvalue_13 (.z (.MaxScale_35_FCE6C7E041374C609DB9D58E2E74BBA8 (Utilities|Struct|BreakProcAssetInfo (Utilities|Array|Get(aref) _assetlist _array_element))))) 2.0)) (/ (- _size_x_4 (- _currentdistance_3 (* _size_x_6 (Math|Float|Min(Float) (Math|Float|Min(Float) _returnvalue_14 (.x (.MaxScale_35_FCE6C7E041374C609DB9D58E2E74BBA8 (Utilities|Struct|BreakProcAssetInfo (Utilities|Array|Get(aref) _assetlist _array_element))))) (* (Math|Float|Min(Float) _returnvalue_13 (.z (.MaxScale_35_FCE6C7E041374C609DB9D58E2E74BBA8 (Utilities|Struct|BreakProcAssetInfo (Utilities|Array|Get(aref) _assetlist _array_element))))) 2.0))))) _size_x_5)) (Math|Float|Min(Float) (Math|Float|Min(Float) _returnvalue_14 (.x (.MaxScale_35_FCE6C7E041374C609DB9D58E2E74BBA8 (Utilities|Struct|BreakProcAssetInfo (Utilities|Array|Get(aref) _assetlist _array_element))))) (* (Math|Float|Min(Float) _returnvalue_13 (.z (.MaxScale_35_FCE6C7E041374C609DB9D58E2E74BBA8 (Utilities|Struct|BreakProcAssetInfo (Utilities|Array|Get(aref) _assetlist _array_element))))) 2.0)) (and (< (- _currentdistance_3 (* _size_x_6 (Math|Float|Min(Float) (Math|Float|Min(Float) _returnvalue_14 (.x (.MaxScale_35_FCE6C7E041374C609DB9D58E2E74BBA8 (Utilities|Struct|BreakProcAssetInfo (Utilities|Array|Get(aref) _assetlist _array_element))))) (* (Math|Float|Min(Float) _returnvalue_13 (.z (.MaxScale_35_FCE6C7E041374C609DB9D58E2E74BBA8 (Utilities|Struct|BreakProcAssetInfo (Utilities|Array|Get(aref) _assetlist _array_element))))) 2.0)))) _size_x_4) (> (- _currentdistance_3 (* _size_x_6 (Math|Float|Min(Float) (Math|Float|Min(Float) _returnvalue_14 (.x (.MaxScale_35_FCE6C7E041374C609DB9D58E2E74BBA8 (Utilities|Struct|BreakProcAssetInfo (Utilities|Array|Get(aref) _assetlist _array_element))))) (* (Math|Float|Min(Float) _returnvalue_13 (.z (.MaxScale_35_FCE6C7E041374C609DB9D58E2E74BBA8 (Utilities|Struct|BreakProcAssetInfo (Utilities|Array|Get(aref) _assetlist _array_element))))) 2.0)))) 0)))) (* _size_x_6 (Math|Float|Min(Float) (Math|Float|Min(Float) _returnvalue_14 (.x (.MaxScale_35_FCE6C7E041374C609DB9D58E2E74BBA8 (Utilities|Struct|BreakProcAssetInfo (Utilities|Array|Get(aref) _assetlist _array_element))))) (* (Math|Float|Min(Float) _returnvalue_13 (.z (.MaxScale_35_FCE6C7E041374C609DB9D58E2E74BBA8 (Utilities|Struct|BreakProcAssetInfo (Utilities|Array|Get(aref) _assetlist _array_element))))) 2.0))) (and (< (- _currentdistance_3 (* _size_x_6 (Math|Float|Min(Float) (Math|Float|Min(Float) _returnvalue_14 (.x (.MaxScale_35_FCE6C7E041374C609DB9D58E2E74BBA8 (Utilities|Struct|BreakProcAssetInfo (Utilities|Array|Get(aref) _assetlist _array_element))))) (* (Math|Float|Min(Float) _returnvalue_13 (.z (.MaxScale_35_FCE6C7E041374C609DB9D58E2E74BBA8 (Utilities|Struct|BreakProcAssetInfo (Utilities|Array|Get(aref) _assetlist _array_element))))) 2.0)))) _size_x_4) (> (- _currentdistance_3 (* _size_x_6 (Math|Float|Min(Float) (Math|Float|Min(Float) _returnvalue_14 (.x (.MaxScale_35_FCE6C7E041374C609DB9D58E2E74BBA8 (Utilities|Struct|BreakProcAssetInfo (Utilities|Array|Get(aref) _assetlist _array_element))))) (* (Math|Float|Min(Float) _returnvalue_13 (.z (.MaxScale_35_FCE6C7E041374C609DB9D58E2E74BBA8 (Utilities|Struct|BreakProcAssetInfo (Utilities|Array|Get(aref) _assetlist _array_element))))) 2.0)))) 0))))
                          (bind _index_3 (Utilities|FlowControl|ForLoopwithBreak 0 (- _returnvalue_7 1))
                            (:LoopBody
                              (Development|PrintString (Utilities|String|Append (Utilities|String|ToString(Integer) _index_3) " _HeightIteration"))
                              (if (<= (LocalVariables|GetCurrentHeight) 0)
                                _)
                              (Development|PrintString (Utilities|String|Append (Utilities|String|Append (Utilities|String|Append "CurrenDistance = " (Utilities|String|ToString(Float) (LocalVariables|GetCurrentDistance))) "  ___  CurrentHeight = ") (Utilities|String|ToString(Float) (LocalVariables|GetCurrentHeight))))
                              (Utilities|Array|Add (Variables|X.%29DoNOTEdit|GetComponent%20Transform%20List) (Utilities|Struct|MakeProcComponenttoSpawn (Math|Transform|MakeTransform (Utilities|Operators|Add _current (* (Math|Vector|GetUnitDirection(Vector) _current _next) (Math|Float|SelectFloat (* _size_x_3 (Math|Float|SelectFloat (- (Math|Float|Min(Float) (Math|Float|Min(Float) _returnvalue_14 (.x (.MaxScale_35_FCE6C7E041374C609DB9D58E2E74BBA8 (Utilities|Struct|BreakProcAssetInfo (Utilities|Array|Get(aref) _assetlist _array_element))))) (* (Math|Float|Min(Float) _returnvalue_13 (.z (.MaxScale_35_FCE6C7E041374C609DB9D58E2E74BBA8 (Utilities|Struct|BreakProcAssetInfo (Utilities|Array|Get(aref) _assetlist _array_element))))) 2.0)) (/ (- _size_x_4 (- _currentdistance_3 (* _size_x_6 (Math|Float|Min(Float) (Math|Float|Min(Float) _returnvalue_14 (.x (.MaxScale_35_FCE6C7E041374C609DB9D58E2E74BBA8 (Utilities|Struct|BreakProcAssetInfo (Utilities|Array|Get(aref) _assetlist _array_element))))) (* (Math|Float|Min(Float) _returnvalue_13 (.z (.MaxScale_35_FCE6C7E041374C609DB9D58E2E74BBA8 (Utilities|Struct|BreakProcAssetInfo (Utilities|Array|Get(aref) _assetlist _array_element))))) 2.0))))) _size_x_5)) (Math|Float|Min(Float) (Math|Float|Min(Float) _returnvalue_14 (.x (.MaxScale_35_FCE6C7E041374C609DB9D58E2E74BBA8 (Utilities|Struct|BreakProcAssetInfo (Utilities|Array|Get(aref) _assetlist _array_element))))) (* (Math|Float|Min(Float) _returnvalue_13 (.z (.MaxScale_35_FCE6C7E041374C609DB9D58E2E74BBA8 (Utilities|Struct|BreakProcAssetInfo (Utilities|Array|Get(aref) _assetlist _array_element))))) 2.0)) (and (< (- _currentdistance_3 (* _size_x_6 (Math|Float|Min(Float) (Math|Float|Min(Float) _returnvalue_14 (.x (.MaxScale_35_FCE6C7E041374C609DB9D58E2E74BBA8 (Utilities|Struct|BreakProcAssetInfo (Utilities|Array|Get(aref) _assetlist _array_element))))) (* (Math|Float|Min(Float) _returnvalue_13 (.z (.MaxScale_35_FCE6C7E041374C609DB9D58E2E74BBA8 (Utilities|Struct|BreakProcAssetInfo (Utilities|Array|Get(aref) _assetlist _array_element))))) 2.0)))) _size_x_4) (> (- _currentdistance_3 (* _size_x_6 (Math|Float|Min(Float) (Math|Float|Min(Float) _returnvalue_14 (.x (.MaxScale_35_FCE6C7E041374C609DB9D58E2E74BBA8 (Utilities|Struct|BreakProcAssetInfo (Utilities|Array|Get(aref) _assetlist _array_element))))) (* (Math|Float|Min(Float) _returnvalue_13 (.z (.MaxScale_35_FCE6C7E041374C609DB9D58E2E74BBA8 (Utilities|Struct|BreakProcAssetInfo (Utilities|Array|Get(aref) _assetlist _array_element))))) 2.0)))) 0)))) (* _size_x_6 (Math|Float|Min(Float) (Math|Float|Min(Float) _returnvalue_14 (.x (.MaxScale_35_FCE6C7E041374C609DB9D58E2E74BBA8 (Utilities|Struct|BreakProcAssetInfo (Utilities|Array|Get(aref) _assetlist _array_element))))) (* (Math|Float|Min(Float) _returnvalue_13 (.z (.MaxScale_35_FCE6C7E041374C609DB9D58E2E74BBA8 (Utilities|Struct|BreakProcAssetInfo (Utilities|Array|Get(aref) _assetlist _array_element))))) 2.0))) (and (< (- _currentdistance_3 (* _size_x_6 (Math|Float|Min(Float) (Math|Float|Min(Float) _returnvalue_14 (.x (.MaxScale_35_FCE6C7E041374C609DB9D58E2E74BBA8 (Utilities|Struct|BreakProcAssetInfo (Utilities|Array|Get(aref) _assetlist _array_element))))) (* (Math|Float|Min(Float) _returnvalue_13 (.z (.MaxScale_35_FCE6C7E041374C609DB9D58E2E74BBA8 (Utilities|Struct|BreakProcAssetInfo (Utilities|Array|Get(aref) _assetlist _array_element))))) 2.0)))) _size_x_4) (> (- _currentdistance_3 (* _size_x_6 (Math|Float|Min(Float) (Math|Float|Min(Float) _returnvalue_14 (.x (.MaxScale_35_FCE6C7E041374C609DB9D58E2E74BBA8 (Utilities|Struct|BreakProcAssetInfo (Utilities|Array|Get(aref) _assetlist _array_element))))) (* (Math|Float|Min(Float) _returnvalue_13 (.z (.MaxScale_35_FCE6C7E041374C609DB9D58E2E74BBA8 (Utilities|Struct|BreakProcAssetInfo (Utilities|Array|Get(aref) _assetlist _array_element))))) 2.0)))) 0)))) (* (Math|Vector|GetUnitDirection(Vector) _current _next) (- (LocalVariables|GetOriginalDistance) (LocalVariables|GetCurrentDistance))) (Math|Vector|MakeVector 0.0 0.0 (- (LocalVariables|GetCurrentHeight) (LocalVariables|GetOriginalheight))) (Math|Vector|SelectVector (* (Math|Vector|GetUnitDirection(Vector) _current _next) (* 256.0 -1.0)) (* (Math|Vector|GetUnitDirection(Vector) _current _next) 224.0) (Utilities|IsOuterCorner _next _current _previous))) (Math|Rotator|MakeRotationfromAxes (Math|Vector|NegateVector (Math|Vector|GetUnitDirection(Vector) _current _next)) (Math|Vector|CrossProduct (Math|Vector|GetUnitDirection(Vector) _current _next) "0, 0,1.000000") "0, 0,1.000000") (Math|Vector|MakeVector (Math|Float|SelectFloat (- (Math|Float|Min(Float) (Math|Float|Min(Float) _returnvalue_14 (.x (.MaxScale_35_FCE6C7E041374C609DB9D58E2E74BBA8 (Utilities|Struct|BreakProcAssetInfo (Utilities|Array|Get(aref) _assetlist _array_element))))) (* (Math|Float|Min(Float) _returnvalue_13 (.z (.MaxScale_35_FCE6C7E041374C609DB9D58E2E74BBA8 (Utilities|Struct|BreakProcAssetInfo (Utilities|Array|Get(aref) _assetlist _array_element))))) 2.0)) (/ (- _size_x_4 (- _currentdistance_3 (* _size_x_6 (Math|Float|Min(Float) (Math|Float|Min(Float) _returnvalue_14 (.x (.MaxScale_35_FCE6C7E041374C609DB9D58E2E74BBA8 (Utilities|Struct|BreakProcAssetInfo (Utilities|Array|Get(aref) _assetlist _array_element))))) (* (Math|Float|Min(Float) _returnvalue_13 (.z (.MaxScale_35_FCE6C7E041374C609DB9D58E2E74BBA8 (Utilities|Struct|BreakProcAssetInfo (Utilities|Array|Get(aref) _assetlist _array_element))))) 2.0))))) _size_x_5)) (Math|Float|Min(Float) (Math|Float|Min(Float) _returnvalue_14 (.x (.MaxScale_35_FCE6C7E041374C609DB9D58E2E74BBA8 (Utilities|Struct|BreakProcAssetInfo (Utilities|Array|Get(aref) _assetlist _array_element))))) (* (Math|Float|Min(Float) _returnvalue_13 (.z (.MaxScale_35_FCE6C7E041374C609DB9D58E2E74BBA8 (Utilities|Struct|BreakProcAssetInfo (Utilities|Array|Get(aref) _assetlist _array_element))))) 2.0)) (and (< (- _currentdistance_3 (* _size_x_6 (Math|Float|Min(Float) (Math|Float|Min(Float) _returnvalue_14 (.x (.MaxScale_35_FCE6C7E041374C609DB9D58E2E74BBA8 (Utilities|Struct|BreakProcAssetInfo (Utilities|Array|Get(aref) _assetlist _array_element))))) (* (Math|Float|Min(Float) _returnvalue_13 (.z (.MaxScale_35_FCE6C7E041374C609DB9D58E2E74BBA8 (Utilities|Struct|BreakProcAssetInfo (Utilities|Array|Get(aref) _assetlist _array_element))))) 2.0)))) _size_x_4) (> (- _currentdistance_3 (* _size_x_6 (Math|Float|Min(Float) (Math|Float|Min(Float) _returnvalue_14 (.x (.MaxScale_35_FCE6C7E041374C609DB9D58E2E74BBA8 (Utilities|Struct|BreakProcAssetInfo (Utilities|Array|Get(aref) _assetlist _array_element))))) (* (Math|Float|Min(Float) _returnvalue_13 (.z (.MaxScale_35_FCE6C7E041374C609DB9D58E2E74BBA8 (Utilities|Struct|BreakProcAssetInfo (Utilities|Array|Get(aref) _assetlist _array_element))))) 2.0)))) 0))) 1.0 (Math|Float|Min(Float) (Math|Float|Min(Float) _returnvalue_13 (.z (.MaxScale_35_FCE6C7E041374C609DB9D58E2E74BBA8 (Utilities|Struct|BreakProcAssetInfo (Utilities|Array|Get(aref) _assetlist _array_element))))) (* (Math|Float|Min(Float) _returnvalue_14 (.x (.MaxScale_35_FCE6C7E041374C609DB9D58E2E74BBA8 (Utilities|Struct|BreakProcAssetInfo (Utilities|Array|Get(aref) _assetlist _array_element))))) 2.0)))) (Variables|E.%29Meshes|GetCliff_Asset_List_) _array_element))
                              (LocalVariables|SetCurrentHeight (- (LocalVariables|GetCurrentHeight) (* (LocalVariables|GetSize%20Z) (Math|Float|Min(Float) (Math|Float|Min(Float) _returnvalue_13 (.z (.MaxScale_35_FCE6C7E041374C609DB9D58E2E74BBA8 (Utilities|Struct|BreakProcAssetInfo (Utilities|Array|Get(aref) _assetlist _array_element))))) (* (Math|Float|Min(Float) _returnvalue_14 (.x (.MaxScale_35_FCE6C7E041374C609DB9D58E2E74BBA8 (Utilities|Struct|BreakProcAssetInfo (Utilities|Array|Get(aref) _assetlist _array_element))))) 2.0)))))
                              (Development|PrintString _returnvalue_12)
                              (if (<= (LocalVariables|GetCurrentHeight) 0)
                                (LocalVariables|SetCurrentDistance (- (LocalVariables|GetCurrentDistance) (LocalVariables|GetCurrentDistance_toSubtract)))
                                (Development|PrintString _returnvalue_12))
                              (if (and (< (LocalVariables|GetCurrentDistance) (LocalVariables|GetSize%20X)) (<= (LocalVariables|GetCurrentHeight) 0))
                                _)))))))
                  (else
                    (Development|PrintString "ExistLoop CurrentDistance small asset Size x")))))))))))

(fn CreateComponentsFromSavedList
  (if false
    (Utilities|FlowControl|Sequence
      (:then_0
        (for _array_element (Variables|X.%29DoNOTEdit|GetComponent%20Transform%20List)
          (Utilities|Array|AddUnique (LocalVariables|GetUniqueStaticmeshList) (Utilities|Struct|BreakProcAssetInfo (Utilities|Array|Get(acopy) (Variables|/Game/ArtTools/Procedural/Foundation/Proc_AssetList.Proc_AssetList_C|GetAssetList (.ProAssetList_8_82A9873447ED3A9EA990E390BE75AB8D (Utilities|Struct|BreakProcComponenttoSpawn _array_element))) (.Index_6_9AC1D1DE4901E9AF2516CA98A2CDC90C (Utilities|Struct|BreakProcComponenttoSpawn _array_element)))))))
      (:then_1
        (for _array_element_2 (LocalVariables|GetUniqueStaticmeshList)
          (bind _returnvalue (AddComponent|NODE_AddHierarchicalInstancedStaticMeshComponent-0 :RelativeTransform (Math|Transform|MakeTransform "0, 0, 0")))
          (Components|StaticMesh|SetStaticMesh _returnvalue _array_element_2)
          (for _array_element_3 (Variables|X.%29DoNOTEdit|GetComponent%20Transform%20List)
            (if (== _array_element_2 (Utilities|Struct|BreakProcAssetInfo (Utilities|Array|Get(acopy) (Variables|/Game/ArtTools/Procedural/Foundation/Proc_AssetList.Proc_AssetList_C|GetAssetList (.ProAssetList_8_82A9873447ED3A9EA990E390BE75AB8D (Utilities|Struct|BreakProcComponenttoSpawn _array_element_3))) (.Index_6_9AC1D1DE4901E9AF2516CA98A2CDC90C (Utilities|Struct|BreakProcComponenttoSpawn _array_element_3)))))
              (Components|InstancedStaticMesh|AddInstance _returnvalue (Math|Transform|MakeTransform (.location (Utilities|Struct|BreakProcComponenttoSpawn _array_element_3)) (.rotation (Utilities|Struct|BreakProcComponenttoSpawn _array_element_3)) (.scale (Utilities|Struct|BreakProcComponenttoSpawn _array_element_3))) false))))))
    (else
      (for _array_element_4 (Variables|X.%29DoNOTEdit|GetComponent%20Transform%20List)
        (bind _returnvalue_2 (AddComponent|NODE_AddStaticMeshComponent-56 :RelativeTransform (Utilities|Struct|BreakProcComponenttoSpawn _array_element_4)))
        (Tools|SetStaticmeshfromAssetData _returnvalue_2 (.ProAssetList_8_82A9873447ED3A9EA990E390BE75AB8D (Utilities|Struct|BreakProcComponenttoSpawn _array_element_4)) (.Index_6_9AC1D1DE4901E9AF2516CA98A2CDC90C (Utilities|Struct|BreakProcComponenttoSpawn _array_element_4)))
        (Utilities|Array|Add (Variables|MeshLists|GetComponentMeshList) _returnvalue_2)))))

(fn CreateVisualDebugComponents
  (bind _returnvalue 4096.0)
  (bind _returnvalue_2 2048.0)
  (Utilities|FlowControl|Sequence
    (:then_0
      (Utilities|Array|Add (LocalVariables|GetList) (Variables|E.%29Meshes|GetSidewalk_Asset_List))
      (Utilities|Array|Add (LocalVariables|GetList) (Variables|E.%29Meshes|GetBasement_Asset_List))
      (Utilities|Array|Add (LocalVariables|GetList) (Variables|E.%29Meshes|GetTerrainBlend_Asset_List))
      (Utilities|Array|Add (LocalVariables|GetList) (Variables|E.%29Meshes|GetCliff_Asset_List_)))
    (:then_1
      (for (_array_element _array_index) (LocalVariables|GetList)
        (for (_array_element_2 _array_index_2) (Variables|/Game/ArtTools/Procedural/Foundation/Proc_AssetList.Proc_AssetList_C|GetAssetList _array_element)
          (bind _transform (CallFunction|TranformfromData (Math|Transform|MakeTransform (Math|Vector|MakeVector (* _returnvalue _array_index) (* _returnvalue_2 _array_index_2))) _array_element_2))
          (Tools|AddComponentandSetfromData _transform _array_element _array_index_2)
          (bind _returnvalue_3 (AddComponent|NODE_AddStaticMeshComponent-57 :RelativeTransform (Math|Transform|MakeTransform (+ (Math|Vector|MakeVector (* _returnvalue _array_index) (* _returnvalue_2 _array_index_2)) (Math|Vector|MakeVector 0.0 0.0 (* (/ (.SizeZ_15_C53BDD28478B1F2003457E995BF7FDC6 (Utilities|Struct|BreakProcAssetInfo _array_element_2)) 128.0) 64.0))) "0, 0, 0" (Math|Vector|MakeVector (/ (.SizeX_11_DFCCFEB9407AC738B8EC5AAF3A48E13C (Utilities|Struct|BreakProcAssetInfo _array_element_2)) 512.0) (* (/ (.SizeY_14_2F84FFCF464B2148EFD89E96265F5278 (Utilities|Struct|BreakProcAssetInfo _array_element_2)) 512.0) -1.0) (/ (.SizeZ_15_C53BDD28478B1F2003457E995BF7FDC6 (Utilities|Struct|BreakProcAssetInfo _array_element_2)) 128.0)))))
          (bind _returnvalue_4 (AddComponent|NODE_AddTextRenderComponent-18 :RelativeTransform (Math|Transform|MakeTransform (+ (.location (Math|Transform|MakeTransform (+ (Math|Vector|MakeVector (* _returnvalue _array_index) (* _returnvalue_2 _array_index_2)) (Math|Vector|MakeVector 0.0 0.0 (* (/ (.SizeZ_15_C53BDD28478B1F2003457E995BF7FDC6 (Utilities|Struct|BreakProcAssetInfo _array_element_2)) 128.0) 64.0))) "0, 0, 0" (Math|Vector|MakeVector (/ (.SizeX_11_DFCCFEB9407AC738B8EC5AAF3A48E13C (Utilities|Struct|BreakProcAssetInfo _array_element_2)) 512.0) (* (/ (.SizeY_14_2F84FFCF464B2148EFD89E96265F5278 (Utilities|Struct|BreakProcAssetInfo _array_element_2)) 512.0) -1.0) (/ (.SizeZ_15_C53BDD28478B1F2003457E995BF7FDC6 (Utilities|Struct|BreakProcAssetInfo _array_element_2)) 128.0)))) (* (.Max (Components|StaticMesh|GetLocalBounds _returnvalue_3)) "0.000000,0,4.000000")) (.rotation (Math|Transform|MakeTransform (+ (Math|Vector|MakeVector (* _returnvalue _array_index) (* _returnvalue_2 _array_index_2)) (Math|Vector|MakeVector 0.0 0.0 (* (/ (.SizeZ_15_C53BDD28478B1F2003457E995BF7FDC6 (Utilities|Struct|BreakProcAssetInfo _array_element_2)) 128.0) 64.0))) "0, 0, 0" (Math|Vector|MakeVector (/ (.SizeX_11_DFCCFEB9407AC738B8EC5AAF3A48E13C (Utilities|Struct|BreakProcAssetInfo _array_element_2)) 512.0) (* (/ (.SizeY_14_2F84FFCF464B2148EFD89E96265F5278 (Utilities|Struct|BreakProcAssetInfo _array_element_2)) 512.0) -1.0) (/ (.SizeZ_15_C53BDD28478B1F2003457E995BF7FDC6 (Utilities|Struct|BreakProcAssetInfo _array_element_2)) 128.0)))) (.scale (Math|Transform|MakeTransform (+ (Math|Vector|MakeVector (* _returnvalue _array_index) (* _returnvalue_2 _array_index_2)) (Math|Vector|MakeVector 0.0 0.0 (* (/ (.SizeZ_15_C53BDD28478B1F2003457E995BF7FDC6 (Utilities|Struct|BreakProcAssetInfo _array_element_2)) 128.0) 64.0))) "0, 0, 0" (Math|Vector|MakeVector (/ (.SizeX_11_DFCCFEB9407AC738B8EC5AAF3A48E13C (Utilities|Struct|BreakProcAssetInfo _array_element_2)) 512.0) (* (/ (.SizeY_14_2F84FFCF464B2148EFD89E96265F5278 (Utilities|Struct|BreakProcAssetInfo _array_element_2)) 512.0) -1.0) (/ (.SizeZ_15_C53BDD28478B1F2003457E995BF7FDC6 (Utilities|Struct|BreakProcAssetInfo _array_element_2)) 128.0)))))))
          (Rendering|Components|TextRender|SetText _returnvalue_4 (Utilities|Text|ToText(String) (Utilities|String|Append (Utilities|String|Append (Utilities|String|Append "INDEX = " (Utilities|String|ToString(Integer) _array_index_2)) "   _   ") (Utilities|GetDisplayName _array_element))))
          (AddComponent|NODE_AddArrowComponent-1 :RelativeTransform (Math|Transform|MakeTransform "0, 0,2048.000000" "0, 0, 0" "40.000000,40.000000,40.000000"))
          (AddComponent|NODE_AddTextRenderComponent-19 :RelativeTransform (Math|Transform|MakeTransform "0, 0,2048.000000" "0,90.000000, 0" "40.000000,40.000000,40.000000")))))))

(fn DebugText
  (params (NewParam :type Transform) (NewParam1 :type Text) (NewParam2 :type Text) (NewParam3 :type Text) (NewParam4 :type Text) (NewParam5 :type Text) (NewParam6 :type Text) (NewParam7 :type Integer) (NewParam8 :type Integer) (Debug :type Boolean))
  (bind _returnvalue (+ NewParam8 NewParam7))
  (if Debug
    (bind _returnvalue_2 (AddComponent|NODE_AddTextRenderComponent-9 :RelativeTransform (Math|Transform|MakeTransform (.location NewParam))))
    (Transformation|AddRelativeLocation _returnvalue_2 "0, 0,200.000000" false false)
    (Transformation|AddRelativeLocation _returnvalue_2 (* (Math|Vector|MakeVector 5.0 5.0 5.0) _returnvalue) false false)
    (Rendering|Components|TextRender|SetText _returnvalue_2 (Utilities|Text|ToText(String) (Utilities|String|Append (Utilities|String|Append (Utilities|String|Append (Utilities|String|Append (Utilities|String|Append (Utilities|String|Append (Utilities|String|Append (Utilities|String|Append (Utilities|String|Append (Utilities|String|Append (Utilities|String|ToString(Text) NewParam1) " _ ") (Utilities|String|ToString(Text) NewParam2)) " _ ") (Utilities|String|ToString(Text) NewParam3)) " _ ") (Utilities|String|ToString(Text) NewParam4)) " _ ") (Utilities|String|ToString(Text) NewParam5)) " _ ") (Utilities|String|ToString(Text) NewParam6))))
    (Rendering|Components|TextRender|SetTextRenderColor _returnvalue_2 (select (> (Math|Float|Fraction (/ (* _returnvalue 1.0) 2.0)) 0.2) (Utilities|Struct|MakeColor 0 255 0 255) (Utilities|Struct|MakeColor 255 0 0 1)))))

(fn DeleteAllTempSplines
  (bind _tempsplinelist (Variables|SplineList|GetTempSplineList))
  (for _array_element _tempsplinelist
    (Components|DestroyComponent _array_element))
  (Utilities|Array|Clear _tempsplinelist))

(fn DeleteOuterCornerCollision
  (bind _cornercollisionlist (Variables|MeshLists|GetCornerCollisionList))
  (for _array_element _cornercollisionlist
    (Components|DestroyComponent _array_element))
  (Utilities|Array|Clear _cornercollisionlist))

(fn DestroyAllComponents
  (bind _floormeshlist (Variables|MeshLists|GetFloorMeshList))
  (bind _basementwallmeshlist (Variables|MeshLists|GetBasementWallMeshList))
  (bind _terrainblendmeshlist (Variables|MeshLists|GetTerrainBlendMeshList))
  (bind _bordercollisionlist (Variables|MeshLists|GetBorderCollisionList))
  (bind _componentmeshlist (Variables|MeshLists|GetComponentMeshList))
  (Utilities|FlowControl|Sequence
    (:then_0
      (for _array_element _floormeshlist
        (Components|DestroyComponent _array_element))
      (Utilities|Array|Clear _floormeshlist))
    (:then_1
      (for _array_element_2 _basementwallmeshlist
        (Components|DestroyComponent _array_element_2))
      (Utilities|Array|Clear _basementwallmeshlist))
    (:then_2
      (for _array_element_3 _terrainblendmeshlist
        (Components|DestroyComponent _array_element_3))
      (Utilities|Array|Clear _terrainblendmeshlist))
    (:then_3
      (for _array_element_4 _bordercollisionlist
        (Components|DestroyComponent _array_element_4))
      (Utilities|Array|Clear _bordercollisionlist))
    (:then_4
      (for _array_element_5 _componentmeshlist
        (Components|DestroyComponent _array_element_5))
      (Utilities|Array|Clear _componentmeshlist))))

(fn DoNotUseText
  (AddComponent|NODE_AddTextRenderComponent-21 :RelativeTransform (Math|Transform|MakeTransform "0, 0, 0" "0.000000, 0,90.000000")))

(fn DuplicateSpline
  (params (Spline :type Spline%20Component%20Object%20Reference) (Reverse :type Boolean) (CoordinateSpace :type ESplineCoordinateSpace%20Enum :default "Local") (AlignZtoActor :type Boolean) (Debug :type Boolean))
  (bind _spline (Variables|Default|GetSpline))
  (bind _returnvalue (Spline|GetNumberOfSplinePoints _spline))
  (bind _returnvalue_2 (- _returnvalue 1))
  (bind _returnvalue_3 (AddComponent|NODE_AddSplineComponent-1 :RelativeTransform (Utilities|Select :Index (Variables|GetCoordinateSpace) :Local (Transformation|GetRelativeTransform Spline) :World (Math|Transform|MakeTransform "0, 0, 0"))))
  (if (not Debug)
    (Utilities|Array|Add (Variables|SplineList|GetTempSplineList) _returnvalue_3))
  (Spline|RemoveSplinePoint _returnvalue_3 1 true)
  (Spline|RemoveSplinePoint _returnvalue_3 0 true)
  (Spline|SetClosedLoop _returnvalue_3 (Spline|IsClosedLoop (Variables|Default|GetSpline)) true)
  (for _index (range _returnvalue)
    (Spline|AddSplinePoint _returnvalue_3 (Utilities|Select :Index (Variables|GetCoordinateSpace) :Local (* (Spline|GetLocationatSplinePoint Spline (Math|Integer|SelectInt (- _returnvalue_2 _index) _index Reverse) (Variables|GetCoordinateSpace)) (Math|Vector|SelectVector "1.000000,1.000000, 0" "1.000000,1.000000,1.000000" AlignZtoActor)) :World (Math|Vector|MakeVector (.x (Spline|GetLocationatSplinePoint Spline (Math|Integer|SelectInt (- _returnvalue_2 _index) _index Reverse) (Variables|GetCoordinateSpace))) (.y (Spline|GetLocationatSplinePoint Spline (Math|Integer|SelectInt (- _returnvalue_2 _index) _index Reverse) (Variables|GetCoordinateSpace))) (.z (Transformation|GetActorLocation)))) (Variables|GetCoordinateSpace) true)
    (Spline|SetSplinePointType _returnvalue_3 (Math|Integer|SelectInt (- _returnvalue_2 _index) _index Reverse) "Linear" true))
  (return _returnvalue_3))

(fn EnableLiveEditUp
  (params (InputObject :type Proc%20POI%20Tool%20Object%20Reference))
  (Utilities|IsValid InputObject
    (:Is%20Valid
      (Variables|Default|SetLiveUpdate true InputObject)
      (CreateComponents|CreateAllComponents InputObject)
      (CallFunction|EnableLiveEditUp (Variables|X.%29DoNOTEdit|GetConnectedBasementsUp InputObject)))))

(fn ExpandSpline
  (params (Spline :type Spline%20Component%20Object%20Reference) (Distance :type Float%20%28double-precision%29))
  (bind _returnvalue (Spline|GetNumberOfSplinePoints Spline))
  (bind _returnvalue_2 (- _returnvalue 1))
  (Utilities|Array|Clear (Variables|Default|GetOriginalSplinePoints))
  (for _index (range _returnvalue)
    (Utilities|Array|Add (Variables|Default|GetOriginalSplinePoints) (Spline|GetLocationatSplinePoint Spline _index "Local")))
  (Development|PrintString (Utilities|String|ToString(Integer) _returnvalue_2))
  (for _index_2 (range _returnvalue)
    (Spline|SetLocationatSplinePoint Spline _index_2 (+ (Spline|GetLocationatSplinePoint Spline _index_2 "Local") (* (+ (* (Math|Vector|Normalize (- (Spline|GetLocationatSplinePoint Spline _index_2 "Local") (Utilities|Array|Get(aref) (Variables|Default|GetOriginalSplinePoints) (Math|Integer|SelectInt 0 (+ _index_2 1) (== _index_2 _returnvalue_2))))) Distance) (* (Math|Vector|Normalize (- (Spline|GetLocationatSplinePoint Spline _index_2 "Local") (Utilities|Array|Get(aref) (Variables|Default|GetOriginalSplinePoints) (Math|Integer|SelectInt _returnvalue_2 (+ _index_2 -1) (== _index_2 0))))) Distance)) (Math|Float|SelectFloat 1.0 -1.0 (> (Math|Vector|DotProduct (Math|Vector|CrossProduct (- (Utilities|Array|Get(aref) (Variables|Default|GetOriginalSplinePoints) (Math|Integer|SelectInt 0 (+ _index_2 1) (== _index_2 _returnvalue_2))) (Spline|GetLocationatSplinePoint Spline _index_2 "Local")) (Math|Vector|MakeVector 0.0 0.0 1.0)) (- (Spline|GetLocationatSplinePoint Spline _index_2 "Local") (Utilities|Array|Get(aref) (Variables|Default|GetOriginalSplinePoints) (Math|Integer|SelectInt _returnvalue_2 (+ _index_2 -1) (== _index_2 0))))) 0.0)))) "Local" true))
  (Utilities|Array|Clear (Variables|Default|GetOriginalSplinePoints))
  (return Spline))

(fn FillArea
  (params (PaintLayer :type Landscape%20Layer%20Info%20Object%20Object%20Reference))
  (bind _returnvalue 128.0)
  (bind _terraincellarmargin (Variables|Default|GetTerrainCellarMargin))
  (bind _returnvalue_2 384.0)
  (bind _numberoffloors (Variables|Default|GetNumberOfFloors))
  (bind _returnvalue_3 (* _numberoffloors _returnvalue_2))
  (bind _returnvalue_4 (+ _returnvalue_3 _terraincellarmargin))
  (bind _returnvalue_5 (Math|Vector|MakeVector 0.0 0.0 _returnvalue_4))
  (bind _bordercollisionlist (Variables|MeshLists|GetBorderCollisionList))
  (bind _fillsplines (Variables|SplineList|GetFillSplines))
  (bind _spline (Variables|Default|GetSpline))
  (Utilities|FlowControl|Sequence
    (:then_0
      (CreateCollisionComponents|MakeBorderCollisionFromSpline (Variables|Default|GetSpline)))
    (:then_1
      (Utilities|Array|Clear (Variables|SplineList|GetFillSplines))
      (for _index (range (Spline|GetNumberOfSplinePoints _spline))
        (bind (_next__1 _next _current _previous _previous__1 _world_next__1 _world_next _world_current) (Tools|GetCurrentNextPreviousSplinePoint _spline _index))
        (for _index_2 (range (Math|Float|Round (/ (Math|Vector|VectorLength (- _world_next _world_current)) 512.0)))
          (bind _outhits (Collision|MultiLineTraceForObjects :Start (+ (+ (* (* (Math|Vector|Normalize (- _world_next _world_current)) 512.0) _index_2) (* (Math|Vector|Normalize (- _world_next _world_current)) 256.0)) _world_current) :End (+ (* (Math|Vector|Normalize (Math|Vector|CrossProduct (- _world_next _world_current) "0, 0,-1.000000")) 9.99e+05) (+ (+ (* (* (Math|Vector|Normalize (- _world_next _world_current)) 512.0) _index_2) (* (Math|Vector|Normalize (- _world_next _world_current)) 256.0)) _world_current)) :ObjectTypes (Utilities|Array|MakeArray "ObjectTypeQuery3") :bTraceComplex true :DrawDebugType "None" :bIgnoreSelf false :TraceColor "(R=1.000000,G=0.000000,B=0.000000,A=1.000000)" :TraceHitColor "(R=0.000000,G=1.000000,B=0.000000,A=1.000000)" :DrawTime 1e+01))
          (bind _hitresoultoutput (Tools|ClosestHitResult _outhits (+ (+ (* (* (Math|Vector|Normalize (- _world_next _world_current)) 512.0) _index_2) (* (Math|Vector|Normalize (- _world_next _world_current)) 256.0)) _world_current)))
          (bind _returnvalue_6 (AddComponent|NODE_AddSplineComponent-2))
          (Spline|SetLocationatSplinePoint _returnvalue_6 0 (- (+ (- (+ (+ (* (* (Math|Vector|Normalize (- _world_next _world_current)) 512.0) _index_2) (* (Math|Vector|Normalize (- _world_next _world_current)) 256.0)) _world_current) "0, 0, 0") (* (Math|Vector|Normalize (- (- (+ (+ (* (* (Math|Vector|Normalize (- _world_next _world_current)) 512.0) _index_2) (* (Math|Vector|Normalize (- _world_next _world_current)) 256.0)) _world_current) "0, 0, 0") (- (.Location (Collision|BreakHitResult _hitresoultoutput)) "0, 0, 0"))) _returnvalue)) _returnvalue_5) "World" true)
          (Spline|SetLocationatSplinePoint _returnvalue_6 1 (- (- (- (.Location (Collision|BreakHitResult _hitresoultoutput)) "0, 0, 0") (* (Math|Vector|Normalize (- (- (+ (+ (* (* (Math|Vector|Normalize (- _world_next _world_current)) 512.0) _index_2) (* (Math|Vector|Normalize (- _world_next _world_current)) 256.0)) _world_current) "0, 0, 0") (- (.Location (Collision|BreakHitResult _hitresoultoutput)) "0, 0, 0"))) _returnvalue)) _returnvalue_5) "World" true)
          (Utilities|Array|Add (Variables|SplineList|GetFillSplines) _returnvalue_6))))
    (:then_2
      (bind _outactors (Actor|GetAllActorsOfClass "/Script/Landscape.LandscapeProxy"))
      (for _array_element _outactors
        (for _array_element_2 (Variables|SplineList|GetFillSplines)
          (for _array_element_3 (Variables|PaintLayer|GetEdit%20Layer%20Name)
            (Landscape|Editor|EditorApplySpline _array_element _array_element_2 (+ 512.0 0.0) (+ 512.0 0.0) 0.0 0.0 0.0 0.0 20 true true PaintLayer _array_element_3)))))
    (:then_3
      (for _array_element_4 _bordercollisionlist
        (Components|DestroyComponent _array_element_4))
      (Utilities|Array|Clear _bordercollisionlist)
      (for _array_element_5 _fillsplines
        (Components|DestroyComponent _array_element_5))
      (Utilities|Array|Clear _fillsplines))))

(fn Force4SplinePoints
  (bind _spline (Variables|Default|GetSpline))
  (bind _returnvalue (Spline|GetNumberOfSplinePoints _spline))
  (bind _returnvalue_2 (Spline|GetLocationatSplinePoint _spline 1 "Local"))
  (bind _returnvalue_3 (Spline|GetLocationatSplinePoint _spline 0 "Local"))
  (bind _returnvalue_4 (Math|Vector|GetUnitDirection(Vector) _returnvalue_3 _returnvalue_2))
  (bind _returnvalue_5 (Math|Vector|CrossProduct _returnvalue_4 "0, 0,-1.000000"))
  (bind _returnvalue_6 (Math|Vector|Normalize _returnvalue_5))
  (bind _returnvalue_7 (* _returnvalue_6 512.0))
  (if (<= _returnvalue 2)
    (if (== _returnvalue 2)
      (Spline|AddSplinePoint _spline (+ _returnvalue_2 _returnvalue_7) "Local" true)
      (Spline|AddSplinePoint _spline (+ _returnvalue_3 _returnvalue_7) "Local" true)
      (else
        (Spline|SetSplinePoints _spline (Utilities|Array|MakeArray "0, 0, 0" "-512.000000,0.000000, 0" "-512.000000,-512.000000, 0" "0.000000,-512.000000, 0") "Local" true)))))

(fn ForceSubBasmentMode
  (for _array_element (Variables|C.%29ConnectingComplexPOIs|GetConnectedActors)
    (Utilities|IsValid _array_element
      (:Is%20Valid
        (switch Utilities|FlowControl|Switch|SwitchonProc_FoundationSelection (Variables|Default|GetMode _array_element)
          (:NewEnumerator0
            _)
          (:NewEnumerator2
            _))
        (Variables|Default|SetMode "NewEnumerator1" _array_element)))))

(fn FurthestHitResult
  (params (HitResult :type Array%20of%20Hit%20Result%20Structures) (WorldSpaceLocation :type Vector) (Debug :type Boolean))
  (bind _self self)
  (bind _returnvalue (Utilities|GetDisplayName _self))
  (bind _returnvalue_2 (Utilities|String|Append _returnvalue "   _HIT_    "))
  (if (> (Utilities|Array|Length HitResult) 0)
    (LocalVariables|SetClosestHitIndex -1)
    (bind (_array_element _array_index) (Utilities|Array|ForEachLoopwithBreak HitResult)
      (:LoopBody
        (if (Utilities|Equal(Object) (.HitActor (Collision|BreakHitResult _array_element)) _self)
          (Development|PrintString (Utilities|String|Append (Utilities|String|Append (Utilities|String|Append _returnvalue_2 (Utilities|GetDisplayName (.HitActor (Collision|BreakHitResult _array_element)))) "   _Bool_    ") (Utilities|String|ToString(Boolean) (Utilities|Equal(Object) (.HitActor (Collision|BreakHitResult _array_element)) _self))) Debug Debug "(R=0.142391,G=1.000000,B=0.000000,A=1.000000)")
          (if (== (LocalVariables|GetClosestHitIndex) -1)
            (LocalVariables|SetMaxLength (Math|Vector|VectorLength (- (.Location (Collision|BreakHitResult _array_element)) WorldSpaceLocation)))
            (LocalVariables|SetClosestHitIndex 0)
            (elif (> (Math|Vector|VectorLength (- (.Location (Collision|BreakHitResult _array_element)) WorldSpaceLocation)) (LocalVariables|GetMaxLength))
              (LocalVariables|SetMaxLength (Math|Vector|VectorLength (- (.Location (Collision|BreakHitResult _array_element)) WorldSpaceLocation)))
              (LocalVariables|SetClosestHitIndex _array_index)))
          (else
            (Development|PrintString (Utilities|String|Append (Utilities|String|Append (Utilities|String|Append _returnvalue_2 (Utilities|GetDisplayName (.HitActor (Collision|BreakHitResult _array_element)))) "   _Bool_    ") (Utilities|String|ToString(Boolean) (Utilities|Equal(Object) (.HitActor (Collision|BreakHitResult _array_element)) _self))) Debug Debug "(R=1.000000,G=0.000000,B=0.000000,A=1.000000)"))))
      (:Completed
        (return (Utilities|Array|Get(aref) HitResult (LocalVariables|GetClosestHitIndex)))))
    (else
      (return (Collision|MakeHitResult false)))))

(fn GetAssetsOfType
  (params (AssetList :type Array%20of%20Proc%20Asset%20Info%20Structures) (Type :type Proc_AssetType%20Enum :default "NewEnumerator0") (SortAxis :type Proc_Axis%20Enum :default "NewEnumerator0") (Debug :type Boolean))
  (bind _listsorted (LocalVariables|GetListSorted))
  (bind _listsorted_2 (LocalVariables|GetListSorted))
  (Utilities|FlowControl|Sequence
    (:then_0
      (for (_array_element _array_index) AssetList
        (if (Utilities|Enum|Equal(Enum) (.Type_6_B4CD9FF0498F54172801239A22E7A5CA (Utilities|Struct|BreakProcAssetInfo _array_element)) Type)
          (Utilities|Array|Add (LocalVariables|GetList) _array_index))))
    (:then_1
      (for _array_element_2 (LocalVariables|GetList)
        (if (== (Utilities|Array|Length _listsorted) 0)
          (Utilities|Array|Add _listsorted _array_element_2)
          (else
            (LocalVariables|SetSmallIndex 0)
            (bind (_array_element_3 _array_index_2) (Utilities|Array|ForEachLoopwithBreak (LocalVariables|GetListSorted))
              (:LoopBody
                (if (>= (Utilities|Select :Index SortAxis :NewEnumerator0 (.SizeX_11_DFCCFEB9407AC738B8EC5AAF3A48E13C (Utilities|Struct|BreakProcAssetInfo (Utilities|Array|Get(acopy) AssetList _array_element_3))) :NewEnumerator1 (.SizeY_14_2F84FFCF464B2148EFD89E96265F5278 (Utilities|Struct|BreakProcAssetInfo (Utilities|Array|Get(acopy) AssetList _array_element_3))) :NewEnumerator2 (.SizeZ_15_C53BDD28478B1F2003457E995BF7FDC6 (Utilities|Struct|BreakProcAssetInfo (Utilities|Array|Get(acopy) AssetList _array_element_3)))) (Utilities|Select :Index SortAxis :NewEnumerator0 (.SizeX_11_DFCCFEB9407AC738B8EC5AAF3A48E13C (Utilities|Struct|BreakProcAssetInfo (Utilities|Array|Get(acopy) AssetList _array_element_2))) :NewEnumerator1 (.SizeY_14_2F84FFCF464B2148EFD89E96265F5278 (Utilities|Struct|BreakProcAssetInfo (Utilities|Array|Get(acopy) AssetList _array_element_2))) :NewEnumerator2 (.SizeZ_15_C53BDD28478B1F2003457E995BF7FDC6 (Utilities|Struct|BreakProcAssetInfo (Utilities|Array|Get(acopy) AssetList _array_element_2)))))
                  _
                  (else
                    _))
                (LocalVariables|SetSmallIndex (+ _array_index_2 1)))
              (:Completed
                _))
            (Utilities|Array|Insert (LocalVariables|GetListSorted) _array_element_2 (LocalVariables|GetSmallIndex))))))
    (:then_2
      (if Debug
        (for _array_element_4 _listsorted_2
          (Development|PrintString (Utilities|String|ToString(Integer) _array_element_4) true true "(R=0.000000,G=1.000000,B=0.120580,A=1.000000)")))
      (return _listsorted_2))))

(fn getCollisionComponentTag
  (params (Component :type Actor%20Component%20Object%20Reference) (Debug :type Boolean))
  (bind _componenttags (Variables|/Script/Engine.ActorComponent|GetComponentTags Component))
  (bind _output (Utilities|Array|Get(acopy) _componenttags))
  (if (> (Utilities|Array|Length _componenttags) 0)
    (Development|PrintString (Utilities|String|ToString(Name) _output) Debug Debug)
    (return _output)
    (else
      (return "None"))))

(fn GetCurrentNextPreviousSplinePoint
  (params (Spline :type Spline%20Component%20Object%20Reference) (CurrentIndex :type Integer) (DrawDebugNumber :type Boolean) (Reverse :type Boolean))
  (bind _returnvalue (Spline|GetLocationatSplinePoint Spline CurrentIndex "World"))
  (Utilities|FlowControl|Sequence
    (:then_0
      (if DrawDebugNumber
        (bind _returnvalue_2 (AddComponent|NODE_AddTextRenderComponent-13 :RelativeTransform (Math|Transform|MakeTransform (- _returnvalue (Transformation|GetActorLocation)))))
        (Rendering|Components|TextRender|SetText _returnvalue_2 (Utilities|Text|ToText(Integer) CurrentIndex))))
    (:then_1
      (return (Math|Vector|SelectVector (Spline|GetLocationatSplinePoint Spline (Math|Integer|SelectInt (- (Spline|GetNumberOfSplinePoints Spline) 2) (Math|Integer|SelectInt (- (Spline|GetNumberOfSplinePoints Spline) 1) (- CurrentIndex 2) (== CurrentIndex 1)) (== CurrentIndex 0)) "Local") (Spline|GetLocationatSplinePoint Spline (Math|Integer|SelectInt 1 (Math|Integer|SelectInt 0 (+ CurrentIndex 2) (== CurrentIndex (- (- (Spline|GetNumberOfSplinePoints Spline) 1) 1))) (== CurrentIndex (- (Spline|GetNumberOfSplinePoints Spline) 1))) "Local") (Variables|GetReverse)) (Math|Vector|SelectVector (Spline|GetLocationatSplinePoint Spline (Math|Integer|SelectInt (- (Spline|GetNumberOfSplinePoints Spline) 1) (+ CurrentIndex -1) (== CurrentIndex 0)) "Local") (Spline|GetLocationatSplinePoint Spline (Math|Integer|SelectInt 0 (+ CurrentIndex 1) (== CurrentIndex (- (Spline|GetNumberOfSplinePoints Spline) 1))) "Local") (Variables|GetReverse)) (Spline|GetLocationatSplinePoint Spline CurrentIndex "Local") (Math|Vector|SelectVector (Spline|GetLocationatSplinePoint Spline (Math|Integer|SelectInt 0 (+ CurrentIndex 1) (== CurrentIndex (- (Spline|GetNumberOfSplinePoints Spline) 1))) "Local") (Spline|GetLocationatSplinePoint Spline (Math|Integer|SelectInt (- (Spline|GetNumberOfSplinePoints Spline) 1) (+ CurrentIndex -1) (== CurrentIndex 0)) "Local") (Variables|GetReverse)) (Math|Vector|SelectVector (Spline|GetLocationatSplinePoint Spline (Math|Integer|SelectInt 1 (Math|Integer|SelectInt 0 (+ CurrentIndex 2) (== CurrentIndex (- (- (Spline|GetNumberOfSplinePoints Spline) 1) 1))) (== CurrentIndex (- (Spline|GetNumberOfSplinePoints Spline) 1))) "Local") (Spline|GetLocationatSplinePoint Spline (Math|Integer|SelectInt (- (Spline|GetNumberOfSplinePoints Spline) 2) (Math|Integer|SelectInt (- (Spline|GetNumberOfSplinePoints Spline) 1) (- CurrentIndex 2) (== CurrentIndex 1)) (== CurrentIndex 0)) "Local") (Variables|GetReverse)) (Math|Vector|SelectVector (Spline|GetLocationatSplinePoint Spline (Math|Integer|SelectInt (- (Spline|GetNumberOfSplinePoints Spline) 2) (Math|Integer|SelectInt (- (Spline|GetNumberOfSplinePoints Spline) 1) (- CurrentIndex 2) (== CurrentIndex 1)) (== CurrentIndex 0)) "World") (Spline|GetLocationatSplinePoint Spline (Math|Integer|SelectInt 1 (Math|Integer|SelectInt 0 (+ CurrentIndex 2) (== CurrentIndex (- (- (Spline|GetNumberOfSplinePoints Spline) 1) 1))) (== CurrentIndex (- (Spline|GetNumberOfSplinePoints Spline) 1))) "World") (Variables|GetReverse)) (Math|Vector|SelectVector (Spline|GetLocationatSplinePoint Spline (Math|Integer|SelectInt (- (Spline|GetNumberOfSplinePoints Spline) 1) (+ CurrentIndex -1) (== CurrentIndex 0)) "World") (Spline|GetLocationatSplinePoint Spline (Math|Integer|SelectInt 0 (+ CurrentIndex 1) (== CurrentIndex (- (Spline|GetNumberOfSplinePoints Spline) 1))) "World") (Variables|GetReverse)) _returnvalue (Math|Vector|SelectVector (Spline|GetLocationatSplinePoint Spline (Math|Integer|SelectInt 1 (Math|Integer|SelectInt 0 (+ CurrentIndex 2) (== CurrentIndex (- (- (Spline|GetNumberOfSplinePoints Spline) 1) 1))) (== CurrentIndex (- (Spline|GetNumberOfSplinePoints Spline) 1))) "World") (Spline|GetLocationatSplinePoint Spline (Math|Integer|SelectInt (- (Spline|GetNumberOfSplinePoints Spline) 1) (+ CurrentIndex -1) (== CurrentIndex 0)) "World") (Variables|GetReverse)) (Math|Vector|SelectVector (Spline|GetLocationatSplinePoint Spline (Math|Integer|SelectInt 1 (Math|Integer|SelectInt 0 (+ CurrentIndex 2) (== CurrentIndex (- (- (Spline|GetNumberOfSplinePoints Spline) 1) 1))) (== CurrentIndex (- (Spline|GetNumberOfSplinePoints Spline) 1))) "World") (Spline|GetLocationatSplinePoint Spline (Math|Integer|SelectInt (- (Spline|GetNumberOfSplinePoints Spline) 2) (Math|Integer|SelectInt (- (Spline|GetNumberOfSplinePoints Spline) 1) (- CurrentIndex 2) (== CurrentIndex 1)) (== CurrentIndex 0)) "World") (Variables|GetReverse))))))

(fn GetIndexFromKeyArray
  (params (Array :type Array%20of%20Integers))
  (Utilities|FlowControl|Sequence
    (:then_0
      (for _array_element Array
        (bind _keys (Utilities|Map|Keys (Variables|Default|GetConnectedSplineList)))
        (Utilities|Array|Add (LocalVariables|GetTempArray) (Utilities|Array|FindItem _keys _array_element))))
    (:then_1
      (return (LocalVariables|GetTempArray)))))

(fn InsetSplineForCliffs
  (params (Spline :type Spline%20Component%20Object%20Reference) (Open%20CliffSpline :type Spline%20Component%20Object%20Reference))
  (bind _cliffsplinepoints (LocalVariables|GetCliffSplinePoints))
  (Utilities|FlowControl|Sequence
    (:then_0
      (for _index (range (+ (Spline|GetNumberOfSplinePoints Open%20CliffSpline) 1))
        (Utilities|Array|Add (LocalVariables|GetCliffSplinePoints) (Spline|GetLocationatSplinePoint Open%20CliffSpline _index "World"))))
    (:then_1
      (for _index_2 (range (+ (Spline|GetNumberOfSplinePoints Spline) 1))
        (bind (_next__1 _next _current _previous _previous__1 _world_next__1 _world_next _world_current _world_previous) (Tools|GetCurrentNextPreviousSplinePoint Spline _index_2))
        (if (>= (Utilities|Array|FindItem _cliffsplinepoints _world_current) 0)
          (Utilities|Array|Add (LocalVariables|GetSplinePoints) (Utilities|Operators|Add _world_current (Math|Vector|SelectVector (* (Math|Vector|CrossProduct (Math|Vector|GetUnitDirection(Vector) _current _next) "0, 0,-1.000000") 256.0) "0, 0, 0" (>= (Utilities|Array|FindItem _cliffsplinepoints _world_next) 0)) (Math|Vector|SelectVector (* (Math|Vector|CrossProduct (Math|Vector|GetUnitDirection(Vector) _current _previous) "0, 0,1.000000") 256.0) "0, 0, 0" (>= (Utilities|Array|FindItem _cliffsplinepoints _world_previous) 0))))
          (else
            (Utilities|Array|Add (LocalVariables|GetSplinePoints) _world_current)))))
    (:then_2
      (bind _returnvalue (AddComponent|NODE_AddSplineComponent-12 :RelativeTransform (Math|Transform|MakeTransform "0, 0, 0")))
      (Utilities|FlowControl|Sequence
        (:then_0
          (if (not false)
            (Utilities|Array|Add (Variables|SplineList|GetTempSplineList) _returnvalue)))
        (:then_1
          (Spline|SetSplinePoints _returnvalue (LocalVariables|GetSplinePoints) "World" true)
          (return _returnvalue))))))

(fn MakeBorderCollisionfromSpline
  (params (Spline :type Spline%20Component%20Object%20Reference) (CollisionDirection :type Boolean) (CoordinateSpace :type ESplineCoordinateSpace%20Enum :default "Local") (Debug :type Boolean) (Thick :type Boolean))
  (bind _returnvalue (Transformation|GetActorLocation))
  (bind _returnvalue_2 (Transformation|GetActorRotation))
  (bind _thick (Variables|GetThick))
  (bind _returnvalue_3 (Math|Float|SelectFloat 512.0 256.0 _thick))
  (bind _returnvalue_4 (Spline|GetNumberOfSplinePoints Spline))
  (bind _returnvalue_5 (- _returnvalue_4 1))
  (bind _collisiondirection (Variables|GetCollisionDirection))
  (bind _returnvalue_6 (Math|Float|SelectFloat -1.0 1.0 _collisiondirection))
  (bind _returnvalue_7 (Math|Rotator|InvertRotator _returnvalue_2))
  (bind _returnvalue_8 (Transformation|GetActorRotation))
  (bind _returnvalue_9 (Math|Rotator|InvertRotator _returnvalue_8))
  (bind _coordinatespace (Variables|GetCoordinateSpace))
  (bind _returnvalue_10 (Utilities|Select :Index _coordinatespace :Local "0, 0, 0" :World _returnvalue))
  (bind _coordinatespace_2 (Variables|GetCoordinateSpace))
  (bind _coordinatespace_3 (Variables|GetCoordinateSpace))
  (bind _thick_2 (Variables|GetThick))
  (bind _debug (Variables|GetDebug))
  ;[1 1 1] CollisionMeshes Makes a border around the defined spline as collision for the traces
  (for _index (range _returnvalue_4)
    (bind (_next__1 _next _current _previous) (Tools|GetCurrentNextPreviousSplinePoint Spline _index))
    (if (> (Math|Vector|VectorLength (- (Spline|GetLocationatSplinePoint Spline (Math|Integer|SelectInt 0 (+ _index 1) (== _index _returnvalue_5)) CoordinateSpace) (Spline|GetLocationatSplinePoint Spline _index CoordinateSpace))) 0)
      (if (Variables|GetThick)
        (Utilities|FlowControl|Sequence
          (:then_0
            (bind _returnvalue_11 (AddComponent|NODE_AddStaticMeshComponent-49 :RelativeTransform (Math|Transform|MakeTransform (Utilities|Select :Index _coordinatespace_3 :Local (Math|Vector|MakeVector (.x (- (+ (Spline|GetLocationatSplinePoint Spline _index CoordinateSpace) (Math|Vector|SelectVector "0, 0, 0" (* (Math|Vector|Normalize (- (Spline|GetLocationatSplinePoint Spline (Math|Integer|SelectInt 0 (+ _index 1) (== _index _returnvalue_5)) CoordinateSpace) (Spline|GetLocationatSplinePoint Spline _index CoordinateSpace))) (/ (Math|Vector|VectorLength (- (Spline|GetLocationatSplinePoint Spline (Math|Integer|SelectInt 0 (+ _index 1) (== _index _returnvalue_5)) CoordinateSpace) (Spline|GetLocationatSplinePoint Spline _index CoordinateSpace))) 2.0)) _thick_2)) _returnvalue_10)) (.y (- (+ (Spline|GetLocationatSplinePoint Spline _index CoordinateSpace) (Math|Vector|SelectVector "0, 0, 0" (* (Math|Vector|Normalize (- (Spline|GetLocationatSplinePoint Spline (Math|Integer|SelectInt 0 (+ _index 1) (== _index _returnvalue_5)) CoordinateSpace) (Spline|GetLocationatSplinePoint Spline _index CoordinateSpace))) (/ (Math|Vector|VectorLength (- (Spline|GetLocationatSplinePoint Spline (Math|Integer|SelectInt 0 (+ _index 1) (== _index _returnvalue_5)) CoordinateSpace) (Spline|GetLocationatSplinePoint Spline _index CoordinateSpace))) 2.0)) _thick_2)) _returnvalue_10))) :World (Math|Vector|RotateVector (Math|Vector|MakeVector (.x (- (+ (Spline|GetLocationatSplinePoint Spline _index CoordinateSpace) (Math|Vector|SelectVector "0, 0, 0" (* (Math|Vector|Normalize (- (Spline|GetLocationatSplinePoint Spline (Math|Integer|SelectInt 0 (+ _index 1) (== _index _returnvalue_5)) CoordinateSpace) (Spline|GetLocationatSplinePoint Spline _index CoordinateSpace))) (/ (Math|Vector|VectorLength (- (Spline|GetLocationatSplinePoint Spline (Math|Integer|SelectInt 0 (+ _index 1) (== _index _returnvalue_5)) CoordinateSpace) (Spline|GetLocationatSplinePoint Spline _index CoordinateSpace))) 2.0)) _thick_2)) _returnvalue_10)) (.y (- (+ (Spline|GetLocationatSplinePoint Spline _index CoordinateSpace) (Math|Vector|SelectVector "0, 0, 0" (* (Math|Vector|Normalize (- (Spline|GetLocationatSplinePoint Spline (Math|Integer|SelectInt 0 (+ _index 1) (== _index _returnvalue_5)) CoordinateSpace) (Spline|GetLocationatSplinePoint Spline _index CoordinateSpace))) (/ (Math|Vector|VectorLength (- (Spline|GetLocationatSplinePoint Spline (Math|Integer|SelectInt 0 (+ _index 1) (== _index _returnvalue_5)) CoordinateSpace) (Spline|GetLocationatSplinePoint Spline _index CoordinateSpace))) 2.0)) _thick_2)) _returnvalue_10))) _returnvalue_9)) (Math|Rotator|CombineRotators (Math|Rotator|MakeRotationfromAxes (Utilities|Select :Index _coordinatespace_2 :Local (- (Spline|GetLocationatSplinePoint Spline (Math|Integer|SelectInt 0 (+ _index 1) (== _index _returnvalue_5)) CoordinateSpace) (Spline|GetLocationatSplinePoint Spline _index CoordinateSpace)) :World (Math|Vector|RotateVector (- (Spline|GetLocationatSplinePoint Spline (Math|Integer|SelectInt 0 (+ _index 1) (== _index _returnvalue_5)) CoordinateSpace) (Spline|GetLocationatSplinePoint Spline _index CoordinateSpace)) _returnvalue_7)) (Math|Vector|CrossProduct (Utilities|Select :Index _coordinatespace_2 :Local (- (Spline|GetLocationatSplinePoint Spline (Math|Integer|SelectInt 0 (+ _index 1) (== _index _returnvalue_5)) CoordinateSpace) (Spline|GetLocationatSplinePoint Spline _index CoordinateSpace)) :World (Math|Vector|RotateVector (- (Spline|GetLocationatSplinePoint Spline (Math|Integer|SelectInt 0 (+ _index 1) (== _index _returnvalue_5)) CoordinateSpace) (Spline|GetLocationatSplinePoint Spline _index CoordinateSpace)) _returnvalue_7)) "0, 0,1.000000") "0, 0,1.000000") "0.000000,-90.000000,0.000000") (Math|Vector|MakeVector _returnvalue_6 (/ (Math|Vector|VectorLength (- (Spline|GetLocationatSplinePoint Spline (Math|Integer|SelectInt 0 (+ _index 1) (== _index _returnvalue_5)) CoordinateSpace) (Spline|GetLocationatSplinePoint Spline _index CoordinateSpace))) _returnvalue_3) 1.0))))
            (Variables|/Script/Engine.ActorComponent|SetComponentTags (Utilities|Array|MakeArray "border") _returnvalue_11)
            (if (not _debug)
              (Utilities|Array|Add (Variables|MeshLists|GetBorderCollisionList) _returnvalue_11)))
          (:then_1
            (if (not (Utilities|IsOuterCorner _next _current _previous))
              (bind _returnvalue_12 (AddComponent|NODE_AddStaticMeshComponent-50 :RelativeTransform (Math|Transform|MakeTransform (Utilities|Select :Index (Variables|GetCoordinateSpace) :Local (Spline|GetLocationatSplinePoint Spline _index CoordinateSpace) :World (Math|Vector|RotateVector (Spline|GetLocationatSplinePoint Spline _index CoordinateSpace) (Math|Rotator|InvertRotator (Transformation|GetActorRotation)))) (Math|Rotator|CombineRotators (Math|Rotator|MakeRotationfromAxes (Utilities|Select :Index _coordinatespace_2 :Local (- (Spline|GetLocationatSplinePoint Spline (Math|Integer|SelectInt 0 (+ _index 1) (== _index _returnvalue_5)) CoordinateSpace) (Spline|GetLocationatSplinePoint Spline _index CoordinateSpace)) :World (Math|Vector|RotateVector (- (Spline|GetLocationatSplinePoint Spline (Math|Integer|SelectInt 0 (+ _index 1) (== _index _returnvalue_5)) CoordinateSpace) (Spline|GetLocationatSplinePoint Spline _index CoordinateSpace)) _returnvalue_7)) (Math|Vector|CrossProduct (Utilities|Select :Index _coordinatespace_2 :Local (- (Spline|GetLocationatSplinePoint Spline (Math|Integer|SelectInt 0 (+ _index 1) (== _index _returnvalue_5)) CoordinateSpace) (Spline|GetLocationatSplinePoint Spline _index CoordinateSpace)) :World (Math|Vector|RotateVector (- (Spline|GetLocationatSplinePoint Spline (Math|Integer|SelectInt 0 (+ _index 1) (== _index _returnvalue_5)) CoordinateSpace) (Spline|GetLocationatSplinePoint Spline _index CoordinateSpace)) _returnvalue_7)) "0, 0,1.000000") "0, 0,1.000000") "0.000000,180.000000,0.000000"))))
              (Variables|/Script/Engine.ActorComponent|SetComponentTags (Utilities|Array|MakeArray "border") _returnvalue_12)
              (if (not _debug)
                (Utilities|Array|Add (Variables|MeshLists|GetBorderCollisionList) _returnvalue_12)))))
        (else
          (Development|PrintString (Utilities|String|ToString(Transform) (Math|Transform|MakeTransform (Utilities|Select :Index _coordinatespace_3 :Local (Math|Vector|MakeVector (.x (- (+ (Spline|GetLocationatSplinePoint Spline _index CoordinateSpace) (Math|Vector|SelectVector "0, 0, 0" (* (Math|Vector|Normalize (- (Spline|GetLocationatSplinePoint Spline (Math|Integer|SelectInt 0 (+ _index 1) (== _index _returnvalue_5)) CoordinateSpace) (Spline|GetLocationatSplinePoint Spline _index CoordinateSpace))) (/ (Math|Vector|VectorLength (- (Spline|GetLocationatSplinePoint Spline (Math|Integer|SelectInt 0 (+ _index 1) (== _index _returnvalue_5)) CoordinateSpace) (Spline|GetLocationatSplinePoint Spline _index CoordinateSpace))) 2.0)) _thick_2)) _returnvalue_10)) (.y (- (+ (Spline|GetLocationatSplinePoint Spline _index CoordinateSpace) (Math|Vector|SelectVector "0, 0, 0" (* (Math|Vector|Normalize (- (Spline|GetLocationatSplinePoint Spline (Math|Integer|SelectInt 0 (+ _index 1) (== _index _returnvalue_5)) CoordinateSpace) (Spline|GetLocationatSplinePoint Spline _index CoordinateSpace))) (/ (Math|Vector|VectorLength (- (Spline|GetLocationatSplinePoint Spline (Math|Integer|SelectInt 0 (+ _index 1) (== _index _returnvalue_5)) CoordinateSpace) (Spline|GetLocationatSplinePoint Spline _index CoordinateSpace))) 2.0)) _thick_2)) _returnvalue_10))) :World (Math|Vector|RotateVector (Math|Vector|MakeVector (.x (- (+ (Spline|GetLocationatSplinePoint Spline _index CoordinateSpace) (Math|Vector|SelectVector "0, 0, 0" (* (Math|Vector|Normalize (- (Spline|GetLocationatSplinePoint Spline (Math|Integer|SelectInt 0 (+ _index 1) (== _index _returnvalue_5)) CoordinateSpace) (Spline|GetLocationatSplinePoint Spline _index CoordinateSpace))) (/ (Math|Vector|VectorLength (- (Spline|GetLocationatSplinePoint Spline (Math|Integer|SelectInt 0 (+ _index 1) (== _index _returnvalue_5)) CoordinateSpace) (Spline|GetLocationatSplinePoint Spline _index CoordinateSpace))) 2.0)) _thick_2)) _returnvalue_10)) (.y (- (+ (Spline|GetLocationatSplinePoint Spline _index CoordinateSpace) (Math|Vector|SelectVector "0, 0, 0" (* (Math|Vector|Normalize (- (Spline|GetLocationatSplinePoint Spline (Math|Integer|SelectInt 0 (+ _index 1) (== _index _returnvalue_5)) CoordinateSpace) (Spline|GetLocationatSplinePoint Spline _index CoordinateSpace))) (/ (Math|Vector|VectorLength (- (Spline|GetLocationatSplinePoint Spline (Math|Integer|SelectInt 0 (+ _index 1) (== _index _returnvalue_5)) CoordinateSpace) (Spline|GetLocationatSplinePoint Spline _index CoordinateSpace))) 2.0)) _thick_2)) _returnvalue_10))) _returnvalue_9)) (Math|Rotator|CombineRotators (Math|Rotator|MakeRotationfromAxes (Utilities|Select :Index _coordinatespace_2 :Local (- (Spline|GetLocationatSplinePoint Spline (Math|Integer|SelectInt 0 (+ _index 1) (== _index _returnvalue_5)) CoordinateSpace) (Spline|GetLocationatSplinePoint Spline _index CoordinateSpace)) :World (Math|Vector|RotateVector (- (Spline|GetLocationatSplinePoint Spline (Math|Integer|SelectInt 0 (+ _index 1) (== _index _returnvalue_5)) CoordinateSpace) (Spline|GetLocationatSplinePoint Spline _index CoordinateSpace)) _returnvalue_7)) (Math|Vector|CrossProduct (Utilities|Select :Index _coordinatespace_2 :Local (- (Spline|GetLocationatSplinePoint Spline (Math|Integer|SelectInt 0 (+ _index 1) (== _index _returnvalue_5)) CoordinateSpace) (Spline|GetLocationatSplinePoint Spline _index CoordinateSpace)) :World (Math|Vector|RotateVector (- (Spline|GetLocationatSplinePoint Spline (Math|Integer|SelectInt 0 (+ _index 1) (== _index _returnvalue_5)) CoordinateSpace) (Spline|GetLocationatSplinePoint Spline _index CoordinateSpace)) _returnvalue_7)) "0, 0,1.000000") "0, 0,1.000000") "0.000000,-90.000000,0.000000") (Math|Vector|MakeVector _returnvalue_6 (/ (Math|Vector|VectorLength (- (Spline|GetLocationatSplinePoint Spline (Math|Integer|SelectInt 0 (+ _index 1) (== _index _returnvalue_5)) CoordinateSpace) (Spline|GetLocationatSplinePoint Spline _index CoordinateSpace))) _returnvalue_3) 1.0))))
          (bind _returnvalue_13 (AddComponent|NODE_AddStaticMeshComponent-15 :RelativeTransform (Math|Transform|MakeTransform (Utilities|Select :Index _coordinatespace_3 :Local (Math|Vector|MakeVector (.x (- (+ (Spline|GetLocationatSplinePoint Spline _index CoordinateSpace) (Math|Vector|SelectVector "0, 0, 0" (* (Math|Vector|Normalize (- (Spline|GetLocationatSplinePoint Spline (Math|Integer|SelectInt 0 (+ _index 1) (== _index _returnvalue_5)) CoordinateSpace) (Spline|GetLocationatSplinePoint Spline _index CoordinateSpace))) (/ (Math|Vector|VectorLength (- (Spline|GetLocationatSplinePoint Spline (Math|Integer|SelectInt 0 (+ _index 1) (== _index _returnvalue_5)) CoordinateSpace) (Spline|GetLocationatSplinePoint Spline _index CoordinateSpace))) 2.0)) _thick_2)) _returnvalue_10)) (.y (- (+ (Spline|GetLocationatSplinePoint Spline _index CoordinateSpace) (Math|Vector|SelectVector "0, 0, 0" (* (Math|Vector|Normalize (- (Spline|GetLocationatSplinePoint Spline (Math|Integer|SelectInt 0 (+ _index 1) (== _index _returnvalue_5)) CoordinateSpace) (Spline|GetLocationatSplinePoint Spline _index CoordinateSpace))) (/ (Math|Vector|VectorLength (- (Spline|GetLocationatSplinePoint Spline (Math|Integer|SelectInt 0 (+ _index 1) (== _index _returnvalue_5)) CoordinateSpace) (Spline|GetLocationatSplinePoint Spline _index CoordinateSpace))) 2.0)) _thick_2)) _returnvalue_10))) :World (Math|Vector|RotateVector (Math|Vector|MakeVector (.x (- (+ (Spline|GetLocationatSplinePoint Spline _index CoordinateSpace) (Math|Vector|SelectVector "0, 0, 0" (* (Math|Vector|Normalize (- (Spline|GetLocationatSplinePoint Spline (Math|Integer|SelectInt 0 (+ _index 1) (== _index _returnvalue_5)) CoordinateSpace) (Spline|GetLocationatSplinePoint Spline _index CoordinateSpace))) (/ (Math|Vector|VectorLength (- (Spline|GetLocationatSplinePoint Spline (Math|Integer|SelectInt 0 (+ _index 1) (== _index _returnvalue_5)) CoordinateSpace) (Spline|GetLocationatSplinePoint Spline _index CoordinateSpace))) 2.0)) _thick_2)) _returnvalue_10)) (.y (- (+ (Spline|GetLocationatSplinePoint Spline _index CoordinateSpace) (Math|Vector|SelectVector "0, 0, 0" (* (Math|Vector|Normalize (- (Spline|GetLocationatSplinePoint Spline (Math|Integer|SelectInt 0 (+ _index 1) (== _index _returnvalue_5)) CoordinateSpace) (Spline|GetLocationatSplinePoint Spline _index CoordinateSpace))) (/ (Math|Vector|VectorLength (- (Spline|GetLocationatSplinePoint Spline (Math|Integer|SelectInt 0 (+ _index 1) (== _index _returnvalue_5)) CoordinateSpace) (Spline|GetLocationatSplinePoint Spline _index CoordinateSpace))) 2.0)) _thick_2)) _returnvalue_10))) _returnvalue_9)) (Math|Rotator|CombineRotators (Math|Rotator|MakeRotationfromAxes (Utilities|Select :Index _coordinatespace_2 :Local (- (Spline|GetLocationatSplinePoint Spline (Math|Integer|SelectInt 0 (+ _index 1) (== _index _returnvalue_5)) CoordinateSpace) (Spline|GetLocationatSplinePoint Spline _index CoordinateSpace)) :World (Math|Vector|RotateVector (- (Spline|GetLocationatSplinePoint Spline (Math|Integer|SelectInt 0 (+ _index 1) (== _index _returnvalue_5)) CoordinateSpace) (Spline|GetLocationatSplinePoint Spline _index CoordinateSpace)) _returnvalue_7)) (Math|Vector|CrossProduct (Utilities|Select :Index _coordinatespace_2 :Local (- (Spline|GetLocationatSplinePoint Spline (Math|Integer|SelectInt 0 (+ _index 1) (== _index _returnvalue_5)) CoordinateSpace) (Spline|GetLocationatSplinePoint Spline _index CoordinateSpace)) :World (Math|Vector|RotateVector (- (Spline|GetLocationatSplinePoint Spline (Math|Integer|SelectInt 0 (+ _index 1) (== _index _returnvalue_5)) CoordinateSpace) (Spline|GetLocationatSplinePoint Spline _index CoordinateSpace)) _returnvalue_7)) "0, 0,1.000000") "0, 0,1.000000") "0.000000,-90.000000,0.000000") (Math|Vector|MakeVector _returnvalue_6 (/ (Math|Vector|VectorLength (- (Spline|GetLocationatSplinePoint Spline (Math|Integer|SelectInt 0 (+ _index 1) (== _index _returnvalue_5)) CoordinateSpace) (Spline|GetLocationatSplinePoint Spline _index CoordinateSpace))) _returnvalue_3) 1.0))))
          (if (Variables|GetCollisionDirection)
            (Transformation|AddLocalOffset _returnvalue_13 "-20.000000, 0, 0" false false))
          (Variables|/Script/Engine.ActorComponent|SetComponentTags (Utilities|Array|MakeArray "border") _returnvalue_13)
          (if (not _debug)
            (Utilities|Array|Add (Variables|MeshLists|GetBorderCollisionList) _returnvalue_13)))))))

(fn MakeBasementFloor
  (params (IterationStartOffset :type Integer) (Optimize :type Boolean :default true) (Top :type Boolean) (Proc%20AssetList :type Proc%20Asset%20List%20Object%20Reference) (InputSpline :type Spline%20Component%20Object%20Reference))
  (bind _spline (Variables|Default|GetSpline))
  (bind _returnvalue (Spline|GetNumberOfSplinePoints _spline))
  (bind _returnvalue_2 (- _returnvalue 1))
  (bind _minnumberspawnedarray (LocalVariables|GetMinNumberSpawnedArray))
  (bind _indexofminvalue (Math|Integer|MinOfIntArray _minnumberspawnedarray))
  (bind _maxoptimizationiterations (Variables|Default|GetMaxOptimizationIterations))
  (bind _returnvalue_3 (Math|Integer|Min(Integer) _maxoptimizationiterations _returnvalue_2))
  (bind _returnvalue_4 (Math|Integer|Max(Integer) _returnvalue_3 1))
  (bind _returnvalue_5 (/ _returnvalue_2 _returnvalue_4))
  (bind _returnvalue_6 (* _returnvalue_5 _indexofminvalue))
  (bind _fillcollisionlist (Variables|MeshLists|GetFillCollisionList))
  (bind _bordercollisionlist (Variables|MeshLists|GetBorderCollisionList))
  (bind _tempcollisionlist (Variables|MeshLists|GetTempCollisionList))
  (bind _tempsplinelist (Variables|SplineList|GetTempSplineList))
  (bind _tempcount (LocalVariables|Gettempcount))
  (bind _insidesplines (LocalVariables|GetInsideSplines))
  (bind _tempcollisionlist_2 (Variables|MeshLists|GetTempCollisionList))
  (LocalVariables|SetCurrentSpline InputSpline)
  (Utilities|FlowControl|Sequence
    (:then_0
      ;[1 1 1] Adjust splie for cliff
      (Variables|Default|SetIs256Grid false)
      (for _array_element (Variables|C.%29ConnectingComplexPOIs|GetConnectedActors)
        (if (Utilities|Enum|Equal(Enum) (Variables|Default|GetMode _array_element) "NewEnumerator6")
          (Variables|Default|SetIs256Grid true)
          (bind _outputspline (CallFunction|InsetSplineforCliffs (LocalVariables|GetCurrentSpline) (Variables|Default|GetSpline _array_element)))
          (LocalVariables|SetCurrentSpline _outputspline))))
    (:then_1
      (Utilities|FlowControl|Sequence
        (:then_0
          (Utilities|FlowControl|Sequence
            (:then_0
              ;[1 1 1] Create temp colission for traces
              (Utilities|Array|Clear (Variables|MeshLists|GetBorderCollisionList))
              (Utilities|FlowControl|Sequence
                (:then_0
                  (CreateCollisionComponents|MakeBorderCollisionFromSpline (LocalVariables|GetCurrentSpline)))
                (:then_1
                  (for _array_element_2 (Variables|C.%29ConnectingComplexPOIs|GetConnectedActors)
                    (if (not (Utilities|ExcludeConnectedActors (Variables|Default|GetMode _array_element_2)))
                      (CreateCollisionComponents|MakeFillCollisionfromSpline (Variables|Default|GetSpline _array_element_2) "World"))))))
            (:then_1
              ;[1 1 1] List of Assets used and sorted by size
              (bind _assetindexlist (Tools|GetAssetsOfType (Variables|/Game/ArtTools/Procedural/Foundation/Proc_AssetList.Proc_AssetList_C|GetAssetList (Variables|GetProc%20AssetList)) "NewEnumerator6"))
              (LocalVariables|SetFloorIndexList _assetindexlist))))
        (:then_1
          (Utilities|FlowControl|Sequence
            (:then_0
              (if (Variables|GetOptimize)
                (Utilities|Array|Clear (LocalVariables|GetMinNumberSpawnedArray))
                (Development|PrintString (Utilities|String|Append "MaxLoops" (Utilities|String|ToString(Integer) _returnvalue_4)) false false)
                (for _index (range (+ _returnvalue_4 1))
                  (LocalVariables|Settempcount 0)
                  (bind _numberofspawnedcomponents (CreateComponents|MakeBasementFloorComponents (LocalVariables|GetCurrentSpline) (* _returnvalue_5 _index) (LocalVariables|GetFloorIndexList) false (Variables|GetProc%20AssetList) false))
                  (LocalVariables|Settempcount (+ (LocalVariables|Gettempcount) _numberofspawnedcomponents))
                  (Utilities|FlowControl|Sequence
                    (:then_0
                      (for _array_element_3 (Variables|C.%29ConnectingComplexPOIs|GetConnectedActors)
                        (Utilities|IsValid _array_element_3
                          (:Is%20Valid
                            (if (not (Utilities|ExcludeConnectedActors (Variables|Default|GetMode _array_element_3)))
                              (bind _numberofspawnedcomponents_2 (CreateComponents|MakeBasementFloorComponents (Variables|Default|GetSpline _array_element_3) 0 (LocalVariables|GetFloorIndexList) false (Variables|GetProc%20AssetList) false true "World"))
                              (LocalVariables|Settempcount (+ (LocalVariables|Gettempcount) _numberofspawnedcomponents_2)))))))
                    (:then_1
                      (bind _numberofcomponentsspawned (CreateComponents|MakeBasementFloorCornerMissing (LocalVariables|GetCurrentSpline) (LocalVariables|GetFloorIndexList) (Variables|GetProc%20AssetList) false "Local" false))
                      (LocalVariables|Settempcount (+ (LocalVariables|Gettempcount) _numberofcomponentsspawned)))
                    (:then_2
                      (for _array_element_4 (Variables|C.%29ConnectingComplexPOIs|GetConnectedActors)
                        (Utilities|IsValid _array_element_4
                          (:Is%20Valid
                            (if (not (Utilities|ExcludeConnectedActors (Variables|Default|GetMode _array_element_4)))
                              (bind _numberofcomponentsspawned_2 (CreateComponents|MakeBasementFloorCornerMissing (Variables|Default|GetSpline _array_element_4) (LocalVariables|GetFloorIndexList) (Variables|GetProc%20AssetList) true "World" false))
                              (LocalVariables|Settempcount (+ (LocalVariables|Gettempcount) _numberofcomponentsspawned_2)))))))
                    (:then_3
                      (Development|PrintString (Utilities|String|Append (Utilities|String|Append "OptimizationOFFSET = " (Utilities|String|ToString(Integer) (* _returnvalue_5 _index))) (Utilities|String|Append "   AND Spwned Components =  " (Utilities|String|ToString(Integer) _tempcount))) false false)
                      (Utilities|Array|Add (LocalVariables|GetMinNumberSpawnedArray) _tempcount)
                      (for _array_element_5 _tempcollisionlist_2
                        (Components|DestroyComponent _array_element_5))
                      (Utilities|Array|Clear _tempcollisionlist_2))))
                (Development|PrintString (Utilities|String|Append "Usin OPTIZATION = = " (Utilities|String|ToString(Integer) _returnvalue_6)) false false)
                (bind _numberofspawnedcomponents_3 (CreateComponents|MakeBasementFloorComponents (LocalVariables|GetCurrentSpline) _returnvalue_6 (LocalVariables|GetFloorIndexList) false (Variables|GetProc%20AssetList) true false "Local" (Variables|GetTop)))
                (else
                  (CreateComponents|MakeBasementFloorComponents (LocalVariables|GetCurrentSpline) 0 (LocalVariables|GetFloorIndexList) false (Variables|GetProc%20AssetList) true false "Local" (Variables|GetTop))))
              (Utilities|FlowControl|Sequence
                (:then_0
                  (for _array_element_6 (Variables|C.%29ConnectingComplexPOIs|GetConnectedActors)
                    (Utilities|IsValid _array_element_6
                      (:Is%20Valid
                        (if (not (Utilities|ExcludeConnectedActors (Variables|Default|GetMode _array_element_6)))
                          (bind _numberofspawnedcomponents_4 (CreateComponents|MakeBasementFloorComponents (Variables|Default|GetSpline _array_element_6) 0 (LocalVariables|GetFloorIndexList) false (Variables|GetProc%20AssetList) true true "World" (Variables|GetTop))))))))
                (:then_1
                  (bind _numberofcomponentsspawned_3 (CreateComponents|MakeBasementFloorCornerMissing (LocalVariables|GetCurrentSpline) (LocalVariables|GetFloorIndexList) (Variables|GetProc%20AssetList) false "Local" true false (Variables|GetTop))))
                (:then_2
                  (for _array_element_7 (Variables|C.%29ConnectingComplexPOIs|GetConnectedActors)
                    (Utilities|IsValid _array_element_7
                      (:Is%20Valid
                        (if (not (Utilities|ExcludeConnectedActors (Variables|Default|GetMode _array_element_7)))
                          (bind _numberofcomponentsspawned_4 (CreateComponents|MakeBasementFloorCornerMissing (Variables|Default|GetSpline _array_element_7) (LocalVariables|GetFloorIndexList) (Variables|GetProc%20AssetList) true "World" true false (Variables|GetTop))))))))))
            (:then_3
              (for _array_element_8 _tempcollisionlist
                (Components|DestroyComponent _array_element_8))
              (Utilities|Array|Clear _tempcollisionlist))))
        (:then_2
          ;[1 1 1] Delete all temp components
          (Utilities|FlowControl|Sequence
            (:then_0
              (if (not false)
                (for _array_element_9 _bordercollisionlist
                  (Components|DestroyComponent _array_element_9))
                (Utilities|Array|Clear _bordercollisionlist)))
            (:then_1
              (for _array_element_10 _tempsplinelist
                (Components|DestroyComponent _array_element_10))
              (Utilities|Array|Clear _tempsplinelist))
            (:then_2
              (if (not false)
                (for _array_element_11 _fillcollisionlist
                  (Components|DestroyComponent _array_element_11))
                (Utilities|Array|Clear _fillcollisionlist)))
            (:then_3
              (for _array_element_12 _insidesplines
                (Components|DestroyComponent _array_element_12))
              (Utilities|Array|Clear _insidesplines))
            (:then_4
              (Cleanup|DeleteAllTempSplines))))
        (:then_4
          (return))))))

(fn MakeBasementFloorComponents
  (params (Spline :type Spline%20Component%20Object%20Reference) (OffsetIndex :type Integer) (AssetIndexList :type Array%20of%20Integers) (Debug :type Boolean) (Proc%20Asset%20List :type Proc%20Asset%20List%20Object%20Reference) (SpawnComponents :type Boolean :default true) (Reverse :type Boolean) (CoordinateSpace :type ESplineCoordinateSpace%20Enum :default "Local") (Top :type Boolean))
  (bind (_staticmesh_2_d8df7d7b4556a6c1fde39bb633b7e288 _type_6_b4cd9ff0498f54172801239a22e7a5ca _prepivot_29_094082c34c836f4f1ab3deb660eef580 _forwarddirection_32_09c0a6674706473836365795dce83b27 _sizex_11_dfccfeb9407ac738b8ec5aaf3a48e13c) (Utilities|Struct|BreakProcAssetInfo (Utilities|Array|Get(acopy) (Variables|/Game/ArtTools/Procedural/Foundation/Proc_AssetList.Proc_AssetList_C|GetAssetList (Variables|GetProc%20Asset%20List)) (Utilities|Array|Get(acopy) (Variables|GetAssetIndexList)))))
  (bind _assetindexlist (Variables|GetAssetIndexList))
  (bind _output (Utilities|Array|Get(acopy) _assetindexlist))
  (bind _proc_asset_list (Variables|GetProc%20Asset%20List))
  (bind _assetlist (Variables|/Game/ArtTools/Procedural/Foundation/Proc_AssetList.Proc_AssetList_C|GetAssetList _proc_asset_list))
  (bind _output_2 (Utilities|Array|Get(acopy) _assetlist _output))
  (bind (_staticmesh_2_d8df7d7b4556a6c1fde39bb633b7e288_2 _type_6_b4cd9ff0498f54172801239a22e7a5ca_2 _prepivot_29_094082c34c836f4f1ab3deb660eef580_2 _forwarddirection_32_09c0a6674706473836365795dce83b27_2 _sizex_11_dfccfeb9407ac738b8ec5aaf3a48e13c_2) (Utilities|Struct|BreakProcAssetInfo _output_2))
  (bind _tempcollisionlist (Variables|MeshLists|GetTempCollisionList))
  (bind _returnvalue (Utilities|Array|LastIndex _tempcollisionlist))
  (bind _returnvalue_2 (Spline|GetNumberOfSplinePoints Spline))
  (bind _returnvalue_3 (- _returnvalue_2 1))
  (bind _top (Variables|GetTop))
  (bind _numberoffloors (Variables|Default|GetNumberOfFloors))
  (bind _returnvalue_4 (* _numberoffloors 384.0))
  (bind _returnvalue_5 (Math|Float|SelectFloat 0.0 _returnvalue_4 _top))
  (bind _gridsize (LocalVariables|GetGridSize))
  (bind _meshspawn_iteration (Variables|MakeBasementFloor|GetMeshSpawn%20Iteration))
  (bind _returnvalue_6 (* _meshspawn_iteration _gridsize))
  (bind _basement_floor_iteration (Variables|MakeBasementFloor|GetBasement%20Floor%20Iteration))
  (bind _returnvalue_7 (+ _basement_floor_iteration 1))
  (bind _returnvalue_8 (* _returnvalue_7 _gridsize))
  (bind _returnvalue_9 (- _returnvalue_8 _returnvalue_6))
  (bind _returnvalue_10 (/ _returnvalue_9 _sizex_11_dfccfeb9407ac738b8ec5aaf3a48e13c_2))
  (bind _returnvalue_11 (Math|Vector|MakeVector 0.0 0.0 _returnvalue_5))
  (bind _currentindex (LocalVariables|GetCurrentIndex))
  (bind _transformlist (LocalVariables|GetTransformList))
  (bind _output_3 (Utilities|Array|Get(aref) _transformlist _currentindex))
  (bind _scale (.scale _output_3))
  (bind _z (.z _scale))
  (bind _y (.y _scale))
  (bind _reverse (Variables|GetReverse))
  (bind _returnvalue_12 (Math|Float|SelectFloat 1.0 -1.0 _reverse))
  (bind _returnvalue_13 (* _returnvalue_10 _returnvalue_12))
  (bind _returnvalue_14 (Math|Vector|MakeVector _returnvalue_13 _y _z))
  (bind _gridsize_2 (LocalVariables|GetGridSize))
  (bind _basement_floor_iteration_2 (Variables|MakeBasementFloor|GetBasement%20Floor%20Iteration))
  (bind _returnvalue_15 (* _basement_floor_iteration_2 _gridsize_2))
  (bind _returnvalue_16 (Transformation|GetActorTransform))
  (bind _reverse_2 (Variables|GetReverse))
  (bind _returnvalue_17 (Math|Float|SelectFloat 1.0 -1.0 _reverse_2))
  (bind _output_4 (Utilities|Array|Get(acopy) _tempcollisionlist _returnvalue))
  (bind _meshsizex (LocalVariables|GetMeshSizeX))
  (bind _gridsize_3 (LocalVariables|GetGridSize))
  (bind _returnvalue_18 (/ _gridsize_3 _meshsizex))
  (bind _returnvalue_19 (* _returnvalue_17 _returnvalue_18))
  (bind _gridsize_4 (LocalVariables|GetGridSize))
  (bind _returnvalue_20 (/ _gridsize_4 2.0))
  (bind _coordinatespace (Variables|GetCoordinateSpace))
  (bind _coordinatespace_2 (Variables|GetCoordinateSpace))
  (bind _coordinatespace_3 (Variables|GetCoordinateSpace))
  (bind _gridsize_5 (LocalVariables|GetGridSize))
  (bind _meshsizex_2 (LocalVariables|GetMeshSizeX))
  (Utilities|FlowControl|Sequence
    (:then_0
      (if (or (Utilities|Enum|Equal(Enum) (Variables|Default|GetMode) "NewEnumerator5") (Variables|Default|GetIs256Grid))
        (LocalVariables|SetGridSize 256.0)
        (else
          (LocalVariables|SetGridSize 512.0))))
    (:then_1
      (Utilities|FlowControl|Sequence
        (:then_0
          (for _index (range _returnvalue_2)
            (Variables|MakeBasementFloor|SetcurrentLength 0.0)
            (LocalVariables|SetMeshSizeX _sizex_11_dfccfeb9407ac738b8ec5aaf3a48e13c)
            (Variables|MakeBasementFloor|SetCurrentFloorMesh 0)
            (LocalVariables|SetCurrentIndex -1)
            (Variables|MakeBasementFloor|SetNewFloor true)
            (for _index_2 (range (+ (- (Math|Float|Round (/ (Math|Vector|VectorLength (- (Utilities|ZeroLocationWorldSpaceZ (Spline|GetLocationatSplinePoint Spline (Math|Integer|SelectInt 0 (+ (Math|Integer|SelectInt (- (+ _index OffsetIndex) _returnvalue_2) (+ _index OffsetIndex) (> (+ _index OffsetIndex) _returnvalue_3)) 1) (== (Math|Integer|SelectInt (- (+ _index OffsetIndex) _returnvalue_2) (+ _index OffsetIndex) (> (+ _index OffsetIndex) _returnvalue_3)) _returnvalue_3)) "World")) (Utilities|ZeroLocationWorldSpaceZ (Spline|GetLocationatSplinePoint Spline (Math|Integer|SelectInt (- (+ _index OffsetIndex) _returnvalue_2) (+ _index OffsetIndex) (> (+ _index OffsetIndex) _returnvalue_3)) "World")))) _gridsize_4)) 1) 1))
              (Variables|MakeBasementFloor|SetBasement%20Floor%20Iteration _index_2)
              (bind _outhits (Collision|MultiLineTraceForObjects :Start (- (+ (Utilities|ZeroLocationWorldSpaceZ (Spline|GetLocationatSplinePoint Spline (Math|Integer|SelectInt (- (+ _index OffsetIndex) _returnvalue_2) (+ _index OffsetIndex) (> (+ _index OffsetIndex) _returnvalue_3)) "World")) (+ (* (* (Math|Vector|Normalize (- (Utilities|ZeroLocationWorldSpaceZ (Spline|GetLocationatSplinePoint Spline (Math|Integer|SelectInt 0 (+ (Math|Integer|SelectInt (- (+ _index OffsetIndex) _returnvalue_2) (+ _index OffsetIndex) (> (+ _index OffsetIndex) _returnvalue_3)) 1) (== (Math|Integer|SelectInt (- (+ _index OffsetIndex) _returnvalue_2) (+ _index OffsetIndex) (> (+ _index OffsetIndex) _returnvalue_3)) _returnvalue_3)) "World")) (Utilities|ZeroLocationWorldSpaceZ (Spline|GetLocationatSplinePoint Spline (Math|Integer|SelectInt (- (+ _index OffsetIndex) _returnvalue_2) (+ _index OffsetIndex) (> (+ _index OffsetIndex) _returnvalue_3)) "World")))) _gridsize_4) _index_2) (* (Math|Vector|Normalize (- (Utilities|ZeroLocationWorldSpaceZ (Spline|GetLocationatSplinePoint Spline (Math|Integer|SelectInt 0 (+ (Math|Integer|SelectInt (- (+ _index OffsetIndex) _returnvalue_2) (+ _index OffsetIndex) (> (+ _index OffsetIndex) _returnvalue_3)) 1) (== (Math|Integer|SelectInt (- (+ _index OffsetIndex) _returnvalue_2) (+ _index OffsetIndex) (> (+ _index OffsetIndex) _returnvalue_3)) _returnvalue_3)) "World")) (Utilities|ZeroLocationWorldSpaceZ (Spline|GetLocationatSplinePoint Spline (Math|Integer|SelectInt (- (+ _index OffsetIndex) _returnvalue_2) (+ _index OffsetIndex) (> (+ _index OffsetIndex) _returnvalue_3)) "World")))) _returnvalue_20))) (* (Math|Vector|SelectVector (Math|Vector|NegateVector (Math|Vector|Normalize (Math|Vector|CrossProduct (- (Utilities|ZeroLocationWorldSpaceZ (Spline|GetLocationatSplinePoint Spline (Math|Integer|SelectInt 0 (+ (Math|Integer|SelectInt (- (+ _index OffsetIndex) _returnvalue_2) (+ _index OffsetIndex) (> (+ _index OffsetIndex) _returnvalue_3)) 1) (== (Math|Integer|SelectInt (- (+ _index OffsetIndex) _returnvalue_2) (+ _index OffsetIndex) (> (+ _index OffsetIndex) _returnvalue_3)) _returnvalue_3)) "World")) (Utilities|ZeroLocationWorldSpaceZ (Spline|GetLocationatSplinePoint Spline (Math|Integer|SelectInt (- (+ _index OffsetIndex) _returnvalue_2) (+ _index OffsetIndex) (> (+ _index OffsetIndex) _returnvalue_3)) "World"))) "0, 0,-1.000000"))) (Math|Vector|Normalize (Math|Vector|CrossProduct (- (Utilities|ZeroLocationWorldSpaceZ (Spline|GetLocationatSplinePoint Spline (Math|Integer|SelectInt 0 (+ (Math|Integer|SelectInt (- (+ _index OffsetIndex) _returnvalue_2) (+ _index OffsetIndex) (> (+ _index OffsetIndex) _returnvalue_3)) 1) (== (Math|Integer|SelectInt (- (+ _index OffsetIndex) _returnvalue_2) (+ _index OffsetIndex) (> (+ _index OffsetIndex) _returnvalue_3)) _returnvalue_3)) "World")) (Utilities|ZeroLocationWorldSpaceZ (Spline|GetLocationatSplinePoint Spline (Math|Integer|SelectInt (- (+ _index OffsetIndex) _returnvalue_2) (+ _index OffsetIndex) (> (+ _index OffsetIndex) _returnvalue_3)) "World"))) "0, 0,-1.000000")) (Variables|GetReverse)) -1.0)) :End (+ (+ (Utilities|ZeroLocationWorldSpaceZ (Spline|GetLocationatSplinePoint Spline (Math|Integer|SelectInt (- (+ _index OffsetIndex) _returnvalue_2) (+ _index OffsetIndex) (> (+ _index OffsetIndex) _returnvalue_3)) "World")) (+ (* (* (Math|Vector|Normalize (- (Utilities|ZeroLocationWorldSpaceZ (Spline|GetLocationatSplinePoint Spline (Math|Integer|SelectInt 0 (+ (Math|Integer|SelectInt (- (+ _index OffsetIndex) _returnvalue_2) (+ _index OffsetIndex) (> (+ _index OffsetIndex) _returnvalue_3)) 1) (== (Math|Integer|SelectInt (- (+ _index OffsetIndex) _returnvalue_2) (+ _index OffsetIndex) (> (+ _index OffsetIndex) _returnvalue_3)) _returnvalue_3)) "World")) (Utilities|ZeroLocationWorldSpaceZ (Spline|GetLocationatSplinePoint Spline (Math|Integer|SelectInt (- (+ _index OffsetIndex) _returnvalue_2) (+ _index OffsetIndex) (> (+ _index OffsetIndex) _returnvalue_3)) "World")))) _gridsize_4) _index_2) (* (Math|Vector|Normalize (- (Utilities|ZeroLocationWorldSpaceZ (Spline|GetLocationatSplinePoint Spline (Math|Integer|SelectInt 0 (+ (Math|Integer|SelectInt (- (+ _index OffsetIndex) _returnvalue_2) (+ _index OffsetIndex) (> (+ _index OffsetIndex) _returnvalue_3)) 1) (== (Math|Integer|SelectInt (- (+ _index OffsetIndex) _returnvalue_2) (+ _index OffsetIndex) (> (+ _index OffsetIndex) _returnvalue_3)) _returnvalue_3)) "World")) (Utilities|ZeroLocationWorldSpaceZ (Spline|GetLocationatSplinePoint Spline (Math|Integer|SelectInt (- (+ _index OffsetIndex) _returnvalue_2) (+ _index OffsetIndex) (> (+ _index OffsetIndex) _returnvalue_3)) "World")))) _returnvalue_20))) (* (Math|Vector|SelectVector (Math|Vector|NegateVector (Math|Vector|Normalize (Math|Vector|CrossProduct (- (Utilities|ZeroLocationWorldSpaceZ (Spline|GetLocationatSplinePoint Spline (Math|Integer|SelectInt 0 (+ (Math|Integer|SelectInt (- (+ _index OffsetIndex) _returnvalue_2) (+ _index OffsetIndex) (> (+ _index OffsetIndex) _returnvalue_3)) 1) (== (Math|Integer|SelectInt (- (+ _index OffsetIndex) _returnvalue_2) (+ _index OffsetIndex) (> (+ _index OffsetIndex) _returnvalue_3)) _returnvalue_3)) "World")) (Utilities|ZeroLocationWorldSpaceZ (Spline|GetLocationatSplinePoint Spline (Math|Integer|SelectInt (- (+ _index OffsetIndex) _returnvalue_2) (+ _index OffsetIndex) (> (+ _index OffsetIndex) _returnvalue_3)) "World"))) "0, 0,-1.000000"))) (Math|Vector|Normalize (Math|Vector|CrossProduct (- (Utilities|ZeroLocationWorldSpaceZ (Spline|GetLocationatSplinePoint Spline (Math|Integer|SelectInt 0 (+ (Math|Integer|SelectInt (- (+ _index OffsetIndex) _returnvalue_2) (+ _index OffsetIndex) (> (+ _index OffsetIndex) _returnvalue_3)) 1) (== (Math|Integer|SelectInt (- (+ _index OffsetIndex) _returnvalue_2) (+ _index OffsetIndex) (> (+ _index OffsetIndex) _returnvalue_3)) _returnvalue_3)) "World")) (Utilities|ZeroLocationWorldSpaceZ (Spline|GetLocationatSplinePoint Spline (Math|Integer|SelectInt (- (+ _index OffsetIndex) _returnvalue_2) (+ _index OffsetIndex) (> (+ _index OffsetIndex) _returnvalue_3)) "World"))) "0, 0,-1.000000")) (Variables|GetReverse)) 4e+04)) :ObjectTypes (Utilities|Array|MakeArray "ObjectTypeQuery3") :bTraceComplex false :DrawDebugType (select (and (== _index (Math|Integer|MakeLiteralInt (Variables|Debug|GetTest))) (Variables|GetDebug)) "ForDuration" "None") :bIgnoreSelf false :TraceColor "(R=0.000000,G=1.000000,B=0.384102,A=1.000000)" :TraceHitColor "(R=0.000000,G=1.000000,B=0.000000,A=1.000000)" :DrawTime 5.0))
              (bind _hitresoultoutput (Tools|ClosestHitResult _outhits (+ (Utilities|ZeroLocationWorldSpaceZ (Spline|GetLocationatSplinePoint Spline (Math|Integer|SelectInt (- (+ _index OffsetIndex) _returnvalue_2) (+ _index OffsetIndex) (> (+ _index OffsetIndex) _returnvalue_3)) "World")) (+ (* (* (Math|Vector|Normalize (- (Utilities|ZeroLocationWorldSpaceZ (Spline|GetLocationatSplinePoint Spline (Math|Integer|SelectInt 0 (+ (Math|Integer|SelectInt (- (+ _index OffsetIndex) _returnvalue_2) (+ _index OffsetIndex) (> (+ _index OffsetIndex) _returnvalue_3)) 1) (== (Math|Integer|SelectInt (- (+ _index OffsetIndex) _returnvalue_2) (+ _index OffsetIndex) (> (+ _index OffsetIndex) _returnvalue_3)) _returnvalue_3)) "World")) (Utilities|ZeroLocationWorldSpaceZ (Spline|GetLocationatSplinePoint Spline (Math|Integer|SelectInt (- (+ _index OffsetIndex) _returnvalue_2) (+ _index OffsetIndex) (> (+ _index OffsetIndex) _returnvalue_3)) "World")))) _gridsize_4) _index_2) (* (Math|Vector|Normalize (- (Utilities|ZeroLocationWorldSpaceZ (Spline|GetLocationatSplinePoint Spline (Math|Integer|SelectInt 0 (+ (Math|Integer|SelectInt (- (+ _index OffsetIndex) _returnvalue_2) (+ _index OffsetIndex) (> (+ _index OffsetIndex) _returnvalue_3)) 1) (== (Math|Integer|SelectInt (- (+ _index OffsetIndex) _returnvalue_2) (+ _index OffsetIndex) (> (+ _index OffsetIndex) _returnvalue_3)) _returnvalue_3)) "World")) (Utilities|ZeroLocationWorldSpaceZ (Spline|GetLocationatSplinePoint Spline (Math|Integer|SelectInt (- (+ _index OffsetIndex) _returnvalue_2) (+ _index OffsetIndex) (> (+ _index OffsetIndex) _returnvalue_3)) "World")))) _returnvalue_20)))))
              (if (> (* _gridsize_5 (Math|Float|Round (/ (.Distance (Collision|BreakHitResult _hitresoultoutput)) _gridsize_5))) 0.5)
                (if (and (Math|Float|NearlyEqual(Float) (* _gridsize_5 (Math|Float|Round (/ (.Distance (Collision|BreakHitResult _hitresoultoutput)) _gridsize_5))) (Variables|MakeBasementFloor|GetcurrentLength) 0.5) (> (Variables|MakeBasementFloor|GetcurrentLength) 0.5))
                  (Variables|MakeBasementFloor|SetNewFloor false)
                  (else
                    (Variables|MakeBasementFloor|SetNewFloor true)))
                (if (Variables|MakeBasementFloor|GetNewFloor)
                  (Variables|MakeBasementFloor|SetcurrentLength (* _gridsize_5 (Math|Float|Round (/ (.Distance (Collision|BreakHitResult _hitresoultoutput)) _gridsize_5))))
                  (Utilities|FlowControl|Sequence
                    (:then_0
                      (Development|PrintString (Utilities|String|ToString(Transform) (Utilities|Select :Index _coordinatespace :Local (Math|Transform|MakeTransform (- (+ (Utilities|ZeroLocalorWorldSpaceZ _coordinatespace_3 (Spline|GetLocationatSplinePoint Spline (Math|Integer|SelectInt (- (+ _index OffsetIndex) _returnvalue_2) (+ _index OffsetIndex) (> (+ _index OffsetIndex) _returnvalue_3)) CoordinateSpace)) (* (Math|Vector|Normalize (- (Utilities|ZeroLocalorWorldSpaceZ _coordinatespace_2 (Spline|GetLocationatSplinePoint Spline (Math|Integer|SelectInt 0 (+ (Math|Integer|SelectInt (- (+ _index OffsetIndex) _returnvalue_2) (+ _index OffsetIndex) (> (+ _index OffsetIndex) _returnvalue_3)) 1) (== (Math|Integer|SelectInt (- (+ _index OffsetIndex) _returnvalue_2) (+ _index OffsetIndex) (> (+ _index OffsetIndex) _returnvalue_3)) _returnvalue_3)) CoordinateSpace)) (Utilities|ZeroLocalorWorldSpaceZ _coordinatespace_3 (Spline|GetLocationatSplinePoint Spline (Math|Integer|SelectInt (- (+ _index OffsetIndex) _returnvalue_2) (+ _index OffsetIndex) (> (+ _index OffsetIndex) _returnvalue_3)) CoordinateSpace)))) _returnvalue_15)) _returnvalue_11) (Math|Rotator|SelectRotator (Math|Rotator|MakeRotationfromAxes (- (Utilities|ZeroLocalorWorldSpaceZ _coordinatespace_2 (Spline|GetLocationatSplinePoint Spline (Math|Integer|SelectInt 0 (+ (Math|Integer|SelectInt (- (+ _index OffsetIndex) _returnvalue_2) (+ _index OffsetIndex) (> (+ _index OffsetIndex) _returnvalue_3)) 1) (== (Math|Integer|SelectInt (- (+ _index OffsetIndex) _returnvalue_2) (+ _index OffsetIndex) (> (+ _index OffsetIndex) _returnvalue_3)) _returnvalue_3)) CoordinateSpace)) (Utilities|ZeroLocalorWorldSpaceZ _coordinatespace_3 (Spline|GetLocationatSplinePoint Spline (Math|Integer|SelectInt (- (+ _index OffsetIndex) _returnvalue_2) (+ _index OffsetIndex) (> (+ _index OffsetIndex) _returnvalue_3)) CoordinateSpace))) (Math|Vector|NegateVector (Math|Vector|CrossProduct (- (Utilities|ZeroLocalorWorldSpaceZ _coordinatespace_2 (Spline|GetLocationatSplinePoint Spline (Math|Integer|SelectInt 0 (+ (Math|Integer|SelectInt (- (+ _index OffsetIndex) _returnvalue_2) (+ _index OffsetIndex) (> (+ _index OffsetIndex) _returnvalue_3)) 1) (== (Math|Integer|SelectInt (- (+ _index OffsetIndex) _returnvalue_2) (+ _index OffsetIndex) (> (+ _index OffsetIndex) _returnvalue_3)) _returnvalue_3)) CoordinateSpace)) (Utilities|ZeroLocalorWorldSpaceZ _coordinatespace_3 (Spline|GetLocationatSplinePoint Spline (Math|Integer|SelectInt (- (+ _index OffsetIndex) _returnvalue_2) (+ _index OffsetIndex) (> (+ _index OffsetIndex) _returnvalue_3)) CoordinateSpace))) "0, 0,1.000000")) "0, 0,1.000000") (Math|Rotator|MakeRotfromX (Math|Vector|NegateVector (- (Utilities|ZeroLocalorWorldSpaceZ _coordinatespace_2 (Spline|GetLocationatSplinePoint Spline (Math|Integer|SelectInt 0 (+ (Math|Integer|SelectInt (- (+ _index OffsetIndex) _returnvalue_2) (+ _index OffsetIndex) (> (+ _index OffsetIndex) _returnvalue_3)) 1) (== (Math|Integer|SelectInt (- (+ _index OffsetIndex) _returnvalue_2) (+ _index OffsetIndex) (> (+ _index OffsetIndex) _returnvalue_3)) _returnvalue_3)) CoordinateSpace)) (Utilities|ZeroLocalorWorldSpaceZ _coordinatespace_3 (Spline|GetLocationatSplinePoint Spline (Math|Integer|SelectInt (- (+ _index OffsetIndex) _returnvalue_2) (+ _index OffsetIndex) (> (+ _index OffsetIndex) _returnvalue_3)) CoordinateSpace))))) _reverse_2) (Math|Vector|MakeVector _returnvalue_19 (/ (* _gridsize_5 (Math|Float|Round (/ (.Distance (Collision|BreakHitResult _hitresoultoutput)) _gridsize_5))) _meshsizex_2) 1.0)) :World (Math|Transform|MakeRelativeTransform (Math|Transform|MakeTransform (- (+ (Utilities|ZeroLocalorWorldSpaceZ _coordinatespace_3 (Spline|GetLocationatSplinePoint Spline (Math|Integer|SelectInt (- (+ _index OffsetIndex) _returnvalue_2) (+ _index OffsetIndex) (> (+ _index OffsetIndex) _returnvalue_3)) CoordinateSpace)) (* (Math|Vector|Normalize (- (Utilities|ZeroLocalorWorldSpaceZ _coordinatespace_2 (Spline|GetLocationatSplinePoint Spline (Math|Integer|SelectInt 0 (+ (Math|Integer|SelectInt (- (+ _index OffsetIndex) _returnvalue_2) (+ _index OffsetIndex) (> (+ _index OffsetIndex) _returnvalue_3)) 1) (== (Math|Integer|SelectInt (- (+ _index OffsetIndex) _returnvalue_2) (+ _index OffsetIndex) (> (+ _index OffsetIndex) _returnvalue_3)) _returnvalue_3)) CoordinateSpace)) (Utilities|ZeroLocalorWorldSpaceZ _coordinatespace_3 (Spline|GetLocationatSplinePoint Spline (Math|Integer|SelectInt (- (+ _index OffsetIndex) _returnvalue_2) (+ _index OffsetIndex) (> (+ _index OffsetIndex) _returnvalue_3)) CoordinateSpace)))) _returnvalue_15)) _returnvalue_11) (Math|Rotator|SelectRotator (Math|Rotator|MakeRotationfromAxes (- (Utilities|ZeroLocalorWorldSpaceZ _coordinatespace_2 (Spline|GetLocationatSplinePoint Spline (Math|Integer|SelectInt 0 (+ (Math|Integer|SelectInt (- (+ _index OffsetIndex) _returnvalue_2) (+ _index OffsetIndex) (> (+ _index OffsetIndex) _returnvalue_3)) 1) (== (Math|Integer|SelectInt (- (+ _index OffsetIndex) _returnvalue_2) (+ _index OffsetIndex) (> (+ _index OffsetIndex) _returnvalue_3)) _returnvalue_3)) CoordinateSpace)) (Utilities|ZeroLocalorWorldSpaceZ _coordinatespace_3 (Spline|GetLocationatSplinePoint Spline (Math|Integer|SelectInt (- (+ _index OffsetIndex) _returnvalue_2) (+ _index OffsetIndex) (> (+ _index OffsetIndex) _returnvalue_3)) CoordinateSpace))) (Math|Vector|NegateVector (Math|Vector|CrossProduct (- (Utilities|ZeroLocalorWorldSpaceZ _coordinatespace_2 (Spline|GetLocationatSplinePoint Spline (Math|Integer|SelectInt 0 (+ (Math|Integer|SelectInt (- (+ _index OffsetIndex) _returnvalue_2) (+ _index OffsetIndex) (> (+ _index OffsetIndex) _returnvalue_3)) 1) (== (Math|Integer|SelectInt (- (+ _index OffsetIndex) _returnvalue_2) (+ _index OffsetIndex) (> (+ _index OffsetIndex) _returnvalue_3)) _returnvalue_3)) CoordinateSpace)) (Utilities|ZeroLocalorWorldSpaceZ _coordinatespace_3 (Spline|GetLocationatSplinePoint Spline (Math|Integer|SelectInt (- (+ _index OffsetIndex) _returnvalue_2) (+ _index OffsetIndex) (> (+ _index OffsetIndex) _returnvalue_3)) CoordinateSpace))) "0, 0,1.000000")) "0, 0,1.000000") (Math|Rotator|MakeRotfromX (Math|Vector|NegateVector (- (Utilities|ZeroLocalorWorldSpaceZ _coordinatespace_2 (Spline|GetLocationatSplinePoint Spline (Math|Integer|SelectInt 0 (+ (Math|Integer|SelectInt (- (+ _index OffsetIndex) _returnvalue_2) (+ _index OffsetIndex) (> (+ _index OffsetIndex) _returnvalue_3)) 1) (== (Math|Integer|SelectInt (- (+ _index OffsetIndex) _returnvalue_2) (+ _index OffsetIndex) (> (+ _index OffsetIndex) _returnvalue_3)) _returnvalue_3)) CoordinateSpace)) (Utilities|ZeroLocalorWorldSpaceZ _coordinatespace_3 (Spline|GetLocationatSplinePoint Spline (Math|Integer|SelectInt (- (+ _index OffsetIndex) _returnvalue_2) (+ _index OffsetIndex) (> (+ _index OffsetIndex) _returnvalue_3)) CoordinateSpace))))) _reverse_2) (Math|Vector|MakeVector _returnvalue_19 (/ (* _gridsize_5 (Math|Float|Round (/ (.Distance (Collision|BreakHitResult _hitresoultoutput)) _gridsize_5))) _meshsizex_2) 1.0)) _returnvalue_16))) false false "(R=1.000000,G=0.028558,B=0.000000,A=1.000000)")
                      (bind _returnvalue_21 (Utilities|Array|Add (LocalVariables|GetTransformList) (Utilities|Select :Index _coordinatespace :Local (Math|Transform|MakeTransform (- (+ (Utilities|ZeroLocalorWorldSpaceZ _coordinatespace_3 (Spline|GetLocationatSplinePoint Spline (Math|Integer|SelectInt (- (+ _index OffsetIndex) _returnvalue_2) (+ _index OffsetIndex) (> (+ _index OffsetIndex) _returnvalue_3)) CoordinateSpace)) (* (Math|Vector|Normalize (- (Utilities|ZeroLocalorWorldSpaceZ _coordinatespace_2 (Spline|GetLocationatSplinePoint Spline (Math|Integer|SelectInt 0 (+ (Math|Integer|SelectInt (- (+ _index OffsetIndex) _returnvalue_2) (+ _index OffsetIndex) (> (+ _index OffsetIndex) _returnvalue_3)) 1) (== (Math|Integer|SelectInt (- (+ _index OffsetIndex) _returnvalue_2) (+ _index OffsetIndex) (> (+ _index OffsetIndex) _returnvalue_3)) _returnvalue_3)) CoordinateSpace)) (Utilities|ZeroLocalorWorldSpaceZ _coordinatespace_3 (Spline|GetLocationatSplinePoint Spline (Math|Integer|SelectInt (- (+ _index OffsetIndex) _returnvalue_2) (+ _index OffsetIndex) (> (+ _index OffsetIndex) _returnvalue_3)) CoordinateSpace)))) _returnvalue_15)) _returnvalue_11) (Math|Rotator|SelectRotator (Math|Rotator|MakeRotationfromAxes (- (Utilities|ZeroLocalorWorldSpaceZ _coordinatespace_2 (Spline|GetLocationatSplinePoint Spline (Math|Integer|SelectInt 0 (+ (Math|Integer|SelectInt (- (+ _index OffsetIndex) _returnvalue_2) (+ _index OffsetIndex) (> (+ _index OffsetIndex) _returnvalue_3)) 1) (== (Math|Integer|SelectInt (- (+ _index OffsetIndex) _returnvalue_2) (+ _index OffsetIndex) (> (+ _index OffsetIndex) _returnvalue_3)) _returnvalue_3)) CoordinateSpace)) (Utilities|ZeroLocalorWorldSpaceZ _coordinatespace_3 (Spline|GetLocationatSplinePoint Spline (Math|Integer|SelectInt (- (+ _index OffsetIndex) _returnvalue_2) (+ _index OffsetIndex) (> (+ _index OffsetIndex) _returnvalue_3)) CoordinateSpace))) (Math|Vector|NegateVector (Math|Vector|CrossProduct (- (Utilities|ZeroLocalorWorldSpaceZ _coordinatespace_2 (Spline|GetLocationatSplinePoint Spline (Math|Integer|SelectInt 0 (+ (Math|Integer|SelectInt (- (+ _index OffsetIndex) _returnvalue_2) (+ _index OffsetIndex) (> (+ _index OffsetIndex) _returnvalue_3)) 1) (== (Math|Integer|SelectInt (- (+ _index OffsetIndex) _returnvalue_2) (+ _index OffsetIndex) (> (+ _index OffsetIndex) _returnvalue_3)) _returnvalue_3)) CoordinateSpace)) (Utilities|ZeroLocalorWorldSpaceZ _coordinatespace_3 (Spline|GetLocationatSplinePoint Spline (Math|Integer|SelectInt (- (+ _index OffsetIndex) _returnvalue_2) (+ _index OffsetIndex) (> (+ _index OffsetIndex) _returnvalue_3)) CoordinateSpace))) "0, 0,1.000000")) "0, 0,1.000000") (Math|Rotator|MakeRotfromX (Math|Vector|NegateVector (- (Utilities|ZeroLocalorWorldSpaceZ _coordinatespace_2 (Spline|GetLocationatSplinePoint Spline (Math|Integer|SelectInt 0 (+ (Math|Integer|SelectInt (- (+ _index OffsetIndex) _returnvalue_2) (+ _index OffsetIndex) (> (+ _index OffsetIndex) _returnvalue_3)) 1) (== (Math|Integer|SelectInt (- (+ _index OffsetIndex) _returnvalue_2) (+ _index OffsetIndex) (> (+ _index OffsetIndex) _returnvalue_3)) _returnvalue_3)) CoordinateSpace)) (Utilities|ZeroLocalorWorldSpaceZ _coordinatespace_3 (Spline|GetLocationatSplinePoint Spline (Math|Integer|SelectInt (- (+ _index OffsetIndex) _returnvalue_2) (+ _index OffsetIndex) (> (+ _index OffsetIndex) _returnvalue_3)) CoordinateSpace))))) _reverse_2) (Math|Vector|MakeVector _returnvalue_19 (/ (* _gridsize_5 (Math|Float|Round (/ (.Distance (Collision|BreakHitResult _hitresoultoutput)) _gridsize_5))) _meshsizex_2) 1.0)) :World (Math|Transform|MakeRelativeTransform (Math|Transform|MakeTransform (- (+ (Utilities|ZeroLocalorWorldSpaceZ _coordinatespace_3 (Spline|GetLocationatSplinePoint Spline (Math|Integer|SelectInt (- (+ _index OffsetIndex) _returnvalue_2) (+ _index OffsetIndex) (> (+ _index OffsetIndex) _returnvalue_3)) CoordinateSpace)) (* (Math|Vector|Normalize (- (Utilities|ZeroLocalorWorldSpaceZ _coordinatespace_2 (Spline|GetLocationatSplinePoint Spline (Math|Integer|SelectInt 0 (+ (Math|Integer|SelectInt (- (+ _index OffsetIndex) _returnvalue_2) (+ _index OffsetIndex) (> (+ _index OffsetIndex) _returnvalue_3)) 1) (== (Math|Integer|SelectInt (- (+ _index OffsetIndex) _returnvalue_2) (+ _index OffsetIndex) (> (+ _index OffsetIndex) _returnvalue_3)) _returnvalue_3)) CoordinateSpace)) (Utilities|ZeroLocalorWorldSpaceZ _coordinatespace_3 (Spline|GetLocationatSplinePoint Spline (Math|Integer|SelectInt (- (+ _index OffsetIndex) _returnvalue_2) (+ _index OffsetIndex) (> (+ _index OffsetIndex) _returnvalue_3)) CoordinateSpace)))) _returnvalue_15)) _returnvalue_11) (Math|Rotator|SelectRotator (Math|Rotator|MakeRotationfromAxes (- (Utilities|ZeroLocalorWorldSpaceZ _coordinatespace_2 (Spline|GetLocationatSplinePoint Spline (Math|Integer|SelectInt 0 (+ (Math|Integer|SelectInt (- (+ _index OffsetIndex) _returnvalue_2) (+ _index OffsetIndex) (> (+ _index OffsetIndex) _returnvalue_3)) 1) (== (Math|Integer|SelectInt (- (+ _index OffsetIndex) _returnvalue_2) (+ _index OffsetIndex) (> (+ _index OffsetIndex) _returnvalue_3)) _returnvalue_3)) CoordinateSpace)) (Utilities|ZeroLocalorWorldSpaceZ _coordinatespace_3 (Spline|GetLocationatSplinePoint Spline (Math|Integer|SelectInt (- (+ _index OffsetIndex) _returnvalue_2) (+ _index OffsetIndex) (> (+ _index OffsetIndex) _returnvalue_3)) CoordinateSpace))) (Math|Vector|NegateVector (Math|Vector|CrossProduct (- (Utilities|ZeroLocalorWorldSpaceZ _coordinatespace_2 (Spline|GetLocationatSplinePoint Spline (Math|Integer|SelectInt 0 (+ (Math|Integer|SelectInt (- (+ _index OffsetIndex) _returnvalue_2) (+ _index OffsetIndex) (> (+ _index OffsetIndex) _returnvalue_3)) 1) (== (Math|Integer|SelectInt (- (+ _index OffsetIndex) _returnvalue_2) (+ _index OffsetIndex) (> (+ _index OffsetIndex) _returnvalue_3)) _returnvalue_3)) CoordinateSpace)) (Utilities|ZeroLocalorWorldSpaceZ _coordinatespace_3 (Spline|GetLocationatSplinePoint Spline (Math|Integer|SelectInt (- (+ _index OffsetIndex) _returnvalue_2) (+ _index OffsetIndex) (> (+ _index OffsetIndex) _returnvalue_3)) CoordinateSpace))) "0, 0,1.000000")) "0, 0,1.000000") (Math|Rotator|MakeRotfromX (Math|Vector|NegateVector (- (Utilities|ZeroLocalorWorldSpaceZ _coordinatespace_2 (Spline|GetLocationatSplinePoint Spline (Math|Integer|SelectInt 0 (+ (Math|Integer|SelectInt (- (+ _index OffsetIndex) _returnvalue_2) (+ _index OffsetIndex) (> (+ _index OffsetIndex) _returnvalue_3)) 1) (== (Math|Integer|SelectInt (- (+ _index OffsetIndex) _returnvalue_2) (+ _index OffsetIndex) (> (+ _index OffsetIndex) _returnvalue_3)) _returnvalue_3)) CoordinateSpace)) (Utilities|ZeroLocalorWorldSpaceZ _coordinatespace_3 (Spline|GetLocationatSplinePoint Spline (Math|Integer|SelectInt (- (+ _index OffsetIndex) _returnvalue_2) (+ _index OffsetIndex) (> (+ _index OffsetIndex) _returnvalue_3)) CoordinateSpace))))) _reverse_2) (Math|Vector|MakeVector _returnvalue_19 (/ (* _gridsize_5 (Math|Float|Round (/ (.Distance (Collision|BreakHitResult _hitresoultoutput)) _gridsize_5))) _meshsizex_2) 1.0)) _returnvalue_16))))
                      (LocalVariables|SetCurrentIndex _returnvalue_21)
                      (Utilities|FlowControl|Sequence
                        (:then_1
                          (Variables|MakeBasementFloor|SetNewFloor false)
                          (Variables|MakeBasementFloor|SetMeshSpawn%20Iteration (Variables|MakeBasementFloor|GetBasement%20Floor%20Iteration)))))
                    (:then_1
                      (bind _returnvalue_22 (AddComponent|NODE_AddStaticMeshComponent-18 :RelativeTransform (Math|Transform|ComposeTransforms (Math|Transform|MakeTransform (* (.location (Utilities|Select :Index _coordinatespace :Local (Math|Transform|MakeTransform (- (+ (Utilities|ZeroLocalorWorldSpaceZ _coordinatespace_3 (Spline|GetLocationatSplinePoint Spline (Math|Integer|SelectInt (- (+ _index OffsetIndex) _returnvalue_2) (+ _index OffsetIndex) (> (+ _index OffsetIndex) _returnvalue_3)) CoordinateSpace)) (* (Math|Vector|Normalize (- (Utilities|ZeroLocalorWorldSpaceZ _coordinatespace_2 (Spline|GetLocationatSplinePoint Spline (Math|Integer|SelectInt 0 (+ (Math|Integer|SelectInt (- (+ _index OffsetIndex) _returnvalue_2) (+ _index OffsetIndex) (> (+ _index OffsetIndex) _returnvalue_3)) 1) (== (Math|Integer|SelectInt (- (+ _index OffsetIndex) _returnvalue_2) (+ _index OffsetIndex) (> (+ _index OffsetIndex) _returnvalue_3)) _returnvalue_3)) CoordinateSpace)) (Utilities|ZeroLocalorWorldSpaceZ _coordinatespace_3 (Spline|GetLocationatSplinePoint Spline (Math|Integer|SelectInt (- (+ _index OffsetIndex) _returnvalue_2) (+ _index OffsetIndex) (> (+ _index OffsetIndex) _returnvalue_3)) CoordinateSpace)))) _returnvalue_15)) _returnvalue_11) (Math|Rotator|SelectRotator (Math|Rotator|MakeRotationfromAxes (- (Utilities|ZeroLocalorWorldSpaceZ _coordinatespace_2 (Spline|GetLocationatSplinePoint Spline (Math|Integer|SelectInt 0 (+ (Math|Integer|SelectInt (- (+ _index OffsetIndex) _returnvalue_2) (+ _index OffsetIndex) (> (+ _index OffsetIndex) _returnvalue_3)) 1) (== (Math|Integer|SelectInt (- (+ _index OffsetIndex) _returnvalue_2) (+ _index OffsetIndex) (> (+ _index OffsetIndex) _returnvalue_3)) _returnvalue_3)) CoordinateSpace)) (Utilities|ZeroLocalorWorldSpaceZ _coordinatespace_3 (Spline|GetLocationatSplinePoint Spline (Math|Integer|SelectInt (- (+ _index OffsetIndex) _returnvalue_2) (+ _index OffsetIndex) (> (+ _index OffsetIndex) _returnvalue_3)) CoordinateSpace))) (Math|Vector|NegateVector (Math|Vector|CrossProduct (- (Utilities|ZeroLocalorWorldSpaceZ _coordinatespace_2 (Spline|GetLocationatSplinePoint Spline (Math|Integer|SelectInt 0 (+ (Math|Integer|SelectInt (- (+ _index OffsetIndex) _returnvalue_2) (+ _index OffsetIndex) (> (+ _index OffsetIndex) _returnvalue_3)) 1) (== (Math|Integer|SelectInt (- (+ _index OffsetIndex) _returnvalue_2) (+ _index OffsetIndex) (> (+ _index OffsetIndex) _returnvalue_3)) _returnvalue_3)) CoordinateSpace)) (Utilities|ZeroLocalorWorldSpaceZ _coordinatespace_3 (Spline|GetLocationatSplinePoint Spline (Math|Integer|SelectInt (- (+ _index OffsetIndex) _returnvalue_2) (+ _index OffsetIndex) (> (+ _index OffsetIndex) _returnvalue_3)) CoordinateSpace))) "0, 0,1.000000")) "0, 0,1.000000") (Math|Rotator|MakeRotfromX (Math|Vector|NegateVector (- (Utilities|ZeroLocalorWorldSpaceZ _coordinatespace_2 (Spline|GetLocationatSplinePoint Spline (Math|Integer|SelectInt 0 (+ (Math|Integer|SelectInt (- (+ _index OffsetIndex) _returnvalue_2) (+ _index OffsetIndex) (> (+ _index OffsetIndex) _returnvalue_3)) 1) (== (Math|Integer|SelectInt (- (+ _index OffsetIndex) _returnvalue_2) (+ _index OffsetIndex) (> (+ _index OffsetIndex) _returnvalue_3)) _returnvalue_3)) CoordinateSpace)) (Utilities|ZeroLocalorWorldSpaceZ _coordinatespace_3 (Spline|GetLocationatSplinePoint Spline (Math|Integer|SelectInt (- (+ _index OffsetIndex) _returnvalue_2) (+ _index OffsetIndex) (> (+ _index OffsetIndex) _returnvalue_3)) CoordinateSpace))))) _reverse_2) (Math|Vector|MakeVector _returnvalue_19 (/ (* _gridsize_5 (Math|Float|Round (/ (.Distance (Collision|BreakHitResult _hitresoultoutput)) _gridsize_5))) _meshsizex_2) 1.0)) :World (Math|Transform|MakeRelativeTransform (Math|Transform|MakeTransform (- (+ (Utilities|ZeroLocalorWorldSpaceZ _coordinatespace_3 (Spline|GetLocationatSplinePoint Spline (Math|Integer|SelectInt (- (+ _index OffsetIndex) _returnvalue_2) (+ _index OffsetIndex) (> (+ _index OffsetIndex) _returnvalue_3)) CoordinateSpace)) (* (Math|Vector|Normalize (- (Utilities|ZeroLocalorWorldSpaceZ _coordinatespace_2 (Spline|GetLocationatSplinePoint Spline (Math|Integer|SelectInt 0 (+ (Math|Integer|SelectInt (- (+ _index OffsetIndex) _returnvalue_2) (+ _index OffsetIndex) (> (+ _index OffsetIndex) _returnvalue_3)) 1) (== (Math|Integer|SelectInt (- (+ _index OffsetIndex) _returnvalue_2) (+ _index OffsetIndex) (> (+ _index OffsetIndex) _returnvalue_3)) _returnvalue_3)) CoordinateSpace)) (Utilities|ZeroLocalorWorldSpaceZ _coordinatespace_3 (Spline|GetLocationatSplinePoint Spline (Math|Integer|SelectInt (- (+ _index OffsetIndex) _returnvalue_2) (+ _index OffsetIndex) (> (+ _index OffsetIndex) _returnvalue_3)) CoordinateSpace)))) _returnvalue_15)) _returnvalue_11) (Math|Rotator|SelectRotator (Math|Rotator|MakeRotationfromAxes (- (Utilities|ZeroLocalorWorldSpaceZ _coordinatespace_2 (Spline|GetLocationatSplinePoint Spline (Math|Integer|SelectInt 0 (+ (Math|Integer|SelectInt (- (+ _index OffsetIndex) _returnvalue_2) (+ _index OffsetIndex) (> (+ _index OffsetIndex) _returnvalue_3)) 1) (== (Math|Integer|SelectInt (- (+ _index OffsetIndex) _returnvalue_2) (+ _index OffsetIndex) (> (+ _index OffsetIndex) _returnvalue_3)) _returnvalue_3)) CoordinateSpace)) (Utilities|ZeroLocalorWorldSpaceZ _coordinatespace_3 (Spline|GetLocationatSplinePoint Spline (Math|Integer|SelectInt (- (+ _index OffsetIndex) _returnvalue_2) (+ _index OffsetIndex) (> (+ _index OffsetIndex) _returnvalue_3)) CoordinateSpace))) (Math|Vector|NegateVector (Math|Vector|CrossProduct (- (Utilities|ZeroLocalorWorldSpaceZ _coordinatespace_2 (Spline|GetLocationatSplinePoint Spline (Math|Integer|SelectInt 0 (+ (Math|Integer|SelectInt (- (+ _index OffsetIndex) _returnvalue_2) (+ _index OffsetIndex) (> (+ _index OffsetIndex) _returnvalue_3)) 1) (== (Math|Integer|SelectInt (- (+ _index OffsetIndex) _returnvalue_2) (+ _index OffsetIndex) (> (+ _index OffsetIndex) _returnvalue_3)) _returnvalue_3)) CoordinateSpace)) (Utilities|ZeroLocalorWorldSpaceZ _coordinatespace_3 (Spline|GetLocationatSplinePoint Spline (Math|Integer|SelectInt (- (+ _index OffsetIndex) _returnvalue_2) (+ _index OffsetIndex) (> (+ _index OffsetIndex) _returnvalue_3)) CoordinateSpace))) "0, 0,1.000000")) "0, 0,1.000000") (Math|Rotator|MakeRotfromX (Math|Vector|NegateVector (- (Utilities|ZeroLocalorWorldSpaceZ _coordinatespace_2 (Spline|GetLocationatSplinePoint Spline (Math|Integer|SelectInt 0 (+ (Math|Integer|SelectInt (- (+ _index OffsetIndex) _returnvalue_2) (+ _index OffsetIndex) (> (+ _index OffsetIndex) _returnvalue_3)) 1) (== (Math|Integer|SelectInt (- (+ _index OffsetIndex) _returnvalue_2) (+ _index OffsetIndex) (> (+ _index OffsetIndex) _returnvalue_3)) _returnvalue_3)) CoordinateSpace)) (Utilities|ZeroLocalorWorldSpaceZ _coordinatespace_3 (Spline|GetLocationatSplinePoint Spline (Math|Integer|SelectInt (- (+ _index OffsetIndex) _returnvalue_2) (+ _index OffsetIndex) (> (+ _index OffsetIndex) _returnvalue_3)) CoordinateSpace))))) _reverse_2) (Math|Vector|MakeVector _returnvalue_19 (/ (* _gridsize_5 (Math|Float|Round (/ (.Distance (Collision|BreakHitResult _hitresoultoutput)) _gridsize_5))) _meshsizex_2) 1.0)) _returnvalue_16))) "1.000000,1.000000, 0") (.rotation (Utilities|Select :Index _coordinatespace :Local (Math|Transform|MakeTransform (- (+ (Utilities|ZeroLocalorWorldSpaceZ _coordinatespace_3 (Spline|GetLocationatSplinePoint Spline (Math|Integer|SelectInt (- (+ _index OffsetIndex) _returnvalue_2) (+ _index OffsetIndex) (> (+ _index OffsetIndex) _returnvalue_3)) CoordinateSpace)) (* (Math|Vector|Normalize (- (Utilities|ZeroLocalorWorldSpaceZ _coordinatespace_2 (Spline|GetLocationatSplinePoint Spline (Math|Integer|SelectInt 0 (+ (Math|Integer|SelectInt (- (+ _index OffsetIndex) _returnvalue_2) (+ _index OffsetIndex) (> (+ _index OffsetIndex) _returnvalue_3)) 1) (== (Math|Integer|SelectInt (- (+ _index OffsetIndex) _returnvalue_2) (+ _index OffsetIndex) (> (+ _index OffsetIndex) _returnvalue_3)) _returnvalue_3)) CoordinateSpace)) (Utilities|ZeroLocalorWorldSpaceZ _coordinatespace_3 (Spline|GetLocationatSplinePoint Spline (Math|Integer|SelectInt (- (+ _index OffsetIndex) _returnvalue_2) (+ _index OffsetIndex) (> (+ _index OffsetIndex) _returnvalue_3)) CoordinateSpace)))) _returnvalue_15)) _returnvalue_11) (Math|Rotator|SelectRotator (Math|Rotator|MakeRotationfromAxes (- (Utilities|ZeroLocalorWorldSpaceZ _coordinatespace_2 (Spline|GetLocationatSplinePoint Spline (Math|Integer|SelectInt 0 (+ (Math|Integer|SelectInt (- (+ _index OffsetIndex) _returnvalue_2) (+ _index OffsetIndex) (> (+ _index OffsetIndex) _returnvalue_3)) 1) (== (Math|Integer|SelectInt (- (+ _index OffsetIndex) _returnvalue_2) (+ _index OffsetIndex) (> (+ _index OffsetIndex) _returnvalue_3)) _returnvalue_3)) CoordinateSpace)) (Utilities|ZeroLocalorWorldSpaceZ _coordinatespace_3 (Spline|GetLocationatSplinePoint Spline (Math|Integer|SelectInt (- (+ _index OffsetIndex) _returnvalue_2) (+ _index OffsetIndex) (> (+ _index OffsetIndex) _returnvalue_3)) CoordinateSpace))) (Math|Vector|NegateVector (Math|Vector|CrossProduct (- (Utilities|ZeroLocalorWorldSpaceZ _coordinatespace_2 (Spline|GetLocationatSplinePoint Spline (Math|Integer|SelectInt 0 (+ (Math|Integer|SelectInt (- (+ _index OffsetIndex) _returnvalue_2) (+ _index OffsetIndex) (> (+ _index OffsetIndex) _returnvalue_3)) 1) (== (Math|Integer|SelectInt (- (+ _index OffsetIndex) _returnvalue_2) (+ _index OffsetIndex) (> (+ _index OffsetIndex) _returnvalue_3)) _returnvalue_3)) CoordinateSpace)) (Utilities|ZeroLocalorWorldSpaceZ _coordinatespace_3 (Spline|GetLocationatSplinePoint Spline (Math|Integer|SelectInt (- (+ _index OffsetIndex) _returnvalue_2) (+ _index OffsetIndex) (> (+ _index OffsetIndex) _returnvalue_3)) CoordinateSpace))) "0, 0,1.000000")) "0, 0,1.000000") (Math|Rotator|MakeRotfromX (Math|Vector|NegateVector (- (Utilities|ZeroLocalorWorldSpaceZ _coordinatespace_2 (Spline|GetLocationatSplinePoint Spline (Math|Integer|SelectInt 0 (+ (Math|Integer|SelectInt (- (+ _index OffsetIndex) _returnvalue_2) (+ _index OffsetIndex) (> (+ _index OffsetIndex) _returnvalue_3)) 1) (== (Math|Integer|SelectInt (- (+ _index OffsetIndex) _returnvalue_2) (+ _index OffsetIndex) (> (+ _index OffsetIndex) _returnvalue_3)) _returnvalue_3)) CoordinateSpace)) (Utilities|ZeroLocalorWorldSpaceZ _coordinatespace_3 (Spline|GetLocationatSplinePoint Spline (Math|Integer|SelectInt (- (+ _index OffsetIndex) _returnvalue_2) (+ _index OffsetIndex) (> (+ _index OffsetIndex) _returnvalue_3)) CoordinateSpace))))) _reverse_2) (Math|Vector|MakeVector _returnvalue_19 (/ (* _gridsize_5 (Math|Float|Round (/ (.Distance (Collision|BreakHitResult _hitresoultoutput)) _gridsize_5))) _meshsizex_2) 1.0)) :World (Math|Transform|MakeRelativeTransform (Math|Transform|MakeTransform (- (+ (Utilities|ZeroLocalorWorldSpaceZ _coordinatespace_3 (Spline|GetLocationatSplinePoint Spline (Math|Integer|SelectInt (- (+ _index OffsetIndex) _returnvalue_2) (+ _index OffsetIndex) (> (+ _index OffsetIndex) _returnvalue_3)) CoordinateSpace)) (* (Math|Vector|Normalize (- (Utilities|ZeroLocalorWorldSpaceZ _coordinatespace_2 (Spline|GetLocationatSplinePoint Spline (Math|Integer|SelectInt 0 (+ (Math|Integer|SelectInt (- (+ _index OffsetIndex) _returnvalue_2) (+ _index OffsetIndex) (> (+ _index OffsetIndex) _returnvalue_3)) 1) (== (Math|Integer|SelectInt (- (+ _index OffsetIndex) _returnvalue_2) (+ _index OffsetIndex) (> (+ _index OffsetIndex) _returnvalue_3)) _returnvalue_3)) CoordinateSpace)) (Utilities|ZeroLocalorWorldSpaceZ _coordinatespace_3 (Spline|GetLocationatSplinePoint Spline (Math|Integer|SelectInt (- (+ _index OffsetIndex) _returnvalue_2) (+ _index OffsetIndex) (> (+ _index OffsetIndex) _returnvalue_3)) CoordinateSpace)))) _returnvalue_15)) _returnvalue_11) (Math|Rotator|SelectRotator (Math|Rotator|MakeRotationfromAxes (- (Utilities|ZeroLocalorWorldSpaceZ _coordinatespace_2 (Spline|GetLocationatSplinePoint Spline (Math|Integer|SelectInt 0 (+ (Math|Integer|SelectInt (- (+ _index OffsetIndex) _returnvalue_2) (+ _index OffsetIndex) (> (+ _index OffsetIndex) _returnvalue_3)) 1) (== (Math|Integer|SelectInt (- (+ _index OffsetIndex) _returnvalue_2) (+ _index OffsetIndex) (> (+ _index OffsetIndex) _returnvalue_3)) _returnvalue_3)) CoordinateSpace)) (Utilities|ZeroLocalorWorldSpaceZ _coordinatespace_3 (Spline|GetLocationatSplinePoint Spline (Math|Integer|SelectInt (- (+ _index OffsetIndex) _returnvalue_2) (+ _index OffsetIndex) (> (+ _index OffsetIndex) _returnvalue_3)) CoordinateSpace))) (Math|Vector|NegateVector (Math|Vector|CrossProduct (- (Utilities|ZeroLocalorWorldSpaceZ _coordinatespace_2 (Spline|GetLocationatSplinePoint Spline (Math|Integer|SelectInt 0 (+ (Math|Integer|SelectInt (- (+ _index OffsetIndex) _returnvalue_2) (+ _index OffsetIndex) (> (+ _index OffsetIndex) _returnvalue_3)) 1) (== (Math|Integer|SelectInt (- (+ _index OffsetIndex) _returnvalue_2) (+ _index OffsetIndex) (> (+ _index OffsetIndex) _returnvalue_3)) _returnvalue_3)) CoordinateSpace)) (Utilities|ZeroLocalorWorldSpaceZ _coordinatespace_3 (Spline|GetLocationatSplinePoint Spline (Math|Integer|SelectInt (- (+ _index OffsetIndex) _returnvalue_2) (+ _index OffsetIndex) (> (+ _index OffsetIndex) _returnvalue_3)) CoordinateSpace))) "0, 0,1.000000")) "0, 0,1.000000") (Math|Rotator|MakeRotfromX (Math|Vector|NegateVector (- (Utilities|ZeroLocalorWorldSpaceZ _coordinatespace_2 (Spline|GetLocationatSplinePoint Spline (Math|Integer|SelectInt 0 (+ (Math|Integer|SelectInt (- (+ _index OffsetIndex) _returnvalue_2) (+ _index OffsetIndex) (> (+ _index OffsetIndex) _returnvalue_3)) 1) (== (Math|Integer|SelectInt (- (+ _index OffsetIndex) _returnvalue_2) (+ _index OffsetIndex) (> (+ _index OffsetIndex) _returnvalue_3)) _returnvalue_3)) CoordinateSpace)) (Utilities|ZeroLocalorWorldSpaceZ _coordinatespace_3 (Spline|GetLocationatSplinePoint Spline (Math|Integer|SelectInt (- (+ _index OffsetIndex) _returnvalue_2) (+ _index OffsetIndex) (> (+ _index OffsetIndex) _returnvalue_3)) CoordinateSpace))))) _reverse_2) (Math|Vector|MakeVector _returnvalue_19 (/ (* _gridsize_5 (Math|Float|Round (/ (.Distance (Collision|BreakHitResult _hitresoultoutput)) _gridsize_5))) _meshsizex_2) 1.0)) _returnvalue_16))) (Math|Vector|MakeVector (/ (- (* (.x (.scale (Utilities|Select :Index _coordinatespace :Local (Math|Transform|MakeTransform (- (+ (Utilities|ZeroLocalorWorldSpaceZ _coordinatespace_3 (Spline|GetLocationatSplinePoint Spline (Math|Integer|SelectInt (- (+ _index OffsetIndex) _returnvalue_2) (+ _index OffsetIndex) (> (+ _index OffsetIndex) _returnvalue_3)) CoordinateSpace)) (* (Math|Vector|Normalize (- (Utilities|ZeroLocalorWorldSpaceZ _coordinatespace_2 (Spline|GetLocationatSplinePoint Spline (Math|Integer|SelectInt 0 (+ (Math|Integer|SelectInt (- (+ _index OffsetIndex) _returnvalue_2) (+ _index OffsetIndex) (> (+ _index OffsetIndex) _returnvalue_3)) 1) (== (Math|Integer|SelectInt (- (+ _index OffsetIndex) _returnvalue_2) (+ _index OffsetIndex) (> (+ _index OffsetIndex) _returnvalue_3)) _returnvalue_3)) CoordinateSpace)) (Utilities|ZeroLocalorWorldSpaceZ _coordinatespace_3 (Spline|GetLocationatSplinePoint Spline (Math|Integer|SelectInt (- (+ _index OffsetIndex) _returnvalue_2) (+ _index OffsetIndex) (> (+ _index OffsetIndex) _returnvalue_3)) CoordinateSpace)))) _returnvalue_15)) _returnvalue_11) (Math|Rotator|SelectRotator (Math|Rotator|MakeRotationfromAxes (- (Utilities|ZeroLocalorWorldSpaceZ _coordinatespace_2 (Spline|GetLocationatSplinePoint Spline (Math|Integer|SelectInt 0 (+ (Math|Integer|SelectInt (- (+ _index OffsetIndex) _returnvalue_2) (+ _index OffsetIndex) (> (+ _index OffsetIndex) _returnvalue_3)) 1) (== (Math|Integer|SelectInt (- (+ _index OffsetIndex) _returnvalue_2) (+ _index OffsetIndex) (> (+ _index OffsetIndex) _returnvalue_3)) _returnvalue_3)) CoordinateSpace)) (Utilities|ZeroLocalorWorldSpaceZ _coordinatespace_3 (Spline|GetLocationatSplinePoint Spline (Math|Integer|SelectInt (- (+ _index OffsetIndex) _returnvalue_2) (+ _index OffsetIndex) (> (+ _index OffsetIndex) _returnvalue_3)) CoordinateSpace))) (Math|Vector|NegateVector (Math|Vector|CrossProduct (- (Utilities|ZeroLocalorWorldSpaceZ _coordinatespace_2 (Spline|GetLocationatSplinePoint Spline (Math|Integer|SelectInt 0 (+ (Math|Integer|SelectInt (- (+ _index OffsetIndex) _returnvalue_2) (+ _index OffsetIndex) (> (+ _index OffsetIndex) _returnvalue_3)) 1) (== (Math|Integer|SelectInt (- (+ _index OffsetIndex) _returnvalue_2) (+ _index OffsetIndex) (> (+ _index OffsetIndex) _returnvalue_3)) _returnvalue_3)) CoordinateSpace)) (Utilities|ZeroLocalorWorldSpaceZ _coordinatespace_3 (Spline|GetLocationatSplinePoint Spline (Math|Integer|SelectInt (- (+ _index OffsetIndex) _returnvalue_2) (+ _index OffsetIndex) (> (+ _index OffsetIndex) _returnvalue_3)) CoordinateSpace))) "0, 0,1.000000")) "0, 0,1.000000") (Math|Rotator|MakeRotfromX (Math|Vector|NegateVector (- (Utilities|ZeroLocalorWorldSpaceZ _coordinatespace_2 (Spline|GetLocationatSplinePoint Spline (Math|Integer|SelectInt 0 (+ (Math|Integer|SelectInt (- (+ _index OffsetIndex) _returnvalue_2) (+ _index OffsetIndex) (> (+ _index OffsetIndex) _returnvalue_3)) 1) (== (Math|Integer|SelectInt (- (+ _index OffsetIndex) _returnvalue_2) (+ _index OffsetIndex) (> (+ _index OffsetIndex) _returnvalue_3)) _returnvalue_3)) CoordinateSpace)) (Utilities|ZeroLocalorWorldSpaceZ _coordinatespace_3 (Spline|GetLocationatSplinePoint Spline (Math|Integer|SelectInt (- (+ _index OffsetIndex) _returnvalue_2) (+ _index OffsetIndex) (> (+ _index OffsetIndex) _returnvalue_3)) CoordinateSpace))))) _reverse_2) (Math|Vector|MakeVector _returnvalue_19 (/ (* _gridsize_5 (Math|Float|Round (/ (.Distance (Collision|BreakHitResult _hitresoultoutput)) _gridsize_5))) _meshsizex_2) 1.0)) :World (Math|Transform|MakeRelativeTransform (Math|Transform|MakeTransform (- (+ (Utilities|ZeroLocalorWorldSpaceZ _coordinatespace_3 (Spline|GetLocationatSplinePoint Spline (Math|Integer|SelectInt (- (+ _index OffsetIndex) _returnvalue_2) (+ _index OffsetIndex) (> (+ _index OffsetIndex) _returnvalue_3)) CoordinateSpace)) (* (Math|Vector|Normalize (- (Utilities|ZeroLocalorWorldSpaceZ _coordinatespace_2 (Spline|GetLocationatSplinePoint Spline (Math|Integer|SelectInt 0 (+ (Math|Integer|SelectInt (- (+ _index OffsetIndex) _returnvalue_2) (+ _index OffsetIndex) (> (+ _index OffsetIndex) _returnvalue_3)) 1) (== (Math|Integer|SelectInt (- (+ _index OffsetIndex) _returnvalue_2) (+ _index OffsetIndex) (> (+ _index OffsetIndex) _returnvalue_3)) _returnvalue_3)) CoordinateSpace)) (Utilities|ZeroLocalorWorldSpaceZ _coordinatespace_3 (Spline|GetLocationatSplinePoint Spline (Math|Integer|SelectInt (- (+ _index OffsetIndex) _returnvalue_2) (+ _index OffsetIndex) (> (+ _index OffsetIndex) _returnvalue_3)) CoordinateSpace)))) _returnvalue_15)) _returnvalue_11) (Math|Rotator|SelectRotator (Math|Rotator|MakeRotationfromAxes (- (Utilities|ZeroLocalorWorldSpaceZ _coordinatespace_2 (Spline|GetLocationatSplinePoint Spline (Math|Integer|SelectInt 0 (+ (Math|Integer|SelectInt (- (+ _index OffsetIndex) _returnvalue_2) (+ _index OffsetIndex) (> (+ _index OffsetIndex) _returnvalue_3)) 1) (== (Math|Integer|SelectInt (- (+ _index OffsetIndex) _returnvalue_2) (+ _index OffsetIndex) (> (+ _index OffsetIndex) _returnvalue_3)) _returnvalue_3)) CoordinateSpace)) (Utilities|ZeroLocalorWorldSpaceZ _coordinatespace_3 (Spline|GetLocationatSplinePoint Spline (Math|Integer|SelectInt (- (+ _index OffsetIndex) _returnvalue_2) (+ _index OffsetIndex) (> (+ _index OffsetIndex) _returnvalue_3)) CoordinateSpace))) (Math|Vector|NegateVector (Math|Vector|CrossProduct (- (Utilities|ZeroLocalorWorldSpaceZ _coordinatespace_2 (Spline|GetLocationatSplinePoint Spline (Math|Integer|SelectInt 0 (+ (Math|Integer|SelectInt (- (+ _index OffsetIndex) _returnvalue_2) (+ _index OffsetIndex) (> (+ _index OffsetIndex) _returnvalue_3)) 1) (== (Math|Integer|SelectInt (- (+ _index OffsetIndex) _returnvalue_2) (+ _index OffsetIndex) (> (+ _index OffsetIndex) _returnvalue_3)) _returnvalue_3)) CoordinateSpace)) (Utilities|ZeroLocalorWorldSpaceZ _coordinatespace_3 (Spline|GetLocationatSplinePoint Spline (Math|Integer|SelectInt (- (+ _index OffsetIndex) _returnvalue_2) (+ _index OffsetIndex) (> (+ _index OffsetIndex) _returnvalue_3)) CoordinateSpace))) "0, 0,1.000000")) "0, 0,1.000000") (Math|Rotator|MakeRotfromX (Math|Vector|NegateVector (- (Utilities|ZeroLocalorWorldSpaceZ _coordinatespace_2 (Spline|GetLocationatSplinePoint Spline (Math|Integer|SelectInt 0 (+ (Math|Integer|SelectInt (- (+ _index OffsetIndex) _returnvalue_2) (+ _index OffsetIndex) (> (+ _index OffsetIndex) _returnvalue_3)) 1) (== (Math|Integer|SelectInt (- (+ _index OffsetIndex) _returnvalue_2) (+ _index OffsetIndex) (> (+ _index OffsetIndex) _returnvalue_3)) _returnvalue_3)) CoordinateSpace)) (Utilities|ZeroLocalorWorldSpaceZ _coordinatespace_3 (Spline|GetLocationatSplinePoint Spline (Math|Integer|SelectInt (- (+ _index OffsetIndex) _returnvalue_2) (+ _index OffsetIndex) (> (+ _index OffsetIndex) _returnvalue_3)) CoordinateSpace))))) _reverse_2) (Math|Vector|MakeVector _returnvalue_19 (/ (* _gridsize_5 (Math|Float|Round (/ (.Distance (Collision|BreakHitResult _hitresoultoutput)) _gridsize_5))) _meshsizex_2) 1.0)) _returnvalue_16)))) (LocalVariables|GetGridSize)) (Math|Float|SelectFloat 4e+01 -4e+01 (Variables|GetReverse))) (LocalVariables|GetGridSize)) (/ (- (* (.y (.scale (Utilities|Select :Index _coordinatespace :Local (Math|Transform|MakeTransform (- (+ (Utilities|ZeroLocalorWorldSpaceZ _coordinatespace_3 (Spline|GetLocationatSplinePoint Spline (Math|Integer|SelectInt (- (+ _index OffsetIndex) _returnvalue_2) (+ _index OffsetIndex) (> (+ _index OffsetIndex) _returnvalue_3)) CoordinateSpace)) (* (Math|Vector|Normalize (- (Utilities|ZeroLocalorWorldSpaceZ _coordinatespace_2 (Spline|GetLocationatSplinePoint Spline (Math|Integer|SelectInt 0 (+ (Math|Integer|SelectInt (- (+ _index OffsetIndex) _returnvalue_2) (+ _index OffsetIndex) (> (+ _index OffsetIndex) _returnvalue_3)) 1) (== (Math|Integer|SelectInt (- (+ _index OffsetIndex) _returnvalue_2) (+ _index OffsetIndex) (> (+ _index OffsetIndex) _returnvalue_3)) _returnvalue_3)) CoordinateSpace)) (Utilities|ZeroLocalorWorldSpaceZ _coordinatespace_3 (Spline|GetLocationatSplinePoint Spline (Math|Integer|SelectInt (- (+ _index OffsetIndex) _returnvalue_2) (+ _index OffsetIndex) (> (+ _index OffsetIndex) _returnvalue_3)) CoordinateSpace)))) _returnvalue_15)) _returnvalue_11) (Math|Rotator|SelectRotator (Math|Rotator|MakeRotationfromAxes (- (Utilities|ZeroLocalorWorldSpaceZ _coordinatespace_2 (Spline|GetLocationatSplinePoint Spline (Math|Integer|SelectInt 0 (+ (Math|Integer|SelectInt (- (+ _index OffsetIndex) _returnvalue_2) (+ _index OffsetIndex) (> (+ _index OffsetIndex) _returnvalue_3)) 1) (== (Math|Integer|SelectInt (- (+ _index OffsetIndex) _returnvalue_2) (+ _index OffsetIndex) (> (+ _index OffsetIndex) _returnvalue_3)) _returnvalue_3)) CoordinateSpace)) (Utilities|ZeroLocalorWorldSpaceZ _coordinatespace_3 (Spline|GetLocationatSplinePoint Spline (Math|Integer|SelectInt (- (+ _index OffsetIndex) _returnvalue_2) (+ _index OffsetIndex) (> (+ _index OffsetIndex) _returnvalue_3)) CoordinateSpace))) (Math|Vector|NegateVector (Math|Vector|CrossProduct (- (Utilities|ZeroLocalorWorldSpaceZ _coordinatespace_2 (Spline|GetLocationatSplinePoint Spline (Math|Integer|SelectInt 0 (+ (Math|Integer|SelectInt (- (+ _index OffsetIndex) _returnvalue_2) (+ _index OffsetIndex) (> (+ _index OffsetIndex) _returnvalue_3)) 1) (== (Math|Integer|SelectInt (- (+ _index OffsetIndex) _returnvalue_2) (+ _index OffsetIndex) (> (+ _index OffsetIndex) _returnvalue_3)) _returnvalue_3)) CoordinateSpace)) (Utilities|ZeroLocalorWorldSpaceZ _coordinatespace_3 (Spline|GetLocationatSplinePoint Spline (Math|Integer|SelectInt (- (+ _index OffsetIndex) _returnvalue_2) (+ _index OffsetIndex) (> (+ _index OffsetIndex) _returnvalue_3)) CoordinateSpace))) "0, 0,1.000000")) "0, 0,1.000000") (Math|Rotator|MakeRotfromX (Math|Vector|NegateVector (- (Utilities|ZeroLocalorWorldSpaceZ _coordinatespace_2 (Spline|GetLocationatSplinePoint Spline (Math|Integer|SelectInt 0 (+ (Math|Integer|SelectInt (- (+ _index OffsetIndex) _returnvalue_2) (+ _index OffsetIndex) (> (+ _index OffsetIndex) _returnvalue_3)) 1) (== (Math|Integer|SelectInt (- (+ _index OffsetIndex) _returnvalue_2) (+ _index OffsetIndex) (> (+ _index OffsetIndex) _returnvalue_3)) _returnvalue_3)) CoordinateSpace)) (Utilities|ZeroLocalorWorldSpaceZ _coordinatespace_3 (Spline|GetLocationatSplinePoint Spline (Math|Integer|SelectInt (- (+ _index OffsetIndex) _returnvalue_2) (+ _index OffsetIndex) (> (+ _index OffsetIndex) _returnvalue_3)) CoordinateSpace))))) _reverse_2) (Math|Vector|MakeVector _returnvalue_19 (/ (* _gridsize_5 (Math|Float|Round (/ (.Distance (Collision|BreakHitResult _hitresoultoutput)) _gridsize_5))) _meshsizex_2) 1.0)) :World (Math|Transform|MakeRelativeTransform (Math|Transform|MakeTransform (- (+ (Utilities|ZeroLocalorWorldSpaceZ _coordinatespace_3 (Spline|GetLocationatSplinePoint Spline (Math|Integer|SelectInt (- (+ _index OffsetIndex) _returnvalue_2) (+ _index OffsetIndex) (> (+ _index OffsetIndex) _returnvalue_3)) CoordinateSpace)) (* (Math|Vector|Normalize (- (Utilities|ZeroLocalorWorldSpaceZ _coordinatespace_2 (Spline|GetLocationatSplinePoint Spline (Math|Integer|SelectInt 0 (+ (Math|Integer|SelectInt (- (+ _index OffsetIndex) _returnvalue_2) (+ _index OffsetIndex) (> (+ _index OffsetIndex) _returnvalue_3)) 1) (== (Math|Integer|SelectInt (- (+ _index OffsetIndex) _returnvalue_2) (+ _index OffsetIndex) (> (+ _index OffsetIndex) _returnvalue_3)) _returnvalue_3)) CoordinateSpace)) (Utilities|ZeroLocalorWorldSpaceZ _coordinatespace_3 (Spline|GetLocationatSplinePoint Spline (Math|Integer|SelectInt (- (+ _index OffsetIndex) _returnvalue_2) (+ _index OffsetIndex) (> (+ _index OffsetIndex) _returnvalue_3)) CoordinateSpace)))) _returnvalue_15)) _returnvalue_11) (Math|Rotator|SelectRotator (Math|Rotator|MakeRotationfromAxes (- (Utilities|ZeroLocalorWorldSpaceZ _coordinatespace_2 (Spline|GetLocationatSplinePoint Spline (Math|Integer|SelectInt 0 (+ (Math|Integer|SelectInt (- (+ _index OffsetIndex) _returnvalue_2) (+ _index OffsetIndex) (> (+ _index OffsetIndex) _returnvalue_3)) 1) (== (Math|Integer|SelectInt (- (+ _index OffsetIndex) _returnvalue_2) (+ _index OffsetIndex) (> (+ _index OffsetIndex) _returnvalue_3)) _returnvalue_3)) CoordinateSpace)) (Utilities|ZeroLocalorWorldSpaceZ _coordinatespace_3 (Spline|GetLocationatSplinePoint Spline (Math|Integer|SelectInt (- (+ _index OffsetIndex) _returnvalue_2) (+ _index OffsetIndex) (> (+ _index OffsetIndex) _returnvalue_3)) CoordinateSpace))) (Math|Vector|NegateVector (Math|Vector|CrossProduct (- (Utilities|ZeroLocalorWorldSpaceZ _coordinatespace_2 (Spline|GetLocationatSplinePoint Spline (Math|Integer|SelectInt 0 (+ (Math|Integer|SelectInt (- (+ _index OffsetIndex) _returnvalue_2) (+ _index OffsetIndex) (> (+ _index OffsetIndex) _returnvalue_3)) 1) (== (Math|Integer|SelectInt (- (+ _index OffsetIndex) _returnvalue_2) (+ _index OffsetIndex) (> (+ _index OffsetIndex) _returnvalue_3)) _returnvalue_3)) CoordinateSpace)) (Utilities|ZeroLocalorWorldSpaceZ _coordinatespace_3 (Spline|GetLocationatSplinePoint Spline (Math|Integer|SelectInt (- (+ _index OffsetIndex) _returnvalue_2) (+ _index OffsetIndex) (> (+ _index OffsetIndex) _returnvalue_3)) CoordinateSpace))) "0, 0,1.000000")) "0, 0,1.000000") (Math|Rotator|MakeRotfromX (Math|Vector|NegateVector (- (Utilities|ZeroLocalorWorldSpaceZ _coordinatespace_2 (Spline|GetLocationatSplinePoint Spline (Math|Integer|SelectInt 0 (+ (Math|Integer|SelectInt (- (+ _index OffsetIndex) _returnvalue_2) (+ _index OffsetIndex) (> (+ _index OffsetIndex) _returnvalue_3)) 1) (== (Math|Integer|SelectInt (- (+ _index OffsetIndex) _returnvalue_2) (+ _index OffsetIndex) (> (+ _index OffsetIndex) _returnvalue_3)) _returnvalue_3)) CoordinateSpace)) (Utilities|ZeroLocalorWorldSpaceZ _coordinatespace_3 (Spline|GetLocationatSplinePoint Spline (Math|Integer|SelectInt (- (+ _index OffsetIndex) _returnvalue_2) (+ _index OffsetIndex) (> (+ _index OffsetIndex) _returnvalue_3)) CoordinateSpace))))) _reverse_2) (Math|Vector|MakeVector _returnvalue_19 (/ (* _gridsize_5 (Math|Float|Round (/ (.Distance (Collision|BreakHitResult _hitresoultoutput)) _gridsize_5))) _meshsizex_2) 1.0)) _returnvalue_16)))) (LocalVariables|GetMeshSizeX)) 4e+01) (LocalVariables|GetMeshSizeX)) (.z (.scale (Utilities|Select :Index _coordinatespace :Local (Math|Transform|MakeTransform (- (+ (Utilities|ZeroLocalorWorldSpaceZ _coordinatespace_3 (Spline|GetLocationatSplinePoint Spline (Math|Integer|SelectInt (- (+ _index OffsetIndex) _returnvalue_2) (+ _index OffsetIndex) (> (+ _index OffsetIndex) _returnvalue_3)) CoordinateSpace)) (* (Math|Vector|Normalize (- (Utilities|ZeroLocalorWorldSpaceZ _coordinatespace_2 (Spline|GetLocationatSplinePoint Spline (Math|Integer|SelectInt 0 (+ (Math|Integer|SelectInt (- (+ _index OffsetIndex) _returnvalue_2) (+ _index OffsetIndex) (> (+ _index OffsetIndex) _returnvalue_3)) 1) (== (Math|Integer|SelectInt (- (+ _index OffsetIndex) _returnvalue_2) (+ _index OffsetIndex) (> (+ _index OffsetIndex) _returnvalue_3)) _returnvalue_3)) CoordinateSpace)) (Utilities|ZeroLocalorWorldSpaceZ _coordinatespace_3 (Spline|GetLocationatSplinePoint Spline (Math|Integer|SelectInt (- (+ _index OffsetIndex) _returnvalue_2) (+ _index OffsetIndex) (> (+ _index OffsetIndex) _returnvalue_3)) CoordinateSpace)))) _returnvalue_15)) _returnvalue_11) (Math|Rotator|SelectRotator (Math|Rotator|MakeRotationfromAxes (- (Utilities|ZeroLocalorWorldSpaceZ _coordinatespace_2 (Spline|GetLocationatSplinePoint Spline (Math|Integer|SelectInt 0 (+ (Math|Integer|SelectInt (- (+ _index OffsetIndex) _returnvalue_2) (+ _index OffsetIndex) (> (+ _index OffsetIndex) _returnvalue_3)) 1) (== (Math|Integer|SelectInt (- (+ _index OffsetIndex) _returnvalue_2) (+ _index OffsetIndex) (> (+ _index OffsetIndex) _returnvalue_3)) _returnvalue_3)) CoordinateSpace)) (Utilities|ZeroLocalorWorldSpaceZ _coordinatespace_3 (Spline|GetLocationatSplinePoint Spline (Math|Integer|SelectInt (- (+ _index OffsetIndex) _returnvalue_2) (+ _index OffsetIndex) (> (+ _index OffsetIndex) _returnvalue_3)) CoordinateSpace))) (Math|Vector|NegateVector (Math|Vector|CrossProduct (- (Utilities|ZeroLocalorWorldSpaceZ _coordinatespace_2 (Spline|GetLocationatSplinePoint Spline (Math|Integer|SelectInt 0 (+ (Math|Integer|SelectInt (- (+ _index OffsetIndex) _returnvalue_2) (+ _index OffsetIndex) (> (+ _index OffsetIndex) _returnvalue_3)) 1) (== (Math|Integer|SelectInt (- (+ _index OffsetIndex) _returnvalue_2) (+ _index OffsetIndex) (> (+ _index OffsetIndex) _returnvalue_3)) _returnvalue_3)) CoordinateSpace)) (Utilities|ZeroLocalorWorldSpaceZ _coordinatespace_3 (Spline|GetLocationatSplinePoint Spline (Math|Integer|SelectInt (- (+ _index OffsetIndex) _returnvalue_2) (+ _index OffsetIndex) (> (+ _index OffsetIndex) _returnvalue_3)) CoordinateSpace))) "0, 0,1.000000")) "0, 0,1.000000") (Math|Rotator|MakeRotfromX (Math|Vector|NegateVector (- (Utilities|ZeroLocalorWorldSpaceZ _coordinatespace_2 (Spline|GetLocationatSplinePoint Spline (Math|Integer|SelectInt 0 (+ (Math|Integer|SelectInt (- (+ _index OffsetIndex) _returnvalue_2) (+ _index OffsetIndex) (> (+ _index OffsetIndex) _returnvalue_3)) 1) (== (Math|Integer|SelectInt (- (+ _index OffsetIndex) _returnvalue_2) (+ _index OffsetIndex) (> (+ _index OffsetIndex) _returnvalue_3)) _returnvalue_3)) CoordinateSpace)) (Utilities|ZeroLocalorWorldSpaceZ _coordinatespace_3 (Spline|GetLocationatSplinePoint Spline (Math|Integer|SelectInt (- (+ _index OffsetIndex) _returnvalue_2) (+ _index OffsetIndex) (> (+ _index OffsetIndex) _returnvalue_3)) CoordinateSpace))))) _reverse_2) (Math|Vector|MakeVector _returnvalue_19 (/ (* _gridsize_5 (Math|Float|Round (/ (.Distance (Collision|BreakHitResult _hitresoultoutput)) _gridsize_5))) _meshsizex_2) 1.0)) :World (Math|Transform|MakeRelativeTransform (Math|Transform|MakeTransform (- (+ (Utilities|ZeroLocalorWorldSpaceZ _coordinatespace_3 (Spline|GetLocationatSplinePoint Spline (Math|Integer|SelectInt (- (+ _index OffsetIndex) _returnvalue_2) (+ _index OffsetIndex) (> (+ _index OffsetIndex) _returnvalue_3)) CoordinateSpace)) (* (Math|Vector|Normalize (- (Utilities|ZeroLocalorWorldSpaceZ _coordinatespace_2 (Spline|GetLocationatSplinePoint Spline (Math|Integer|SelectInt 0 (+ (Math|Integer|SelectInt (- (+ _index OffsetIndex) _returnvalue_2) (+ _index OffsetIndex) (> (+ _index OffsetIndex) _returnvalue_3)) 1) (== (Math|Integer|SelectInt (- (+ _index OffsetIndex) _returnvalue_2) (+ _index OffsetIndex) (> (+ _index OffsetIndex) _returnvalue_3)) _returnvalue_3)) CoordinateSpace)) (Utilities|ZeroLocalorWorldSpaceZ _coordinatespace_3 (Spline|GetLocationatSplinePoint Spline (Math|Integer|SelectInt (- (+ _index OffsetIndex) _returnvalue_2) (+ _index OffsetIndex) (> (+ _index OffsetIndex) _returnvalue_3)) CoordinateSpace)))) _returnvalue_15)) _returnvalue_11) (Math|Rotator|SelectRotator (Math|Rotator|MakeRotationfromAxes (- (Utilities|ZeroLocalorWorldSpaceZ _coordinatespace_2 (Spline|GetLocationatSplinePoint Spline (Math|Integer|SelectInt 0 (+ (Math|Integer|SelectInt (- (+ _index OffsetIndex) _returnvalue_2) (+ _index OffsetIndex) (> (+ _index OffsetIndex) _returnvalue_3)) 1) (== (Math|Integer|SelectInt (- (+ _index OffsetIndex) _returnvalue_2) (+ _index OffsetIndex) (> (+ _index OffsetIndex) _returnvalue_3)) _returnvalue_3)) CoordinateSpace)) (Utilities|ZeroLocalorWorldSpaceZ _coordinatespace_3 (Spline|GetLocationatSplinePoint Spline (Math|Integer|SelectInt (- (+ _index OffsetIndex) _returnvalue_2) (+ _index OffsetIndex) (> (+ _index OffsetIndex) _returnvalue_3)) CoordinateSpace))) (Math|Vector|NegateVector (Math|Vector|CrossProduct (- (Utilities|ZeroLocalorWorldSpaceZ _coordinatespace_2 (Spline|GetLocationatSplinePoint Spline (Math|Integer|SelectInt 0 (+ (Math|Integer|SelectInt (- (+ _index OffsetIndex) _returnvalue_2) (+ _index OffsetIndex) (> (+ _index OffsetIndex) _returnvalue_3)) 1) (== (Math|Integer|SelectInt (- (+ _index OffsetIndex) _returnvalue_2) (+ _index OffsetIndex) (> (+ _index OffsetIndex) _returnvalue_3)) _returnvalue_3)) CoordinateSpace)) (Utilities|ZeroLocalorWorldSpaceZ _coordinatespace_3 (Spline|GetLocationatSplinePoint Spline (Math|Integer|SelectInt (- (+ _index OffsetIndex) _returnvalue_2) (+ _index OffsetIndex) (> (+ _index OffsetIndex) _returnvalue_3)) CoordinateSpace))) "0, 0,1.000000")) "0, 0,1.000000") (Math|Rotator|MakeRotfromX (Math|Vector|NegateVector (- (Utilities|ZeroLocalorWorldSpaceZ _coordinatespace_2 (Spline|GetLocationatSplinePoint Spline (Math|Integer|SelectInt 0 (+ (Math|Integer|SelectInt (- (+ _index OffsetIndex) _returnvalue_2) (+ _index OffsetIndex) (> (+ _index OffsetIndex) _returnvalue_3)) 1) (== (Math|Integer|SelectInt (- (+ _index OffsetIndex) _returnvalue_2) (+ _index OffsetIndex) (> (+ _index OffsetIndex) _returnvalue_3)) _returnvalue_3)) CoordinateSpace)) (Utilities|ZeroLocalorWorldSpaceZ _coordinatespace_3 (Spline|GetLocationatSplinePoint Spline (Math|Integer|SelectInt (- (+ _index OffsetIndex) _returnvalue_2) (+ _index OffsetIndex) (> (+ _index OffsetIndex) _returnvalue_3)) CoordinateSpace))))) _reverse_2) (Math|Vector|MakeVector _returnvalue_19 (/ (* _gridsize_5 (Math|Float|Round (/ (.Distance (Collision|BreakHitResult _hitresoultoutput)) _gridsize_5))) _meshsizex_2) 1.0)) _returnvalue_16)))))) (Math|Transform|MakeTransform (Math|Vector|RotateVector (Math|Vector|SelectVector "20.000000,-20.000000, 0" "-20.000000,-20.000000, 0" (Variables|GetReverse)) (.rotation (Utilities|Select :Index _coordinatespace :Local (Math|Transform|MakeTransform (- (+ (Utilities|ZeroLocalorWorldSpaceZ _coordinatespace_3 (Spline|GetLocationatSplinePoint Spline (Math|Integer|SelectInt (- (+ _index OffsetIndex) _returnvalue_2) (+ _index OffsetIndex) (> (+ _index OffsetIndex) _returnvalue_3)) CoordinateSpace)) (* (Math|Vector|Normalize (- (Utilities|ZeroLocalorWorldSpaceZ _coordinatespace_2 (Spline|GetLocationatSplinePoint Spline (Math|Integer|SelectInt 0 (+ (Math|Integer|SelectInt (- (+ _index OffsetIndex) _returnvalue_2) (+ _index OffsetIndex) (> (+ _index OffsetIndex) _returnvalue_3)) 1) (== (Math|Integer|SelectInt (- (+ _index OffsetIndex) _returnvalue_2) (+ _index OffsetIndex) (> (+ _index OffsetIndex) _returnvalue_3)) _returnvalue_3)) CoordinateSpace)) (Utilities|ZeroLocalorWorldSpaceZ _coordinatespace_3 (Spline|GetLocationatSplinePoint Spline (Math|Integer|SelectInt (- (+ _index OffsetIndex) _returnvalue_2) (+ _index OffsetIndex) (> (+ _index OffsetIndex) _returnvalue_3)) CoordinateSpace)))) _returnvalue_15)) _returnvalue_11) (Math|Rotator|SelectRotator (Math|Rotator|MakeRotationfromAxes (- (Utilities|ZeroLocalorWorldSpaceZ _coordinatespace_2 (Spline|GetLocationatSplinePoint Spline (Math|Integer|SelectInt 0 (+ (Math|Integer|SelectInt (- (+ _index OffsetIndex) _returnvalue_2) (+ _index OffsetIndex) (> (+ _index OffsetIndex) _returnvalue_3)) 1) (== (Math|Integer|SelectInt (- (+ _index OffsetIndex) _returnvalue_2) (+ _index OffsetIndex) (> (+ _index OffsetIndex) _returnvalue_3)) _returnvalue_3)) CoordinateSpace)) (Utilities|ZeroLocalorWorldSpaceZ _coordinatespace_3 (Spline|GetLocationatSplinePoint Spline (Math|Integer|SelectInt (- (+ _index OffsetIndex) _returnvalue_2) (+ _index OffsetIndex) (> (+ _index OffsetIndex) _returnvalue_3)) CoordinateSpace))) (Math|Vector|NegateVector (Math|Vector|CrossProduct (- (Utilities|ZeroLocalorWorldSpaceZ _coordinatespace_2 (Spline|GetLocationatSplinePoint Spline (Math|Integer|SelectInt 0 (+ (Math|Integer|SelectInt (- (+ _index OffsetIndex) _returnvalue_2) (+ _index OffsetIndex) (> (+ _index OffsetIndex) _returnvalue_3)) 1) (== (Math|Integer|SelectInt (- (+ _index OffsetIndex) _returnvalue_2) (+ _index OffsetIndex) (> (+ _index OffsetIndex) _returnvalue_3)) _returnvalue_3)) CoordinateSpace)) (Utilities|ZeroLocalorWorldSpaceZ _coordinatespace_3 (Spline|GetLocationatSplinePoint Spline (Math|Integer|SelectInt (- (+ _index OffsetIndex) _returnvalue_2) (+ _index OffsetIndex) (> (+ _index OffsetIndex) _returnvalue_3)) CoordinateSpace))) "0, 0,1.000000")) "0, 0,1.000000") (Math|Rotator|MakeRotfromX (Math|Vector|NegateVector (- (Utilities|ZeroLocalorWorldSpaceZ _coordinatespace_2 (Spline|GetLocationatSplinePoint Spline (Math|Integer|SelectInt 0 (+ (Math|Integer|SelectInt (- (+ _index OffsetIndex) _returnvalue_2) (+ _index OffsetIndex) (> (+ _index OffsetIndex) _returnvalue_3)) 1) (== (Math|Integer|SelectInt (- (+ _index OffsetIndex) _returnvalue_2) (+ _index OffsetIndex) (> (+ _index OffsetIndex) _returnvalue_3)) _returnvalue_3)) CoordinateSpace)) (Utilities|ZeroLocalorWorldSpaceZ _coordinatespace_3 (Spline|GetLocationatSplinePoint Spline (Math|Integer|SelectInt (- (+ _index OffsetIndex) _returnvalue_2) (+ _index OffsetIndex) (> (+ _index OffsetIndex) _returnvalue_3)) CoordinateSpace))))) _reverse_2) (Math|Vector|MakeVector _returnvalue_19 (/ (* _gridsize_5 (Math|Float|Round (/ (.Distance (Collision|BreakHitResult _hitresoultoutput)) _gridsize_5))) _meshsizex_2) 1.0)) :World (Math|Transform|MakeRelativeTransform (Math|Transform|MakeTransform (- (+ (Utilities|ZeroLocalorWorldSpaceZ _coordinatespace_3 (Spline|GetLocationatSplinePoint Spline (Math|Integer|SelectInt (- (+ _index OffsetIndex) _returnvalue_2) (+ _index OffsetIndex) (> (+ _index OffsetIndex) _returnvalue_3)) CoordinateSpace)) (* (Math|Vector|Normalize (- (Utilities|ZeroLocalorWorldSpaceZ _coordinatespace_2 (Spline|GetLocationatSplinePoint Spline (Math|Integer|SelectInt 0 (+ (Math|Integer|SelectInt (- (+ _index OffsetIndex) _returnvalue_2) (+ _index OffsetIndex) (> (+ _index OffsetIndex) _returnvalue_3)) 1) (== (Math|Integer|SelectInt (- (+ _index OffsetIndex) _returnvalue_2) (+ _index OffsetIndex) (> (+ _index OffsetIndex) _returnvalue_3)) _returnvalue_3)) CoordinateSpace)) (Utilities|ZeroLocalorWorldSpaceZ _coordinatespace_3 (Spline|GetLocationatSplinePoint Spline (Math|Integer|SelectInt (- (+ _index OffsetIndex) _returnvalue_2) (+ _index OffsetIndex) (> (+ _index OffsetIndex) _returnvalue_3)) CoordinateSpace)))) _returnvalue_15)) _returnvalue_11) (Math|Rotator|SelectRotator (Math|Rotator|MakeRotationfromAxes (- (Utilities|ZeroLocalorWorldSpaceZ _coordinatespace_2 (Spline|GetLocationatSplinePoint Spline (Math|Integer|SelectInt 0 (+ (Math|Integer|SelectInt (- (+ _index OffsetIndex) _returnvalue_2) (+ _index OffsetIndex) (> (+ _index OffsetIndex) _returnvalue_3)) 1) (== (Math|Integer|SelectInt (- (+ _index OffsetIndex) _returnvalue_2) (+ _index OffsetIndex) (> (+ _index OffsetIndex) _returnvalue_3)) _returnvalue_3)) CoordinateSpace)) (Utilities|ZeroLocalorWorldSpaceZ _coordinatespace_3 (Spline|GetLocationatSplinePoint Spline (Math|Integer|SelectInt (- (+ _index OffsetIndex) _returnvalue_2) (+ _index OffsetIndex) (> (+ _index OffsetIndex) _returnvalue_3)) CoordinateSpace))) (Math|Vector|NegateVector (Math|Vector|CrossProduct (- (Utilities|ZeroLocalorWorldSpaceZ _coordinatespace_2 (Spline|GetLocationatSplinePoint Spline (Math|Integer|SelectInt 0 (+ (Math|Integer|SelectInt (- (+ _index OffsetIndex) _returnvalue_2) (+ _index OffsetIndex) (> (+ _index OffsetIndex) _returnvalue_3)) 1) (== (Math|Integer|SelectInt (- (+ _index OffsetIndex) _returnvalue_2) (+ _index OffsetIndex) (> (+ _index OffsetIndex) _returnvalue_3)) _returnvalue_3)) CoordinateSpace)) (Utilities|ZeroLocalorWorldSpaceZ _coordinatespace_3 (Spline|GetLocationatSplinePoint Spline (Math|Integer|SelectInt (- (+ _index OffsetIndex) _returnvalue_2) (+ _index OffsetIndex) (> (+ _index OffsetIndex) _returnvalue_3)) CoordinateSpace))) "0, 0,1.000000")) "0, 0,1.000000") (Math|Rotator|MakeRotfromX (Math|Vector|NegateVector (- (Utilities|ZeroLocalorWorldSpaceZ _coordinatespace_2 (Spline|GetLocationatSplinePoint Spline (Math|Integer|SelectInt 0 (+ (Math|Integer|SelectInt (- (+ _index OffsetIndex) _returnvalue_2) (+ _index OffsetIndex) (> (+ _index OffsetIndex) _returnvalue_3)) 1) (== (Math|Integer|SelectInt (- (+ _index OffsetIndex) _returnvalue_2) (+ _index OffsetIndex) (> (+ _index OffsetIndex) _returnvalue_3)) _returnvalue_3)) CoordinateSpace)) (Utilities|ZeroLocalorWorldSpaceZ _coordinatespace_3 (Spline|GetLocationatSplinePoint Spline (Math|Integer|SelectInt (- (+ _index OffsetIndex) _returnvalue_2) (+ _index OffsetIndex) (> (+ _index OffsetIndex) _returnvalue_3)) CoordinateSpace))))) _reverse_2) (Math|Vector|MakeVector _returnvalue_19 (/ (* _gridsize_5 (Math|Float|Round (/ (.Distance (Collision|BreakHitResult _hitresoultoutput)) _gridsize_5))) _meshsizex_2) 1.0)) _returnvalue_16))))))))
                      (Utilities|Array|Add (Variables|MeshLists|GetTempCollisionList) _returnvalue_22)))
                  (else
                    (Utilities|FlowControl|Sequence
                      (:then_1
                        (if (and (> (Utilities|Array|Length _transformlist) 0) (>= (LocalVariables|GetCurrentIndex) 0))
                          (Utilities|FlowControl|Sequence
                            (:then_0
                              (Utilities|Array|SetArrayElem _transformlist _currentindex (Math|Transform|MakeTransform (.location _output_3) (.rotation _output_3) _returnvalue_14))))))
                      (:then_2
                        (Utilities|IsValid _output_4
                          (:Is%20Valid
                            (Transformation|SetRelativeScale3D _output_4 (Math|Vector|MakeVector (/ (- (* (.x _returnvalue_14) (LocalVariables|GetGridSize)) (Math|Float|SelectFloat 4e+01 -4e+01 (Variables|GetReverse))) (LocalVariables|GetGridSize)) (/ (- (* (.y _returnvalue_14) (LocalVariables|GetMeshSizeX)) 4e+01) (LocalVariables|GetMeshSizeX)) (.z _returnvalue_14)))))))))
                (else
                  (Variables|MakeBasementFloor|SetNewFloor true)
                  (Variables|MakeBasementFloor|SetcurrentLength 0.0))))
            (Variables|MakeBasementFloor|SetcurrentLength 0.0)
            (Variables|MakeBasementFloor|SetMeshSpawn%20Iteration 0)
            (Variables|MakeBasementFloor|SetNewFloor true))
          ;[1 1 1] Reset
          (Variables|MakeBasementFloor|SetcurrentLength 0.0)
          (Variables|MakeBasementFloor|SetCurrentFloorMesh 0)
          (LocalVariables|SetCurrentIndex -1))
        (:then_1
          ;[1 1 1] Spawn From List
          (if (Variables|GetSpawnComponents)
            (for _array_element (LocalVariables|GetTransformList)
              (Utilities|Array|Add (Variables|X.%29DoNOTEdit|GetComponent%20Transform%20List) (Utilities|Struct|MakeProcComponenttoSpawn _array_element (Variables|GetProc%20Asset%20List) (Utilities|Array|Get(acopy) (Variables|GetAssetIndexList)))))))
        (:then_2
          (return (Utilities|Array|Length (LocalVariables|GetTransformList))))))))

(fn MakeBasementFloorCornerMissing
  (params (Spline :type Spline%20Component%20Object%20Reference) (Asset%20Index%20List :type Array%20of%20Integers) (Proc%20Asset%20List :type Proc%20Asset%20List%20Object%20Reference) (Reverse :type Boolean) (CoordinateSpace :type ESplineCoordinateSpace%20Enum :default "Local") (SpawnComponents :type Boolean :default true) (Debug :type Boolean) (Top :type Boolean))
  (bind _top (Variables|GetTop))
  (bind _numberoffloors (Variables|Default|GetNumberOfFloors))
  (bind _returnvalue (* _numberoffloors 384.0))
  (bind _returnvalue_2 (Math|Float|SelectFloat 0.0 _returnvalue _top))
  (bind _returnvalue_3 3e+01)
  (bind _returnvalue_4 0.25)
  (bind _returnvalue_5 (Math|Vector|MakeVector 0.0 0.0 _returnvalue_2))
  (bind _returnvalue_6 (Transformation|GetActorTransform))
  (bind _array (Utilities|Array|MakeArray "ObjectTypeQuery3"))
  (bind _reverse (Variables|GetReverse))
  (bind _coordinatespace (Variables|GetCoordinateSpace))
  (bind _debug (Variables|GetDebug))
  (bind _reverse_2 (Variables|GetReverse))
  (bind _coordinatespace_2 (Variables|GetCoordinateSpace))
  (bind _test (Variables|Debug|GetTest))
  (Utilities|FlowControl|Sequence
    (:then_0
      (for _index (range (Spline|GetNumberOfSplinePoints Spline))
        (bind (_next__1 _next _current _previous _previous__1 _world_next__1 _world_next _world_current _world_previous) (Tools|GetCurrentNextPreviousSplinePoint Spline _index (Variables|GetDebug)))
        (Utilities|FlowControl|Sequence
          (:then_0
            (if false
              (Debug|DebugText (Math|Transform|MakeTransform (Utilities|ZeroLocalorWorldSpaceZ _coordinatespace (Utilities|Select :Index CoordinateSpace :Local _current :World (Utilities|ZeroLocationWorldSpaceZ _world_current)))))))
          (:then_1
            (bind (_outhits _returnvalue_7) (Collision|MultiLineTraceForObjects :Start (+ (Utilities|Operators|Add :A (Utilities|ZeroLocationWorldSpaceZ _world_current) :B (Math|Vector|SelectVector (Math|Vector|NegateVector (+ (* (Math|Vector|GetUnitDirection(Vector) (Utilities|ZeroLocationWorldSpaceZ _world_current) (Utilities|ZeroLocationWorldSpaceZ _world_next)) _returnvalue_3) (* (Math|Vector|GetUnitDirection(Vector) (Utilities|ZeroLocationWorldSpaceZ _world_current) (Utilities|ZeroLocationWorldSpaceZ _world_previous)) _returnvalue_3))) (+ (* (Math|Vector|GetUnitDirection(Vector) (Utilities|ZeroLocationWorldSpaceZ _world_current) (Utilities|ZeroLocationWorldSpaceZ _world_next)) _returnvalue_3) (* (Math|Vector|GetUnitDirection(Vector) (Utilities|ZeroLocationWorldSpaceZ _world_current) (Utilities|ZeroLocationWorldSpaceZ _world_previous)) _returnvalue_3)) (select _reverse (not (Utilities|IsOuterCorner (Utilities|ZeroLocalorWorldSpaceZ _coordinatespace (Utilities|Select :Index CoordinateSpace :Local _next :World (Utilities|ZeroLocationWorldSpaceZ _world_next))) (Utilities|ZeroLocalorWorldSpaceZ _coordinatespace (Utilities|Select :Index CoordinateSpace :Local _current :World (Utilities|ZeroLocationWorldSpaceZ _world_current))) (Utilities|ZeroLocalorWorldSpaceZ _coordinatespace (Utilities|Select :Index CoordinateSpace :Local _previous :World (Utilities|ZeroLocationWorldSpaceZ _world_previous))))) (Utilities|IsOuterCorner (Utilities|ZeroLocalorWorldSpaceZ _coordinatespace (Utilities|Select :Index CoordinateSpace :Local _next :World (Utilities|ZeroLocationWorldSpaceZ _world_next))) (Utilities|ZeroLocalorWorldSpaceZ _coordinatespace (Utilities|Select :Index CoordinateSpace :Local _current :World (Utilities|ZeroLocationWorldSpaceZ _world_current))) (Utilities|ZeroLocalorWorldSpaceZ _coordinatespace (Utilities|Select :Index CoordinateSpace :Local _previous :World (Utilities|ZeroLocationWorldSpaceZ _world_previous))))))) "0,0,256.000000") :End (Utilities|Operators|Add :A (Utilities|ZeroLocationWorldSpaceZ _world_current) :B (Math|Vector|SelectVector (Math|Vector|NegateVector (+ (* (Math|Vector|GetUnitDirection(Vector) (Utilities|ZeroLocationWorldSpaceZ _world_current) (Utilities|ZeroLocationWorldSpaceZ _world_next)) _returnvalue_3) (* (Math|Vector|GetUnitDirection(Vector) (Utilities|ZeroLocationWorldSpaceZ _world_current) (Utilities|ZeroLocationWorldSpaceZ _world_previous)) _returnvalue_3))) (+ (* (Math|Vector|GetUnitDirection(Vector) (Utilities|ZeroLocationWorldSpaceZ _world_current) (Utilities|ZeroLocationWorldSpaceZ _world_next)) _returnvalue_3) (* (Math|Vector|GetUnitDirection(Vector) (Utilities|ZeroLocationWorldSpaceZ _world_current) (Utilities|ZeroLocationWorldSpaceZ _world_previous)) _returnvalue_3)) (select _reverse (not (Utilities|IsOuterCorner (Utilities|ZeroLocalorWorldSpaceZ _coordinatespace (Utilities|Select :Index CoordinateSpace :Local _next :World (Utilities|ZeroLocationWorldSpaceZ _world_next))) (Utilities|ZeroLocalorWorldSpaceZ _coordinatespace (Utilities|Select :Index CoordinateSpace :Local _current :World (Utilities|ZeroLocationWorldSpaceZ _world_current))) (Utilities|ZeroLocalorWorldSpaceZ _coordinatespace (Utilities|Select :Index CoordinateSpace :Local _previous :World (Utilities|ZeroLocationWorldSpaceZ _world_previous))))) (Utilities|IsOuterCorner (Utilities|ZeroLocalorWorldSpaceZ _coordinatespace (Utilities|Select :Index CoordinateSpace :Local _next :World (Utilities|ZeroLocationWorldSpaceZ _world_next))) (Utilities|ZeroLocalorWorldSpaceZ _coordinatespace (Utilities|Select :Index CoordinateSpace :Local _current :World (Utilities|ZeroLocationWorldSpaceZ _world_current))) (Utilities|ZeroLocalorWorldSpaceZ _coordinatespace (Utilities|Select :Index CoordinateSpace :Local _previous :World (Utilities|ZeroLocationWorldSpaceZ _world_previous))))))) :ObjectTypes (Utilities|Array|MakeArray "ObjectTypeQuery3") :bTraceComplex false :DrawDebugType "None" :bIgnoreSelf false :TraceColor "(R=0.000000,G=0.050724,B=1.000000,A=1.000000)" :TraceHitColor "(R=1.000000,G=0.910421,B=0.000000,A=1.000000)" :DrawTime 5.0))
            (if (not _returnvalue_7)
              (bind _outhits_2 (Collision|MultiLineTraceForObjects :Start (- (+ (Utilities|ZeroLocationWorldSpaceZ _world_current) (Math|Vector|SelectVector (Math|Vector|NegateVector (+ (* (Math|Vector|GetUnitDirection(Vector) (Utilities|ZeroLocationWorldSpaceZ _world_current) (Utilities|ZeroLocationWorldSpaceZ _world_next)) _returnvalue_3) (* (Math|Vector|GetUnitDirection(Vector) (Utilities|ZeroLocationWorldSpaceZ _world_current) (Utilities|ZeroLocationWorldSpaceZ _world_previous)) _returnvalue_3))) (+ (* (Math|Vector|GetUnitDirection(Vector) (Utilities|ZeroLocationWorldSpaceZ _world_current) (Utilities|ZeroLocationWorldSpaceZ _world_next)) _returnvalue_3) (* (Math|Vector|GetUnitDirection(Vector) (Utilities|ZeroLocationWorldSpaceZ _world_current) (Utilities|ZeroLocationWorldSpaceZ _world_previous)) _returnvalue_3)) (select _reverse (not (Utilities|IsOuterCorner (Utilities|ZeroLocalorWorldSpaceZ _coordinatespace (Utilities|Select :Index CoordinateSpace :Local _next :World (Utilities|ZeroLocationWorldSpaceZ _world_next))) (Utilities|ZeroLocalorWorldSpaceZ _coordinatespace (Utilities|Select :Index CoordinateSpace :Local _current :World (Utilities|ZeroLocationWorldSpaceZ _world_current))) (Utilities|ZeroLocalorWorldSpaceZ _coordinatespace (Utilities|Select :Index CoordinateSpace :Local _previous :World (Utilities|ZeroLocationWorldSpaceZ _world_previous))))) (Utilities|IsOuterCorner (Utilities|ZeroLocalorWorldSpaceZ _coordinatespace (Utilities|Select :Index CoordinateSpace :Local _next :World (Utilities|ZeroLocationWorldSpaceZ _world_next))) (Utilities|ZeroLocalorWorldSpaceZ _coordinatespace (Utilities|Select :Index CoordinateSpace :Local _current :World (Utilities|ZeroLocationWorldSpaceZ _world_current))) (Utilities|ZeroLocalorWorldSpaceZ _coordinatespace (Utilities|Select :Index CoordinateSpace :Local _previous :World (Utilities|ZeroLocationWorldSpaceZ _world_previous))))))) (* (Math|Vector|SelectVector (Math|Vector|GetUnitDirection(Vector) (Utilities|ZeroLocationWorldSpaceZ _world_previous) (Utilities|ZeroLocationWorldSpaceZ _world_current)) (Math|Vector|NegateVector (Math|Vector|GetUnitDirection(Vector) (Utilities|ZeroLocationWorldSpaceZ _world_previous) (Utilities|ZeroLocationWorldSpaceZ _world_current))) (select _reverse (not (Utilities|IsOuterCorner (Utilities|ZeroLocalorWorldSpaceZ _coordinatespace (Utilities|Select :Index CoordinateSpace :Local _next :World (Utilities|ZeroLocationWorldSpaceZ _world_next))) (Utilities|ZeroLocalorWorldSpaceZ _coordinatespace (Utilities|Select :Index CoordinateSpace :Local _current :World (Utilities|ZeroLocationWorldSpaceZ _world_current))) (Utilities|ZeroLocalorWorldSpaceZ _coordinatespace (Utilities|Select :Index CoordinateSpace :Local _previous :World (Utilities|ZeroLocationWorldSpaceZ _world_previous))))) (Utilities|IsOuterCorner (Utilities|ZeroLocalorWorldSpaceZ _coordinatespace (Utilities|Select :Index CoordinateSpace :Local _next :World (Utilities|ZeroLocationWorldSpaceZ _world_next))) (Utilities|ZeroLocalorWorldSpaceZ _coordinatespace (Utilities|Select :Index CoordinateSpace :Local _current :World (Utilities|ZeroLocationWorldSpaceZ _world_current))) (Utilities|ZeroLocalorWorldSpaceZ _coordinatespace (Utilities|Select :Index CoordinateSpace :Local _previous :World (Utilities|ZeroLocationWorldSpaceZ _world_previous)))))) _returnvalue_4)) :End (+ (- (+ (Utilities|ZeroLocationWorldSpaceZ _world_current) (Math|Vector|SelectVector (Math|Vector|NegateVector (+ (* (Math|Vector|GetUnitDirection(Vector) (Utilities|ZeroLocationWorldSpaceZ _world_current) (Utilities|ZeroLocationWorldSpaceZ _world_next)) _returnvalue_3) (* (Math|Vector|GetUnitDirection(Vector) (Utilities|ZeroLocationWorldSpaceZ _world_current) (Utilities|ZeroLocationWorldSpaceZ _world_previous)) _returnvalue_3))) (+ (* (Math|Vector|GetUnitDirection(Vector) (Utilities|ZeroLocationWorldSpaceZ _world_current) (Utilities|ZeroLocationWorldSpaceZ _world_next)) _returnvalue_3) (* (Math|Vector|GetUnitDirection(Vector) (Utilities|ZeroLocationWorldSpaceZ _world_current) (Utilities|ZeroLocationWorldSpaceZ _world_previous)) _returnvalue_3)) (select _reverse (not (Utilities|IsOuterCorner (Utilities|ZeroLocalorWorldSpaceZ _coordinatespace (Utilities|Select :Index CoordinateSpace :Local _next :World (Utilities|ZeroLocationWorldSpaceZ _world_next))) (Utilities|ZeroLocalorWorldSpaceZ _coordinatespace (Utilities|Select :Index CoordinateSpace :Local _current :World (Utilities|ZeroLocationWorldSpaceZ _world_current))) (Utilities|ZeroLocalorWorldSpaceZ _coordinatespace (Utilities|Select :Index CoordinateSpace :Local _previous :World (Utilities|ZeroLocationWorldSpaceZ _world_previous))))) (Utilities|IsOuterCorner (Utilities|ZeroLocalorWorldSpaceZ _coordinatespace (Utilities|Select :Index CoordinateSpace :Local _next :World (Utilities|ZeroLocationWorldSpaceZ _world_next))) (Utilities|ZeroLocalorWorldSpaceZ _coordinatespace (Utilities|Select :Index CoordinateSpace :Local _current :World (Utilities|ZeroLocationWorldSpaceZ _world_current))) (Utilities|ZeroLocalorWorldSpaceZ _coordinatespace (Utilities|Select :Index CoordinateSpace :Local _previous :World (Utilities|ZeroLocationWorldSpaceZ _world_previous))))))) (* (Math|Vector|SelectVector (Math|Vector|GetUnitDirection(Vector) (Utilities|ZeroLocationWorldSpaceZ _world_previous) (Utilities|ZeroLocationWorldSpaceZ _world_current)) (Math|Vector|NegateVector (Math|Vector|GetUnitDirection(Vector) (Utilities|ZeroLocationWorldSpaceZ _world_previous) (Utilities|ZeroLocationWorldSpaceZ _world_current))) (select _reverse (not (Utilities|IsOuterCorner (Utilities|ZeroLocalorWorldSpaceZ _coordinatespace (Utilities|Select :Index CoordinateSpace :Local _next :World (Utilities|ZeroLocationWorldSpaceZ _world_next))) (Utilities|ZeroLocalorWorldSpaceZ _coordinatespace (Utilities|Select :Index CoordinateSpace :Local _current :World (Utilities|ZeroLocationWorldSpaceZ _world_current))) (Utilities|ZeroLocalorWorldSpaceZ _coordinatespace (Utilities|Select :Index CoordinateSpace :Local _previous :World (Utilities|ZeroLocationWorldSpaceZ _world_previous))))) (Utilities|IsOuterCorner (Utilities|ZeroLocalorWorldSpaceZ _coordinatespace (Utilities|Select :Index CoordinateSpace :Local _next :World (Utilities|ZeroLocationWorldSpaceZ _world_next))) (Utilities|ZeroLocalorWorldSpaceZ _coordinatespace (Utilities|Select :Index CoordinateSpace :Local _current :World (Utilities|ZeroLocationWorldSpaceZ _world_current))) (Utilities|ZeroLocalorWorldSpaceZ _coordinatespace (Utilities|Select :Index CoordinateSpace :Local _previous :World (Utilities|ZeroLocationWorldSpaceZ _world_previous)))))) _returnvalue_4)) (* (Math|Vector|SelectVector (Math|Vector|GetUnitDirection(Vector) (Utilities|ZeroLocationWorldSpaceZ _world_previous) (Utilities|ZeroLocationWorldSpaceZ _world_current)) (Math|Vector|NegateVector (Math|Vector|GetUnitDirection(Vector) (Utilities|ZeroLocationWorldSpaceZ _world_previous) (Utilities|ZeroLocationWorldSpaceZ _world_current))) (select _reverse (not (Utilities|IsOuterCorner (Utilities|ZeroLocalorWorldSpaceZ _coordinatespace (Utilities|Select :Index CoordinateSpace :Local _next :World (Utilities|ZeroLocationWorldSpaceZ _world_next))) (Utilities|ZeroLocalorWorldSpaceZ _coordinatespace (Utilities|Select :Index CoordinateSpace :Local _current :World (Utilities|ZeroLocationWorldSpaceZ _world_current))) (Utilities|ZeroLocalorWorldSpaceZ _coordinatespace (Utilities|Select :Index CoordinateSpace :Local _previous :World (Utilities|ZeroLocationWorldSpaceZ _world_previous))))) (Utilities|IsOuterCorner (Utilities|ZeroLocalorWorldSpaceZ _coordinatespace (Utilities|Select :Index CoordinateSpace :Local _next :World (Utilities|ZeroLocationWorldSpaceZ _world_next))) (Utilities|ZeroLocalorWorldSpaceZ _coordinatespace (Utilities|Select :Index CoordinateSpace :Local _current :World (Utilities|ZeroLocationWorldSpaceZ _world_current))) (Utilities|ZeroLocalorWorldSpaceZ _coordinatespace (Utilities|Select :Index CoordinateSpace :Local _previous :World (Utilities|ZeroLocationWorldSpaceZ _world_previous)))))) 2e+04)) :ObjectTypes _array :bTraceComplex false :DrawDebugType (select (and _debug (== _index _test)) "ForDuration" "None") :bIgnoreSelf false :TraceColor "(R=0.000000,G=0.050724,B=1.000000,A=1.000000)" :TraceHitColor "(R=1.000000,G=0.910421,B=0.000000,A=1.000000)" :DrawTime 5.0))
              (bind _hitresoultoutput (Tools|ClosestHitResult _outhits_2 (- (+ (Utilities|ZeroLocationWorldSpaceZ _world_current) (Math|Vector|SelectVector (Math|Vector|NegateVector (+ (* (Math|Vector|GetUnitDirection(Vector) (Utilities|ZeroLocationWorldSpaceZ _world_current) (Utilities|ZeroLocationWorldSpaceZ _world_next)) _returnvalue_3) (* (Math|Vector|GetUnitDirection(Vector) (Utilities|ZeroLocationWorldSpaceZ _world_current) (Utilities|ZeroLocationWorldSpaceZ _world_previous)) _returnvalue_3))) (+ (* (Math|Vector|GetUnitDirection(Vector) (Utilities|ZeroLocationWorldSpaceZ _world_current) (Utilities|ZeroLocationWorldSpaceZ _world_next)) _returnvalue_3) (* (Math|Vector|GetUnitDirection(Vector) (Utilities|ZeroLocationWorldSpaceZ _world_current) (Utilities|ZeroLocationWorldSpaceZ _world_previous)) _returnvalue_3)) (select _reverse (not (Utilities|IsOuterCorner (Utilities|ZeroLocalorWorldSpaceZ _coordinatespace (Utilities|Select :Index CoordinateSpace :Local _next :World (Utilities|ZeroLocationWorldSpaceZ _world_next))) (Utilities|ZeroLocalorWorldSpaceZ _coordinatespace (Utilities|Select :Index CoordinateSpace :Local _current :World (Utilities|ZeroLocationWorldSpaceZ _world_current))) (Utilities|ZeroLocalorWorldSpaceZ _coordinatespace (Utilities|Select :Index CoordinateSpace :Local _previous :World (Utilities|ZeroLocationWorldSpaceZ _world_previous))))) (Utilities|IsOuterCorner (Utilities|ZeroLocalorWorldSpaceZ _coordinatespace (Utilities|Select :Index CoordinateSpace :Local _next :World (Utilities|ZeroLocationWorldSpaceZ _world_next))) (Utilities|ZeroLocalorWorldSpaceZ _coordinatespace (Utilities|Select :Index CoordinateSpace :Local _current :World (Utilities|ZeroLocationWorldSpaceZ _world_current))) (Utilities|ZeroLocalorWorldSpaceZ _coordinatespace (Utilities|Select :Index CoordinateSpace :Local _previous :World (Utilities|ZeroLocationWorldSpaceZ _world_previous))))))) (* (Math|Vector|SelectVector (Math|Vector|GetUnitDirection(Vector) (Utilities|ZeroLocationWorldSpaceZ _world_previous) (Utilities|ZeroLocationWorldSpaceZ _world_current)) (Math|Vector|NegateVector (Math|Vector|GetUnitDirection(Vector) (Utilities|ZeroLocationWorldSpaceZ _world_previous) (Utilities|ZeroLocationWorldSpaceZ _world_current))) (select _reverse (not (Utilities|IsOuterCorner (Utilities|ZeroLocalorWorldSpaceZ _coordinatespace (Utilities|Select :Index CoordinateSpace :Local _next :World (Utilities|ZeroLocationWorldSpaceZ _world_next))) (Utilities|ZeroLocalorWorldSpaceZ _coordinatespace (Utilities|Select :Index CoordinateSpace :Local _current :World (Utilities|ZeroLocationWorldSpaceZ _world_current))) (Utilities|ZeroLocalorWorldSpaceZ _coordinatespace (Utilities|Select :Index CoordinateSpace :Local _previous :World (Utilities|ZeroLocationWorldSpaceZ _world_previous))))) (Utilities|IsOuterCorner (Utilities|ZeroLocalorWorldSpaceZ _coordinatespace (Utilities|Select :Index CoordinateSpace :Local _next :World (Utilities|ZeroLocationWorldSpaceZ _world_next))) (Utilities|ZeroLocalorWorldSpaceZ _coordinatespace (Utilities|Select :Index CoordinateSpace :Local _current :World (Utilities|ZeroLocationWorldSpaceZ _world_current))) (Utilities|ZeroLocalorWorldSpaceZ _coordinatespace (Utilities|Select :Index CoordinateSpace :Local _previous :World (Utilities|ZeroLocationWorldSpaceZ _world_previous)))))) _returnvalue_4))))
              (bind _outhits_3 (Collision|MultiLineTraceForObjects :Start (- (+ (Utilities|ZeroLocationWorldSpaceZ _world_current) (Math|Vector|SelectVector (Math|Vector|NegateVector (+ (* (Math|Vector|GetUnitDirection(Vector) (Utilities|ZeroLocationWorldSpaceZ _world_current) (Utilities|ZeroLocationWorldSpaceZ _world_next)) _returnvalue_3) (* (Math|Vector|GetUnitDirection(Vector) (Utilities|ZeroLocationWorldSpaceZ _world_current) (Utilities|ZeroLocationWorldSpaceZ _world_previous)) _returnvalue_3))) (+ (* (Math|Vector|GetUnitDirection(Vector) (Utilities|ZeroLocationWorldSpaceZ _world_current) (Utilities|ZeroLocationWorldSpaceZ _world_next)) _returnvalue_3) (* (Math|Vector|GetUnitDirection(Vector) (Utilities|ZeroLocationWorldSpaceZ _world_current) (Utilities|ZeroLocationWorldSpaceZ _world_previous)) _returnvalue_3)) (select _reverse (not (Utilities|IsOuterCorner (Utilities|ZeroLocalorWorldSpaceZ _coordinatespace (Utilities|Select :Index CoordinateSpace :Local _next :World (Utilities|ZeroLocationWorldSpaceZ _world_next))) (Utilities|ZeroLocalorWorldSpaceZ _coordinatespace (Utilities|Select :Index CoordinateSpace :Local _current :World (Utilities|ZeroLocationWorldSpaceZ _world_current))) (Utilities|ZeroLocalorWorldSpaceZ _coordinatespace (Utilities|Select :Index CoordinateSpace :Local _previous :World (Utilities|ZeroLocationWorldSpaceZ _world_previous))))) (Utilities|IsOuterCorner (Utilities|ZeroLocalorWorldSpaceZ _coordinatespace (Utilities|Select :Index CoordinateSpace :Local _next :World (Utilities|ZeroLocationWorldSpaceZ _world_next))) (Utilities|ZeroLocalorWorldSpaceZ _coordinatespace (Utilities|Select :Index CoordinateSpace :Local _current :World (Utilities|ZeroLocationWorldSpaceZ _world_current))) (Utilities|ZeroLocalorWorldSpaceZ _coordinatespace (Utilities|Select :Index CoordinateSpace :Local _previous :World (Utilities|ZeroLocationWorldSpaceZ _world_previous))))))) (* (Math|Vector|SelectVector (Math|Vector|GetUnitDirection(Vector) (Utilities|ZeroLocationWorldSpaceZ _world_next) (Utilities|ZeroLocationWorldSpaceZ _world_current)) (Math|Vector|NegateVector (Math|Vector|GetUnitDirection(Vector) (Utilities|ZeroLocationWorldSpaceZ _world_next) (Utilities|ZeroLocationWorldSpaceZ _world_current))) (select _reverse (not (Utilities|IsOuterCorner (Utilities|ZeroLocalorWorldSpaceZ _coordinatespace (Utilities|Select :Index CoordinateSpace :Local _next :World (Utilities|ZeroLocationWorldSpaceZ _world_next))) (Utilities|ZeroLocalorWorldSpaceZ _coordinatespace (Utilities|Select :Index CoordinateSpace :Local _current :World (Utilities|ZeroLocationWorldSpaceZ _world_current))) (Utilities|ZeroLocalorWorldSpaceZ _coordinatespace (Utilities|Select :Index CoordinateSpace :Local _previous :World (Utilities|ZeroLocationWorldSpaceZ _world_previous))))) (Utilities|IsOuterCorner (Utilities|ZeroLocalorWorldSpaceZ _coordinatespace (Utilities|Select :Index CoordinateSpace :Local _next :World (Utilities|ZeroLocationWorldSpaceZ _world_next))) (Utilities|ZeroLocalorWorldSpaceZ _coordinatespace (Utilities|Select :Index CoordinateSpace :Local _current :World (Utilities|ZeroLocationWorldSpaceZ _world_current))) (Utilities|ZeroLocalorWorldSpaceZ _coordinatespace (Utilities|Select :Index CoordinateSpace :Local _previous :World (Utilities|ZeroLocationWorldSpaceZ _world_previous)))))) _returnvalue_4)) :End (+ (- (+ (Utilities|ZeroLocationWorldSpaceZ _world_current) (Math|Vector|SelectVector (Math|Vector|NegateVector (+ (* (Math|Vector|GetUnitDirection(Vector) (Utilities|ZeroLocationWorldSpaceZ _world_current) (Utilities|ZeroLocationWorldSpaceZ _world_next)) _returnvalue_3) (* (Math|Vector|GetUnitDirection(Vector) (Utilities|ZeroLocationWorldSpaceZ _world_current) (Utilities|ZeroLocationWorldSpaceZ _world_previous)) _returnvalue_3))) (+ (* (Math|Vector|GetUnitDirection(Vector) (Utilities|ZeroLocationWorldSpaceZ _world_current) (Utilities|ZeroLocationWorldSpaceZ _world_next)) _returnvalue_3) (* (Math|Vector|GetUnitDirection(Vector) (Utilities|ZeroLocationWorldSpaceZ _world_current) (Utilities|ZeroLocationWorldSpaceZ _world_previous)) _returnvalue_3)) (select _reverse (not (Utilities|IsOuterCorner (Utilities|ZeroLocalorWorldSpaceZ _coordinatespace (Utilities|Select :Index CoordinateSpace :Local _next :World (Utilities|ZeroLocationWorldSpaceZ _world_next))) (Utilities|ZeroLocalorWorldSpaceZ _coordinatespace (Utilities|Select :Index CoordinateSpace :Local _current :World (Utilities|ZeroLocationWorldSpaceZ _world_current))) (Utilities|ZeroLocalorWorldSpaceZ _coordinatespace (Utilities|Select :Index CoordinateSpace :Local _previous :World (Utilities|ZeroLocationWorldSpaceZ _world_previous))))) (Utilities|IsOuterCorner (Utilities|ZeroLocalorWorldSpaceZ _coordinatespace (Utilities|Select :Index CoordinateSpace :Local _next :World (Utilities|ZeroLocationWorldSpaceZ _world_next))) (Utilities|ZeroLocalorWorldSpaceZ _coordinatespace (Utilities|Select :Index CoordinateSpace :Local _current :World (Utilities|ZeroLocationWorldSpaceZ _world_current))) (Utilities|ZeroLocalorWorldSpaceZ _coordinatespace (Utilities|Select :Index CoordinateSpace :Local _previous :World (Utilities|ZeroLocationWorldSpaceZ _world_previous))))))) (* (Math|Vector|SelectVector (Math|Vector|GetUnitDirection(Vector) (Utilities|ZeroLocationWorldSpaceZ _world_next) (Utilities|ZeroLocationWorldSpaceZ _world_current)) (Math|Vector|NegateVector (Math|Vector|GetUnitDirection(Vector) (Utilities|ZeroLocationWorldSpaceZ _world_next) (Utilities|ZeroLocationWorldSpaceZ _world_current))) (select _reverse (not (Utilities|IsOuterCorner (Utilities|ZeroLocalorWorldSpaceZ _coordinatespace (Utilities|Select :Index CoordinateSpace :Local _next :World (Utilities|ZeroLocationWorldSpaceZ _world_next))) (Utilities|ZeroLocalorWorldSpaceZ _coordinatespace (Utilities|Select :Index CoordinateSpace :Local _current :World (Utilities|ZeroLocationWorldSpaceZ _world_current))) (Utilities|ZeroLocalorWorldSpaceZ _coordinatespace (Utilities|Select :Index CoordinateSpace :Local _previous :World (Utilities|ZeroLocationWorldSpaceZ _world_previous))))) (Utilities|IsOuterCorner (Utilities|ZeroLocalorWorldSpaceZ _coordinatespace (Utilities|Select :Index CoordinateSpace :Local _next :World (Utilities|ZeroLocationWorldSpaceZ _world_next))) (Utilities|ZeroLocalorWorldSpaceZ _coordinatespace (Utilities|Select :Index CoordinateSpace :Local _current :World (Utilities|ZeroLocationWorldSpaceZ _world_current))) (Utilities|ZeroLocalorWorldSpaceZ _coordinatespace (Utilities|Select :Index CoordinateSpace :Local _previous :World (Utilities|ZeroLocationWorldSpaceZ _world_previous)))))) _returnvalue_4)) (* (Math|Vector|SelectVector (Math|Vector|GetUnitDirection(Vector) (Utilities|ZeroLocationWorldSpaceZ _world_next) (Utilities|ZeroLocationWorldSpaceZ _world_current)) (Math|Vector|NegateVector (Math|Vector|GetUnitDirection(Vector) (Utilities|ZeroLocationWorldSpaceZ _world_next) (Utilities|ZeroLocationWorldSpaceZ _world_current))) (select _reverse (not (Utilities|IsOuterCorner (Utilities|ZeroLocalorWorldSpaceZ _coordinatespace (Utilities|Select :Index CoordinateSpace :Local _next :World (Utilities|ZeroLocationWorldSpaceZ _world_next))) (Utilities|ZeroLocalorWorldSpaceZ _coordinatespace (Utilities|Select :Index CoordinateSpace :Local _current :World (Utilities|ZeroLocationWorldSpaceZ _world_current))) (Utilities|ZeroLocalorWorldSpaceZ _coordinatespace (Utilities|Select :Index CoordinateSpace :Local _previous :World (Utilities|ZeroLocationWorldSpaceZ _world_previous))))) (Utilities|IsOuterCorner (Utilities|ZeroLocalorWorldSpaceZ _coordinatespace (Utilities|Select :Index CoordinateSpace :Local _next :World (Utilities|ZeroLocationWorldSpaceZ _world_next))) (Utilities|ZeroLocalorWorldSpaceZ _coordinatespace (Utilities|Select :Index CoordinateSpace :Local _current :World (Utilities|ZeroLocationWorldSpaceZ _world_current))) (Utilities|ZeroLocalorWorldSpaceZ _coordinatespace (Utilities|Select :Index CoordinateSpace :Local _previous :World (Utilities|ZeroLocationWorldSpaceZ _world_previous)))))) 2e+04)) :ObjectTypes _array :bTraceComplex false :DrawDebugType (select (and _debug (== _index _test)) "ForDuration" "None") :bIgnoreSelf false :TraceColor "(R=1.000000,G=0.000000,B=0.000000,A=1.000000)" :TraceHitColor "(R=0.000000,G=1.000000,B=0.000000,A=1.000000)" :DrawTime 5.0))
              (bind _hitresoultoutput_2 (Tools|ClosestHitResult _outhits_3 (- (+ (Utilities|ZeroLocationWorldSpaceZ _world_current) (Math|Vector|SelectVector (Math|Vector|NegateVector (+ (* (Math|Vector|GetUnitDirection(Vector) (Utilities|ZeroLocationWorldSpaceZ _world_current) (Utilities|ZeroLocationWorldSpaceZ _world_next)) _returnvalue_3) (* (Math|Vector|GetUnitDirection(Vector) (Utilities|ZeroLocationWorldSpaceZ _world_current) (Utilities|ZeroLocationWorldSpaceZ _world_previous)) _returnvalue_3))) (+ (* (Math|Vector|GetUnitDirection(Vector) (Utilities|ZeroLocationWorldSpaceZ _world_current) (Utilities|ZeroLocationWorldSpaceZ _world_next)) _returnvalue_3) (* (Math|Vector|GetUnitDirection(Vector) (Utilities|ZeroLocationWorldSpaceZ _world_current) (Utilities|ZeroLocationWorldSpaceZ _world_previous)) _returnvalue_3)) (select _reverse (not (Utilities|IsOuterCorner (Utilities|ZeroLocalorWorldSpaceZ _coordinatespace (Utilities|Select :Index CoordinateSpace :Local _next :World (Utilities|ZeroLocationWorldSpaceZ _world_next))) (Utilities|ZeroLocalorWorldSpaceZ _coordinatespace (Utilities|Select :Index CoordinateSpace :Local _current :World (Utilities|ZeroLocationWorldSpaceZ _world_current))) (Utilities|ZeroLocalorWorldSpaceZ _coordinatespace (Utilities|Select :Index CoordinateSpace :Local _previous :World (Utilities|ZeroLocationWorldSpaceZ _world_previous))))) (Utilities|IsOuterCorner (Utilities|ZeroLocalorWorldSpaceZ _coordinatespace (Utilities|Select :Index CoordinateSpace :Local _next :World (Utilities|ZeroLocationWorldSpaceZ _world_next))) (Utilities|ZeroLocalorWorldSpaceZ _coordinatespace (Utilities|Select :Index CoordinateSpace :Local _current :World (Utilities|ZeroLocationWorldSpaceZ _world_current))) (Utilities|ZeroLocalorWorldSpaceZ _coordinatespace (Utilities|Select :Index CoordinateSpace :Local _previous :World (Utilities|ZeroLocationWorldSpaceZ _world_previous))))))) (* (Math|Vector|SelectVector (Math|Vector|GetUnitDirection(Vector) (Utilities|ZeroLocationWorldSpaceZ _world_next) (Utilities|ZeroLocationWorldSpaceZ _world_current)) (Math|Vector|NegateVector (Math|Vector|GetUnitDirection(Vector) (Utilities|ZeroLocationWorldSpaceZ _world_next) (Utilities|ZeroLocationWorldSpaceZ _world_current))) (select _reverse (not (Utilities|IsOuterCorner (Utilities|ZeroLocalorWorldSpaceZ _coordinatespace (Utilities|Select :Index CoordinateSpace :Local _next :World (Utilities|ZeroLocationWorldSpaceZ _world_next))) (Utilities|ZeroLocalorWorldSpaceZ _coordinatespace (Utilities|Select :Index CoordinateSpace :Local _current :World (Utilities|ZeroLocationWorldSpaceZ _world_current))) (Utilities|ZeroLocalorWorldSpaceZ _coordinatespace (Utilities|Select :Index CoordinateSpace :Local _previous :World (Utilities|ZeroLocationWorldSpaceZ _world_previous))))) (Utilities|IsOuterCorner (Utilities|ZeroLocalorWorldSpaceZ _coordinatespace (Utilities|Select :Index CoordinateSpace :Local _next :World (Utilities|ZeroLocationWorldSpaceZ _world_next))) (Utilities|ZeroLocalorWorldSpaceZ _coordinatespace (Utilities|Select :Index CoordinateSpace :Local _current :World (Utilities|ZeroLocationWorldSpaceZ _world_current))) (Utilities|ZeroLocalorWorldSpaceZ _coordinatespace (Utilities|Select :Index CoordinateSpace :Local _previous :World (Utilities|ZeroLocationWorldSpaceZ _world_previous)))))) _returnvalue_4))))
              (if (and (> (* 512.0 (Math|Float|Round (/ (.Distance (Collision|BreakHitResult _hitresoultoutput)) 512.0))) 0.5) (> (* 512.0 (Math|Float|Round (/ (.Distance (Collision|BreakHitResult _hitresoultoutput_2)) 512.0))) 0.5))
                (Utilities|FlowControl|Sequence
                  (:then_0
                    (Utilities|Array|Add (LocalVariables|GetTransformList) (Utilities|Select :Index _coordinatespace_2 :Local (Math|Transform|MakeTransform (- (Utilities|ZeroLocalorWorldSpaceZ _coordinatespace (Utilities|Select :Index CoordinateSpace :Local _current :World (Utilities|ZeroLocationWorldSpaceZ _world_current))) _returnvalue_5) (Math|Rotator|MakeRotationfromAxes (Math|Vector|SelectVector (Math|Vector|GetUnitDirection(Vector) (Utilities|ZeroLocalorWorldSpaceZ _coordinatespace (Utilities|Select :Index CoordinateSpace :Local _current :World (Utilities|ZeroLocationWorldSpaceZ _world_current))) (Utilities|ZeroLocalorWorldSpaceZ _coordinatespace (Utilities|Select :Index CoordinateSpace :Local _previous :World (Utilities|ZeroLocationWorldSpaceZ _world_previous)))) (Math|Vector|NegateVector (Math|Vector|GetUnitDirection(Vector) (Utilities|ZeroLocalorWorldSpaceZ _coordinatespace (Utilities|Select :Index CoordinateSpace :Local _current :World (Utilities|ZeroLocationWorldSpaceZ _world_current))) (Utilities|ZeroLocalorWorldSpaceZ _coordinatespace (Utilities|Select :Index CoordinateSpace :Local _next :World (Utilities|ZeroLocationWorldSpaceZ _world_next))))) _reverse_2) (Math|Vector|SelectVector (Math|Vector|NegateVector (Math|Vector|GetUnitDirection(Vector) (Utilities|ZeroLocalorWorldSpaceZ _coordinatespace (Utilities|Select :Index CoordinateSpace :Local _current :World (Utilities|ZeroLocationWorldSpaceZ _world_current))) (Utilities|ZeroLocalorWorldSpaceZ _coordinatespace (Utilities|Select :Index CoordinateSpace :Local _next :World (Utilities|ZeroLocationWorldSpaceZ _world_next))))) (Math|Vector|GetUnitDirection(Vector) (Utilities|ZeroLocalorWorldSpaceZ _coordinatespace (Utilities|Select :Index CoordinateSpace :Local _current :World (Utilities|ZeroLocationWorldSpaceZ _world_current))) (Utilities|ZeroLocalorWorldSpaceZ _coordinatespace (Utilities|Select :Index CoordinateSpace :Local _previous :World (Utilities|ZeroLocationWorldSpaceZ _world_previous)))) _reverse_2) "0, 0,1.000000") (Math|Vector|MakeVector (/ (* 512.0 (Math|Float|Round (/ (.Distance (Collision|BreakHitResult _hitresoultoutput_2)) 512.0))) 512.0) (/ (* 512.0 (Math|Float|Round (/ (.Distance (Collision|BreakHitResult _hitresoultoutput)) 512.0))) 512.0) 1.0)) :World (Math|Transform|MakeRelativeTransform (Math|Transform|MakeTransform (- (Utilities|ZeroLocalorWorldSpaceZ _coordinatespace (Utilities|Select :Index CoordinateSpace :Local _current :World (Utilities|ZeroLocationWorldSpaceZ _world_current))) _returnvalue_5) (Math|Rotator|MakeRotationfromAxes (Math|Vector|SelectVector (Math|Vector|GetUnitDirection(Vector) (Utilities|ZeroLocalorWorldSpaceZ _coordinatespace (Utilities|Select :Index CoordinateSpace :Local _current :World (Utilities|ZeroLocationWorldSpaceZ _world_current))) (Utilities|ZeroLocalorWorldSpaceZ _coordinatespace (Utilities|Select :Index CoordinateSpace :Local _next :World (Utilities|ZeroLocationWorldSpaceZ _world_next)))) (Math|Vector|NegateVector (Math|Vector|GetUnitDirection(Vector) (Utilities|ZeroLocalorWorldSpaceZ _coordinatespace (Utilities|Select :Index CoordinateSpace :Local _current :World (Utilities|ZeroLocationWorldSpaceZ _world_current))) (Utilities|ZeroLocalorWorldSpaceZ _coordinatespace (Utilities|Select :Index CoordinateSpace :Local _previous :World (Utilities|ZeroLocationWorldSpaceZ _world_previous))))) (Utilities|IsOuterCorner (Utilities|ZeroLocalorWorldSpaceZ _coordinatespace (Utilities|Select :Index CoordinateSpace :Local _next :World (Utilities|ZeroLocationWorldSpaceZ _world_next))) (Utilities|ZeroLocalorWorldSpaceZ _coordinatespace (Utilities|Select :Index CoordinateSpace :Local _current :World (Utilities|ZeroLocationWorldSpaceZ _world_current))) (Utilities|ZeroLocalorWorldSpaceZ _coordinatespace (Utilities|Select :Index CoordinateSpace :Local _previous :World (Utilities|ZeroLocationWorldSpaceZ _world_previous))))) (Math|Vector|SelectVector (Math|Vector|NegateVector (Math|Vector|GetUnitDirection(Vector) (Utilities|ZeroLocalorWorldSpaceZ _coordinatespace (Utilities|Select :Index CoordinateSpace :Local _current :World (Utilities|ZeroLocationWorldSpaceZ _world_current))) (Utilities|ZeroLocalorWorldSpaceZ _coordinatespace (Utilities|Select :Index CoordinateSpace :Local _previous :World (Utilities|ZeroLocationWorldSpaceZ _world_previous))))) (Math|Vector|GetUnitDirection(Vector) (Utilities|ZeroLocalorWorldSpaceZ _coordinatespace (Utilities|Select :Index CoordinateSpace :Local _current :World (Utilities|ZeroLocationWorldSpaceZ _world_current))) (Utilities|ZeroLocalorWorldSpaceZ _coordinatespace (Utilities|Select :Index CoordinateSpace :Local _next :World (Utilities|ZeroLocationWorldSpaceZ _world_next)))) (Utilities|IsOuterCorner (Utilities|ZeroLocalorWorldSpaceZ _coordinatespace (Utilities|Select :Index CoordinateSpace :Local _next :World (Utilities|ZeroLocationWorldSpaceZ _world_next))) (Utilities|ZeroLocalorWorldSpaceZ _coordinatespace (Utilities|Select :Index CoordinateSpace :Local _current :World (Utilities|ZeroLocationWorldSpaceZ _world_current))) (Utilities|ZeroLocalorWorldSpaceZ _coordinatespace (Utilities|Select :Index CoordinateSpace :Local _previous :World (Utilities|ZeroLocationWorldSpaceZ _world_previous))))) "0, 0,1.000000") (Math|Vector|SelectVector (Math|Vector|MakeVector (/ (* 512.0 (Math|Float|Round (/ (.Distance (Collision|BreakHitResult _hitresoultoutput_2)) 512.0))) 512.0) (/ (* 512.0 (Math|Float|Round (/ (.Distance (Collision|BreakHitResult _hitresoultoutput)) 512.0))) 512.0) 1.0) (Math|Vector|MakeVector (/ (* 512.0 (Math|Float|Round (/ (.Distance (Collision|BreakHitResult _hitresoultoutput)) 512.0))) 512.0) (/ (* 512.0 (Math|Float|Round (/ (.Distance (Collision|BreakHitResult _hitresoultoutput_2)) 512.0))) 512.0) 1.0) (Utilities|IsOuterCorner (Utilities|ZeroLocalorWorldSpaceZ _coordinatespace (Utilities|Select :Index CoordinateSpace :Local _next :World (Utilities|ZeroLocationWorldSpaceZ _world_next))) (Utilities|ZeroLocalorWorldSpaceZ _coordinatespace (Utilities|Select :Index CoordinateSpace :Local _current :World (Utilities|ZeroLocationWorldSpaceZ _world_current))) (Utilities|ZeroLocalorWorldSpaceZ _coordinatespace (Utilities|Select :Index CoordinateSpace :Local _previous :World (Utilities|ZeroLocationWorldSpaceZ _world_previous)))))) _returnvalue_6))))
                  (:then_1
                    (bind _returnvalue_8 (AddComponent|NODE_AddStaticMeshComponent-28 :RelativeTransform (Math|Transform|ComposeTransforms (Math|Transform|MakeTransform (* (.location (Utilities|Select :Index _coordinatespace_2 :Local (Math|Transform|MakeTransform (- (Utilities|ZeroLocalorWorldSpaceZ _coordinatespace (Utilities|Select :Index CoordinateSpace :Local _current :World (Utilities|ZeroLocationWorldSpaceZ _world_current))) _returnvalue_5) (Math|Rotator|MakeRotationfromAxes (Math|Vector|SelectVector (Math|Vector|GetUnitDirection(Vector) (Utilities|ZeroLocalorWorldSpaceZ _coordinatespace (Utilities|Select :Index CoordinateSpace :Local _current :World (Utilities|ZeroLocationWorldSpaceZ _world_current))) (Utilities|ZeroLocalorWorldSpaceZ _coordinatespace (Utilities|Select :Index CoordinateSpace :Local _previous :World (Utilities|ZeroLocationWorldSpaceZ _world_previous)))) (Math|Vector|NegateVector (Math|Vector|GetUnitDirection(Vector) (Utilities|ZeroLocalorWorldSpaceZ _coordinatespace (Utilities|Select :Index CoordinateSpace :Local _current :World (Utilities|ZeroLocationWorldSpaceZ _world_current))) (Utilities|ZeroLocalorWorldSpaceZ _coordinatespace (Utilities|Select :Index CoordinateSpace :Local _next :World (Utilities|ZeroLocationWorldSpaceZ _world_next))))) _reverse_2) (Math|Vector|SelectVector (Math|Vector|NegateVector (Math|Vector|GetUnitDirection(Vector) (Utilities|ZeroLocalorWorldSpaceZ _coordinatespace (Utilities|Select :Index CoordinateSpace :Local _current :World (Utilities|ZeroLocationWorldSpaceZ _world_current))) (Utilities|ZeroLocalorWorldSpaceZ _coordinatespace (Utilities|Select :Index CoordinateSpace :Local _next :World (Utilities|ZeroLocationWorldSpaceZ _world_next))))) (Math|Vector|GetUnitDirection(Vector) (Utilities|ZeroLocalorWorldSpaceZ _coordinatespace (Utilities|Select :Index CoordinateSpace :Local _current :World (Utilities|ZeroLocationWorldSpaceZ _world_current))) (Utilities|ZeroLocalorWorldSpaceZ _coordinatespace (Utilities|Select :Index CoordinateSpace :Local _previous :World (Utilities|ZeroLocationWorldSpaceZ _world_previous)))) _reverse_2) "0, 0,1.000000") (Math|Vector|MakeVector (/ (* 512.0 (Math|Float|Round (/ (.Distance (Collision|BreakHitResult _hitresoultoutput_2)) 512.0))) 512.0) (/ (* 512.0 (Math|Float|Round (/ (.Distance (Collision|BreakHitResult _hitresoultoutput)) 512.0))) 512.0) 1.0)) :World (Math|Transform|MakeRelativeTransform (Math|Transform|MakeTransform (- (Utilities|ZeroLocalorWorldSpaceZ _coordinatespace (Utilities|Select :Index CoordinateSpace :Local _current :World (Utilities|ZeroLocationWorldSpaceZ _world_current))) _returnvalue_5) (Math|Rotator|MakeRotationfromAxes (Math|Vector|SelectVector (Math|Vector|GetUnitDirection(Vector) (Utilities|ZeroLocalorWorldSpaceZ _coordinatespace (Utilities|Select :Index CoordinateSpace :Local _current :World (Utilities|ZeroLocationWorldSpaceZ _world_current))) (Utilities|ZeroLocalorWorldSpaceZ _coordinatespace (Utilities|Select :Index CoordinateSpace :Local _next :World (Utilities|ZeroLocationWorldSpaceZ _world_next)))) (Math|Vector|NegateVector (Math|Vector|GetUnitDirection(Vector) (Utilities|ZeroLocalorWorldSpaceZ _coordinatespace (Utilities|Select :Index CoordinateSpace :Local _current :World (Utilities|ZeroLocationWorldSpaceZ _world_current))) (Utilities|ZeroLocalorWorldSpaceZ _coordinatespace (Utilities|Select :Index CoordinateSpace :Local _previous :World (Utilities|ZeroLocationWorldSpaceZ _world_previous))))) (Utilities|IsOuterCorner (Utilities|ZeroLocalorWorldSpaceZ _coordinatespace (Utilities|Select :Index CoordinateSpace :Local _next :World (Utilities|ZeroLocationWorldSpaceZ _world_next))) (Utilities|ZeroLocalorWorldSpaceZ _coordinatespace (Utilities|Select :Index CoordinateSpace :Local _current :World (Utilities|ZeroLocationWorldSpaceZ _world_current))) (Utilities|ZeroLocalorWorldSpaceZ _coordinatespace (Utilities|Select :Index CoordinateSpace :Local _previous :World (Utilities|ZeroLocationWorldSpaceZ _world_previous))))) (Math|Vector|SelectVector (Math|Vector|NegateVector (Math|Vector|GetUnitDirection(Vector) (Utilities|ZeroLocalorWorldSpaceZ _coordinatespace (Utilities|Select :Index CoordinateSpace :Local _current :World (Utilities|ZeroLocationWorldSpaceZ _world_current))) (Utilities|ZeroLocalorWorldSpaceZ _coordinatespace (Utilities|Select :Index CoordinateSpace :Local _previous :World (Utilities|ZeroLocationWorldSpaceZ _world_previous))))) (Math|Vector|GetUnitDirection(Vector) (Utilities|ZeroLocalorWorldSpaceZ _coordinatespace (Utilities|Select :Index CoordinateSpace :Local _current :World (Utilities|ZeroLocationWorldSpaceZ _world_current))) (Utilities|ZeroLocalorWorldSpaceZ _coordinatespace (Utilities|Select :Index CoordinateSpace :Local _next :World (Utilities|ZeroLocationWorldSpaceZ _world_next)))) (Utilities|IsOuterCorner (Utilities|ZeroLocalorWorldSpaceZ _coordinatespace (Utilities|Select :Index CoordinateSpace :Local _next :World (Utilities|ZeroLocationWorldSpaceZ _world_next))) (Utilities|ZeroLocalorWorldSpaceZ _coordinatespace (Utilities|Select :Index CoordinateSpace :Local _current :World (Utilities|ZeroLocationWorldSpaceZ _world_current))) (Utilities|ZeroLocalorWorldSpaceZ _coordinatespace (Utilities|Select :Index CoordinateSpace :Local _previous :World (Utilities|ZeroLocationWorldSpaceZ _world_previous))))) "0, 0,1.000000") (Math|Vector|SelectVector (Math|Vector|MakeVector (/ (* 512.0 (Math|Float|Round (/ (.Distance (Collision|BreakHitResult _hitresoultoutput_2)) 512.0))) 512.0) (/ (* 512.0 (Math|Float|Round (/ (.Distance (Collision|BreakHitResult _hitresoultoutput)) 512.0))) 512.0) 1.0) (Math|Vector|MakeVector (/ (* 512.0 (Math|Float|Round (/ (.Distance (Collision|BreakHitResult _hitresoultoutput)) 512.0))) 512.0) (/ (* 512.0 (Math|Float|Round (/ (.Distance (Collision|BreakHitResult _hitresoultoutput_2)) 512.0))) 512.0) 1.0) (Utilities|IsOuterCorner (Utilities|ZeroLocalorWorldSpaceZ _coordinatespace (Utilities|Select :Index CoordinateSpace :Local _next :World (Utilities|ZeroLocationWorldSpaceZ _world_next))) (Utilities|ZeroLocalorWorldSpaceZ _coordinatespace (Utilities|Select :Index CoordinateSpace :Local _current :World (Utilities|ZeroLocationWorldSpaceZ _world_current))) (Utilities|ZeroLocalorWorldSpaceZ _coordinatespace (Utilities|Select :Index CoordinateSpace :Local _previous :World (Utilities|ZeroLocationWorldSpaceZ _world_previous)))))) _returnvalue_6))) "1.000000,1.000000, 0") (.rotation (Utilities|Select :Index _coordinatespace_2 :Local (Math|Transform|MakeTransform (- (Utilities|ZeroLocalorWorldSpaceZ _coordinatespace (Utilities|Select :Index CoordinateSpace :Local _current :World (Utilities|ZeroLocationWorldSpaceZ _world_current))) _returnvalue_5) (Math|Rotator|MakeRotationfromAxes (Math|Vector|SelectVector (Math|Vector|GetUnitDirection(Vector) (Utilities|ZeroLocalorWorldSpaceZ _coordinatespace (Utilities|Select :Index CoordinateSpace :Local _current :World (Utilities|ZeroLocationWorldSpaceZ _world_current))) (Utilities|ZeroLocalorWorldSpaceZ _coordinatespace (Utilities|Select :Index CoordinateSpace :Local _previous :World (Utilities|ZeroLocationWorldSpaceZ _world_previous)))) (Math|Vector|NegateVector (Math|Vector|GetUnitDirection(Vector) (Utilities|ZeroLocalorWorldSpaceZ _coordinatespace (Utilities|Select :Index CoordinateSpace :Local _current :World (Utilities|ZeroLocationWorldSpaceZ _world_current))) (Utilities|ZeroLocalorWorldSpaceZ _coordinatespace (Utilities|Select :Index CoordinateSpace :Local _next :World (Utilities|ZeroLocationWorldSpaceZ _world_next))))) _reverse_2) (Math|Vector|SelectVector (Math|Vector|NegateVector (Math|Vector|GetUnitDirection(Vector) (Utilities|ZeroLocalorWorldSpaceZ _coordinatespace (Utilities|Select :Index CoordinateSpace :Local _current :World (Utilities|ZeroLocationWorldSpaceZ _world_current))) (Utilities|ZeroLocalorWorldSpaceZ _coordinatespace (Utilities|Select :Index CoordinateSpace :Local _next :World (Utilities|ZeroLocationWorldSpaceZ _world_next))))) (Math|Vector|GetUnitDirection(Vector) (Utilities|ZeroLocalorWorldSpaceZ _coordinatespace (Utilities|Select :Index CoordinateSpace :Local _current :World (Utilities|ZeroLocationWorldSpaceZ _world_current))) (Utilities|ZeroLocalorWorldSpaceZ _coordinatespace (Utilities|Select :Index CoordinateSpace :Local _previous :World (Utilities|ZeroLocationWorldSpaceZ _world_previous)))) _reverse_2) "0, 0,1.000000") (Math|Vector|MakeVector (/ (* 512.0 (Math|Float|Round (/ (.Distance (Collision|BreakHitResult _hitresoultoutput_2)) 512.0))) 512.0) (/ (* 512.0 (Math|Float|Round (/ (.Distance (Collision|BreakHitResult _hitresoultoutput)) 512.0))) 512.0) 1.0)) :World (Math|Transform|MakeRelativeTransform (Math|Transform|MakeTransform (- (Utilities|ZeroLocalorWorldSpaceZ _coordinatespace (Utilities|Select :Index CoordinateSpace :Local _current :World (Utilities|ZeroLocationWorldSpaceZ _world_current))) _returnvalue_5) (Math|Rotator|MakeRotationfromAxes (Math|Vector|SelectVector (Math|Vector|GetUnitDirection(Vector) (Utilities|ZeroLocalorWorldSpaceZ _coordinatespace (Utilities|Select :Index CoordinateSpace :Local _current :World (Utilities|ZeroLocationWorldSpaceZ _world_current))) (Utilities|ZeroLocalorWorldSpaceZ _coordinatespace (Utilities|Select :Index CoordinateSpace :Local _next :World (Utilities|ZeroLocationWorldSpaceZ _world_next)))) (Math|Vector|NegateVector (Math|Vector|GetUnitDirection(Vector) (Utilities|ZeroLocalorWorldSpaceZ _coordinatespace (Utilities|Select :Index CoordinateSpace :Local _current :World (Utilities|ZeroLocationWorldSpaceZ _world_current))) (Utilities|ZeroLocalorWorldSpaceZ _coordinatespace (Utilities|Select :Index CoordinateSpace :Local _previous :World (Utilities|ZeroLocationWorldSpaceZ _world_previous))))) (Utilities|IsOuterCorner (Utilities|ZeroLocalorWorldSpaceZ _coordinatespace (Utilities|Select :Index CoordinateSpace :Local _next :World (Utilities|ZeroLocationWorldSpaceZ _world_next))) (Utilities|ZeroLocalorWorldSpaceZ _coordinatespace (Utilities|Select :Index CoordinateSpace :Local _current :World (Utilities|ZeroLocationWorldSpaceZ _world_current))) (Utilities|ZeroLocalorWorldSpaceZ _coordinatespace (Utilities|Select :Index CoordinateSpace :Local _previous :World (Utilities|ZeroLocationWorldSpaceZ _world_previous))))) (Math|Vector|SelectVector (Math|Vector|NegateVector (Math|Vector|GetUnitDirection(Vector) (Utilities|ZeroLocalorWorldSpaceZ _coordinatespace (Utilities|Select :Index CoordinateSpace :Local _current :World (Utilities|ZeroLocationWorldSpaceZ _world_current))) (Utilities|ZeroLocalorWorldSpaceZ _coordinatespace (Utilities|Select :Index CoordinateSpace :Local _previous :World (Utilities|ZeroLocationWorldSpaceZ _world_previous))))) (Math|Vector|GetUnitDirection(Vector) (Utilities|ZeroLocalorWorldSpaceZ _coordinatespace (Utilities|Select :Index CoordinateSpace :Local _current :World (Utilities|ZeroLocationWorldSpaceZ _world_current))) (Utilities|ZeroLocalorWorldSpaceZ _coordinatespace (Utilities|Select :Index CoordinateSpace :Local _next :World (Utilities|ZeroLocationWorldSpaceZ _world_next)))) (Utilities|IsOuterCorner (Utilities|ZeroLocalorWorldSpaceZ _coordinatespace (Utilities|Select :Index CoordinateSpace :Local _next :World (Utilities|ZeroLocationWorldSpaceZ _world_next))) (Utilities|ZeroLocalorWorldSpaceZ _coordinatespace (Utilities|Select :Index CoordinateSpace :Local _current :World (Utilities|ZeroLocationWorldSpaceZ _world_current))) (Utilities|ZeroLocalorWorldSpaceZ _coordinatespace (Utilities|Select :Index CoordinateSpace :Local _previous :World (Utilities|ZeroLocationWorldSpaceZ _world_previous))))) "0, 0,1.000000") (Math|Vector|SelectVector (Math|Vector|MakeVector (/ (* 512.0 (Math|Float|Round (/ (.Distance (Collision|BreakHitResult _hitresoultoutput_2)) 512.0))) 512.0) (/ (* 512.0 (Math|Float|Round (/ (.Distance (Collision|BreakHitResult _hitresoultoutput)) 512.0))) 512.0) 1.0) (Math|Vector|MakeVector (/ (* 512.0 (Math|Float|Round (/ (.Distance (Collision|BreakHitResult _hitresoultoutput)) 512.0))) 512.0) (/ (* 512.0 (Math|Float|Round (/ (.Distance (Collision|BreakHitResult _hitresoultoutput_2)) 512.0))) 512.0) 1.0) (Utilities|IsOuterCorner (Utilities|ZeroLocalorWorldSpaceZ _coordinatespace (Utilities|Select :Index CoordinateSpace :Local _next :World (Utilities|ZeroLocationWorldSpaceZ _world_next))) (Utilities|ZeroLocalorWorldSpaceZ _coordinatespace (Utilities|Select :Index CoordinateSpace :Local _current :World (Utilities|ZeroLocationWorldSpaceZ _world_current))) (Utilities|ZeroLocalorWorldSpaceZ _coordinatespace (Utilities|Select :Index CoordinateSpace :Local _previous :World (Utilities|ZeroLocationWorldSpaceZ _world_previous)))))) _returnvalue_6))) (Math|Vector|MakeVector (/ (- (* (.x (.scale (Utilities|Select :Index _coordinatespace_2 :Local (Math|Transform|MakeTransform (- (Utilities|ZeroLocalorWorldSpaceZ _coordinatespace (Utilities|Select :Index CoordinateSpace :Local _current :World (Utilities|ZeroLocationWorldSpaceZ _world_current))) _returnvalue_5) (Math|Rotator|MakeRotationfromAxes (Math|Vector|SelectVector (Math|Vector|GetUnitDirection(Vector) (Utilities|ZeroLocalorWorldSpaceZ _coordinatespace (Utilities|Select :Index CoordinateSpace :Local _current :World (Utilities|ZeroLocationWorldSpaceZ _world_current))) (Utilities|ZeroLocalorWorldSpaceZ _coordinatespace (Utilities|Select :Index CoordinateSpace :Local _previous :World (Utilities|ZeroLocationWorldSpaceZ _world_previous)))) (Math|Vector|NegateVector (Math|Vector|GetUnitDirection(Vector) (Utilities|ZeroLocalorWorldSpaceZ _coordinatespace (Utilities|Select :Index CoordinateSpace :Local _current :World (Utilities|ZeroLocationWorldSpaceZ _world_current))) (Utilities|ZeroLocalorWorldSpaceZ _coordinatespace (Utilities|Select :Index CoordinateSpace :Local _next :World (Utilities|ZeroLocationWorldSpaceZ _world_next))))) _reverse_2) (Math|Vector|SelectVector (Math|Vector|NegateVector (Math|Vector|GetUnitDirection(Vector) (Utilities|ZeroLocalorWorldSpaceZ _coordinatespace (Utilities|Select :Index CoordinateSpace :Local _current :World (Utilities|ZeroLocationWorldSpaceZ _world_current))) (Utilities|ZeroLocalorWorldSpaceZ _coordinatespace (Utilities|Select :Index CoordinateSpace :Local _next :World (Utilities|ZeroLocationWorldSpaceZ _world_next))))) (Math|Vector|GetUnitDirection(Vector) (Utilities|ZeroLocalorWorldSpaceZ _coordinatespace (Utilities|Select :Index CoordinateSpace :Local _current :World (Utilities|ZeroLocationWorldSpaceZ _world_current))) (Utilities|ZeroLocalorWorldSpaceZ _coordinatespace (Utilities|Select :Index CoordinateSpace :Local _previous :World (Utilities|ZeroLocationWorldSpaceZ _world_previous)))) _reverse_2) "0, 0,1.000000") (Math|Vector|MakeVector (/ (* 512.0 (Math|Float|Round (/ (.Distance (Collision|BreakHitResult _hitresoultoutput_2)) 512.0))) 512.0) (/ (* 512.0 (Math|Float|Round (/ (.Distance (Collision|BreakHitResult _hitresoultoutput)) 512.0))) 512.0) 1.0)) :World (Math|Transform|MakeRelativeTransform (Math|Transform|MakeTransform (- (Utilities|ZeroLocalorWorldSpaceZ _coordinatespace (Utilities|Select :Index CoordinateSpace :Local _current :World (Utilities|ZeroLocationWorldSpaceZ _world_current))) _returnvalue_5) (Math|Rotator|MakeRotationfromAxes (Math|Vector|SelectVector (Math|Vector|GetUnitDirection(Vector) (Utilities|ZeroLocalorWorldSpaceZ _coordinatespace (Utilities|Select :Index CoordinateSpace :Local _current :World (Utilities|ZeroLocationWorldSpaceZ _world_current))) (Utilities|ZeroLocalorWorldSpaceZ _coordinatespace (Utilities|Select :Index CoordinateSpace :Local _next :World (Utilities|ZeroLocationWorldSpaceZ _world_next)))) (Math|Vector|NegateVector (Math|Vector|GetUnitDirection(Vector) (Utilities|ZeroLocalorWorldSpaceZ _coordinatespace (Utilities|Select :Index CoordinateSpace :Local _current :World (Utilities|ZeroLocationWorldSpaceZ _world_current))) (Utilities|ZeroLocalorWorldSpaceZ _coordinatespace (Utilities|Select :Index CoordinateSpace :Local _previous :World (Utilities|ZeroLocationWorldSpaceZ _world_previous))))) (Utilities|IsOuterCorner (Utilities|ZeroLocalorWorldSpaceZ _coordinatespace (Utilities|Select :Index CoordinateSpace :Local _next :World (Utilities|ZeroLocationWorldSpaceZ _world_next))) (Utilities|ZeroLocalorWorldSpaceZ _coordinatespace (Utilities|Select :Index CoordinateSpace :Local _current :World (Utilities|ZeroLocationWorldSpaceZ _world_current))) (Utilities|ZeroLocalorWorldSpaceZ _coordinatespace (Utilities|Select :Index CoordinateSpace :Local _previous :World (Utilities|ZeroLocationWorldSpaceZ _world_previous))))) (Math|Vector|SelectVector (Math|Vector|NegateVector (Math|Vector|GetUnitDirection(Vector) (Utilities|ZeroLocalorWorldSpaceZ _coordinatespace (Utilities|Select :Index CoordinateSpace :Local _current :World (Utilities|ZeroLocationWorldSpaceZ _world_current))) (Utilities|ZeroLocalorWorldSpaceZ _coordinatespace (Utilities|Select :Index CoordinateSpace :Local _previous :World (Utilities|ZeroLocationWorldSpaceZ _world_previous))))) (Math|Vector|GetUnitDirection(Vector) (Utilities|ZeroLocalorWorldSpaceZ _coordinatespace (Utilities|Select :Index CoordinateSpace :Local _current :World (Utilities|ZeroLocationWorldSpaceZ _world_current))) (Utilities|ZeroLocalorWorldSpaceZ _coordinatespace (Utilities|Select :Index CoordinateSpace :Local _next :World (Utilities|ZeroLocationWorldSpaceZ _world_next)))) (Utilities|IsOuterCorner (Utilities|ZeroLocalorWorldSpaceZ _coordinatespace (Utilities|Select :Index CoordinateSpace :Local _next :World (Utilities|ZeroLocationWorldSpaceZ _world_next))) (Utilities|ZeroLocalorWorldSpaceZ _coordinatespace (Utilities|Select :Index CoordinateSpace :Local _current :World (Utilities|ZeroLocationWorldSpaceZ _world_current))) (Utilities|ZeroLocalorWorldSpaceZ _coordinatespace (Utilities|Select :Index CoordinateSpace :Local _previous :World (Utilities|ZeroLocationWorldSpaceZ _world_previous))))) "0, 0,1.000000") (Math|Vector|SelectVector (Math|Vector|MakeVector (/ (* 512.0 (Math|Float|Round (/ (.Distance (Collision|BreakHitResult _hitresoultoutput_2)) 512.0))) 512.0) (/ (* 512.0 (Math|Float|Round (/ (.Distance (Collision|BreakHitResult _hitresoultoutput)) 512.0))) 512.0) 1.0) (Math|Vector|MakeVector (/ (* 512.0 (Math|Float|Round (/ (.Distance (Collision|BreakHitResult _hitresoultoutput)) 512.0))) 512.0) (/ (* 512.0 (Math|Float|Round (/ (.Distance (Collision|BreakHitResult _hitresoultoutput_2)) 512.0))) 512.0) 1.0) (Utilities|IsOuterCorner (Utilities|ZeroLocalorWorldSpaceZ _coordinatespace (Utilities|Select :Index CoordinateSpace :Local _next :World (Utilities|ZeroLocationWorldSpaceZ _world_next))) (Utilities|ZeroLocalorWorldSpaceZ _coordinatespace (Utilities|Select :Index CoordinateSpace :Local _current :World (Utilities|ZeroLocationWorldSpaceZ _world_current))) (Utilities|ZeroLocalorWorldSpaceZ _coordinatespace (Utilities|Select :Index CoordinateSpace :Local _previous :World (Utilities|ZeroLocationWorldSpaceZ _world_previous)))))) _returnvalue_6)))) 512.0) 4e+01) 512.0) (/ (- (* (.y (.scale (Utilities|Select :Index _coordinatespace_2 :Local (Math|Transform|MakeTransform (- (Utilities|ZeroLocalorWorldSpaceZ _coordinatespace (Utilities|Select :Index CoordinateSpace :Local _current :World (Utilities|ZeroLocationWorldSpaceZ _world_current))) _returnvalue_5) (Math|Rotator|MakeRotationfromAxes (Math|Vector|SelectVector (Math|Vector|GetUnitDirection(Vector) (Utilities|ZeroLocalorWorldSpaceZ _coordinatespace (Utilities|Select :Index CoordinateSpace :Local _current :World (Utilities|ZeroLocationWorldSpaceZ _world_current))) (Utilities|ZeroLocalorWorldSpaceZ _coordinatespace (Utilities|Select :Index CoordinateSpace :Local _previous :World (Utilities|ZeroLocationWorldSpaceZ _world_previous)))) (Math|Vector|NegateVector (Math|Vector|GetUnitDirection(Vector) (Utilities|ZeroLocalorWorldSpaceZ _coordinatespace (Utilities|Select :Index CoordinateSpace :Local _current :World (Utilities|ZeroLocationWorldSpaceZ _world_current))) (Utilities|ZeroLocalorWorldSpaceZ _coordinatespace (Utilities|Select :Index CoordinateSpace :Local _next :World (Utilities|ZeroLocationWorldSpaceZ _world_next))))) _reverse_2) (Math|Vector|SelectVector (Math|Vector|NegateVector (Math|Vector|GetUnitDirection(Vector) (Utilities|ZeroLocalorWorldSpaceZ _coordinatespace (Utilities|Select :Index CoordinateSpace :Local _current :World (Utilities|ZeroLocationWorldSpaceZ _world_current))) (Utilities|ZeroLocalorWorldSpaceZ _coordinatespace (Utilities|Select :Index CoordinateSpace :Local _next :World (Utilities|ZeroLocationWorldSpaceZ _world_next))))) (Math|Vector|GetUnitDirection(Vector) (Utilities|ZeroLocalorWorldSpaceZ _coordinatespace (Utilities|Select :Index CoordinateSpace :Local _current :World (Utilities|ZeroLocationWorldSpaceZ _world_current))) (Utilities|ZeroLocalorWorldSpaceZ _coordinatespace (Utilities|Select :Index CoordinateSpace :Local _previous :World (Utilities|ZeroLocationWorldSpaceZ _world_previous)))) _reverse_2) "0, 0,1.000000") (Math|Vector|MakeVector (/ (* 512.0 (Math|Float|Round (/ (.Distance (Collision|BreakHitResult _hitresoultoutput_2)) 512.0))) 512.0) (/ (* 512.0 (Math|Float|Round (/ (.Distance (Collision|BreakHitResult _hitresoultoutput)) 512.0))) 512.0) 1.0)) :World (Math|Transform|MakeRelativeTransform (Math|Transform|MakeTransform (- (Utilities|ZeroLocalorWorldSpaceZ _coordinatespace (Utilities|Select :Index CoordinateSpace :Local _current :World (Utilities|ZeroLocationWorldSpaceZ _world_current))) _returnvalue_5) (Math|Rotator|MakeRotationfromAxes (Math|Vector|SelectVector (Math|Vector|GetUnitDirection(Vector) (Utilities|ZeroLocalorWorldSpaceZ _coordinatespace (Utilities|Select :Index CoordinateSpace :Local _current :World (Utilities|ZeroLocationWorldSpaceZ _world_current))) (Utilities|ZeroLocalorWorldSpaceZ _coordinatespace (Utilities|Select :Index CoordinateSpace :Local _next :World (Utilities|ZeroLocationWorldSpaceZ _world_next)))) (Math|Vector|NegateVector (Math|Vector|GetUnitDirection(Vector) (Utilities|ZeroLocalorWorldSpaceZ _coordinatespace (Utilities|Select :Index CoordinateSpace :Local _current :World (Utilities|ZeroLocationWorldSpaceZ _world_current))) (Utilities|ZeroLocalorWorldSpaceZ _coordinatespace (Utilities|Select :Index CoordinateSpace :Local _previous :World (Utilities|ZeroLocationWorldSpaceZ _world_previous))))) (Utilities|IsOuterCorner (Utilities|ZeroLocalorWorldSpaceZ _coordinatespace (Utilities|Select :Index CoordinateSpace :Local _next :World (Utilities|ZeroLocationWorldSpaceZ _world_next))) (Utilities|ZeroLocalorWorldSpaceZ _coordinatespace (Utilities|Select :Index CoordinateSpace :Local _current :World (Utilities|ZeroLocationWorldSpaceZ _world_current))) (Utilities|ZeroLocalorWorldSpaceZ _coordinatespace (Utilities|Select :Index CoordinateSpace :Local _previous :World (Utilities|ZeroLocationWorldSpaceZ _world_previous))))) (Math|Vector|SelectVector (Math|Vector|NegateVector (Math|Vector|GetUnitDirection(Vector) (Utilities|ZeroLocalorWorldSpaceZ _coordinatespace (Utilities|Select :Index CoordinateSpace :Local _current :World (Utilities|ZeroLocationWorldSpaceZ _world_current))) (Utilities|ZeroLocalorWorldSpaceZ _coordinatespace (Utilities|Select :Index CoordinateSpace :Local _previous :World (Utilities|ZeroLocationWorldSpaceZ _world_previous))))) (Math|Vector|GetUnitDirection(Vector) (Utilities|ZeroLocalorWorldSpaceZ _coordinatespace (Utilities|Select :Index CoordinateSpace :Local _current :World (Utilities|ZeroLocationWorldSpaceZ _world_current))) (Utilities|ZeroLocalorWorldSpaceZ _coordinatespace (Utilities|Select :Index CoordinateSpace :Local _next :World (Utilities|ZeroLocationWorldSpaceZ _world_next)))) (Utilities|IsOuterCorner (Utilities|ZeroLocalorWorldSpaceZ _coordinatespace (Utilities|Select :Index CoordinateSpace :Local _next :World (Utilities|ZeroLocationWorldSpaceZ _world_next))) (Utilities|ZeroLocalorWorldSpaceZ _coordinatespace (Utilities|Select :Index CoordinateSpace :Local _current :World (Utilities|ZeroLocationWorldSpaceZ _world_current))) (Utilities|ZeroLocalorWorldSpaceZ _coordinatespace (Utilities|Select :Index CoordinateSpace :Local _previous :World (Utilities|ZeroLocationWorldSpaceZ _world_previous))))) "0, 0,1.000000") (Math|Vector|SelectVector (Math|Vector|MakeVector (/ (* 512.0 (Math|Float|Round (/ (.Distance (Collision|BreakHitResult _hitresoultoutput_2)) 512.0))) 512.0) (/ (* 512.0 (Math|Float|Round (/ (.Distance (Collision|BreakHitResult _hitresoultoutput)) 512.0))) 512.0) 1.0) (Math|Vector|MakeVector (/ (* 512.0 (Math|Float|Round (/ (.Distance (Collision|BreakHitResult _hitresoultoutput)) 512.0))) 512.0) (/ (* 512.0 (Math|Float|Round (/ (.Distance (Collision|BreakHitResult _hitresoultoutput_2)) 512.0))) 512.0) 1.0) (Utilities|IsOuterCorner (Utilities|ZeroLocalorWorldSpaceZ _coordinatespace (Utilities|Select :Index CoordinateSpace :Local _next :World (Utilities|ZeroLocationWorldSpaceZ _world_next))) (Utilities|ZeroLocalorWorldSpaceZ _coordinatespace (Utilities|Select :Index CoordinateSpace :Local _current :World (Utilities|ZeroLocationWorldSpaceZ _world_current))) (Utilities|ZeroLocalorWorldSpaceZ _coordinatespace (Utilities|Select :Index CoordinateSpace :Local _previous :World (Utilities|ZeroLocationWorldSpaceZ _world_previous)))))) _returnvalue_6)))) 512.0) 4e+01) 512.0) (.z (.scale (Utilities|Select :Index _coordinatespace_2 :Local (Math|Transform|MakeTransform (- (Utilities|ZeroLocalorWorldSpaceZ _coordinatespace (Utilities|Select :Index CoordinateSpace :Local _current :World (Utilities|ZeroLocationWorldSpaceZ _world_current))) _returnvalue_5) (Math|Rotator|MakeRotationfromAxes (Math|Vector|SelectVector (Math|Vector|GetUnitDirection(Vector) (Utilities|ZeroLocalorWorldSpaceZ _coordinatespace (Utilities|Select :Index CoordinateSpace :Local _current :World (Utilities|ZeroLocationWorldSpaceZ _world_current))) (Utilities|ZeroLocalorWorldSpaceZ _coordinatespace (Utilities|Select :Index CoordinateSpace :Local _previous :World (Utilities|ZeroLocationWorldSpaceZ _world_previous)))) (Math|Vector|NegateVector (Math|Vector|GetUnitDirection(Vector) (Utilities|ZeroLocalorWorldSpaceZ _coordinatespace (Utilities|Select :Index CoordinateSpace :Local _current :World (Utilities|ZeroLocationWorldSpaceZ _world_current))) (Utilities|ZeroLocalorWorldSpaceZ _coordinatespace (Utilities|Select :Index CoordinateSpace :Local _next :World (Utilities|ZeroLocationWorldSpaceZ _world_next))))) _reverse_2) (Math|Vector|SelectVector (Math|Vector|NegateVector (Math|Vector|GetUnitDirection(Vector) (Utilities|ZeroLocalorWorldSpaceZ _coordinatespace (Utilities|Select :Index CoordinateSpace :Local _current :World (Utilities|ZeroLocationWorldSpaceZ _world_current))) (Utilities|ZeroLocalorWorldSpaceZ _coordinatespace (Utilities|Select :Index CoordinateSpace :Local _next :World (Utilities|ZeroLocationWorldSpaceZ _world_next))))) (Math|Vector|GetUnitDirection(Vector) (Utilities|ZeroLocalorWorldSpaceZ _coordinatespace (Utilities|Select :Index CoordinateSpace :Local _current :World (Utilities|ZeroLocationWorldSpaceZ _world_current))) (Utilities|ZeroLocalorWorldSpaceZ _coordinatespace (Utilities|Select :Index CoordinateSpace :Local _previous :World (Utilities|ZeroLocationWorldSpaceZ _world_previous)))) _reverse_2) "0, 0,1.000000") (Math|Vector|MakeVector (/ (* 512.0 (Math|Float|Round (/ (.Distance (Collision|BreakHitResult _hitresoultoutput_2)) 512.0))) 512.0) (/ (* 512.0 (Math|Float|Round (/ (.Distance (Collision|BreakHitResult _hitresoultoutput)) 512.0))) 512.0) 1.0)) :World (Math|Transform|MakeRelativeTransform (Math|Transform|MakeTransform (- (Utilities|ZeroLocalorWorldSpaceZ _coordinatespace (Utilities|Select :Index CoordinateSpace :Local _current :World (Utilities|ZeroLocationWorldSpaceZ _world_current))) _returnvalue_5) (Math|Rotator|MakeRotationfromAxes (Math|Vector|SelectVector (Math|Vector|GetUnitDirection(Vector) (Utilities|ZeroLocalorWorldSpaceZ _coordinatespace (Utilities|Select :Index CoordinateSpace :Local _current :World (Utilities|ZeroLocationWorldSpaceZ _world_current))) (Utilities|ZeroLocalorWorldSpaceZ _coordinatespace (Utilities|Select :Index CoordinateSpace :Local _next :World (Utilities|ZeroLocationWorldSpaceZ _world_next)))) (Math|Vector|NegateVector (Math|Vector|GetUnitDirection(Vector) (Utilities|ZeroLocalorWorldSpaceZ _coordinatespace (Utilities|Select :Index CoordinateSpace :Local _current :World (Utilities|ZeroLocationWorldSpaceZ _world_current))) (Utilities|ZeroLocalorWorldSpaceZ _coordinatespace (Utilities|Select :Index CoordinateSpace :Local _previous :World (Utilities|ZeroLocationWorldSpaceZ _world_previous))))) (Utilities|IsOuterCorner (Utilities|ZeroLocalorWorldSpaceZ _coordinatespace (Utilities|Select :Index CoordinateSpace :Local _next :World (Utilities|ZeroLocationWorldSpaceZ _world_next))) (Utilities|ZeroLocalorWorldSpaceZ _coordinatespace (Utilities|Select :Index CoordinateSpace :Local _current :World (Utilities|ZeroLocationWorldSpaceZ _world_current))) (Utilities|ZeroLocalorWorldSpaceZ _coordinatespace (Utilities|Select :Index CoordinateSpace :Local _previous :World (Utilities|ZeroLocationWorldSpaceZ _world_previous))))) (Math|Vector|SelectVector (Math|Vector|NegateVector (Math|Vector|GetUnitDirection(Vector) (Utilities|ZeroLocalorWorldSpaceZ _coordinatespace (Utilities|Select :Index CoordinateSpace :Local _current :World (Utilities|ZeroLocationWorldSpaceZ _world_current))) (Utilities|ZeroLocalorWorldSpaceZ _coordinatespace (Utilities|Select :Index CoordinateSpace :Local _previous :World (Utilities|ZeroLocationWorldSpaceZ _world_previous))))) (Math|Vector|GetUnitDirection(Vector) (Utilities|ZeroLocalorWorldSpaceZ _coordinatespace (Utilities|Select :Index CoordinateSpace :Local _current :World (Utilities|ZeroLocationWorldSpaceZ _world_current))) (Utilities|ZeroLocalorWorldSpaceZ _coordinatespace (Utilities|Select :Index CoordinateSpace :Local _next :World (Utilities|ZeroLocationWorldSpaceZ _world_next)))) (Utilities|IsOuterCorner (Utilities|ZeroLocalorWorldSpaceZ _coordinatespace (Utilities|Select :Index CoordinateSpace :Local _next :World (Utilities|ZeroLocationWorldSpaceZ _world_next))) (Utilities|ZeroLocalorWorldSpaceZ _coordinatespace (Utilities|Select :Index CoordinateSpace :Local _current :World (Utilities|ZeroLocationWorldSpaceZ _world_current))) (Utilities|ZeroLocalorWorldSpaceZ _coordinatespace (Utilities|Select :Index CoordinateSpace :Local _previous :World (Utilities|ZeroLocationWorldSpaceZ _world_previous))))) "0, 0,1.000000") (Math|Vector|SelectVector (Math|Vector|MakeVector (/ (* 512.0 (Math|Float|Round (/ (.Distance (Collision|BreakHitResult _hitresoultoutput_2)) 512.0))) 512.0) (/ (* 512.0 (Math|Float|Round (/ (.Distance (Collision|BreakHitResult _hitresoultoutput)) 512.0))) 512.0) 1.0) (Math|Vector|MakeVector (/ (* 512.0 (Math|Float|Round (/ (.Distance (Collision|BreakHitResult _hitresoultoutput)) 512.0))) 512.0) (/ (* 512.0 (Math|Float|Round (/ (.Distance (Collision|BreakHitResult _hitresoultoutput_2)) 512.0))) 512.0) 1.0) (Utilities|IsOuterCorner (Utilities|ZeroLocalorWorldSpaceZ _coordinatespace (Utilities|Select :Index CoordinateSpace :Local _next :World (Utilities|ZeroLocationWorldSpaceZ _world_next))) (Utilities|ZeroLocalorWorldSpaceZ _coordinatespace (Utilities|Select :Index CoordinateSpace :Local _current :World (Utilities|ZeroLocationWorldSpaceZ _world_current))) (Utilities|ZeroLocalorWorldSpaceZ _coordinatespace (Utilities|Select :Index CoordinateSpace :Local _previous :World (Utilities|ZeroLocationWorldSpaceZ _world_previous)))))) _returnvalue_6)))))) (Math|Transform|MakeTransform (Math|Vector|RotateVector "20.000000,-20.000000, 0" (.rotation (Utilities|Select :Index _coordinatespace_2 :Local (Math|Transform|MakeTransform (- (Utilities|ZeroLocalorWorldSpaceZ _coordinatespace (Utilities|Select :Index CoordinateSpace :Local _current :World (Utilities|ZeroLocationWorldSpaceZ _world_current))) _returnvalue_5) (Math|Rotator|MakeRotationfromAxes (Math|Vector|SelectVector (Math|Vector|GetUnitDirection(Vector) (Utilities|ZeroLocalorWorldSpaceZ _coordinatespace (Utilities|Select :Index CoordinateSpace :Local _current :World (Utilities|ZeroLocationWorldSpaceZ _world_current))) (Utilities|ZeroLocalorWorldSpaceZ _coordinatespace (Utilities|Select :Index CoordinateSpace :Local _previous :World (Utilities|ZeroLocationWorldSpaceZ _world_previous)))) (Math|Vector|NegateVector (Math|Vector|GetUnitDirection(Vector) (Utilities|ZeroLocalorWorldSpaceZ _coordinatespace (Utilities|Select :Index CoordinateSpace :Local _current :World (Utilities|ZeroLocationWorldSpaceZ _world_current))) (Utilities|ZeroLocalorWorldSpaceZ _coordinatespace (Utilities|Select :Index CoordinateSpace :Local _next :World (Utilities|ZeroLocationWorldSpaceZ _world_next))))) _reverse_2) (Math|Vector|SelectVector (Math|Vector|NegateVector (Math|Vector|GetUnitDirection(Vector) (Utilities|ZeroLocalorWorldSpaceZ _coordinatespace (Utilities|Select :Index CoordinateSpace :Local _current :World (Utilities|ZeroLocationWorldSpaceZ _world_current))) (Utilities|ZeroLocalorWorldSpaceZ _coordinatespace (Utilities|Select :Index CoordinateSpace :Local _next :World (Utilities|ZeroLocationWorldSpaceZ _world_next))))) (Math|Vector|GetUnitDirection(Vector) (Utilities|ZeroLocalorWorldSpaceZ _coordinatespace (Utilities|Select :Index CoordinateSpace :Local _current :World (Utilities|ZeroLocationWorldSpaceZ _world_current))) (Utilities|ZeroLocalorWorldSpaceZ _coordinatespace (Utilities|Select :Index CoordinateSpace :Local _previous :World (Utilities|ZeroLocationWorldSpaceZ _world_previous)))) _reverse_2) "0, 0,1.000000") (Math|Vector|MakeVector (/ (* 512.0 (Math|Float|Round (/ (.Distance (Collision|BreakHitResult _hitresoultoutput_2)) 512.0))) 512.0) (/ (* 512.0 (Math|Float|Round (/ (.Distance (Collision|BreakHitResult _hitresoultoutput)) 512.0))) 512.0) 1.0)) :World (Math|Transform|MakeRelativeTransform (Math|Transform|MakeTransform (- (Utilities|ZeroLocalorWorldSpaceZ _coordinatespace (Utilities|Select :Index CoordinateSpace :Local _current :World (Utilities|ZeroLocationWorldSpaceZ _world_current))) _returnvalue_5) (Math|Rotator|MakeRotationfromAxes (Math|Vector|SelectVector (Math|Vector|GetUnitDirection(Vector) (Utilities|ZeroLocalorWorldSpaceZ _coordinatespace (Utilities|Select :Index CoordinateSpace :Local _current :World (Utilities|ZeroLocationWorldSpaceZ _world_current))) (Utilities|ZeroLocalorWorldSpaceZ _coordinatespace (Utilities|Select :Index CoordinateSpace :Local _next :World (Utilities|ZeroLocationWorldSpaceZ _world_next)))) (Math|Vector|NegateVector (Math|Vector|GetUnitDirection(Vector) (Utilities|ZeroLocalorWorldSpaceZ _coordinatespace (Utilities|Select :Index CoordinateSpace :Local _current :World (Utilities|ZeroLocationWorldSpaceZ _world_current))) (Utilities|ZeroLocalorWorldSpaceZ _coordinatespace (Utilities|Select :Index CoordinateSpace :Local _previous :World (Utilities|ZeroLocationWorldSpaceZ _world_previous))))) (Utilities|IsOuterCorner (Utilities|ZeroLocalorWorldSpaceZ _coordinatespace (Utilities|Select :Index CoordinateSpace :Local _next :World (Utilities|ZeroLocationWorldSpaceZ _world_next))) (Utilities|ZeroLocalorWorldSpaceZ _coordinatespace (Utilities|Select :Index CoordinateSpace :Local _current :World (Utilities|ZeroLocationWorldSpaceZ _world_current))) (Utilities|ZeroLocalorWorldSpaceZ _coordinatespace (Utilities|Select :Index CoordinateSpace :Local _previous :World (Utilities|ZeroLocationWorldSpaceZ _world_previous))))) (Math|Vector|SelectVector (Math|Vector|NegateVector (Math|Vector|GetUnitDirection(Vector) (Utilities|ZeroLocalorWorldSpaceZ _coordinatespace (Utilities|Select :Index CoordinateSpace :Local _current :World (Utilities|ZeroLocationWorldSpaceZ _world_current))) (Utilities|ZeroLocalorWorldSpaceZ _coordinatespace (Utilities|Select :Index CoordinateSpace :Local _previous :World (Utilities|ZeroLocationWorldSpaceZ _world_previous))))) (Math|Vector|GetUnitDirection(Vector) (Utilities|ZeroLocalorWorldSpaceZ _coordinatespace (Utilities|Select :Index CoordinateSpace :Local _current :World (Utilities|ZeroLocationWorldSpaceZ _world_current))) (Utilities|ZeroLocalorWorldSpaceZ _coordinatespace (Utilities|Select :Index CoordinateSpace :Local _next :World (Utilities|ZeroLocationWorldSpaceZ _world_next)))) (Utilities|IsOuterCorner (Utilities|ZeroLocalorWorldSpaceZ _coordinatespace (Utilities|Select :Index CoordinateSpace :Local _next :World (Utilities|ZeroLocationWorldSpaceZ _world_next))) (Utilities|ZeroLocalorWorldSpaceZ _coordinatespace (Utilities|Select :Index CoordinateSpace :Local _current :World (Utilities|ZeroLocationWorldSpaceZ _world_current))) (Utilities|ZeroLocalorWorldSpaceZ _coordinatespace (Utilities|Select :Index CoordinateSpace :Local _previous :World (Utilities|ZeroLocationWorldSpaceZ _world_previous))))) "0, 0,1.000000") (Math|Vector|SelectVector (Math|Vector|MakeVector (/ (* 512.0 (Math|Float|Round (/ (.Distance (Collision|BreakHitResult _hitresoultoutput_2)) 512.0))) 512.0) (/ (* 512.0 (Math|Float|Round (/ (.Distance (Collision|BreakHitResult _hitresoultoutput)) 512.0))) 512.0) 1.0) (Math|Vector|MakeVector (/ (* 512.0 (Math|Float|Round (/ (.Distance (Collision|BreakHitResult _hitresoultoutput)) 512.0))) 512.0) (/ (* 512.0 (Math|Float|Round (/ (.Distance (Collision|BreakHitResult _hitresoultoutput_2)) 512.0))) 512.0) 1.0) (Utilities|IsOuterCorner (Utilities|ZeroLocalorWorldSpaceZ _coordinatespace (Utilities|Select :Index CoordinateSpace :Local _next :World (Utilities|ZeroLocationWorldSpaceZ _world_next))) (Utilities|ZeroLocalorWorldSpaceZ _coordinatespace (Utilities|Select :Index CoordinateSpace :Local _current :World (Utilities|ZeroLocationWorldSpaceZ _world_current))) (Utilities|ZeroLocalorWorldSpaceZ _coordinatespace (Utilities|Select :Index CoordinateSpace :Local _previous :World (Utilities|ZeroLocationWorldSpaceZ _world_previous)))))) _returnvalue_6))))))))
                    (Utilities|Array|Add (Variables|MeshLists|GetTempCollisionList) _returnvalue_8)))))))))
    (:then_1
      (if SpawnComponents
        (for _array_element (LocalVariables|GetTransformList)
          (Utilities|Array|Add (Variables|X.%29DoNOTEdit|GetComponent%20Transform%20List) (Utilities|Struct|MakeProcComponenttoSpawn _array_element (Variables|GetProc%20Asset%20List) (Utilities|Array|Get(acopy) (Variables|GetAsset%20Index%20List)))))))
    (:then_2
      (return (Utilities|Array|Length (LocalVariables|GetTransformList))))))

(fn MakeBasementFloorOptimized
  (params (Proc%20AssetList :type Proc%20Asset%20List%20Object%20Reference) (Top :type Boolean) (InputSpline :type Spline%20Component%20Object%20Reference))
  (CreateComponents|MakeBasementFloor 0 true Top Proc%20AssetList InputSpline))

(fn MakeBasementWalls
  (bind _cornerpadding (Variables|Default|GetCornerPadding))
  (bind _returnvalue (* _cornerpadding 2.0))
  (if (> (Variables|Default|GetNumberOfFloors) 0)
    (Utilities|FlowControl|Sequence
      (:then_0
        (bind _assetindexlist (Tools|GetAssetsOfType (Variables|/Game/ArtTools/Procedural/Foundation/Proc_AssetList.Proc_AssetList_C|GetAssetList (Variables|E.%29Meshes|GetBasement_Asset_List))))
        (LocalVariables|SetWall_Asset_Index_List _assetindexlist)
        (bind _assetindexlist_2 (Tools|GetAssetsOfType (Variables|/Game/ArtTools/Procedural/Foundation/Proc_AssetList.Proc_AssetList_C|GetAssetList (Variables|E.%29Meshes|GetBasement_Asset_List)) "NewEnumerator3"))
        (LocalVariables|SetOuter_Corner_Index_List _assetindexlist_2)
        (bind _assetindexlist_3 (Tools|GetAssetsOfType (Variables|/Game/ArtTools/Procedural/Foundation/Proc_AssetList.Proc_AssetList_C|GetAssetList (Variables|E.%29Meshes|GetBasement_Asset_List)) "NewEnumerator4"))
        (LocalVariables|SetInner_Corner_Index_List _assetindexlist_3))
      (:then_1
        (bind _splinecopy (SplineOperations|MakeCleanOptimizedSplineCopy (Variables|Default|GetSpline)))
        (Utilities|Array|Add (Variables|SplineList|GetTempSplineList) _splinecopy))
      (:then_2
        _)
      (:then_3
        (Cleanup|DeleteAllTempSplines)))
    (for _index (range (Spline|GetNumberOfSplinePoints _splinecopy))
      (bind (_next__1 _next _current _previous) (Tools|GetCurrentNextPreviousSplinePoint _splinecopy _index))
      (Utilities|FlowControl|Sequence
        (:then_0
          (if (> (- (Math|Vector|VectorLength (- _current _next)) _returnvalue) 1.0)
            (if (Variables|Default|GetAllowScaling)
              (Utilities|Array|Add (Variables|X.%29DoNOTEdit|GetComponent%20Transform%20List) (Utilities|Struct|MakeProcComponenttoSpawn (Math|Transform|MakeTransform (+ (* (Math|Vector|Normalize (- _current _next)) _cornerpadding) _next) (Math|Rotator|MakeRotfromX (- _current _next)) (Math|Vector|MakeVector (Math|Float|SelectFloat (/ (- (Math|Vector|VectorLength (- _current _next)) _returnvalue) 1024.0) (/ (- (Math|Vector|VectorLength (- _current _next)) _returnvalue) 512.0) (>= (/ (- (Math|Vector|VectorLength (- _current _next)) _returnvalue) 1024.0) 0.75)) 1.0 (Math|Conversions|ToFloat(Integer) (Variables|Default|GetNumberOfFloors)))) (Variables|E.%29Meshes|GetBasement_Asset_List) (Utilities|Array|Get(acopy) (LocalVariables|GetWall_Asset_Index_List) (Math|Integer|SelectInt 0 1 (>= (/ (- (Math|Vector|VectorLength (- _current _next)) _returnvalue) 1024.0) 0.75)))))
              (elif (> (Math|Float|Floor (/ (- (Math|Vector|VectorLength (- _current _next)) _returnvalue) 1024.0)) 0)
                (for _index_2 (range (Math|Float|Floor (/ (- (Math|Vector|VectorLength (- _current _next)) _returnvalue) 1024.0)))
                  (Utilities|Array|Add (Variables|X.%29DoNOTEdit|GetComponent%20Transform%20List) (Utilities|Struct|MakeProcComponenttoSpawn (Math|Transform|MakeTransform (+ (+ (* (Math|Vector|Normalize (- _current _next)) _cornerpadding) _next) (* (* (Math|Vector|Normalize (- _current _next)) 1024.0) (Math|Conversions|ToFloat(Integer) _index_2))) (Math|Rotator|MakeRotfromX (- _current _next)) (Math|Vector|MakeVector 1.0 1.0 (Math|Conversions|ToFloat(Integer) (Variables|Default|GetNumberOfFloors)))) (Variables|E.%29Meshes|GetBasement_Asset_List) (Utilities|Array|Get(acopy) (LocalVariables|GetWall_Asset_Index_List)))))
                (if (> (Math|Float|Fraction (/ (- (Math|Vector|VectorLength (- _current _next)) _returnvalue) 1024.0)) 0.0)
                  (Utilities|Array|Add (Variables|X.%29DoNOTEdit|GetComponent%20Transform%20List) (Utilities|Struct|MakeProcComponenttoSpawn (Math|Transform|MakeTransform (+ (* (* (Math|Vector|Normalize (- _current _next)) 1024.0) (Math|Conversions|ToFloat(Integer) (Math|Float|Floor (/ (- (Math|Vector|VectorLength (- _current _next)) _returnvalue) 1024.0)))) (+ (* (Math|Vector|Normalize (- _current _next)) _cornerpadding) _next)) (Math|Rotator|MakeRotfromX (- _current _next)) (Math|Vector|MakeVector 1.0 1.0 (Math|Conversions|ToFloat(Integer) (Variables|Default|GetNumberOfFloors)))) (Variables|E.%29Meshes|GetBasement_Asset_List) (Utilities|Array|Get(acopy) (LocalVariables|GetWall_Asset_Index_List) 1))))
                (else
                  _)))))
        (:then_1
          ;[1 1 1] CornerPieces
          (Utilities|Array|Add (Variables|X.%29DoNOTEdit|GetComponent%20Transform%20List) (Utilities|Struct|MakeProcComponenttoSpawn (Math|Transform|MakeTransform (+ _current (Math|Vector|SelectVector (* (Math|Vector|Normalize (- _next _current)) 256.0) (* (Math|Vector|Normalize (- _next _current)) 256.0) (> (Math|Vector|DotProduct (Math|Vector|CrossProduct (- _current _previous) (Math|Vector|MakeVector 0.0 0.0 1.0)) (- _next _current)) 0.0))) (Math|Rotator|SelectRotator (Math|Rotator|MakeRotfromX (Math|Vector|NegateVector (- _next _current))) (Math|Rotator|CombineRotators (Math|Rotator|MakeRotfromX (- _previous _current)) "0,90.000000, 0") (> (Math|Vector|DotProduct (Math|Vector|CrossProduct (- _current _previous) (Math|Vector|MakeVector 0.0 0.0 1.0)) (- _next _current)) 0.0)) (Math|Vector|MakeVector 1.0 1.0 (Math|Conversions|ToFloat(Integer) (Variables|Default|GetNumberOfFloors)))) (Variables|E.%29Meshes|GetBasement_Asset_List) (Math|Integer|SelectInt (Utilities|Array|Get(acopy) (LocalVariables|GetOuter_Corner_Index_List)) (Utilities|Array|Get(acopy) (LocalVariables|GetInner_Corner_Index_List)) (> (Math|Vector|DotProduct (Math|Vector|CrossProduct (- _current _previous) (Math|Vector|MakeVector 0.0 0.0 1.0)) (- _next _current)) 0.0)))))))))

(fn MakeBlendSpline
  (params (Spline_Inner :type Spline%20Component%20Object%20Reference) (Spline_Outer :type Spline%20Component%20Object%20Reference) (Fraction :type Float%20%28double-precision%29) (Heightbias :type Float%20%28double-precision%29))
  (for _index (range (Spline|GetNumberOfSplinePoints Spline_Inner))
    (Spline|SetSplinePointType Spline_Outer _index "Linear" true)
    (Spline|SetRotationatSplinePoint Spline_Outer _index (Math|Rotator|MakeRotationfromAxes (Math|Vector|CrossProduct (- (Spline|GetLocationatSplinePoint Spline_Inner _index "Local") (Spline|GetLocationatSplinePoint Spline_Outer _index "Local")) "0, 0,1.000000") (- (Spline|GetLocationatSplinePoint Spline_Inner _index "Local") (Spline|GetLocationatSplinePoint Spline_Outer _index "Local")) (Math|Vector|CrossProduct (Math|Vector|CrossProduct (- (Spline|GetLocationatSplinePoint Spline_Inner _index "Local") (Spline|GetLocationatSplinePoint Spline_Outer _index "Local")) "0, 0,1.000000") (- (Spline|GetLocationatSplinePoint Spline_Inner _index "Local") (Spline|GetLocationatSplinePoint Spline_Outer _index "Local")))) "Local" true)
    (Spline|SetLocationatSplinePoint Spline_Outer _index (+ (+ (Spline|GetLocationatSplinePoint Spline_Outer _index "Local") (* (Math|Vector|Normalize (- (Spline|GetLocationatSplinePoint Spline_Inner _index "Local") (Spline|GetLocationatSplinePoint Spline_Outer _index "Local"))) (/ (Math|Vector|VectorLength (- (Spline|GetLocationatSplinePoint Spline_Inner _index "Local") (Spline|GetLocationatSplinePoint Spline_Outer _index "Local"))) Fraction))) (* (* (- (Spline|GetLocationatSplinePoint Spline_Inner _index "Local") (Spline|GetLocationatSplinePoint Spline_Outer _index "Local")) "0, 0,1.000000") Heightbias)) "Local" true))
  (return Spline_Outer))

(fn MakeCleanOptimizedSplineCopy
  (params (Spline :type Spline%20Component%20Object%20Reference) (Debug :type Boolean))
  (Utilities|FlowControl|Sequence
    (:then_0
      (bind _returnvalue (AddComponent|NODE_AddSplineComponent-11 :RelativeTransform (Math|Transform|MakeTransform "0, 0, 0")))
      (LocalVariables|SetTempSpline _returnvalue))
    (:then_1
      (for _index (range (Spline|GetNumberOfSplinePoints Spline))
        (bind (_next__1 _next _current _previous _previous__1 _world_next__1 _world_next _world_current) (Tools|GetCurrentNextPreviousSplinePoint Spline _index))
        (if (< (Math|Float|Absolute(Float) (Math|Vector|DotProduct (Math|Vector|GetUnitDirection(Vector) _current _next) (Math|Vector|GetUnitDirection(Vector) _previous _current))) 0.9)
          (Utilities|Array|AddUnique (LocalVariables|GetPointList) (+ _world_current (Math|Vector|SelectVector "0, 0,512.000000" "0, 0, 0" Debug))))))
    (:then_2
      (Spline|SetSplinePoints (LocalVariables|GetTempSpline) (LocalVariables|GetPointList) "World" true))
    (:then_3
      (return (LocalVariables|GetTempSpline)))))

(fn MakeFillCollisionFromSpline
  (params (Spline :type Spline%20Component%20Object%20Reference) (CoordinateSpace :type ESplineCoordinateSpace%20Enum :default "Local") (Debug :type Boolean))
  (bind _returnvalue 512.0)
  (bind _returnvalue_2 (Spline|GetNumberOfSplinePoints Spline))
  (bind _returnvalue_3 (- _returnvalue_2 1))
  (bind _returnvalue_4 (/ _returnvalue 2.0))
  (Utilities|FlowControl|Sequence
    (:then_0
      (CreateCollisionComponents|MakeBorderCollisionFromSpline Spline false CoordinateSpace))
    (:then_1
      ;[1 1 1] CollisionMeshes Makes a border around the defined spline as collision for the traces
      (for _index (range _returnvalue_2)
        (if (>= (- (Math|Float|Round (/ (Math|Vector|VectorLength (- (Spline|GetLocationatSplinePoint Spline (Math|Integer|SelectInt 0 (+ _index 1) (== _index _returnvalue_3)) CoordinateSpace) (Spline|GetLocationatSplinePoint Spline _index CoordinateSpace))) 512.0)) 1) 0)
          (for _index_2 (range (+ (- (Math|Float|Round (/ (Math|Vector|VectorLength (- (Spline|GetLocationatSplinePoint Spline (Math|Integer|SelectInt 0 (+ _index 1) (== _index _returnvalue_3)) CoordinateSpace) (Spline|GetLocationatSplinePoint Spline _index CoordinateSpace))) 512.0)) 1) 1))
            (bind (_outhits _returnvalue_5) (Collision|MultiLineTraceForObjects :Start (Utilities|Select :Index (Variables|GetCoordinateSpace) :Local (+ (+ (Spline|GetLocationatSplinePoint Spline _index "World") (* (Math|Vector|Normalize (- (Spline|GetLocationatSplinePoint Spline (Math|Integer|SelectInt 0 (+ _index 1) (== _index _returnvalue_3)) "World") (Spline|GetLocationatSplinePoint Spline _index "World"))) (+ (* _index_2 _returnvalue) _returnvalue_4))) (* (Math|Vector|Normalize (Math|Vector|CrossProduct (- (Spline|GetLocationatSplinePoint Spline (Math|Integer|SelectInt 0 (+ _index 1) (== _index _returnvalue_3)) "World") (Spline|GetLocationatSplinePoint Spline _index "World")) "0, 0,-1.000000")) 2.0)) :World (Math|Vector|MakeVector (.x (+ (+ (Spline|GetLocationatSplinePoint Spline _index "World") (* (Math|Vector|Normalize (- (Spline|GetLocationatSplinePoint Spline (Math|Integer|SelectInt 0 (+ _index 1) (== _index _returnvalue_3)) "World") (Spline|GetLocationatSplinePoint Spline _index "World"))) (+ (* _index_2 _returnvalue) _returnvalue_4))) (* (Math|Vector|Normalize (Math|Vector|CrossProduct (- (Spline|GetLocationatSplinePoint Spline (Math|Integer|SelectInt 0 (+ _index 1) (== _index _returnvalue_3)) "World") (Spline|GetLocationatSplinePoint Spline _index "World")) "0, 0,-1.000000")) 2.0))) (.y (+ (+ (Spline|GetLocationatSplinePoint Spline _index "World") (* (Math|Vector|Normalize (- (Spline|GetLocationatSplinePoint Spline (Math|Integer|SelectInt 0 (+ _index 1) (== _index _returnvalue_3)) "World") (Spline|GetLocationatSplinePoint Spline _index "World"))) (+ (* _index_2 _returnvalue) _returnvalue_4))) (* (Math|Vector|Normalize (Math|Vector|CrossProduct (- (Spline|GetLocationatSplinePoint Spline (Math|Integer|SelectInt 0 (+ _index 1) (== _index _returnvalue_3)) "World") (Spline|GetLocationatSplinePoint Spline _index "World")) "0, 0,-1.000000")) 2.0))) (.z (Transformation|GetActorLocation)))) :End (Utilities|Select :Index (Variables|GetCoordinateSpace) :Local (+ (* (Math|Vector|Normalize (Math|Vector|CrossProduct (- (Spline|GetLocationatSplinePoint Spline (Math|Integer|SelectInt 0 (+ _index 1) (== _index _returnvalue_3)) "World") (Spline|GetLocationatSplinePoint Spline _index "World")) "0, 0,-1.000000")) 4e+04) (+ (Spline|GetLocationatSplinePoint Spline _index "World") (* (Math|Vector|Normalize (- (Spline|GetLocationatSplinePoint Spline (Math|Integer|SelectInt 0 (+ _index 1) (== _index _returnvalue_3)) "World") (Spline|GetLocationatSplinePoint Spline _index "World"))) (+ (* _index_2 _returnvalue) _returnvalue_4)))) :World (Math|Vector|MakeVector (.x (+ (* (Math|Vector|Normalize (Math|Vector|CrossProduct (- (Spline|GetLocationatSplinePoint Spline (Math|Integer|SelectInt 0 (+ _index 1) (== _index _returnvalue_3)) "World") (Spline|GetLocationatSplinePoint Spline _index "World")) "0, 0,-1.000000")) 4e+04) (+ (Spline|GetLocationatSplinePoint Spline _index "World") (* (Math|Vector|Normalize (- (Spline|GetLocationatSplinePoint Spline (Math|Integer|SelectInt 0 (+ _index 1) (== _index _returnvalue_3)) "World") (Spline|GetLocationatSplinePoint Spline _index "World"))) (+ (* _index_2 _returnvalue) _returnvalue_4))))) (.y (+ (* (Math|Vector|Normalize (Math|Vector|CrossProduct (- (Spline|GetLocationatSplinePoint Spline (Math|Integer|SelectInt 0 (+ _index 1) (== _index _returnvalue_3)) "World") (Spline|GetLocationatSplinePoint Spline _index "World")) "0, 0,-1.000000")) 4e+04) (+ (Spline|GetLocationatSplinePoint Spline _index "World") (* (Math|Vector|Normalize (- (Spline|GetLocationatSplinePoint Spline (Math|Integer|SelectInt 0 (+ _index 1) (== _index _returnvalue_3)) "World") (Spline|GetLocationatSplinePoint Spline _index "World"))) (+ (* _index_2 _returnvalue) _returnvalue_4))))) (.z (Transformation|GetActorLocation)))) :ObjectTypes (Utilities|Array|MakeArray "ObjectTypeQuery3") :bTraceComplex false :DrawDebugType (select (and (== _index (Variables|Debug|GetTest)) (Variables|GetDebug)) "ForDuration" "None") :bIgnoreSelf false :TraceColor "(R=1.000000,G=0.000000,B=0.000000,A=1.000000)" :TraceHitColor "(R=0.000000,G=1.000000,B=0.000000,A=1.000000)" :DrawTime 5.0))
            (if _returnvalue_5
              (bind _hitresoultoutput (Tools|ClosestHitResult _outhits (+ (Spline|GetLocationatSplinePoint Spline _index "World") (* (Math|Vector|Normalize (- (Spline|GetLocationatSplinePoint Spline (Math|Integer|SelectInt 0 (+ _index 1) (== _index _returnvalue_3)) "World") (Spline|GetLocationatSplinePoint Spline _index "World"))) (+ (* _index_2 _returnvalue) _returnvalue_4)))))
              (if (> (Math|Float|Round (/ (.Distance (Collision|BreakHitResult _hitresoultoutput)) 512.0)) 0)
                (for _index_3 (range (Math|Float|Round (/ (.Distance (Collision|BreakHitResult _hitresoultoutput)) 512.0)))
                  (Utilities|Array|AddUnique (LocalVariables|GetPointList) (Math|Vector|MakeVector (Math|Conversions|ToFloat(Integer) (Math|Float|Round (.x (Utilities|Select :Index (Variables|GetCoordinateSpace) :Local (Math|Vector|MakeVector (.x (- (Math|Vector|vector+vector (Spline|GetLocationatSplinePoint Spline _index CoordinateSpace) (* (Math|Vector|Normalize (- (Spline|GetLocationatSplinePoint Spline (Math|Integer|SelectInt 0 (+ _index 1) (== _index _returnvalue_3)) CoordinateSpace) (Spline|GetLocationatSplinePoint Spline _index CoordinateSpace))) (+ (* _index_2 512.0) (/ 512.0 2.0))) (* (Math|Vector|Normalize (Math|Vector|NegateVector (Math|Vector|CrossProduct (- (Spline|GetLocationatSplinePoint Spline (Math|Integer|SelectInt 0 (+ _index 1) (== _index _returnvalue_3)) CoordinateSpace) (Spline|GetLocationatSplinePoint Spline _index CoordinateSpace)) "0, 0,1.000000"))) (+ (* _index_3 512.0) (/ 512.0 2.0)))) (Utilities|Select :Index CoordinateSpace :Local "0, 0, 0" :World (Transformation|GetActorLocation)))) (.y (- (Math|Vector|vector+vector (Spline|GetLocationatSplinePoint Spline _index CoordinateSpace) (* (Math|Vector|Normalize (- (Spline|GetLocationatSplinePoint Spline (Math|Integer|SelectInt 0 (+ _index 1) (== _index _returnvalue_3)) CoordinateSpace) (Spline|GetLocationatSplinePoint Spline _index CoordinateSpace))) (+ (* _index_2 512.0) (/ 512.0 2.0))) (* (Math|Vector|Normalize (Math|Vector|NegateVector (Math|Vector|CrossProduct (- (Spline|GetLocationatSplinePoint Spline (Math|Integer|SelectInt 0 (+ _index 1) (== _index _returnvalue_3)) CoordinateSpace) (Spline|GetLocationatSplinePoint Spline _index CoordinateSpace)) "0, 0,1.000000"))) (+ (* _index_3 512.0) (/ 512.0 2.0)))) (Utilities|Select :Index CoordinateSpace :Local "0, 0, 0" :World (Transformation|GetActorLocation))))) :World (Math|Vector|RotateVector (Math|Vector|MakeVector (.x (- (Math|Vector|vector+vector (Spline|GetLocationatSplinePoint Spline _index CoordinateSpace) (* (Math|Vector|Normalize (- (Spline|GetLocationatSplinePoint Spline (Math|Integer|SelectInt 0 (+ _index 1) (== _index _returnvalue_3)) CoordinateSpace) (Spline|GetLocationatSplinePoint Spline _index CoordinateSpace))) (+ (* _index_2 512.0) (/ 512.0 2.0))) (* (Math|Vector|Normalize (Math|Vector|NegateVector (Math|Vector|CrossProduct (- (Spline|GetLocationatSplinePoint Spline (Math|Integer|SelectInt 0 (+ _index 1) (== _index _returnvalue_3)) CoordinateSpace) (Spline|GetLocationatSplinePoint Spline _index CoordinateSpace)) "0, 0,1.000000"))) (+ (* _index_3 512.0) (/ 512.0 2.0)))) (Utilities|Select :Index CoordinateSpace :Local "0, 0, 0" :World (Transformation|GetActorLocation)))) (.y (- (Math|Vector|vector+vector (Spline|GetLocationatSplinePoint Spline _index CoordinateSpace) (* (Math|Vector|Normalize (- (Spline|GetLocationatSplinePoint Spline (Math|Integer|SelectInt 0 (+ _index 1) (== _index _returnvalue_3)) CoordinateSpace) (Spline|GetLocationatSplinePoint Spline _index CoordinateSpace))) (+ (* _index_2 512.0) (/ 512.0 2.0))) (* (Math|Vector|Normalize (Math|Vector|NegateVector (Math|Vector|CrossProduct (- (Spline|GetLocationatSplinePoint Spline (Math|Integer|SelectInt 0 (+ _index 1) (== _index _returnvalue_3)) CoordinateSpace) (Spline|GetLocationatSplinePoint Spline _index CoordinateSpace)) "0, 0,1.000000"))) (+ (* _index_3 512.0) (/ 512.0 2.0)))) (Utilities|Select :Index CoordinateSpace :Local "0, 0, 0" :World (Transformation|GetActorLocation))))) (Math|Rotator|InvertRotator (Transformation|GetActorRotation))))))) (Math|Conversions|ToFloat(Integer) (Math|Float|Round (.y (Utilities|Select :Index (Variables|GetCoordinateSpace) :Local (Math|Vector|MakeVector (.x (- (Math|Vector|vector+vector (Spline|GetLocationatSplinePoint Spline _index CoordinateSpace) (* (Math|Vector|Normalize (- (Spline|GetLocationatSplinePoint Spline (Math|Integer|SelectInt 0 (+ _index 1) (== _index _returnvalue_3)) CoordinateSpace) (Spline|GetLocationatSplinePoint Spline _index CoordinateSpace))) (+ (* _index_2 512.0) (/ 512.0 2.0))) (* (Math|Vector|Normalize (Math|Vector|NegateVector (Math|Vector|CrossProduct (- (Spline|GetLocationatSplinePoint Spline (Math|Integer|SelectInt 0 (+ _index 1) (== _index _returnvalue_3)) CoordinateSpace) (Spline|GetLocationatSplinePoint Spline _index CoordinateSpace)) "0, 0,1.000000"))) (+ (* _index_3 512.0) (/ 512.0 2.0)))) (Utilities|Select :Index CoordinateSpace :Local "0, 0, 0" :World (Transformation|GetActorLocation)))) (.y (- (Math|Vector|vector+vector (Spline|GetLocationatSplinePoint Spline _index CoordinateSpace) (* (Math|Vector|Normalize (- (Spline|GetLocationatSplinePoint Spline (Math|Integer|SelectInt 0 (+ _index 1) (== _index _returnvalue_3)) CoordinateSpace) (Spline|GetLocationatSplinePoint Spline _index CoordinateSpace))) (+ (* _index_2 512.0) (/ 512.0 2.0))) (* (Math|Vector|Normalize (Math|Vector|NegateVector (Math|Vector|CrossProduct (- (Spline|GetLocationatSplinePoint Spline (Math|Integer|SelectInt 0 (+ _index 1) (== _index _returnvalue_3)) CoordinateSpace) (Spline|GetLocationatSplinePoint Spline _index CoordinateSpace)) "0, 0,1.000000"))) (+ (* _index_3 512.0) (/ 512.0 2.0)))) (Utilities|Select :Index CoordinateSpace :Local "0, 0, 0" :World (Transformation|GetActorLocation))))) :World (Math|Vector|RotateVector (Math|Vector|MakeVector (.x (- (Math|Vector|vector+vector (Spline|GetLocationatSplinePoint Spline _index CoordinateSpace) (* (Math|Vector|Normalize (- (Spline|GetLocationatSplinePoint Spline (Math|Integer|SelectInt 0 (+ _index 1) (== _index _returnvalue_3)) CoordinateSpace) (Spline|GetLocationatSplinePoint Spline _index CoordinateSpace))) (+ (* _index_2 512.0) (/ 512.0 2.0))) (* (Math|Vector|Normalize (Math|Vector|NegateVector (Math|Vector|CrossProduct (- (Spline|GetLocationatSplinePoint Spline (Math|Integer|SelectInt 0 (+ _index 1) (== _index _returnvalue_3)) CoordinateSpace) (Spline|GetLocationatSplinePoint Spline _index CoordinateSpace)) "0, 0,1.000000"))) (+ (* _index_3 512.0) (/ 512.0 2.0)))) (Utilities|Select :Index CoordinateSpace :Local "0, 0, 0" :World (Transformation|GetActorLocation)))) (.y (- (Math|Vector|vector+vector (Spline|GetLocationatSplinePoint Spline _index CoordinateSpace) (* (Math|Vector|Normalize (- (Spline|GetLocationatSplinePoint Spline (Math|Integer|SelectInt 0 (+ _index 1) (== _index _returnvalue_3)) CoordinateSpace) (Spline|GetLocationatSplinePoint Spline _index CoordinateSpace))) (+ (* _index_2 512.0) (/ 512.0 2.0))) (* (Math|Vector|Normalize (Math|Vector|NegateVector (Math|Vector|CrossProduct (- (Spline|GetLocationatSplinePoint Spline (Math|Integer|SelectInt 0 (+ _index 1) (== _index _returnvalue_3)) CoordinateSpace) (Spline|GetLocationatSplinePoint Spline _index CoordinateSpace)) "0, 0,1.000000"))) (+ (* _index_3 512.0) (/ 512.0 2.0)))) (Utilities|Select :Index CoordinateSpace :Local "0, 0, 0" :World (Transformation|GetActorLocation))))) (Math|Rotator|InvertRotator (Transformation|GetActorRotation))))))) (Math|Conversions|ToFloat(Integer) (Math|Float|Round (.z (Utilities|Select :Index (Variables|GetCoordinateSpace) :Local (Math|Vector|MakeVector (.x (- (Math|Vector|vector+vector (Spline|GetLocationatSplinePoint Spline _index CoordinateSpace) (* (Math|Vector|Normalize (- (Spline|GetLocationatSplinePoint Spline (Math|Integer|SelectInt 0 (+ _index 1) (== _index _returnvalue_3)) CoordinateSpace) (Spline|GetLocationatSplinePoint Spline _index CoordinateSpace))) (+ (* _index_2 512.0) (/ 512.0 2.0))) (* (Math|Vector|Normalize (Math|Vector|NegateVector (Math|Vector|CrossProduct (- (Spline|GetLocationatSplinePoint Spline (Math|Integer|SelectInt 0 (+ _index 1) (== _index _returnvalue_3)) CoordinateSpace) (Spline|GetLocationatSplinePoint Spline _index CoordinateSpace)) "0, 0,1.000000"))) (+ (* _index_3 512.0) (/ 512.0 2.0)))) (Utilities|Select :Index CoordinateSpace :Local "0, 0, 0" :World (Transformation|GetActorLocation)))) (.y (- (Math|Vector|vector+vector (Spline|GetLocationatSplinePoint Spline _index CoordinateSpace) (* (Math|Vector|Normalize (- (Spline|GetLocationatSplinePoint Spline (Math|Integer|SelectInt 0 (+ _index 1) (== _index _returnvalue_3)) CoordinateSpace) (Spline|GetLocationatSplinePoint Spline _index CoordinateSpace))) (+ (* _index_2 512.0) (/ 512.0 2.0))) (* (Math|Vector|Normalize (Math|Vector|NegateVector (Math|Vector|CrossProduct (- (Spline|GetLocationatSplinePoint Spline (Math|Integer|SelectInt 0 (+ _index 1) (== _index _returnvalue_3)) CoordinateSpace) (Spline|GetLocationatSplinePoint Spline _index CoordinateSpace)) "0, 0,1.000000"))) (+ (* _index_3 512.0) (/ 512.0 2.0)))) (Utilities|Select :Index CoordinateSpace :Local "0, 0, 0" :World (Transformation|GetActorLocation))))) :World (Math|Vector|RotateVector (Math|Vector|MakeVector (.x (- (Math|Vector|vector+vector (Spline|GetLocationatSplinePoint Spline _index CoordinateSpace) (* (Math|Vector|Normalize (- (Spline|GetLocationatSplinePoint Spline (Math|Integer|SelectInt 0 (+ _index 1) (== _index _returnvalue_3)) CoordinateSpace) (Spline|GetLocationatSplinePoint Spline _index CoordinateSpace))) (+ (* _index_2 512.0) (/ 512.0 2.0))) (* (Math|Vector|Normalize (Math|Vector|NegateVector (Math|Vector|CrossProduct (- (Spline|GetLocationatSplinePoint Spline (Math|Integer|SelectInt 0 (+ _index 1) (== _index _returnvalue_3)) CoordinateSpace) (Spline|GetLocationatSplinePoint Spline _index CoordinateSpace)) "0, 0,1.000000"))) (+ (* _index_3 512.0) (/ 512.0 2.0)))) (Utilities|Select :Index CoordinateSpace :Local "0, 0, 0" :World (Transformation|GetActorLocation)))) (.y (- (Math|Vector|vector+vector (Spline|GetLocationatSplinePoint Spline _index CoordinateSpace) (* (Math|Vector|Normalize (- (Spline|GetLocationatSplinePoint Spline (Math|Integer|SelectInt 0 (+ _index 1) (== _index _returnvalue_3)) CoordinateSpace) (Spline|GetLocationatSplinePoint Spline _index CoordinateSpace))) (+ (* _index_2 512.0) (/ 512.0 2.0))) (* (Math|Vector|Normalize (Math|Vector|NegateVector (Math|Vector|CrossProduct (- (Spline|GetLocationatSplinePoint Spline (Math|Integer|SelectInt 0 (+ _index 1) (== _index _returnvalue_3)) CoordinateSpace) (Spline|GetLocationatSplinePoint Spline _index CoordinateSpace)) "0, 0,1.000000"))) (+ (* _index_3 512.0) (/ 512.0 2.0)))) (Utilities|Select :Index CoordinateSpace :Local "0, 0, 0" :World (Transformation|GetActorLocation))))) (Math|Rotator|InvertRotator (Transformation|GetActorRotation))))))))))))))))
    (:then_2
      (for _array_element (LocalVariables|GetPointList)
        (bind _returnvalue_6 (AddComponent|NODE_AddStaticMeshComponent-25 :RelativeTransform (Math|Transform|MakeTransform (+ _array_element "-256.000000,256.000000, 0"))))
        (Variables|/Script/Engine.ActorComponent|SetComponentTags (Utilities|Array|MakeArray "border") _returnvalue_6)
        (if (not Debug)
          (Utilities|Array|Add (Variables|MeshLists|GetFillCollisionList) _returnvalue_6))))))

(fn MakeLandscapeMeshBlend
  (bind _terrainblend_asset_index_list (LocalVariables|GetTerrainBlend_Asset_Index_List))
  (bind _output (Utilities|Array|Get(acopy) _terrainblend_asset_index_list))
  (bind _terrainblend_asset_list (Variables|E.%29Meshes|GetTerrainBlend_Asset_List))
  (bind _assetlist (Variables|/Game/ArtTools/Procedural/Foundation/Proc_AssetList.Proc_AssetList_C|GetAssetList _terrainblend_asset_list))
  (bind _output_2 (Utilities|Array|Get(acopy) _assetlist _output))
  (bind (_staticmesh_2_d8df7d7b4556a6c1fde39bb633b7e288 _type_6_b4cd9ff0498f54172801239a22e7a5ca _prepivot_29_094082c34c836f4f1ab3deb660eef580 _forwarddirection_32_09c0a6674706473836365795dce83b27 _sizex_11_dfccfeb9407ac738b8ec5aaf3a48e13c) (Utilities|Struct|BreakProcAssetInfo _output_2))
  (bind _returnvalue (Math|Vector|MakeVector 0.0 0.0 1.0))
  (bind _returnvalue_2 (Math|Vector|MakeVector 0.0 0.0 1.0))
  (bind _returnvalue_3 (Math|Vector|MakeVector 0.0 0.0 1.0))
  (bind _terrainblenddistance (Variables|Default|GetTerrainBlendDistance))
  (bind _returnvalue_4 (* _terrainblenddistance -1.0))
  (bind _terrainblenddistance_2 (Variables|Default|GetTerrainBlendDistance))
  (bind _terrainblenddistance_3 (Variables|Default|GetTerrainBlendDistance))
  (bind _terrainblenddistance_4 (Variables|Default|GetTerrainBlendDistance))
  (bind _terrainblenddistance_5 (Variables|Default|GetTerrainBlendDistance))
  (bind _terrainblenddistance_6 (Variables|Default|GetTerrainBlendDistance))
  (Utilities|FlowControl|Sequence
    (:then_0
      ;[1 1 1] OLD need clean up
      (bind _assetindexlist (Tools|GetAssetsOfType (Variables|/Game/ArtTools/Procedural/Foundation/Proc_AssetList.Proc_AssetList_C|GetAssetList (Variables|E.%29Meshes|GetTerrainBlend_Asset_List)) "NewEnumerator6"))
      (LocalVariables|SetTerrainBlend_Asset_Index_List _assetindexlist))
    (:then_1
      _)
    (:then_2
      (bind _splinecopy (SplineOperations|MakeCleanOptimizedSplineCopy (Variables|Default|GetSpline)))
      (Utilities|Array|Add (Variables|SplineList|GetTempSplineList) _splinecopy))
    (:then_3
      (Cleanup|DeleteAllTempSplines)))
  (for _index (range (Spline|GetNumberOfSplinePoints _splinecopy))
    (if (!= (/ (- (+ (Math|Float|SelectFloat (- (- _terrainblenddistance (Math|Vector|VectorLength (- (Spline|GetLocationatSplinePoint _splinecopy (Math|Integer|SelectInt 0 (+ _index 1) (== _index (- (Spline|GetNumberOfSplinePoints _splinecopy) 1))) "Local") (Spline|GetLocationatSplinePoint _splinecopy _index "Local")))) _terrainblenddistance_2) (+ (Math|Vector|VectorLength (- (Spline|GetLocationatSplinePoint _splinecopy (Math|Integer|SelectInt 0 (+ _index 1) (== _index (- (Spline|GetNumberOfSplinePoints _splinecopy) 1))) "Local") (Spline|GetLocationatSplinePoint _splinecopy _index "Local"))) (Math|Float|SelectFloat _returnvalue_4 (Math|Float|SelectFloat 0.0 _terrainblenddistance (< (Math|Vector|VectorLength (- (Spline|GetLocationatSplinePoint _splinecopy (Math|Integer|SelectInt 1 (Math|Integer|SelectInt 0 (+ _index 2) (== _index (- (- (Spline|GetNumberOfSplinePoints _splinecopy) 1) 1))) (== _index (- (Spline|GetNumberOfSplinePoints _splinecopy) 1))) "Local") (Spline|GetLocationatSplinePoint _splinecopy (Math|Integer|SelectInt 0 (+ _index 1) (== _index (- (Spline|GetNumberOfSplinePoints _splinecopy) 1))) "Local"))) _terrainblenddistance)) (> (Math|Vector|DotProduct (Math|Vector|CrossProduct (- (Spline|GetLocationatSplinePoint _splinecopy (Math|Integer|SelectInt 0 (+ _index 1) (== _index (- (Spline|GetNumberOfSplinePoints _splinecopy) 1))) "Local") (Spline|GetLocationatSplinePoint _splinecopy _index "Local")) _returnvalue_3) (- (Spline|GetLocationatSplinePoint _splinecopy (Math|Integer|SelectInt 1 (Math|Integer|SelectInt 0 (+ _index 2) (== _index (- (- (Spline|GetNumberOfSplinePoints _splinecopy) 1) 1))) (== _index (- (Spline|GetNumberOfSplinePoints _splinecopy) 1))) "Local") (Spline|GetLocationatSplinePoint _splinecopy _index "Local"))) 0.0))) (and (< (Math|Vector|VectorLength (- (Spline|GetLocationatSplinePoint _splinecopy (Math|Integer|SelectInt 0 (+ _index 1) (== _index (- (Spline|GetNumberOfSplinePoints _splinecopy) 1))) "Local") (Spline|GetLocationatSplinePoint _splinecopy _index "Local"))) _terrainblenddistance_6) (> (Math|Vector|DotProduct (Math|Vector|CrossProduct (- (Spline|GetLocationatSplinePoint _splinecopy (Math|Integer|SelectInt 0 (+ _index 1) (== _index (- (Spline|GetNumberOfSplinePoints _splinecopy) 1))) "Local") (Spline|GetLocationatSplinePoint _splinecopy _index "Local")) _returnvalue_3) (- (Spline|GetLocationatSplinePoint _splinecopy (Math|Integer|SelectInt 1 (Math|Integer|SelectInt 0 (+ _index 2) (== _index (- (- (Spline|GetNumberOfSplinePoints _splinecopy) 1) 1))) (== _index (- (Spline|GetNumberOfSplinePoints _splinecopy) 1))) "Local") (Spline|GetLocationatSplinePoint _splinecopy _index "Local"))) 0.0))) (Math|Float|SelectFloat _terrainblenddistance_3 0.0 (and (and (< (Math|Vector|VectorLength (- (Spline|GetLocationatSplinePoint _splinecopy (Math|Integer|SelectInt 1 (Math|Integer|SelectInt 0 (+ _index 2) (== _index (- (- (Spline|GetNumberOfSplinePoints _splinecopy) 1) 1))) (== _index (- (Spline|GetNumberOfSplinePoints _splinecopy) 1))) "Local") (Spline|GetLocationatSplinePoint _splinecopy (Math|Integer|SelectInt 0 (+ _index 1) (== _index (- (Spline|GetNumberOfSplinePoints _splinecopy) 1))) "Local"))) _terrainblenddistance) (not (and (< (Math|Vector|VectorLength (- (Spline|GetLocationatSplinePoint _splinecopy (Math|Integer|SelectInt 1 (Math|Integer|SelectInt 0 (+ _index 2) (== _index (- (- (Spline|GetNumberOfSplinePoints _splinecopy) 1) 1))) (== _index (- (Spline|GetNumberOfSplinePoints _splinecopy) 1))) "Local") (Spline|GetLocationatSplinePoint _splinecopy (Math|Integer|SelectInt 0 (+ _index 1) (== _index (- (Spline|GetNumberOfSplinePoints _splinecopy) 1))) "Local"))) _terrainblenddistance) (> (Math|Vector|DotProduct (Math|Vector|CrossProduct (- (Spline|GetLocationatSplinePoint _splinecopy (Math|Integer|SelectInt 0 (+ _index 1) (== _index (- (Spline|GetNumberOfSplinePoints _splinecopy) 1))) "Local") (Spline|GetLocationatSplinePoint _splinecopy _index "Local")) _returnvalue_3) (- (Spline|GetLocationatSplinePoint _splinecopy (Math|Integer|SelectInt 1 (Math|Integer|SelectInt 0 (+ _index 2) (== _index (- (- (Spline|GetNumberOfSplinePoints _splinecopy) 1) 1))) (== _index (- (Spline|GetNumberOfSplinePoints _splinecopy) 1))) "Local") (Spline|GetLocationatSplinePoint _splinecopy _index "Local"))) 0.0)))) (< (Math|Vector|DotProduct (Math|Vector|CrossProduct (- (Spline|GetLocationatSplinePoint _splinecopy (Math|Integer|SelectInt 0 (+ _index 1) (== _index (- (Spline|GetNumberOfSplinePoints _splinecopy) 1))) "Local") (Spline|GetLocationatSplinePoint _splinecopy _index "Local")) _returnvalue) (- (Spline|GetLocationatSplinePoint _splinecopy (Math|Integer|SelectInt 1 (Math|Integer|SelectInt 0 (+ _index 2) (== _index (- (- (Spline|GetNumberOfSplinePoints _splinecopy) 1) 1))) (== _index (- (Spline|GetNumberOfSplinePoints _splinecopy) 1))) "Local") (Spline|GetLocationatSplinePoint _splinecopy (Math|Integer|SelectInt 0 (+ _index 1) (== _index (- (Spline|GetNumberOfSplinePoints _splinecopy) 1))) "Local"))) 0.0)))) (Math|Float|SelectFloat _terrainblenddistance_4 0.0 (and (> (Math|Vector|DotProduct (Math|Vector|CrossProduct (- (Spline|GetLocationatSplinePoint _splinecopy (Math|Integer|SelectInt 1 (Math|Integer|SelectInt 0 (+ _index 2) (== _index (- (- (Spline|GetNumberOfSplinePoints _splinecopy) 1) 1))) (== _index (- (Spline|GetNumberOfSplinePoints _splinecopy) 1))) "Local") (Spline|GetLocationatSplinePoint _splinecopy (Math|Integer|SelectInt 0 (+ _index 1) (== _index (- (Spline|GetNumberOfSplinePoints _splinecopy) 1))) "Local")) _returnvalue_2) (- (Spline|GetLocationatSplinePoint _splinecopy (Math|Integer|SelectInt 2 (Math|Integer|SelectInt 1 (Math|Integer|SelectInt 0 (+ _index 3) (== _index (- (- (Spline|GetNumberOfSplinePoints _splinecopy) 1) 2))) (== _index (- (- (Spline|GetNumberOfSplinePoints _splinecopy) 1) 1))) (== _index (- (Spline|GetNumberOfSplinePoints _splinecopy) 1))) "Local") (Spline|GetLocationatSplinePoint _splinecopy (Math|Integer|SelectInt 1 (Math|Integer|SelectInt 0 (+ _index 2) (== _index (- (- (Spline|GetNumberOfSplinePoints _splinecopy) 1) 1))) (== _index (- (Spline|GetNumberOfSplinePoints _splinecopy) 1))) "Local"))) 0.0) (< (/ (Math|Vector|VectorLength (- (Spline|GetLocationatSplinePoint _splinecopy (Math|Integer|SelectInt 1 (Math|Integer|SelectInt 0 (+ _index 2) (== _index (- (- (Spline|GetNumberOfSplinePoints _splinecopy) 1) 1))) (== _index (- (Spline|GetNumberOfSplinePoints _splinecopy) 1))) "Local") (Spline|GetLocationatSplinePoint _splinecopy (Math|Integer|SelectInt 0 (+ _index 1) (== _index (- (Spline|GetNumberOfSplinePoints _splinecopy) 1))) "Local"))) _terrainblenddistance_5) 1.0)))) _sizex_11_dfccfeb9407ac738b8ec5aaf3a48e13c) 0.0)
      (Utilities|Array|Add (Variables|X.%29DoNOTEdit|GetComponent%20Transform%20List) (Utilities|Struct|MakeProcComponenttoSpawn (Math|Transform|MakeTransform (+ (+ (Spline|GetLocationatSplinePoint _splinecopy _index "Local") (Math|Vector|SelectVector (* (Math|Vector|Normalize (Math|Vector|NegateVector (- (Spline|GetLocationatSplinePoint _splinecopy (Math|Integer|SelectInt 0 (+ _index 1) (== _index (- (Spline|GetNumberOfSplinePoints _splinecopy) 1))) "Local") (Spline|GetLocationatSplinePoint _splinecopy _index "Local")))) (- _terrainblenddistance (Math|Vector|VectorLength (- (Spline|GetLocationatSplinePoint _splinecopy (Math|Integer|SelectInt 0 (+ _index 1) (== _index (- (Spline|GetNumberOfSplinePoints _splinecopy) 1))) "Local") (Spline|GetLocationatSplinePoint _splinecopy _index "Local"))))) "0, 0, 0" (and (< (Math|Vector|VectorLength (- (Spline|GetLocationatSplinePoint _splinecopy (Math|Integer|SelectInt 0 (+ _index 1) (== _index (- (Spline|GetNumberOfSplinePoints _splinecopy) 1))) "Local") (Spline|GetLocationatSplinePoint _splinecopy _index "Local"))) _terrainblenddistance_6) (> (Math|Vector|DotProduct (Math|Vector|CrossProduct (- (Spline|GetLocationatSplinePoint _splinecopy (Math|Integer|SelectInt 0 (+ _index 1) (== _index (- (Spline|GetNumberOfSplinePoints _splinecopy) 1))) "Local") (Spline|GetLocationatSplinePoint _splinecopy _index "Local")) _returnvalue_3) (- (Spline|GetLocationatSplinePoint _splinecopy (Math|Integer|SelectInt 1 (Math|Integer|SelectInt 0 (+ _index 2) (== _index (- (- (Spline|GetNumberOfSplinePoints _splinecopy) 1) 1))) (== _index (- (Spline|GetNumberOfSplinePoints _splinecopy) 1))) "Local") (Spline|GetLocationatSplinePoint _splinecopy _index "Local"))) 0.0)))) (Math|Vector|SelectVector (Math|Vector|SelectVector "0, 0, 0" (* (Math|Vector|Normalize (Math|Vector|NegateVector (- (Spline|GetLocationatSplinePoint _splinecopy (Math|Integer|SelectInt 0 (+ _index 1) (== _index (- (Spline|GetNumberOfSplinePoints _splinecopy) 1))) "Local") (Spline|GetLocationatSplinePoint _splinecopy _index "Local")))) (Variables|Default|GetTerrainBlendDistance)) (> (Math|Vector|DotProduct (Math|Vector|CrossProduct (- (Spline|GetLocationatSplinePoint _splinecopy (Math|Integer|SelectInt 1 (Math|Integer|SelectInt 0 (+ _index 2) (== _index (- (- (Spline|GetNumberOfSplinePoints _splinecopy) 1) 1))) (== _index (- (Spline|GetNumberOfSplinePoints _splinecopy) 1))) "Local") (Spline|GetLocationatSplinePoint _splinecopy _index "Local")) (Math|Vector|MakeVector 0.0 0.0 1.0)) (- (Spline|GetLocationatSplinePoint _splinecopy (Math|Integer|SelectInt 0 (+ _index 1) (== _index (- (Spline|GetNumberOfSplinePoints _splinecopy) 1))) "Local") (Spline|GetLocationatSplinePoint _splinecopy _index "Local"))) 0.0)) "0, 0, 0" (and (< (Math|Vector|VectorLength (- (Spline|GetLocationatSplinePoint _splinecopy (Math|Integer|SelectInt 1 (Math|Integer|SelectInt 0 (+ _index 2) (== _index (- (- (Spline|GetNumberOfSplinePoints _splinecopy) 1) 1))) (== _index (- (Spline|GetNumberOfSplinePoints _splinecopy) 1))) "Local") (Spline|GetLocationatSplinePoint _splinecopy (Math|Integer|SelectInt 0 (+ _index 1) (== _index (- (Spline|GetNumberOfSplinePoints _splinecopy) 1))) "Local"))) _terrainblenddistance) (not (and (< (Math|Vector|VectorLength (- (Spline|GetLocationatSplinePoint _splinecopy (Math|Integer|SelectInt 1 (Math|Integer|SelectInt 0 (+ _index 2) (== _index (- (- (Spline|GetNumberOfSplinePoints _splinecopy) 1) 1))) (== _index (- (Spline|GetNumberOfSplinePoints _splinecopy) 1))) "Local") (Spline|GetLocationatSplinePoint _splinecopy (Math|Integer|SelectInt 0 (+ _index 1) (== _index (- (Spline|GetNumberOfSplinePoints _splinecopy) 1))) "Local"))) _terrainblenddistance) (> (Math|Vector|DotProduct (Math|Vector|CrossProduct (- (Spline|GetLocationatSplinePoint _splinecopy (Math|Integer|SelectInt 0 (+ _index 1) (== _index (- (Spline|GetNumberOfSplinePoints _splinecopy) 1))) "Local") (Spline|GetLocationatSplinePoint _splinecopy _index "Local")) _returnvalue_3) (- (Spline|GetLocationatSplinePoint _splinecopy (Math|Integer|SelectInt 1 (Math|Integer|SelectInt 0 (+ _index 2) (== _index (- (- (Spline|GetNumberOfSplinePoints _splinecopy) 1) 1))) (== _index (- (Spline|GetNumberOfSplinePoints _splinecopy) 1))) "Local") (Spline|GetLocationatSplinePoint _splinecopy _index "Local"))) 0.0)))))) (Math|Rotator|MakeRotfromX (- (Spline|GetLocationatSplinePoint _splinecopy (Math|Integer|SelectInt 0 (+ _index 1) (== _index (- (Spline|GetNumberOfSplinePoints _splinecopy) 1))) "Local") (Spline|GetLocationatSplinePoint _splinecopy _index "Local"))) (Math|Vector|MakeVector (/ (- (+ (Math|Float|SelectFloat (- (- _terrainblenddistance (Math|Vector|VectorLength (- (Spline|GetLocationatSplinePoint _splinecopy (Math|Integer|SelectInt 0 (+ _index 1) (== _index (- (Spline|GetNumberOfSplinePoints _splinecopy) 1))) "Local") (Spline|GetLocationatSplinePoint _splinecopy _index "Local")))) _terrainblenddistance_2) (+ (Math|Vector|VectorLength (- (Spline|GetLocationatSplinePoint _splinecopy (Math|Integer|SelectInt 0 (+ _index 1) (== _index (- (Spline|GetNumberOfSplinePoints _splinecopy) 1))) "Local") (Spline|GetLocationatSplinePoint _splinecopy _index "Local"))) (Math|Float|SelectFloat _returnvalue_4 (Math|Float|SelectFloat 0.0 _terrainblenddistance (< (Math|Vector|VectorLength (- (Spline|GetLocationatSplinePoint _splinecopy (Math|Integer|SelectInt 1 (Math|Integer|SelectInt 0 (+ _index 2) (== _index (- (- (Spline|GetNumberOfSplinePoints _splinecopy) 1) 1))) (== _index (- (Spline|GetNumberOfSplinePoints _splinecopy) 1))) "Local") (Spline|GetLocationatSplinePoint _splinecopy (Math|Integer|SelectInt 0 (+ _index 1) (== _index (- (Spline|GetNumberOfSplinePoints _splinecopy) 1))) "Local"))) _terrainblenddistance)) (> (Math|Vector|DotProduct (Math|Vector|CrossProduct (- (Spline|GetLocationatSplinePoint _splinecopy (Math|Integer|SelectInt 0 (+ _index 1) (== _index (- (Spline|GetNumberOfSplinePoints _splinecopy) 1))) "Local") (Spline|GetLocationatSplinePoint _splinecopy _index "Local")) _returnvalue_3) (- (Spline|GetLocationatSplinePoint _splinecopy (Math|Integer|SelectInt 1 (Math|Integer|SelectInt 0 (+ _index 2) (== _index (- (- (Spline|GetNumberOfSplinePoints _splinecopy) 1) 1))) (== _index (- (Spline|GetNumberOfSplinePoints _splinecopy) 1))) "Local") (Spline|GetLocationatSplinePoint _splinecopy _index "Local"))) 0.0))) (and (< (Math|Vector|VectorLength (- (Spline|GetLocationatSplinePoint _splinecopy (Math|Integer|SelectInt 0 (+ _index 1) (== _index (- (Spline|GetNumberOfSplinePoints _splinecopy) 1))) "Local") (Spline|GetLocationatSplinePoint _splinecopy _index "Local"))) _terrainblenddistance_6) (> (Math|Vector|DotProduct (Math|Vector|CrossProduct (- (Spline|GetLocationatSplinePoint _splinecopy (Math|Integer|SelectInt 0 (+ _index 1) (== _index (- (Spline|GetNumberOfSplinePoints _splinecopy) 1))) "Local") (Spline|GetLocationatSplinePoint _splinecopy _index "Local")) _returnvalue_3) (- (Spline|GetLocationatSplinePoint _splinecopy (Math|Integer|SelectInt 1 (Math|Integer|SelectInt 0 (+ _index 2) (== _index (- (- (Spline|GetNumberOfSplinePoints _splinecopy) 1) 1))) (== _index (- (Spline|GetNumberOfSplinePoints _splinecopy) 1))) "Local") (Spline|GetLocationatSplinePoint _splinecopy _index "Local"))) 0.0))) (Math|Float|SelectFloat _terrainblenddistance_3 0.0 (and (and (< (Math|Vector|VectorLength (- (Spline|GetLocationatSplinePoint _splinecopy (Math|Integer|SelectInt 1 (Math|Integer|SelectInt 0 (+ _index 2) (== _index (- (- (Spline|GetNumberOfSplinePoints _splinecopy) 1) 1))) (== _index (- (Spline|GetNumberOfSplinePoints _splinecopy) 1))) "Local") (Spline|GetLocationatSplinePoint _splinecopy (Math|Integer|SelectInt 0 (+ _index 1) (== _index (- (Spline|GetNumberOfSplinePoints _splinecopy) 1))) "Local"))) _terrainblenddistance) (not (and (< (Math|Vector|VectorLength (- (Spline|GetLocationatSplinePoint _splinecopy (Math|Integer|SelectInt 1 (Math|Integer|SelectInt 0 (+ _index 2) (== _index (- (- (Spline|GetNumberOfSplinePoints _splinecopy) 1) 1))) (== _index (- (Spline|GetNumberOfSplinePoints _splinecopy) 1))) "Local") (Spline|GetLocationatSplinePoint _splinecopy (Math|Integer|SelectInt 0 (+ _index 1) (== _index (- (Spline|GetNumberOfSplinePoints _splinecopy) 1))) "Local"))) _terrainblenddistance) (> (Math|Vector|DotProduct (Math|Vector|CrossProduct (- (Spline|GetLocationatSplinePoint _splinecopy (Math|Integer|SelectInt 0 (+ _index 1) (== _index (- (Spline|GetNumberOfSplinePoints _splinecopy) 1))) "Local") (Spline|GetLocationatSplinePoint _splinecopy _index "Local")) _returnvalue_3) (- (Spline|GetLocationatSplinePoint _splinecopy (Math|Integer|SelectInt 1 (Math|Integer|SelectInt 0 (+ _index 2) (== _index (- (- (Spline|GetNumberOfSplinePoints _splinecopy) 1) 1))) (== _index (- (Spline|GetNumberOfSplinePoints _splinecopy) 1))) "Local") (Spline|GetLocationatSplinePoint _splinecopy _index "Local"))) 0.0)))) (< (Math|Vector|DotProduct (Math|Vector|CrossProduct (- (Spline|GetLocationatSplinePoint _splinecopy (Math|Integer|SelectInt 0 (+ _index 1) (== _index (- (Spline|GetNumberOfSplinePoints _splinecopy) 1))) "Local") (Spline|GetLocationatSplinePoint _splinecopy _index "Local")) _returnvalue) (- (Spline|GetLocationatSplinePoint _splinecopy (Math|Integer|SelectInt 1 (Math|Integer|SelectInt 0 (+ _index 2) (== _index (- (- (Spline|GetNumberOfSplinePoints _splinecopy) 1) 1))) (== _index (- (Spline|GetNumberOfSplinePoints _splinecopy) 1))) "Local") (Spline|GetLocationatSplinePoint _splinecopy (Math|Integer|SelectInt 0 (+ _index 1) (== _index (- (Spline|GetNumberOfSplinePoints _splinecopy) 1))) "Local"))) 0.0)))) (Math|Float|SelectFloat _terrainblenddistance_4 0.0 (and (> (Math|Vector|DotProduct (Math|Vector|CrossProduct (- (Spline|GetLocationatSplinePoint _splinecopy (Math|Integer|SelectInt 1 (Math|Integer|SelectInt 0 (+ _index 2) (== _index (- (- (Spline|GetNumberOfSplinePoints _splinecopy) 1) 1))) (== _index (- (Spline|GetNumberOfSplinePoints _splinecopy) 1))) "Local") (Spline|GetLocationatSplinePoint _splinecopy (Math|Integer|SelectInt 0 (+ _index 1) (== _index (- (Spline|GetNumberOfSplinePoints _splinecopy) 1))) "Local")) _returnvalue_2) (- (Spline|GetLocationatSplinePoint _splinecopy (Math|Integer|SelectInt 2 (Math|Integer|SelectInt 1 (Math|Integer|SelectInt 0 (+ _index 3) (== _index (- (- (Spline|GetNumberOfSplinePoints _splinecopy) 1) 2))) (== _index (- (- (Spline|GetNumberOfSplinePoints _splinecopy) 1) 1))) (== _index (- (Spline|GetNumberOfSplinePoints _splinecopy) 1))) "Local") (Spline|GetLocationatSplinePoint _splinecopy (Math|Integer|SelectInt 1 (Math|Integer|SelectInt 0 (+ _index 2) (== _index (- (- (Spline|GetNumberOfSplinePoints _splinecopy) 1) 1))) (== _index (- (Spline|GetNumberOfSplinePoints _splinecopy) 1))) "Local"))) 0.0) (< (/ (Math|Vector|VectorLength (- (Spline|GetLocationatSplinePoint _splinecopy (Math|Integer|SelectInt 1 (Math|Integer|SelectInt 0 (+ _index 2) (== _index (- (- (Spline|GetNumberOfSplinePoints _splinecopy) 1) 1))) (== _index (- (Spline|GetNumberOfSplinePoints _splinecopy) 1))) "Local") (Spline|GetLocationatSplinePoint _splinecopy (Math|Integer|SelectInt 0 (+ _index 1) (== _index (- (Spline|GetNumberOfSplinePoints _splinecopy) 1))) "Local"))) _terrainblenddistance_5) 1.0)))) _sizex_11_dfccfeb9407ac738b8ec5aaf3a48e13c) (/ (Variables|Default|GetTerrainBlendDistance) _sizex_11_dfccfeb9407ac738b8ec5aaf3a48e13c) 1.0)) (Variables|E.%29Meshes|GetTerrainBlend_Asset_List) (Utilities|Array|Get(acopy) (LocalVariables|GetTerrainBlend_Asset_Index_List)))))))

(fn MakeMeshFillArea
  (params (AssetList :type Proc%20Asset%20List%20Object%20Reference))
  (bind _assetlist (Variables|/Game/ArtTools/Procedural/Foundation/Proc_AssetList.Proc_AssetList_C|GetAssetList AssetList))
  (bind _minnumberofcompoents (LocalVariables|GetMinNumberofCompoents))
  (bind _bordercollisionlist (Variables|MeshLists|GetBorderCollisionList))
  (Utilities|FlowControl|Sequence
    (:then_0
      (CreateCollisionComponents|MakeBorderCollisionFromSpline (Variables|Default|GetSpline))
      (bind _assetindexlist (Tools|GetAssetsOfType _assetlist "NewEnumerator1"))
      (Utilities|Array|Clear (LocalVariables|GetMinNumberofCompoents))
      (for _index (range (Spline|GetNumberOfSplinePoints (Variables|Default|GetSpline)))
        (bind _componentsspawned (CreateComponents|MakeMeshFillAreaComponents _assetlist _assetindexlist (.SizeY_14_2F84FFCF464B2148EFD89E96265F5278 (Utilities|Struct|BreakProcAssetInfo (Utilities|Array|Get(acopy) _assetlist (Utilities|Array|Get(acopy) _assetindexlist)))) _index false AssetList))
        (Utilities|Array|Add _minnumberofcompoents _componentsspawned))
      (bind _componentsspawned_2 (CreateComponents|MakeMeshFillAreaComponents _assetlist _assetindexlist (.SizeY_14_2F84FFCF464B2148EFD89E96265F5278 (Utilities|Struct|BreakProcAssetInfo (Utilities|Array|Get(acopy) _assetlist (Utilities|Array|Get(acopy) _assetindexlist)))) (Math|Integer|MinOfIntArray _minnumberofcompoents) true AssetList)))
    (:then_1
      (for _array_element _bordercollisionlist
        (Components|DestroyComponent _array_element))
      (Utilities|Array|Clear _bordercollisionlist))))

(fn MakeMeshFillAreaComponents
  (params (AssetList1 :type Array%20of%20Proc%20Asset%20Info%20Structures) (AssetIndexList1 :type Array%20of%20Integers) (AssetWidth1 :type Float%20%28double-precision%29) (IndexOffset :type Integer) (SpawnComponents :type Boolean) (Proc%20Asset%20List :type Proc%20Asset%20List%20Object%20Reference))
  (bind _assetwidth (LocalVariables|GetAssetWidth))
  (bind _returnvalue (/ _assetwidth 2.0))
  (bind _spline (Variables|Default|GetSpline))
  (bind _returnvalue_2 (Spline|GetNumberOfSplinePoints _spline))
  (bind _returnvalue_3 (- _returnvalue_2 1))
  (bind _distance (LocalVariables|GetDistance))
  (bind _returnvalue_4 (+ _distance 1.0))
  (bind _assetlist (LocalVariables|GetAssetList))
  (bind _currentdistance (LocalVariables|GetCurrentDistance))
  (bind _fillcollisionlist (Variables|MeshLists|GetFillCollisionList))
  (bind _assetwidth_2 (LocalVariables|GetAssetWidth))
  (Utilities|FlowControl|Sequence
    (:then_0
      (for _array_element (Variables|C.%29ConnectingComplexPOIs|GetConnectedActors)
        (CreateCollisionComponents|MakeBorderCollisionFromSpline (Variables|Default|GetSpline _array_element) true "World")))
    (:then_1
      (LocalVariables|SetComponents%20Spawned 0)
      (LocalVariables|SetAssetList AssetList1)
      (Utilities|FlowControl|Sequence
        (:then_0
          (LocalVariables|SetAssetIndexList AssetIndexList1)
          (LocalVariables|SetAssetWidth AssetWidth1)
          (Utilities|FlowControl|Sequence
            (:then_0
              (for _index (range _returnvalue_2)
                (bind (_next__1 _next _current _previous _previous__1 _world_next__1 _world_next _world_current) (Tools|GetCurrentNextPreviousSplinePoint _spline (Math|Integer|SelectInt (- (+ _index IndexOffset) _returnvalue_2) (+ _index IndexOffset) (> (+ _index IndexOffset) _returnvalue_3))))
                (if (> (Math|Float|Round (/ (Math|Vector|Distance(Vector) _next _current) _assetwidth_2)) 0)
                  (for _index_2 (range (Math|Float|Round (/ (Math|Vector|Distance(Vector) _next _current) _assetwidth_2)))
                    (bind _outhits (Collision|MultiLineTraceForObjects :Start (+ (+ _world_current (* (Math|Vector|GetUnitDirection(Vector) _world_current _world_next) (+ (* _index_2 _assetwidth) _returnvalue))) (* (Math|Vector|Normalize (Math|Vector|CrossProduct (Math|Vector|GetUnitDirection(Vector) _world_current _world_next) "0, 0,-1.000000")) -5.0)) :End (+ (+ _world_current (* (Math|Vector|GetUnitDirection(Vector) _world_current _world_next) (+ (* _index_2 _assetwidth) _returnvalue))) (* (Math|Vector|Normalize (Math|Vector|CrossProduct (Math|Vector|GetUnitDirection(Vector) _world_current _world_next) "0, 0,-1.000000")) 2e+04)) :ObjectTypes (Utilities|Array|MakeArray "ObjectTypeQuery3") :bTraceComplex true :DrawDebugType "None" :bIgnoreSelf false :TraceColor "(R=1.000000,G=0.000000,B=0.000000,A=1.000000)" :TraceHitColor "(R=0.000000,G=1.000000,B=0.000000,A=1.000000)" :DrawTime 5.0))
                    (bind _hitresoultoutput (Tools|ClosestHitResult _outhits (+ _world_current (* (Math|Vector|GetUnitDirection(Vector) _world_current _world_next) (+ (* _index_2 _assetwidth) _returnvalue)))))
                    (LocalVariables|SetDistance (.Distance (Collision|BreakHitResult _hitresoultoutput)))
                    (LocalVariables|SetCurrentDistance 0.0)
                    (for _array_element_2 (LocalVariables|GetAssetIndexList)
                      (if (> (Math|Float|Floor (/ _returnvalue_4 (.SizeX_11_DFCCFEB9407AC738B8EC5AAF3A48E13C (Utilities|Struct|BreakProcAssetInfo (Utilities|Array|Get(acopy) _assetlist _array_element_2))))) 0)
                        (for _index_3 (range (Math|Float|Floor (/ _returnvalue_4 (.SizeX_11_DFCCFEB9407AC738B8EC5AAF3A48E13C (Utilities|Struct|BreakProcAssetInfo (Utilities|Array|Get(acopy) _assetlist _array_element_2))))))
                          (Utilities|FlowControl|Sequence
                            (:then_0
                              (Math|Integer|IncrementInt (LocalVariables|GetComponents%20Spawned)
                                (:%20%20
                                  (if SpawnComponents
                                    (Utilities|Array|Add (Variables|X.%29DoNOTEdit|GetComponent%20Transform%20List) (Utilities|Struct|MakeProcComponenttoSpawn (Math|Transform|MakeTransform (Math|Vector|MakeVector (.x (+ (Math|Vector|vector+vector :A (* (Math|Vector|Normalize (Math|Vector|CrossProduct (Math|Vector|GetUnitDirection(Vector) _current _next) "0, 0,-1.000000")) (+ (* _index_3 (.SizeX_11_DFCCFEB9407AC738B8EC5AAF3A48E13C (Utilities|Struct|BreakProcAssetInfo (Utilities|Array|Get(acopy) _assetlist _array_element_2)))) _currentdistance)) :B (+ _current (* (Math|Vector|GetUnitDirection(Vector) _current _next) (+ (* _index_2 _assetwidth) _assetwidth)))) (* (Math|Vector|GetUnitDirection(Vector) _current _next) (* (LocalVariables|GetAssetWidth) -1.0)))) (.y (+ (Math|Vector|vector+vector :A (* (Math|Vector|Normalize (Math|Vector|CrossProduct (Math|Vector|GetUnitDirection(Vector) _current _next) "0, 0,-1.000000")) (+ (* _index_3 (.SizeX_11_DFCCFEB9407AC738B8EC5AAF3A48E13C (Utilities|Struct|BreakProcAssetInfo (Utilities|Array|Get(acopy) _assetlist _array_element_2)))) _currentdistance)) :B (+ _current (* (Math|Vector|GetUnitDirection(Vector) _current _next) (+ (* _index_2 _assetwidth) _assetwidth)))) (* (Math|Vector|GetUnitDirection(Vector) _current _next) (* (LocalVariables|GetAssetWidth) -1.0))))) (Math|Rotator|MakeRotfromXZ (Math|Vector|Normalize (Math|Vector|CrossProduct (Math|Vector|GetUnitDirection(Vector) _current _next) "0, 0,-1.000000")) (Math|Vector|MakeVector 0.0 0.0 1.0))) Proc%20Asset%20List _array_element_2))))))
                            (:then_1
                              (bind _returnvalue_5 (AddComponent|NODE_AddStaticMeshComponent-32 :RelativeTransform (Math|Transform|MakeTransform (Math|Vector|vector+vector :A (* (Math|Vector|Normalize (Math|Vector|CrossProduct (Math|Vector|GetUnitDirection(Vector) _current _next) "0, 0,-1.000000")) (+ (* _index_3 (.SizeX_11_DFCCFEB9407AC738B8EC5AAF3A48E13C (Utilities|Struct|BreakProcAssetInfo (Utilities|Array|Get(acopy) _assetlist _array_element_2)))) _currentdistance)) :B (+ _current (* (Math|Vector|GetUnitDirection(Vector) _current _next) (+ (* _index_2 _assetwidth) _assetwidth)))) (Math|Rotator|MakeRotfromXY (Math|Vector|Normalize (Math|Vector|CrossProduct (Math|Vector|GetUnitDirection(Vector) _current _next) "0, 0,-1.000000")) (Math|Vector|GetUnitDirection(Vector) _current _next)) (Math|Vector|MakeVector (/ (.SizeX_11_DFCCFEB9407AC738B8EC5AAF3A48E13C (Utilities|Struct|BreakProcAssetInfo (Utilities|Array|Get(acopy) _assetlist _array_element_2))) 512.0) (/ (.SizeY_14_2F84FFCF464B2148EFD89E96265F5278 (Utilities|Struct|BreakProcAssetInfo (Utilities|Array|Get(acopy) _assetlist _array_element_2))) 512.0) 1.0))))
                              (Utilities|Array|Add (Variables|MeshLists|GetFillCollisionList) _returnvalue_5))))
                        (LocalVariables|SetCurrentDistance (+ (* (Math|Float|Floor (/ _returnvalue_4 (.SizeX_11_DFCCFEB9407AC738B8EC5AAF3A48E13C (Utilities|Struct|BreakProcAssetInfo (Utilities|Array|Get(acopy) _assetlist _array_element_2))))) (.SizeX_11_DFCCFEB9407AC738B8EC5AAF3A48E13C (Utilities|Struct|BreakProcAssetInfo (Utilities|Array|Get(acopy) _assetlist _array_element_2)))) (LocalVariables|GetCurrentDistance)))
                        (LocalVariables|SetDistance (- (LocalVariables|GetDistance) (* (Math|Float|Floor (/ _returnvalue_4 (.SizeX_11_DFCCFEB9407AC738B8EC5AAF3A48E13C (Utilities|Struct|BreakProcAssetInfo (Utilities|Array|Get(acopy) _assetlist _array_element_2))))) (.SizeX_11_DFCCFEB9407AC738B8EC5AAF3A48E13C (Utilities|Struct|BreakProcAssetInfo (Utilities|Array|Get(acopy) _assetlist _array_element_2))))))))))))))
        (:then_1
          (if false
            (Development|PrintString (Utilities|String|ToString(Integer) (LocalVariables|GetComponents%20Spawned)))
            (else
              (for _array_element_3 _fillcollisionlist
                (Components|DestroyComponent _array_element_3))
              (Utilities|Array|Clear _fillcollisionlist))))
        (:then_2
          (return (LocalVariables|GetComponents%20Spawned)))))))

(fn MakeOuterCornerCollision
  (params (Spline :type Spline%20Component%20Object%20Reference) (CoordinateSpace :type ESplineCoordinateSpace%20Enum :default "Local") (Reverse :type Boolean))
  (for _index (range (Spline|GetNumberOfSplinePoints Spline))
    (bind (_next__1 _next _current _previous _previous__1 _world_next__1 _world_next _world_current _world_previous) (Tools|GetCurrentNextPreviousSplinePoint Spline _index))
    (if (select Reverse (not (Utilities|IsOuterCorner _next _current _previous)) (Utilities|IsOuterCorner _next _current _previous))
      (bind _returnvalue (AddComponent|NODE_AddStaticMeshComponent-51 :RelativeTransform (Utilities|Select :Index CoordinateSpace :Local (Math|Transform|MakeTransform _current (Math|Rotator|MakeRotationfromAxes (Math|Vector|NegateVector (Math|Vector|GetUnitDirection(Vector) _current _next)) (Math|Vector|GetUnitDirection(Vector) _current _previous) "0, 0,1.000000")) :World (Math|Transform|MakeRelativeTransform (Math|Transform|MakeTransform _world_current (Math|Rotator|MakeRotationfromAxes (Math|Vector|NegateVector (Math|Vector|GetUnitDirection(Vector) _world_current _world_next)) (Math|Vector|GetUnitDirection(Vector) _world_current _world_previous) "0, 0,1.000000")) (Transformation|GetActorTransform)))))
      (Variables|/Script/Engine.ActorComponent|SetComponentTags (Utilities|Array|MakeArray "OuterCorner") _returnvalue)
      (Utilities|Array|Add (Variables|MeshLists|GetCornerCollisionList) _returnvalue))))

(fn MakeOverrideActorCollision
  (for _array_element (Variables|D.%29Overrides|GetOverride%20Actor%20List)
    (bind _returnvalue (AddComponent|NODE_AddStaticMeshComponent-48 :RelativeTransform (Math|Transform|MakeTransform (Math|Vector|MakeVector (.x (.location (Math|Transform|MakeRelativeTransform (Transformation|GetActorTransform _array_element) (Transformation|GetActorTransform)))) (.y (.location (Math|Transform|MakeRelativeTransform (Transformation|GetActorTransform _array_element) (Transformation|GetActorTransform)))) 128.0) (.rotation (Math|Transform|MakeRelativeTransform (Transformation|GetActorTransform _array_element) (Transformation|GetActorTransform))) (Math|Vector|MakeVector (.x (.scale (Math|Transform|MakeRelativeTransform (Transformation|GetActorTransform _array_element) (Transformation|GetActorTransform)))) (.y (.scale (Math|Transform|MakeRelativeTransform (Transformation|GetActorTransform _array_element) (Transformation|GetActorTransform)))) 0.2))))
    (Variables|/Script/Engine.ActorComponent|SetComponentTags (Utilities|Array|MakeArray (Utilities|Name|EnumtoName (Variables|/Game/ArtTools/Procedural/Foundation/Proc_POI_Override_Actor.Proc_POI_Override_Actor_C|GetMode _array_element))) _returnvalue)
    (Utilities|Array|Add (Variables|MeshLists|GetBorderCollisionList) _returnvalue)))

(fn MakeSidewalkComponents
  (params (Spline :type Spline%20Component%20Object%20Reference) (OuterCornerAssetList1 :type Array%20of%20Integers) (S-BendAssetIndexList1 :type Array%20of%20Integers) (Corner_Inner_AssetList1 :type Array%20of%20Integers) (EdgeBorder_AssetList1 :type Array%20of%20Integers) (CheckForInsersectingShapes :type Boolean) (Reverse :type Boolean))
  (bind _s_bendassetindexlist1 (Variables|GetS-BendAssetIndexList1))
  (bind _output (Utilities|Array|Get(acopy) _s_bendassetindexlist1))
  (bind _sidewalk_asset_list (Variables|E.%29Meshes|GetSidewalk_Asset_List))
  (bind _assetlist (Variables|/Game/ArtTools/Procedural/Foundation/Proc_AssetList.Proc_AssetList_C|GetAssetList _sidewalk_asset_list))
  (bind _output_2 (Utilities|Array|Get(acopy) _assetlist _output))
  (bind (_staticmesh_2_d8df7d7b4556a6c1fde39bb633b7e288 _type_6_b4cd9ff0498f54172801239a22e7a5ca _prepivot_29_094082c34c836f4f1ab3deb660eef580 _forwarddirection_32_09c0a6674706473836365795dce83b27 _sizex_11_dfccfeb9407ac738b8ec5aaf3a48e13c) (Utilities|Struct|BreakProcAssetInfo _output_2))
  (bind _edgeborder_assetlist1 (Variables|GetEdgeBorder_AssetList1))
  (bind _returnvalue (Utilities|Array|LastIndex _edgeborder_assetlist1))
  (bind _output_3 (Utilities|Array|Get(acopy) _edgeborder_assetlist1 _returnvalue))
  (bind _sidewalk_asset_list_2 (Variables|E.%29Meshes|GetSidewalk_Asset_List))
  (bind _assetlist_2 (Variables|/Game/ArtTools/Procedural/Foundation/Proc_AssetList.Proc_AssetList_C|GetAssetList _sidewalk_asset_list_2))
  (bind _output_4 (Utilities|Array|Get(acopy) _assetlist_2 _output_3))
  (bind (_staticmesh_2_d8df7d7b4556a6c1fde39bb633b7e288_2 _type_6_b4cd9ff0498f54172801239a22e7a5ca_2 _prepivot_29_094082c34c836f4f1ab3deb660eef580_2 _forwarddirection_32_09c0a6674706473836365795dce83b27_2 _sizex_11_dfccfeb9407ac738b8ec5aaf3a48e13c_2) (Utilities|Struct|BreakProcAssetInfo _output_4))
  (bind _hitleft (LocalVariables|GetHitLeft))
  (bind _returnvalue_2 (not _hitleft))
  (bind _hitdiagonal (LocalVariables|GetHitDiagonal))
  (bind _returnvalue_3 (not _hitdiagonal))
  (bind _hitright (LocalVariables|GetHitRight))
  (bind _returnvalue_4 (not _hitright))
  (bind _distance_to_next_original (LocalVariables|GetDistance%20to%20Next%20Original))
  (bind _returnvalue_5 (<= _distance_to_next_original 512.0))
  (bind _distance_to_previous_original (LocalVariables|GetDistance%20to%20Previous%20Original))
  (bind _returnvalue_6 (+ _distance_to_previous_original 1.0))
  (bind _hitright_2 (LocalVariables|GetHitRight))
  (bind _hitdiagonal_2 (LocalVariables|GetHitDiagonal))
  (bind _returnvalue_7 (Math|Boolean|ANDBoolean _hitdiagonal_2 _hitright_2 _returnvalue_2))
  (bind _hitleft_2 (LocalVariables|GetHitLeft))
  (bind _returnvalue_8 (Math|Boolean|ANDBoolean _returnvalue_3 _returnvalue_4 _hitleft_2))
  (bind _distance (LocalVariables|GetDistance))
  (bind _returnvalue_9 (+ _distance 1.0))
  (bind _array (Utilities|Array|MakeArray "ObjectTypeQuery3"))
  (bind _corner_inner_assetlist1 (Variables|GetCorner_Inner_AssetList1))
  (bind _sidewalk_asset_list_3 (Variables|E.%29Meshes|GetSidewalk_Asset_List))
  (bind _assetlist_3 (Variables|/Game/ArtTools/Procedural/Foundation/Proc_AssetList.Proc_AssetList_C|GetAssetList _sidewalk_asset_list_3))
  (bind _currentsidewalklength (LocalVariables|GetCurrentSidewalkLength))
  (bind _sidewalk_asset_list_4 (Variables|E.%29Meshes|GetSidewalk_Asset_List))
  (bind _assetlist_4 (Variables|/Game/ArtTools/Procedural/Foundation/Proc_AssetList.Proc_AssetList_C|GetAssetList _sidewalk_asset_list_4))
  (bind _sidewalk_asset_list_5 (Variables|E.%29Meshes|GetSidewalk_Asset_List))
  (bind _assetlist_5 (Variables|/Game/ArtTools/Procedural/Foundation/Proc_AssetList.Proc_AssetList_C|GetAssetList _sidewalk_asset_list_5))
  (Utilities|FlowControl|Sequence
    (:then_1
      (if CheckForInsersectingShapes
        (Utilities|FlowControl|Sequence
          (:then_0
            (for _array_element (Variables|C.%29ConnectingComplexPOIs|GetConnectedActors)
              (bind _splinecopy (SplineOperations|DuplicateSpline (Variables|Default|GetSpline _array_element) true "World" true))
              (LocalVariables|SetCurrentSpline _splinecopy)
              (for _index (range (Spline|GetNumberOfSplinePoints (LocalVariables|GetCurrentSpline)))
                (bind (_next__1 _next _current _previous _previous__1 _world_next__1 _world_next _world_current _world_previous) (Tools|GetCurrentNextPreviousSplinePoint (LocalVariables|GetCurrentSpline) _index false (Variables|GetReverse)))
                (if (Utilities|IsOuterCorner _next _current _previous)
                  (bind (_outhits _returnvalue_10) (Collision|MultiLineTraceForObjects :Start (+ (+ (+ (* (Math|Vector|GetUnitDirection(Vector) _world_next _world_current) 2e+01) (* (Math|Vector|GetUnitDirection(Vector) _world_previous _world_current) 2e+01)) _world_current) "0, 0,256.000000") :End (+ (+ (* (Math|Vector|GetUnitDirection(Vector) _world_next _world_current) 2e+01) (* (Math|Vector|GetUnitDirection(Vector) _world_previous _world_current) 2e+01)) _world_current) :ObjectTypes (Utilities|Array|MakeArray "ObjectTypeQuery3") :bTraceComplex true :DrawDebugType "None" :bIgnoreSelf false :TraceColor "(R=1.000000,G=0.000000,B=0.000000,A=1.000000)" :TraceHitColor "(R=0.000000,G=1.000000,B=0.000000,A=1.000000)" :DrawTime 5.0))
                  (LocalVariables|SetHitDiagonal _returnvalue_10)
                  (bind (_outhits_2 _returnvalue_11) (Collision|MultiLineTraceForObjects :Start (+ (+ _world_current (+ (Math|Vector|NegateVector (* (Math|Vector|GetUnitDirection(Vector) _world_next _world_current) 2e+01)) (* (Math|Vector|GetUnitDirection(Vector) _world_previous _world_current) 2e+01))) "0, 0,256.000000") :End (+ _world_current (+ (Math|Vector|NegateVector (* (Math|Vector|GetUnitDirection(Vector) _world_next _world_current) 2e+01)) (* (Math|Vector|GetUnitDirection(Vector) _world_previous _world_current) 2e+01))) :ObjectTypes (Utilities|Array|MakeArray "ObjectTypeQuery3") :bTraceComplex true :DrawDebugType "None" :bIgnoreSelf false :TraceColor "(R=0.000000,G=0.073639,B=1.000000,A=1.000000)" :TraceHitColor "(R=0.000000,G=1.000000,B=0.000000,A=1.000000)" :DrawTime 5.0)) ; right
                  (LocalVariables|SetHitRight _returnvalue_11)
                  (bind (_outhits_3 _returnvalue_12) (Collision|MultiLineTraceForObjects :Start (+ (+ _world_current (+ (* (Math|Vector|GetUnitDirection(Vector) _world_next _world_current) 2e+01) (Math|Vector|NegateVector (* (Math|Vector|GetUnitDirection(Vector) _world_previous _world_current) 2e+01)))) "0, 0,256.000000") :End (+ _world_current (+ (* (Math|Vector|GetUnitDirection(Vector) _world_next _world_current) 2e+01) (Math|Vector|NegateVector (* (Math|Vector|GetUnitDirection(Vector) _world_previous _world_current) 2e+01)))) :ObjectTypes (Utilities|Array|MakeArray "ObjectTypeQuery3") :bTraceComplex true :DrawDebugType "None" :bIgnoreSelf false :TraceColor "(R=1.000000,G=0.000000,B=0.789068,A=1.000000)" :TraceHitColor "(R=0.000000,G=1.000000,B=0.000000,A=1.000000)" :DrawTime 5.0))
                  (LocalVariables|SetHitLeft _returnvalue_12)
                  (if (Math|Boolean|ANDBoolean (LocalVariables|GetHitDiagonal) (not (LocalVariables|GetHitRight)) (not (LocalVariables|GetHitLeft)))
                    _
                    (elif (or _returnvalue_7 (Math|Boolean|ANDBoolean (LocalVariables|GetHitDiagonal) (not (LocalVariables|GetHitRight)) (LocalVariables|GetHitLeft)))
                      (Utilities|FlowControl|Sequence
                        (:then_0
                          (CreateCollisionComponents|MakeOuterCornerCollision (Variables|Default|GetSpline)))
                        (:then_1
                          (for _array_element_2 (Variables|C.%29ConnectingComplexPOIs|GetConnectedActors)
                            (CreateCollisionComponents|MakeOuterCornerCollision (Variables|Default|GetSpline _array_element_2) "World" true)))
                        (:then_2
                          (bind (_next__1_2 _next_2 _current_2 _previous_2 _previous__1_2 _world_next__1_2 _world_next_2 _world_current_2 _world_previous_2) (Tools|GetCurrentNextPreviousSplinePoint (LocalVariables|GetCurrentSpline) _index false (Variables|GetReverse)))
                          (bind _array_element_3 (Utilities|Array|ForEachLoopwithBreak (Variables|GetCorner_Inner_AssetList1))
                            (:LoopBody
                              (LocalVariables|SetDistanceToNext 0.0)
                              (LocalVariables|SetDistanceToPrevious 0.0)
                              (bind (_outhits_4 _returnvalue_13) (Collision|MultiLineTraceForObjects :Start (Math|Vector|vector+vector _world_current_2 (* (Math|Vector|SelectVector (Math|Vector|GetUnitDirection(Vector) _world_current_2 _world_previous_2) (Math|Vector|NegateVector (Math|Vector|GetUnitDirection(Vector) _world_current_2 _world_previous_2)) _returnvalue_7) 1e+01) (* (Math|Vector|SelectVector (Math|Vector|NegateVector (Math|Vector|GetUnitDirection(Vector) _world_current_2 _world_next_2)) (Math|Vector|GetUnitDirection(Vector) _world_current_2 _world_next_2) _returnvalue_7) 1e+01)) :End (Math|Vector|vector+vector (Math|Vector|vector+vector _world_current_2 (* (Math|Vector|SelectVector (Math|Vector|GetUnitDirection(Vector) _world_current_2 _world_previous_2) (Math|Vector|NegateVector (Math|Vector|GetUnitDirection(Vector) _world_current_2 _world_previous_2)) _returnvalue_7) 1e+01) (* (Math|Vector|SelectVector (Math|Vector|NegateVector (Math|Vector|GetUnitDirection(Vector) _world_current_2 _world_next_2)) (Math|Vector|GetUnitDirection(Vector) _world_current_2 _world_next_2) _returnvalue_7) 1e+01)) "0, 0, 0" (* (Math|Vector|SelectVector (Math|Vector|GetUnitDirection(Vector) _world_current_2 _world_previous_2) (Math|Vector|NegateVector (Math|Vector|GetUnitDirection(Vector) _world_current_2 _world_previous_2)) _returnvalue_7) (- (* (.SizeX_11_DFCCFEB9407AC738B8EC5AAF3A48E13C (Utilities|Struct|BreakProcAssetInfo (Utilities|Array|Get(acopy) _assetlist_5 _array_element_3))) 2.0) 2e+01))) :ObjectTypes (Utilities|Array|MakeArray "ObjectTypeQuery3") :bTraceComplex true :DrawDebugType "None" :bIgnoreSelf false :TraceColor "(R=1.000000,G=0.000000,B=0.789068,A=1.000000)" :TraceHitColor "(R=0.000000,G=1.000000,B=0.000000,A=1.000000)" :DrawTime 5.0))
                              (if _returnvalue_13
                                (bind _hitresoultoutput (Tools|ClosestHitResult _outhits_4 (Math|Vector|vector+vector _world_current_2 (* (Math|Vector|SelectVector (Math|Vector|GetUnitDirection(Vector) _world_current_2 _world_previous_2) (Math|Vector|NegateVector (Math|Vector|GetUnitDirection(Vector) _world_current_2 _world_previous_2)) _returnvalue_7) 1e+01) (* (Math|Vector|SelectVector (Math|Vector|NegateVector (Math|Vector|GetUnitDirection(Vector) _world_current_2 _world_next_2)) (Math|Vector|GetUnitDirection(Vector) _world_current_2 _world_next_2) _returnvalue_7) 1e+01)))))
                              (bind _output_5 (Tools|GetCollisionComponentTag (.HitComponent (Collision|BreakHitResult _hitresoultoutput))))
                              (bind (_outhits_5 _returnvalue_14) (Collision|MultiLineTraceForObjects :Start (Math|Vector|vector+vector _world_current_2 (* (Math|Vector|SelectVector (Math|Vector|NegateVector (Math|Vector|GetUnitDirection(Vector) _world_current_2 _world_next_2)) (Math|Vector|GetUnitDirection(Vector) _world_current_2 _world_next_2) _returnvalue_7) 1e+01) (* (Math|Vector|SelectVector (Math|Vector|GetUnitDirection(Vector) _world_current_2 _world_previous_2) (Math|Vector|NegateVector (Math|Vector|GetUnitDirection(Vector) _world_current_2 _world_previous_2)) _returnvalue_7) 1e+01)) :End (Math|Vector|vector+vector (Math|Vector|vector+vector _world_current_2 (* (Math|Vector|SelectVector (Math|Vector|NegateVector (Math|Vector|GetUnitDirection(Vector) _world_current_2 _world_next_2)) (Math|Vector|GetUnitDirection(Vector) _world_current_2 _world_next_2) _returnvalue_7) 1e+01) (* (Math|Vector|SelectVector (Math|Vector|GetUnitDirection(Vector) _world_current_2 _world_previous_2) (Math|Vector|NegateVector (Math|Vector|GetUnitDirection(Vector) _world_current_2 _world_previous_2)) _returnvalue_7) 1e+01)) "0, 0, 0" (* (Math|Vector|SelectVector (Math|Vector|NegateVector (Math|Vector|GetUnitDirection(Vector) _world_current_2 _world_next_2)) (Math|Vector|GetUnitDirection(Vector) _world_current_2 _world_next_2) _returnvalue_7) (- (* (.SizeX_11_DFCCFEB9407AC738B8EC5AAF3A48E13C (Utilities|Struct|BreakProcAssetInfo (Utilities|Array|Get(acopy) _assetlist_5 _array_element_3))) 2.0) 2e+01))) :ObjectTypes (Utilities|Array|MakeArray "ObjectTypeQuery3") :bTraceComplex true :DrawDebugType "None" :bIgnoreSelf false :TraceColor "(R=0.000000,G=0.085981,B=1.000000,A=1.000000)" :TraceHitColor "(R=0.000000,G=1.000000,B=0.000000,A=1.000000)" :DrawTime 5.0))
                              (if _returnvalue_14
                                (bind _hitresoultoutput_2 (Tools|ClosestHitResult _outhits_5 (Math|Vector|vector+vector _world_current_2 (* (Math|Vector|SelectVector (Math|Vector|NegateVector (Math|Vector|GetUnitDirection(Vector) _world_current_2 _world_next_2)) (Math|Vector|GetUnitDirection(Vector) _world_current_2 _world_next_2) _returnvalue_7) 1e+01) (* (Math|Vector|SelectVector (Math|Vector|GetUnitDirection(Vector) _world_current_2 _world_previous_2) (Math|Vector|NegateVector (Math|Vector|GetUnitDirection(Vector) _world_current_2 _world_previous_2)) _returnvalue_7) 1e+01)))))
                              (bind _output_6 (Tools|GetCollisionComponentTag (.HitComponent (Collision|BreakHitResult _hitresoultoutput_2))))
                              (if (and (or (>= (+ (.Distance (Collision|BreakHitResult _hitresoultoutput)) 12.0) (select (== _output_5 "OuterCorner") (.SizeX_11_DFCCFEB9407AC738B8EC5AAF3A48E13C (Utilities|Struct|BreakProcAssetInfo (Utilities|Array|Get(acopy) _assetlist_5 _array_element_3))) (* (.SizeX_11_DFCCFEB9407AC738B8EC5AAF3A48E13C (Utilities|Struct|BreakProcAssetInfo (Utilities|Array|Get(acopy) _assetlist_5 _array_element_3))) 2.0))) (not _returnvalue_13)) (or (not _returnvalue_14) (>= (+ (.Distance (Collision|BreakHitResult _hitresoultoutput_2)) 12.0) (select (== _output_6 "OuterCorner") (.SizeX_11_DFCCFEB9407AC738B8EC5AAF3A48E13C (Utilities|Struct|BreakProcAssetInfo (Utilities|Array|Get(acopy) _assetlist_5 _array_element_3))) (* (.SizeX_11_DFCCFEB9407AC738B8EC5AAF3A48E13C (Utilities|Struct|BreakProcAssetInfo (Utilities|Array|Get(acopy) _assetlist_5 _array_element_3))) 2.0)))))
                                (Utilities|FlowControl|Sequence
                                  (:then_0
                                    _)
                                  (:then_1
                                    (Utilities|Array|Add (LocalVariables|GetListToMake%20Collision%20from) (Math|Transform|MakeTransform (+ _current_2 (Math|Vector|SelectVector (* (Math|Vector|GetUnitDirection(Vector) _current_2 _previous_2) (.SizeX_11_DFCCFEB9407AC738B8EC5AAF3A48E13C (Utilities|Struct|BreakProcAssetInfo (Utilities|Array|Get(acopy) _assetlist_5 _array_element_3)))) (* (Math|Vector|NegateVector (Math|Vector|GetUnitDirection(Vector) _current_2 _previous_2)) (.SizeX_11_DFCCFEB9407AC738B8EC5AAF3A48E13C (Utilities|Struct|BreakProcAssetInfo (Utilities|Array|Get(acopy) _assetlist_5 _array_element_3)))) _returnvalue_7)) (Math|Rotator|SelectRotator (Math|Rotator|MakeRotationfromAxes (Math|Vector|NegateVector (Math|Vector|GetUnitDirection(Vector) _current_2 _next_2)) (Math|Vector|GetUnitDirection(Vector) _current_2 _previous_2) "0, 0,1.000000") (Math|Rotator|MakeRotationfromAxes (Math|Vector|GetUnitDirection(Vector) _current_2 _next_2) (Math|Vector|NegateVector (Math|Vector|GetUnitDirection(Vector) _current_2 _previous_2)) "0, 0,1.000000") _returnvalue_7) (Math|Vector|MakeVector (/ (.SizeX_11_DFCCFEB9407AC738B8EC5AAF3A48E13C (Utilities|Struct|BreakProcAssetInfo (Utilities|Array|Get(acopy) _assetlist_5 _array_element_3))) 512.0) (/ (.SizeX_11_DFCCFEB9407AC738B8EC5AAF3A48E13C (Utilities|Struct|BreakProcAssetInfo (Utilities|Array|Get(acopy) _assetlist_5 _array_element_3))) 512.0) 1.0))))
                                  (:then_2
                                    _))
                                (Utilities|Array|Add (Variables|X.%29DoNOTEdit|GetComponent%20Transform%20List) (Utilities|Struct|MakeProcComponenttoSpawn (Math|Transform|MakeTransform (+ _current_2 (Math|Vector|SelectVector (* (Math|Vector|GetUnitDirection(Vector) _current_2 _previous_2) (.SizeX_11_DFCCFEB9407AC738B8EC5AAF3A48E13C (Utilities|Struct|BreakProcAssetInfo (Utilities|Array|Get(acopy) _assetlist_5 _array_element_3)))) (* (Math|Vector|NegateVector (Math|Vector|GetUnitDirection(Vector) _current_2 _previous_2)) (.SizeX_11_DFCCFEB9407AC738B8EC5AAF3A48E13C (Utilities|Struct|BreakProcAssetInfo (Utilities|Array|Get(acopy) _assetlist_5 _array_element_3)))) _returnvalue_7)) (Math|Rotator|SelectRotator (Math|Rotator|MakeRotationfromAxes (Math|Vector|NegateVector (Math|Vector|GetUnitDirection(Vector) _current_2 _next_2)) (Math|Vector|GetUnitDirection(Vector) _current_2 _previous_2) "0, 0,1.000000") (Math|Rotator|MakeRotationfromAxes (Math|Vector|GetUnitDirection(Vector) _current_2 _next_2) (Math|Vector|NegateVector (Math|Vector|GetUnitDirection(Vector) _current_2 _previous_2)) "0, 0,1.000000") _returnvalue_7)) (Variables|E.%29Meshes|GetSidewalk_Asset_List) _array_element_3))))))
                        (:then_3
                          (Cleanup|DeleteOuterCornerCollision)))
                      (elif (or _returnvalue_8 (Math|Boolean|ANDBoolean (not (LocalVariables|GetHitDiagonal)) (LocalVariables|GetHitRight) (not (LocalVariables|GetHitLeft))))
                        (bind (_next__1_3 _next_3 _current_3 _previous_3) (Tools|GetCurrentNextPreviousSplinePoint (LocalVariables|GetCurrentSpline) _index false (Variables|GetReverse)))
                        (Utilities|FlowControl|Sequence
                          (:then_0
                            (Utilities|Array|Add (Variables|X.%29DoNOTEdit|GetComponent%20Transform%20List) (Utilities|Struct|MakeProcComponenttoSpawn (Math|Transform|MakeTransform (+ _current_3 (* (Math|Vector|SelectVector (Math|Vector|GetUnitDirection(Vector) _current_3 _next_3) "0, 0, 0" _returnvalue_8) _sizex_11_dfccfeb9407ac738b8ec5aaf3a48e13c_2)) (select _returnvalue_8 (Math|Rotator|MakeRotationfromAxes (Math|Vector|NegateVector (Math|Vector|GetUnitDirection(Vector) _current_3 _next_3)) (Math|Vector|GetUnitDirection(Vector) _current_3 _previous_3) "0, 0,1.000000") (Math|Rotator|MakeRotationfromAxes (Math|Vector|GetUnitDirection(Vector) _current_3 _previous_3) (Math|Vector|GetUnitDirection(Vector) _current_3 _next_3) "0, 0,1.000000"))) (Variables|E.%29Meshes|GetSidewalk_Asset_List) _output_3)))
                          (:then_1
                            (Utilities|Array|Add (LocalVariables|GetListToMake%20Collision%20from) (Math|Transform|MakeTransform (+ _current_3 (* (Math|Vector|SelectVector (Math|Vector|GetUnitDirection(Vector) _current_3 _next_3) "0, 0, 0" _returnvalue_8) _sizex_11_dfccfeb9407ac738b8ec5aaf3a48e13c_2)) (select _returnvalue_8 (Math|Rotator|MakeRotationfromAxes (Math|Vector|NegateVector (Math|Vector|GetUnitDirection(Vector) _current_3 _next_3)) (Math|Vector|GetUnitDirection(Vector) _current_3 _previous_3) "0, 0,1.000000") (Math|Rotator|MakeRotationfromAxes (Math|Vector|GetUnitDirection(Vector) _current_3 _previous_3) (Math|Vector|GetUnitDirection(Vector) _current_3 _next_3) "0, 0,1.000000")))))))))))))
          (:then_1
            (for _array_element_4 (LocalVariables|GetListToMake%20Collision%20from)
              (bind _returnvalue_15 (AddComponent|NODE_AddStaticMeshComponent-47 :RelativeTransform _array_element_4))
              (Variables|/Script/Engine.ActorComponent|SetComponentTags (Utilities|Array|MakeArray "inner") _returnvalue_15)
              (Utilities|Array|Add (Variables|MeshLists|GetBorderCollisionList) _returnvalue_15))))))
    (:then_2
      (for _index_2 (range (Spline|GetNumberOfSplinePoints Spline))
        (bind (_next__1_4 _next_4 _current_4 _previous_4 _previous__1_3 _world_next__1_3 _world_next_3 _world_current_3 _world_previous_3) (Tools|GetCurrentNextPreviousSplinePoint Spline _index_2 false (Variables|GetReverse)))
        (LocalVariables|SetDistance%20to%20Next%20Original (Utilities|Roundto512Grid (Math|Vector|Distance(Vector) _current_4 _next_4)))
        (LocalVariables|SetDistance%20to%20Previous%20Original (Utilities|Roundto512Grid (Math|Vector|Distance(Vector) _current_4 _previous_4)))
        (LocalVariables|SetDistanceToNext (Utilities|Roundto512Grid (Math|Vector|Distance(Vector) _current_4 _next_4)))
        (LocalVariables|SetDistanceToPrevious (Utilities|Roundto512Grid (Math|Vector|Distance(Vector) _current_4 _previous_4)))
        (if (not (Utilities|IsOuterCorner _next_4 _current_4 _previous_4))
          (bind (_outhits_6 _returnvalue_16) (Collision|MultiLineTraceForObjects :Start (+ (+ _world_current_3 (* (+ (Math|Vector|GetUnitDirection(Vector) _world_current_3 _world_previous_3) (Math|Vector|GetUnitDirection(Vector) _world_current_3 _world_next_3)) 2e+01)) "0, 0,256.000000") :End (+ _world_current_3 (* (+ (Math|Vector|GetUnitDirection(Vector) _world_current_3 _world_previous_3) (Math|Vector|GetUnitDirection(Vector) _world_current_3 _world_next_3)) 2e+01)) :ObjectTypes (Utilities|Array|MakeArray "ObjectTypeQuery3") :bTraceComplex false :DrawDebugType "None" :bIgnoreSelf false :TraceColor "(R=1.000000,G=0.913331,B=0.000000,A=1.000000)" :TraceHitColor "(R=0.000000,G=1.000000,B=0.000000,A=1.000000)" :DrawTime 5.0))
          (if (not _returnvalue_16)
            (bind (_outhits_7 _returnvalue_17) (Collision|MultiLineTraceForObjects :Start (+ (+ _world_current_3 (* (Math|Vector|GetUnitDirection(Vector) _world_current_3 _world_next_3) 2e+01)) (* (Math|Vector|GetUnitDirection(Vector) _world_current_3 _world_previous_3) 1.0)) :End (+ (+ _world_current_3 (* (Math|Vector|GetUnitDirection(Vector) _world_current_3 _world_next_3) 2e+01)) (* (Math|Vector|GetUnitDirection(Vector) _world_current_3 _world_previous_3) (LocalVariables|GetDistanceToPrevious))) :ObjectTypes (Utilities|Array|MakeArray "ObjectTypeQuery3") :bTraceComplex false :DrawDebugType "None" :bIgnoreSelf false :TraceColor "(R=0.000000,G=0.964484,B=1.000000,A=1.000000)" :TraceHitColor "(R=0.000000,G=1.000000,B=0.000000,A=1.000000)" :DrawTime 5.0))
            (LocalVariables|SetHitRight _returnvalue_17)
            (if _returnvalue_17
              (bind _hitresoultoutput_3 (Tools|ClosestHitResult _outhits_7 (+ (+ _world_current_3 (* (Math|Vector|GetUnitDirection(Vector) _world_current_3 _world_next_3) 2e+01)) (* (Math|Vector|GetUnitDirection(Vector) _world_current_3 _world_previous_3) 1.0))))
              (LocalVariables|SetDistanceToPrevious (Utilities|Roundto512Grid (.Distance (Collision|BreakHitResult _hitresoultoutput_3))))
              (bind _output_7 (Tools|GetCollisionComponentTag (.HitComponent (Collision|BreakHitResult _hitresoultoutput_3))))
              (LocalVariables|SetHitTagRight _output_7))
            (bind (_outhits_8 _returnvalue_18) (Collision|MultiLineTraceForObjects :Start (+ (+ _world_current_3 (* (Math|Vector|GetUnitDirection(Vector) _world_current_3 _world_previous_3) 2e+01)) (* (Math|Vector|GetUnitDirection(Vector) _world_current_3 _world_next_3) 1.0)) :End (+ (+ _world_current_3 (* (Math|Vector|GetUnitDirection(Vector) _world_current_3 _world_previous_3) 2e+01)) (* (Math|Vector|GetUnitDirection(Vector) _world_current_3 _world_next_3) (LocalVariables|GetDistanceToNext))) :ObjectTypes (Utilities|Array|MakeArray "ObjectTypeQuery3") :bTraceComplex false :DrawDebugType "None" :bIgnoreSelf false :TraceColor "(R=0.986906,G=1.000000,B=0.000000,A=1.000000)" :TraceHitColor "(R=0.000000,G=1.000000,B=0.000000,A=1.000000)" :DrawTime 5.0))
            (LocalVariables|SetHitLeft _returnvalue_18)
            (if _returnvalue_18
              (bind _hitresoultoutput_4 (Tools|ClosestHitResult _outhits_8 (+ (+ _world_current_3 (* (Math|Vector|GetUnitDirection(Vector) _world_current_3 _world_previous_3) 2e+01)) (* (Math|Vector|GetUnitDirection(Vector) _world_current_3 _world_next_3) 1.0))))
              (LocalVariables|SetDistanceToNext (Utilities|Roundto512Grid (.Distance (Collision|BreakHitResult _hitresoultoutput_4)))))
            (bind _output_8 (Tools|GetCollisionComponentTag (.HitComponent (Collision|BreakHitResult _hitresoultoutput_4))))
            (LocalVariables|SetHitTagLeft _output_8)
            (Utilities|FlowControl|Sequence
              (:then_0
                ;[1 1 1] S BEnd Logic
                (if (and (or (Math|Boolean|ANDBoolean (Utilities|IsOuterCorner _next__1_4 _next_4 _current_4) _returnvalue_5 (>= _returnvalue_6 (Math|Float|SelectFloat 1024.0 2048.0 (Utilities|IsOuterCorner _current_4 _previous_4 _previous__1_3)))) (Math|Boolean|ANDBoolean (Utilities|IsOuterCorner _current_4 _previous_4 _previous__1_3) (<= _distance_to_previous_original 512.0) (>= (+ _distance_to_next_original 1.0) (Math|Float|SelectFloat 1024.0 2048.0 (Utilities|IsOuterCorner _next__1_4 _next_4 _current_4))))) (> (Utilities|Array|Length (Variables|GetS-BendAssetIndexList1)) 0))
                  ;[1 1 1] Extra Check if area is clear and not occupied by another S Bend already
                  (bind (_outhits_9 _returnvalue_19) (Collision|MultiLineTraceForObjects :Start (+ (Utilities|Operators|Add _world_current_3 (* (Math|Vector|GetUnitDirection(Vector) _world_current_3 _world_previous_3) 64.0) (* (Math|Vector|GetUnitDirection(Vector) _world_current_3 _world_next_3) 768.0)) "0,0,256.000000") :End (+ (Utilities|Operators|Add _world_current_3 (* (Math|Vector|GetUnitDirection(Vector) _world_current_3 _world_previous_3) 64.0) (* (Math|Vector|GetUnitDirection(Vector) _world_current_3 _world_next_3) 768.0)) "0,0,-512.000000") :ObjectTypes _array :bTraceComplex false :DrawDebugType "None" :bIgnoreSelf false :TraceColor "(R=0.000000,G=0.964484,B=1.000000,A=1.000000)" :TraceHitColor "(R=0.000000,G=1.000000,B=0.000000,A=1.000000)" :DrawTime 5.0))
                  (bind (_outhits_10 _returnvalue_20) (Collision|MultiLineTraceForObjects :Start (+ (Utilities|Operators|Add _world_current_3 (* (Math|Vector|GetUnitDirection(Vector) _world_current_3 _world_next_3) 64.0) (* (Math|Vector|GetUnitDirection(Vector) _world_current_3 _world_previous_3) 768.0)) "0,0,256.000000") :End (+ (Utilities|Operators|Add _world_current_3 (* (Math|Vector|GetUnitDirection(Vector) _world_current_3 _world_next_3) 64.0) (* (Math|Vector|GetUnitDirection(Vector) _world_current_3 _world_previous_3) 768.0)) "0,0,-512.000000") :ObjectTypes _array :bTraceComplex false :DrawDebugType "None" :bIgnoreSelf false :TraceColor "(R=0.000000,G=0.964484,B=1.000000,A=1.000000)" :TraceHitColor "(R=0.000000,G=1.000000,B=0.000000,A=1.000000)" :DrawTime 5.0))
                  (if (not (or _returnvalue_19 _returnvalue_20))
                    (Utilities|FlowControl|Sequence
                      (:then_0
                        (Utilities|Array|Add (Variables|X.%29DoNOTEdit|GetComponent%20Transform%20List) (Utilities|Struct|MakeProcComponenttoSpawn (Math|Transform|MakeTransform (+ _current_4 (Math|Vector|SelectVector (* (Math|Vector|GetUnitDirection(Vector) _current_4 _previous_4) _sizex_11_dfccfeb9407ac738b8ec5aaf3a48e13c) (* (Math|Vector|NegateVector (Math|Vector|GetUnitDirection(Vector) _next_4 _current_4)) _sizex_11_dfccfeb9407ac738b8ec5aaf3a48e13c) (Math|Boolean|ANDBoolean (Utilities|IsOuterCorner _next__1_4 _next_4 _current_4) _returnvalue_5 (>= _returnvalue_6 (Math|Float|SelectFloat 1024.0 2048.0 (Utilities|IsOuterCorner _current_4 _previous_4 _previous__1_3)))))) (select (Math|Boolean|ANDBoolean (Utilities|IsOuterCorner _next__1_4 _next_4 _current_4) _returnvalue_5 (>= _returnvalue_6 (Math|Float|SelectFloat 1024.0 2048.0 (Utilities|IsOuterCorner _current_4 _previous_4 _previous__1_3)))) (Math|Rotator|MakeRotationfromAxes (Math|Vector|NegateVector (Math|Vector|GetUnitDirection(Vector) _next_4 _current_4)) (Math|Vector|GetUnitDirection(Vector) _current_4 _previous_4) "0, 0,1.000000") (Math|Rotator|MakeRotationfromAxes (Math|Vector|GetUnitDirection(Vector) _current_4 _previous_4) (Math|Vector|GetUnitDirection(Vector) _next_4 _current_4) "0, 0,1.000000")) (Math|Vector|SelectVector "1.000000,1.000000,1.000000" "1.000000,-1.000000,1.000000" (Math|Boolean|ANDBoolean (Utilities|IsOuterCorner _next__1_4 _next_4 _current_4) _returnvalue_5 (>= _returnvalue_6 (Math|Float|SelectFloat 1024.0 2048.0 (Utilities|IsOuterCorner _current_4 _previous_4 _previous__1_3)))))) (Variables|E.%29Meshes|GetSidewalk_Asset_List) _output)))
                      (:then_1
                        (bind _returnvalue_21 (AddComponent|NODE_AddStaticMeshComponent-37 :RelativeTransform (Math|Transform|MakeTransform (+ _current_4 (Math|Vector|SelectVector (* (Math|Vector|GetUnitDirection(Vector) _current_4 _previous_4) _sizex_11_dfccfeb9407ac738b8ec5aaf3a48e13c) (* (Math|Vector|NegateVector (Math|Vector|GetUnitDirection(Vector) _next_4 _current_4)) _sizex_11_dfccfeb9407ac738b8ec5aaf3a48e13c) (Math|Boolean|ANDBoolean (Utilities|IsOuterCorner _next__1_4 _next_4 _current_4) _returnvalue_5 (>= _returnvalue_6 (Math|Float|SelectFloat 1024.0 2048.0 (Utilities|IsOuterCorner _current_4 _previous_4 _previous__1_3)))))) (select (Math|Boolean|ANDBoolean (Utilities|IsOuterCorner _next__1_4 _next_4 _current_4) _returnvalue_5 (>= _returnvalue_6 (Math|Float|SelectFloat 1024.0 2048.0 (Utilities|IsOuterCorner _current_4 _previous_4 _previous__1_3)))) (Math|Rotator|MakeRotationfromAxes (Math|Vector|NegateVector (Math|Vector|GetUnitDirection(Vector) _next_4 _current_4)) (Math|Vector|GetUnitDirection(Vector) _current_4 _previous_4) "0, 0,1.000000") (Math|Rotator|MakeRotationfromAxes (Math|Vector|GetUnitDirection(Vector) _current_4 _previous_4) (Math|Vector|GetUnitDirection(Vector) _next_4 _current_4) "0, 0,1.000000")) (* (Math|Vector|SelectVector "1.000000,1.000000,1.000000" "1.000000,-1.000000,1.000000" (Math|Boolean|ANDBoolean (Utilities|IsOuterCorner _next__1_4 _next_4 _current_4) _returnvalue_5 (>= _returnvalue_6 (Math|Float|SelectFloat 1024.0 2048.0 (Utilities|IsOuterCorner _current_4 _previous_4 _previous__1_3))))) "2.000000,2.000000,1.000000"))))
                        (Utilities|Array|Add (Variables|MeshLists|GetBorderCollisionList) _returnvalue_21)))))
                (bind _index_3 (Utilities|FlowControl|ForLoopwithBreak 0 (Utilities|Array|LastIndex _corner_inner_assetlist1))
                  (:LoopBody
                    (if (and (and (or (Utilities|IsOuterCorner _next__1_4 _next_4 _current_4) (>= (+ (LocalVariables|GetDistance%20to%20Next%20Original) 2.0) (* (.SizeX_11_DFCCFEB9407AC738B8EC5AAF3A48E13C (Utilities|Struct|BreakProcAssetInfo (Utilities|Array|Get(acopy) _assetlist_3 (Utilities|Array|Get(acopy) _corner_inner_assetlist1 _index_3)))) 2.0))) (or (Utilities|IsOuterCorner _current_4 _previous_4 _previous__1_3) (>= (+ (LocalVariables|GetDistance%20to%20Previous%20Original) 2.0) (* (.SizeX_11_DFCCFEB9407AC738B8EC5AAF3A48E13C (Utilities|Struct|BreakProcAssetInfo (Utilities|Array|Get(acopy) _assetlist_3 (Utilities|Array|Get(acopy) _corner_inner_assetlist1 _index_3)))) 2.0)))) (and (>= (+ (LocalVariables|GetDistanceToPrevious) 2.0) (Math|Float|SelectFloat (* (.SizeX_11_DFCCFEB9407AC738B8EC5AAF3A48E13C (Utilities|Struct|BreakProcAssetInfo (Utilities|Array|Get(acopy) _assetlist_3 (Utilities|Array|Get(acopy) _corner_inner_assetlist1 _index_3)))) 2.0) (.SizeX_11_DFCCFEB9407AC738B8EC5AAF3A48E13C (Utilities|Struct|BreakProcAssetInfo (Utilities|Array|Get(acopy) _assetlist_3 (Utilities|Array|Get(acopy) _corner_inner_assetlist1 _index_3)))) (or (Utilities|Name|Equal(Name) (LocalVariables|GetHitTagRight) "inner") (Utilities|Name|Equal(Name) (LocalVariables|GetHitTagLeft) "inner")))) (>= (+ (LocalVariables|GetDistanceToNext) 2.0) (Math|Float|SelectFloat (* (.SizeX_11_DFCCFEB9407AC738B8EC5AAF3A48E13C (Utilities|Struct|BreakProcAssetInfo (Utilities|Array|Get(acopy) _assetlist_3 (Utilities|Array|Get(acopy) _corner_inner_assetlist1 _index_3)))) 2.0) (.SizeX_11_DFCCFEB9407AC738B8EC5AAF3A48E13C (Utilities|Struct|BreakProcAssetInfo (Utilities|Array|Get(acopy) _assetlist_3 (Utilities|Array|Get(acopy) _corner_inner_assetlist1 _index_3)))) (or (Utilities|Name|Equal(Name) (LocalVariables|GetHitTagRight) "inner") (Utilities|Name|Equal(Name) (LocalVariables|GetHitTagLeft) "inner"))))))
                      (Utilities|FlowControl|Sequence
                        (:then_0
                          _)
                        (:then_1
                          (bind _returnvalue_22 (AddComponent|NODE_AddStaticMeshComponent-38 :RelativeTransform (Math|Transform|MakeTransform (+ (* (Math|Vector|GetUnitDirection(Vector) _current_4 _previous_4) (.SizeX_11_DFCCFEB9407AC738B8EC5AAF3A48E13C (Utilities|Struct|BreakProcAssetInfo (Utilities|Array|Get(acopy) _assetlist_3 (Utilities|Array|Get(acopy) _corner_inner_assetlist1 _index_3))))) _current_4) (Math|Rotator|MakeRotationfromAxes (Math|Vector|NegateVector (Math|Vector|GetUnitDirection(Vector) _next_4 _current_4)) (Math|Vector|GetUnitDirection(Vector) _current_4 _previous_4) "0, 0,1.000000") (Math|Vector|MakeVector (/ (.SizeX_11_DFCCFEB9407AC738B8EC5AAF3A48E13C (Utilities|Struct|BreakProcAssetInfo (Utilities|Array|Get(acopy) _assetlist_3 (Utilities|Array|Get(acopy) _corner_inner_assetlist1 _index_3)))) 512.0) (/ (.SizeX_11_DFCCFEB9407AC738B8EC5AAF3A48E13C (Utilities|Struct|BreakProcAssetInfo (Utilities|Array|Get(acopy) _assetlist_3 (Utilities|Array|Get(acopy) _corner_inner_assetlist1 _index_3)))) 512.0) 1.0))))
                          (Utilities|Array|Add (Variables|MeshLists|GetBorderCollisionList) _returnvalue_22))
                        (:then_2
                          _))
                      (Utilities|Array|Add (Variables|X.%29DoNOTEdit|GetComponent%20Transform%20List) (Utilities|Struct|MakeProcComponenttoSpawn (Math|Transform|MakeTransform (+ (* (Math|Vector|GetUnitDirection(Vector) _current_4 _previous_4) (.SizeX_11_DFCCFEB9407AC738B8EC5AAF3A48E13C (Utilities|Struct|BreakProcAssetInfo (Utilities|Array|Get(acopy) _assetlist_3 (Utilities|Array|Get(acopy) _corner_inner_assetlist1 _index_3))))) _current_4) (Math|Rotator|MakeRotationfromAxes (Math|Vector|NegateVector (Math|Vector|GetUnitDirection(Vector) _next_4 _current_4)) (Math|Vector|GetUnitDirection(Vector) _current_4 _previous_4) "0, 0,1.000000")) (Variables|E.%29Meshes|GetSidewalk_Asset_List) (Utilities|Array|Get(acopy) _corner_inner_assetlist1 _index_3)))))))
              (:then_1
                (if false
                  (Debug|DebugText (Math|Transform|MakeTransform _current_4) (Utilities|Text|ToText(Name) "None") "" "" "" "" "" 0 0 true))))))))
    (:then_3
      (for _index_4 (range (Spline|GetNumberOfSplinePoints Spline))
        (bind (_next__1_5 _next_5 _current_5 _previous_5 _previous__1_4 _world_next__1_4 _world_next_4 _world_current_4 _world_previous_4) (Tools|GetCurrentNextPreviousSplinePoint Spline _index_4 false (Variables|GetReverse)))
        (if (Utilities|IsOuterCorner _next_5 _current_5 _previous_5)
          (Utilities|FlowControl|Sequence
            (:then_0
              (bind (_outhits_11 _returnvalue_23) (Collision|MultiLineTraceForObjects :Start (+ (+ _world_current_4 (* (+ (Math|Vector|GetUnitDirection(Vector) _world_previous_4 _world_current_4) (Math|Vector|GetUnitDirection(Vector) _world_next_4 _world_current_4)) 2e+01)) "0, 0,256.000000") :End (+ _world_current_4 (* (+ (Math|Vector|GetUnitDirection(Vector) _world_previous_4 _world_current_4) (Math|Vector|GetUnitDirection(Vector) _world_next_4 _world_current_4)) 2e+01)) :ObjectTypes (Utilities|Array|MakeArray "ObjectTypeQuery3") :bTraceComplex false :DrawDebugType "None" :bIgnoreSelf false :TraceColor "(R=1.000000,G=0.000000,B=0.000000,A=1.000000)" :TraceHitColor "(R=0.000000,G=1.000000,B=0.000000,A=1.000000)" :DrawTime 5.0))
              (if (not _returnvalue_23)
                (Utilities|Array|Add (Variables|X.%29DoNOTEdit|GetComponent%20Transform%20List) (Utilities|Struct|MakeProcComponenttoSpawn (Math|Transform|MakeTransform _current_5 (Math|Rotator|MakeRotationfromAxes (Math|Vector|GetUnitDirection(Vector) _next_5 _current_5) (Math|Vector|GetUnitDirection(Vector) _current_5 _previous_5) "0, 0,1.000000")) (Variables|E.%29Meshes|GetSidewalk_Asset_List) (Utilities|Array|Get(acopy) (Variables|GetOuterCornerAssetList1))))))))))
    (:then_4
      (for _index_5 (range (Spline|GetNumberOfSplinePoints Spline))
        (bind (_next__1_6 _next_6 _current_6 _previous_6 _previous__1_5 _world_next__1_5 _world_next_5 _world_current_5) (Tools|GetCurrentNextPreviousSplinePoint Spline _index_5 false (Variables|GetReverse)))
        (LocalVariables|SetCurrentSidewalkLength 0)
        (for _index_6 (range (Math|Float|Round (/ (Math|Vector|Distance(Vector) _next_6 _current_6) 512.0)))
          (bind (_outhits_12 _returnvalue_24) (Collision|MultiLineTraceForObjects :Start (+ (Math|Vector|vector+vector (* (* (Math|Vector|GetUnitDirection(Vector) _world_current_5 _world_next_5) 512.0) (Math|Conversions|ToFloat(Integer) _index_6)) (* (Math|Vector|GetUnitDirection(Vector) _world_current_5 _world_next_5) 256.0) _world_current_5 (* (Math|Vector|Normalize (Math|Vector|CrossProduct (Math|Vector|GetUnitDirection(Vector) _world_current_5 _world_next_5) "0, 0,-1.000000")) 256.0)) "0, 0,256.000000") :End (Math|Vector|vector+vector (* (* (Math|Vector|GetUnitDirection(Vector) _world_current_5 _world_next_5) 512.0) (Math|Conversions|ToFloat(Integer) _index_6)) (* (Math|Vector|GetUnitDirection(Vector) _world_current_5 _world_next_5) 256.0) _world_current_5 (* (Math|Vector|Normalize (Math|Vector|CrossProduct (Math|Vector|GetUnitDirection(Vector) _world_current_5 _world_next_5) "0, 0,-1.000000")) 256.0)) :ObjectTypes (Utilities|Array|MakeArray "ObjectTypeQuery3") :bTraceComplex false :DrawDebugType "None" :bIgnoreSelf false :TraceColor "(R=1.000000,G=0.000000,B=0.000000,A=1.000000)" :TraceHitColor "(R=0.000000,G=1.000000,B=0.000000,A=1.000000)" :DrawTime 5.0))
          (if (and (== _index_6 (- (Math|Float|Round (/ (Math|Vector|Distance(Vector) _next_6 _current_6) 512.0)) 1)) (not _returnvalue_24))
            (Math|Integer|IncrementInt _currentsidewalklength
              (:%20%20
                _))
            (elif _returnvalue_24
              _
              (else
                (Math|Integer|IncrementInt _currentsidewalklength))))
          (if (> _currentsidewalklength 0)
            (LocalVariables|SetDistance (* _currentsidewalklength 512.0))
            (LocalVariables|SetCurrent%20Distance 0.0)
            (Utilities|FlowControl|Sequence
              (:then_0
                (for _array_element_5 (Variables|GetEdgeBorder_AssetList1)
                  (if (> (Math|Float|Floor (/ _returnvalue_9 (.SizeX_11_DFCCFEB9407AC738B8EC5AAF3A48E13C (Utilities|Struct|BreakProcAssetInfo (Utilities|Array|Get(acopy) _assetlist_4 _array_element_5))))) 0)
                    (for _index_7 (range (Math|Float|Floor (/ _returnvalue_9 (.SizeX_11_DFCCFEB9407AC738B8EC5AAF3A48E13C (Utilities|Struct|BreakProcAssetInfo (Utilities|Array|Get(acopy) _assetlist_4 _array_element_5))))))
                      (Utilities|FlowControl|Sequence
                        (:then_0
                          (Utilities|Array|Add (Variables|X.%29DoNOTEdit|GetComponent%20Transform%20List) (Utilities|Struct|MakeProcComponenttoSpawn (Math|Transform|MakeTransform (+ (Math|Vector|vector+vector _current_6 (* (Math|Vector|GetUnitDirection(Vector) _current_6 _next_6) (* (- (Math|Integer|SelectInt (+ _index_6 1) _index_6 (and (== _index_6 (- (Math|Float|Round (/ (Math|Vector|Distance(Vector) _next_6 _current_6) 512.0)) 1)) (not _returnvalue_24))) _currentsidewalklength) 512.0)) (* (Math|Vector|GetUnitDirection(Vector) _current_6 _next_6) (.SizeX_11_DFCCFEB9407AC738B8EC5AAF3A48E13C (Utilities|Struct|BreakProcAssetInfo (Utilities|Array|Get(acopy) _assetlist_4 _array_element_5))))) (* (Math|Vector|GetUnitDirection(Vector) _current_6 _next_6) (+ (* _index_7 (.SizeX_11_DFCCFEB9407AC738B8EC5AAF3A48E13C (Utilities|Struct|BreakProcAssetInfo (Utilities|Array|Get(acopy) _assetlist_4 _array_element_5)))) (LocalVariables|GetCurrent%20Distance)))) (Math|Rotator|MakeRotationfromAxes (Math|Vector|NegateVector (Math|Vector|GetUnitDirection(Vector) _current_6 _next_6)) (Math|Vector|NegateVector (Math|Vector|Normalize (Math|Vector|CrossProduct (Math|Vector|GetUnitDirection(Vector) _current_6 _next_6) "0, 0,-1.000000"))) "0, 0,1.000000")) (Variables|E.%29Meshes|GetSidewalk_Asset_List) _array_element_5)))))
                    (LocalVariables|SetCurrent%20Distance (+ (* (Math|Float|Floor (/ _returnvalue_9 (.SizeX_11_DFCCFEB9407AC738B8EC5AAF3A48E13C (Utilities|Struct|BreakProcAssetInfo (Utilities|Array|Get(acopy) _assetlist_4 _array_element_5))))) (.SizeX_11_DFCCFEB9407AC738B8EC5AAF3A48E13C (Utilities|Struct|BreakProcAssetInfo (Utilities|Array|Get(acopy) _assetlist_4 _array_element_5)))) (LocalVariables|GetCurrent%20Distance)))
                    (LocalVariables|SetDistance (- (LocalVariables|GetDistance) (* (Math|Float|Floor (/ _returnvalue_9 (.SizeX_11_DFCCFEB9407AC738B8EC5AAF3A48E13C (Utilities|Struct|BreakProcAssetInfo (Utilities|Array|Get(acopy) _assetlist_4 _array_element_5))))) (.SizeX_11_DFCCFEB9407AC738B8EC5AAF3A48E13C (Utilities|Struct|BreakProcAssetInfo (Utilities|Array|Get(acopy) _assetlist_4 _array_element_5)))))))))
              (:then_1
                _)))
          (LocalVariables|SetCurrentSidewalkLength 0))))))

(fn MakeSidewalks
  (Utilities|FlowControl|Sequence
    (:then_0
      (Utilities|FlowControl|Sequence
        (:then_0
          (CreateCollisionComponents|MakeBorderCollisionFromSpline (Variables|Default|GetSpline) false "Local" false true))
        (:then_1
          (for _array_element (Variables|C.%29ConnectingComplexPOIs|GetConnectedActors)
            (CreateCollisionComponents|MakeBorderCollisionFromSpline (Variables|Default|GetSpline _array_element) false "World")
            (CreateCollisionComponents|MakeFillCollisionfromSpline (Variables|Default|GetSpline _array_element) "World")))
        (:then_2
          (CreateCollisionComponents|MakeOverrideActorCollision))))
    (:then_1
      (bind _assetindexlist (Tools|GetAssetsOfType (Variables|/Game/ArtTools/Procedural/Foundation/Proc_AssetList.Proc_AssetList_C|GetAssetList (Variables|E.%29Meshes|GetSidewalk_Asset_List)) "NewEnumerator3"))
      (LocalVariables|SetOuterCornerAssetList _assetindexlist)
      (bind _assetindexlist_2 (Tools|GetAssetsOfType (Variables|/Game/ArtTools/Procedural/Foundation/Proc_AssetList.Proc_AssetList_C|GetAssetList (Variables|E.%29Meshes|GetSidewalk_Asset_List)) "NewEnumerator5"))
      (LocalVariables|SetS-BendAssetIndexList _assetindexlist_2)
      (bind _assetindexlist_3 (Tools|GetAssetsOfType (Variables|/Game/ArtTools/Procedural/Foundation/Proc_AssetList.Proc_AssetList_C|GetAssetList (Variables|E.%29Meshes|GetSidewalk_Asset_List)) "NewEnumerator4"))
      (LocalVariables|SetCornerInnerAssetList _assetindexlist_3)
      (bind _assetindexlist_4 (Tools|GetAssetsOfType (Variables|/Game/ArtTools/Procedural/Foundation/Proc_AssetList.Proc_AssetList_C|GetAssetList (Variables|E.%29Meshes|GetSidewalk_Asset_List)) "NewEnumerator2"))
      (LocalVariables|SetEdgeBorderAssetLis _assetindexlist_4))
    (:then_2
      (CreateComponents|MakeSidewalkComponents (Variables|Default|GetSpline) (LocalVariables|GetOuterCornerAssetList) (LocalVariables|GetS-BendAssetIndexList) (LocalVariables|GetCornerInnerAssetList) (LocalVariables|GetEdgeBorderAssetLis) true))
    (:then_3
      (for _array_element_2 (Variables|C.%29ConnectingComplexPOIs|GetConnectedActors)
        (bind _splinecopy (SplineOperations|DuplicateSpline (Variables|Default|GetSpline _array_element_2) true "World" true))
        (CreateComponents|MakeSidewalkComponents _splinecopy (LocalVariables|GetOuterCornerAssetList) (LocalVariables|GetS-BendAssetIndexList) (LocalVariables|GetCornerInnerAssetList) (LocalVariables|GetEdgeBorderAssetLis))))
    (:then_5
      (Cleanup|CleanUpBorderCollisionList))))

(fn MakeSidewalks_WithMergedSplines
  (bind _whileloopiterationscount (LocalVariables|GetWhileLoopIterationsCount))
  (bind _connectedsplinelist (Variables|Default|GetConnectedSplineList))
  (bind _connectedsplinelist_2 (Variables|Default|GetConnectedSplineList))
  (bind _connectedsplinelist_3 (Variables|Default|GetConnectedSplineList))
  (bind _addsplinenumber (LocalVariables|GetAddSplineNumber))
  (bind _connectedsplinelist_4 (Variables|Default|GetConnectedSplineList))
  (bind _whileloopiterationscount_2 (LocalVariables|GetWhileLoopIterationsCount))
  (bind _mergekey (LocalVariables|GetMergeKey))
  (Utilities|FlowControl|Sequence
    (:then_0
      (Utilities|FlowControl|Sequence
        (:then_0
          (LocalVariables|SetAddSplineNumber 0)
          (Utilities|Map|Clear (Variables|Default|GetConnectedSplineList))
          (Utilities|FlowControl|Sequence
            (:then_0
              (Utilities|Map|Add (Variables|Default|GetConnectedSplineList) 0 (Variables|Default|GetSpline)))
            (:then_1
              (for _array_element (Variables|C.%29ConnectingComplexPOIs|GetConnectedActors)
                (Utilities|IsValid _array_element
                  (:Is%20Valid
                    (Math|Integer|IncrementInt _addsplinenumber
                      (:%20%20
                        (Utilities|Map|Add (Variables|Default|GetConnectedSplineList) _addsplinenumber (Variables|Default|GetSpline _array_element))))))))
            (:then_2
              (LocalVariables|SetKeepMerging true)
              (while (LocalVariables|GetKeepMerging)
                (if (> _whileloopiterationscount_2 100)
                  (LocalVariables|SetKeepMerging false)
                  (Development|PrintString "EXIT WHILELOOP REACH LIMIT")
                  (else
                    (Math|Integer|IncrementInt _whileloopiterationscount_2
                      (:%20%20
                        (bind _values (Utilities|Map|Values _connectedsplinelist_4))
                        (Development|PrintString "WhileLoopStarted" false false)
                        (bind (_array_element_2 _array_index) (Utilities|Array|ForEachLoopwithBreak _values)
                          (:LoopBody
                            (bind _keys (Utilities|Map|Keys _connectedsplinelist_4))
                            (bind (_mergedspline _merged _keytoremovefrommap1) (SplineOperations|MergeSplines _array_element_2 true (Variables|Default|GetConnectedSplineList)))
                            (Development|PrintString (Utilities|String|Append (Utilities|String|Append (Utilities|String|Append "WhileLoopItr" (Utilities|String|ToString(Integer) _array_index)) "  _Merge = ") (Utilities|String|ToString(Boolean) _merged)) false false)
                            (LocalVariables|SetKeepMerging _merged)
                            (if _merged
                              (Utilities|Array|Add (Variables|SplineList|GetTempSplineList) _mergedspline)
                              (bind _keys_2 (Utilities|Map|Keys _connectedsplinelist))
                              (LocalVariables|SetMergeKey (Utilities|Array|FindItem _keys_2 (Utilities|Array|Get(aref) _keys _array_index)))
                              (bind _values_2 (Utilities|Map|Values _connectedsplinelist))
                              (Development|PrintString (Utilities|String|Append (Utilities|String|Append (Utilities|String|Append "SettingItem = " (Utilities|String|ToString(Integer) (Utilities|Array|FindItem _keys_2 (Utilities|Array|Get(aref) _keys _array_index)))) "Iterration is _ ") (Utilities|String|ToString(Integer) _array_index)) false false)
                              (Utilities|Map|Remove _connectedsplinelist _mergekey)
                              (Utilities|Map|Add _connectedsplinelist _mergekey _mergedspline)
                              (Utilities|FlowControl|Sequence
                                (:then_0
                                  _)
                                (:then_1
                                  (for _array_element_3 _keytoremovefrommap1
                                    (bind _returnvalue (Utilities|Map|Remove (Variables|Default|GetConnectedSplineList) _array_element_3))
                                    (Development|PrintString (Utilities|String|Append (Utilities|String|Append (Utilities|String|Append "RemovingIndex" (Utilities|String|ToString(Integer) _array_element_3)) "  =  ") (Utilities|String|ToString(Boolean) _returnvalue)) false false)
                                    (if false
                                      (bind _values_3 (Utilities|Map|Values (Variables|Default|GetConnectedSplineList)))
                                      (for (_array_element_4 _array_index_2) _values_3
                                        (Utilities|FlowControl|Sequence
                                          (:then_0
                                            (bind _keys_3 (Utilities|Map|Keys (Variables|Default|GetConnectedSplineList)))
                                            (Development|PrintString (Utilities|String|Append (Utilities|String|Append (Utilities|String|Append (Utilities|String|Append (Utilities|GetDisplayName _array_element_4) "  = Key  = ") (Utilities|String|ToString(Integer) (Utilities|Array|Get(aref) _keys_3 _array_index_2))) "  INDEX  =  ") (Utilities|String|ToString(Integer) _array_index_2))))))))
                                  (Utilities|FlowControl|Sequence
                                    (:then_1
                                      _)
                                    (:then_2
                                      _))
                                  (if false
                                    (bind _values_4 (Utilities|Map|Values (Variables|Default|GetConnectedSplineList)))
                                    (Development|PrintString (Utilities|String|Append "NumberOfSplines_" (Utilities|String|ToString(Integer) (Utilities|Array|Length _values_4))))
                                    (for (_array_element_5 _array_index_3) _values_4
                                      (Utilities|FlowControl|Sequence
                                        (:then_0
                                          (bind _keys_4 (Utilities|Map|Keys (Variables|Default|GetConnectedSplineList)))
                                          (Development|PrintString (Utilities|String|Append (Utilities|String|Append (Utilities|String|Append (Utilities|String|Append (Utilities|GetDisplayName _array_element_5) "  = Key  = ") (Utilities|String|ToString(Integer) (Utilities|Array|Get(aref) _keys_4 _array_index_3))) "  INDEX  =  ") (Utilities|String|ToString(Integer) _array_index_3))))
                                        (:then_1
                                          (bind _splinecopy (SplineOperations|DuplicateSpline _array_element_5 false "World" false true))
                                          (bind _splineout (SplineOperations|OffsetSpline _splinecopy (+ (* 2048.0 (+ _whileloopiterationscount 1)) (* 128.0 _array_index_3))))
                                          (Editor|SetSelectedSplineSegmentColor _splineout (Utilities|Struct|MakeLinearColor (Math|Random|RandomFloatinRangefromStream (Math|Random|MakeRandomStream (Utilities|Operators|Add _whileloopiterationscount _array_index_3 0)) 0.0 256.0) (Math|Random|RandomFloatinRangefromStream (Math|Random|MakeRandomStream (+ (Utilities|Operators|Add _whileloopiterationscount _array_index_3 0) 1)) 0.0 256.0) (Math|Random|RandomFloatinRangefromStream (Math|Random|MakeRandomStream (+ (Utilities|Operators|Add _whileloopiterationscount _array_index_3 0) 2)) 0.0 256.0) 1.0))
                                          (Editor|SetUnselectedSplineSegmentColor _splinecopy (Utilities|Struct|MakeLinearColor (Math|Random|RandomFloatinRangefromStream (Math|Random|MakeRandomStream (Utilities|Operators|Add _whileloopiterationscount _array_index_3 0)) 0.0 256.0) (Math|Random|RandomFloatinRangefromStream (Math|Random|MakeRandomStream (+ (Utilities|Operators|Add _whileloopiterationscount _array_index_3 0) 1)) 0.0 256.0) (Math|Random|RandomFloatinRangefromStream (Math|Random|MakeRandomStream (+ (Utilities|Operators|Add _whileloopiterationscount _array_index_3 0) 2)) 0.0 256.0) 1.0))
                                          (Editor|SetTangentColor _splinecopy (Utilities|Struct|MakeLinearColor (Math|Random|RandomFloatinRangefromStream (Math|Random|MakeRandomStream (Utilities|Operators|Add _whileloopiterationscount _array_index_3 0)) 0.0 256.0) (Math|Random|RandomFloatinRangefromStream (Math|Random|MakeRandomStream (+ (Utilities|Operators|Add _whileloopiterationscount _array_index_3 0) 1)) 0.0 256.0) (Math|Random|RandomFloatinRangefromStream (Math|Random|MakeRandomStream (+ (Utilities|Operators|Add _whileloopiterationscount _array_index_3 0) 2)) 0.0 256.0) 1.0))))))))
                              (if false
                                (bind _values_5 (Utilities|Map|Values (Variables|Default|GetConnectedSplineList)))
                                (for (_array_element_6 _array_index_4) _values_5
                                  (Utilities|FlowControl|Sequence
                                    (:then_0
                                      (bind _keys_5 (Utilities|Map|Keys (Variables|Default|GetConnectedSplineList)))
                                      (Development|PrintString (Utilities|String|Append (Utilities|String|Append (Utilities|String|Append (Utilities|String|Append (Utilities|GetDisplayName _array_element_6) "  = Key  = ") (Utilities|String|ToString(Integer) (Utilities|Array|Get(aref) _keys_5 _array_index_4))) "  INDEX  =  ") (Utilities|String|ToString(Integer) _array_index_4))))))))))))))))))
        (:then_1
          (bind _values_6 (Utilities|Map|Values _connectedsplinelist_3))
          (bind _keys_6 (Utilities|Map|Keys _connectedsplinelist_3))
          (for (_array_element_7 _array_index_5) _values_6
            (bind _splinecopy_2 (SplineOperations|MakeCleanOptimizedSplineCopy _array_element_7))
            (if (not false)
              (Utilities|Array|Add (Variables|SplineList|GetTempSplineList) _splinecopy_2))
            (Utilities|Map|Remove _connectedsplinelist_2 (Utilities|Array|Get(aref) _keys_6 _array_index_5))
            (Utilities|Map|Add _connectedsplinelist_2 (Utilities|Array|Get(aref) _keys_6 _array_index_5) _splinecopy_2)))
        (:then_2
          (if false
            (bind _values_7 (Utilities|Map|Values (Variables|Default|GetConnectedSplineList)))
            (Development|PrintString (Utilities|String|Append (Utilities|String|Append (Utilities|String|Append "LIST OF SPLINES length is " (Utilities|String|ToString(Integer) (Utilities|Array|Length _values_7))) " _FirstSplineNumberofPoints") (Utilities|String|ToString(Integer) (Spline|GetNumberOfSplinePoints (Utilities|Array|Get(acopy) _values_7)))))))
        (:then_3
          (Utilities|FlowControl|Sequence
            (:then_0
              (bind _values_8 (Utilities|Map|Values (Variables|Default|GetConnectedSplineList)))
              (for (_array_element_8 _array_index_6) _values_8
                (if (== _array_index_6 0)
                  (CreateCollisionComponents|MakeBorderCollisionFromSpline _array_element_8)
                  (else
                    (CreateCollisionComponents|MakeBorderCollisionFromSpline _array_element_8)))))
            (:then_2
              (CreateCollisionComponents|MakeOverrideActorCollision))))))
    (:then_1
      (bind _assetindexlist (Tools|GetAssetsOfType (Variables|/Game/ArtTools/Procedural/Foundation/Proc_AssetList.Proc_AssetList_C|GetAssetList (Variables|E.%29Meshes|GetSidewalk_Asset_List)) "NewEnumerator3"))
      (LocalVariables|SetOuterCornerAssetList _assetindexlist)
      (bind _assetindexlist_2 (Tools|GetAssetsOfType (Variables|/Game/ArtTools/Procedural/Foundation/Proc_AssetList.Proc_AssetList_C|GetAssetList (Variables|E.%29Meshes|GetSidewalk_Asset_List)) "NewEnumerator5"))
      (LocalVariables|SetS-BendAssetIndexList _assetindexlist_2)
      (bind _assetindexlist_3 (Tools|GetAssetsOfType (Variables|/Game/ArtTools/Procedural/Foundation/Proc_AssetList.Proc_AssetList_C|GetAssetList (Variables|E.%29Meshes|GetSidewalk_Asset_List)) "NewEnumerator4"))
      (LocalVariables|SetCornerInnerAssetList _assetindexlist_3)
      (bind _assetindexlist_4 (Tools|GetAssetsOfType (Variables|/Game/ArtTools/Procedural/Foundation/Proc_AssetList.Proc_AssetList_C|GetAssetList (Variables|E.%29Meshes|GetSidewalk_Asset_List)) "NewEnumerator2"))
      (LocalVariables|SetEdgeBorderAssetLis _assetindexlist_4))
    (:then_2
      (bind _values_9 (Utilities|Map|Values (Variables|Default|GetConnectedSplineList)))
      (for (_array_element_9 _array_index_7) _values_9
        (if (== _array_index_7 0)
          (CreateComponents|MakeSidewalkComponents _array_element_9 (LocalVariables|GetOuterCornerAssetList) (LocalVariables|GetS-BendAssetIndexList) (LocalVariables|GetCornerInnerAssetList) (LocalVariables|GetEdgeBorderAssetLis))
          (else
            (bind _splinecopy_3 (SplineOperations|DuplicateSpline _array_element_9 true "Local" true))
            (CreateComponents|MakeSidewalkComponents _splinecopy_3 (LocalVariables|GetOuterCornerAssetList) (LocalVariables|GetS-BendAssetIndexList) (LocalVariables|GetCornerInnerAssetList) (LocalVariables|GetEdgeBorderAssetLis))))))
    (:then_5
      (Cleanup|CleanUpBorderCollisionList)
      (Cleanup|DeleteAllTempSplines))))

(fn MakeSplinePointChecklist
  (bind _spline (Variables|Default|GetSpline))
  (Utilities|Array|Clear (Variables|X.%29DoNOTEdit|GetSplinePointListCheck))
  (for _index (range (Spline|GetNumberOfSplinePoints _spline))
    (Utilities|Array|Add (Variables|X.%29DoNOTEdit|GetSplinePointListCheck) (Spline|GetLocationatSplinePoint _spline _index "Local"))))

(fn MergeSplines
  (params (Spline_Input_1 :type Spline%20Component%20Object%20Reference) (CurveTypeLinear :type Boolean) (ConnectedSplineList_Input :type Map%20of%20Integers%20to%20Spline%20Component%20Object%20References))
  (bind _intersectingpoints (LocalVariables|GetIntersectingPoints))
  (bind _returnvalue (Utilities|Array|LastIndex _intersectingpoints))
  (bind _returnvalue_2 15.0)
  (bind _currentspline (LocalVariables|GetCurrentSpline))
  (bind _returnvalue_3 (Spline|GetNumberOfSplinePoints _currentspline))
  (bind _returnvalue_4 (Math|Boolean|MakeLiteralBool (LocalVariables|GetIs_Inside_Of_Shape)))
  (bind _intersectingpoints_2 (LocalVariables|GetIntersectingPoints))
  (bind _lastisinverse (LocalVariables|GetLastIsInverse))
  (bind _lastindexinverse (LocalVariables|GetLastIndexInverse))
  (bind _intersectingpoints_3 (LocalVariables|GetIntersectingPoints))
  (bind _templist (LocalVariables|GetTempList))
  (bind _splinecornercollision (LocalVariables|GetSplineCornerCollision))
  (bind _connectedsplinelist_input (Variables|GetConnectedSplineList_Input))
  (bind _connectedsplinelist_input_2 (Variables|GetConnectedSplineList_Input))
  (bind _tempsplinelistmerged (LocalVariables|GetTempSplineListMerged))
  (bind _splinetomergewith (LocalVariables|GetSplineToMergeWith))
  (bind _currentspline_2 (LocalVariables|GetCurrentSpline))
  (LocalVariables|SetCurrentSpline Spline_Input_1)
  (LocalVariables|SetMergeHappened false)
  (Utilities|FlowControl|Sequence
    (:then_0
      (Utilities|Array|Clear (LocalVariables|GetKeyToRemoveFromMap))
      (bind _values (Utilities|Map|Values (Variables|GetConnectedSplineList_Input)))
      (for (_array_element _array_index) _values
        (if (!= _array_element (Variables|GetSpline_Input_1))
          (LocalVariables|SetSplineToMergeWith _array_element)
          (LocalVariables|SetCurrentIndex _array_index)
          (Development|PrintString "NEWLOOP" false false)
          (Utilities|FlowControl|Sequence
            (:then_0
              ;[1 1 1] Create Collision
              (bind _keys (Utilities|Map|Keys (Variables|GetConnectedSplineList_Input)))
              (for _index (range (Spline|GetNumberOfSplinePoints _splinetomergewith))
                (bind (_next__1 _next _current _previous _previous__1 _world_next__1 _world_next _world_current _world_previous) (Tools|GetCurrentNextPreviousSplinePoint _splinetomergewith _index))
                (if (Math|Float|NearlyEqual(Float) (Math|Vector|DotProduct (Math|Vector|GetUnitDirection(Vector) _world_previous _world_current) (Math|Vector|GetUnitDirection(Vector) _world_current _world_next)) 0.0 0.1)
                  (bind _returnvalue_5 (AddComponent|NODE_AddStaticMeshComponent-54 :RelativeTransform (Math|Transform|MakeTransform (- _world_current (Transformation|GetActorLocation)) "0, 0, 0" "0.100000,0.100000,0.100000")))
                  (Variables|/Script/Engine.ActorComponent|SetComponentTags (Utilities|Array|MakeArray (Utilities|String|StringToName (Utilities|String|ToString(Integer) (Utilities|Array|Get(aref) _keys (LocalVariables|GetCurrentIndex)))) (Utilities|String|StringToName (Utilities|String|ToString(Integer) _index))) _returnvalue_5)
                  (Utilities|Array|Add (LocalVariables|GetSplineCornerCollision) _returnvalue_5))))
            (:then_1
              ;[1 1 1] Trace for Overlaps
              (for _index_2 (range (Spline|GetNumberOfSplinePoints _currentspline_2))
                (bind (_next__1_2 _next_2 _current_2 _previous_2 _previous__1_2 _world_next__1_2 _world_next_2 _world_current_2) (Tools|GetCurrentNextPreviousSplinePoint _currentspline_2 _index_2))
                (bind (_outhits _returnvalue_6) (Collision|MultiLineTraceForObjects :Start _world_current_2 :End _world_next_2 :ObjectTypes (Utilities|Array|MakeArray "ObjectTypeQuery3") :bTraceComplex false :DrawDebugType (select (== _index_2 0) "None" "None") :bIgnoreSelf false :TraceColor "(R=1.000000,G=0.000000,B=0.973890,A=1.000000)" :TraceHitColor "(R=0.000000,G=1.000000,B=0.000000,A=1.000000)" :DrawTime 1e+01))
                (if _returnvalue_6
                  (Development|PrintString (Utilities|String|Append "Number of Hits" (Utilities|String|ToString(Integer) (Utilities|Array|Length _outhits))) false false)
                  (for _array_element_2 _outhits
                    (Utilities|FlowControl|Sequence
                      (:then_1
                        (if (and (== _index_2 0) (< (Math|Vector|Distance(Vector) (.Location (Collision|BreakHitResult _array_element_2)) _world_current_2) _returnvalue_2))
                          (bind _output_get (LocalVariables|SetLastIsInverse true))
                          (Development|PrintString (Utilities|String|Append (Utilities|String|ToString(Boolean) _output_get) " -LastisInverseSet") false false))
                        (Utilities|FlowControl|Sequence
                          (:then_0
                            (if (LocalVariables|GetLastIsInverse)
                              (if (== (Utilities|Array|Length _outhits) 1)
                                (LocalVariables|SetIs_Inside_Of_Shape false)
                                (elif (and (> (Math|Vector|Distance(Vector) (.Location (Collision|BreakHitResult _array_element_2)) _world_current_2) _returnvalue_2) (> (Math|Vector|Distance(Vector) (.Location (Collision|BreakHitResult _array_element_2)) _world_next_2) _returnvalue_2))
                                  (bind _keys_2 (Utilities|Map|Keys _connectedsplinelist_input_2))
                                  (bind _values_2 (Utilities|Map|Values _connectedsplinelist_input_2))
                                  (bind (_next__1_3 _next_3 _current_3 _previous_3 _previous__1_3 _world_next__1_3 _world_next_3 _world_current_3 _world_previous_2) (Tools|GetCurrentNextPreviousSplinePoint (Utilities|Array|Get(acopy) _values_2 (Utilities|Array|FindItem _keys_2 (Utilities|String|StringToInteger (Utilities|String|ToString(Name) (Utilities|Array|Get(acopy) (Variables|/Script/Engine.ActorComponent|GetComponentTags (.HitComponent (Collision|BreakHitResult _array_element_2)))))))) (Utilities|String|StringToInteger (Utilities|String|ToString(Name) (Utilities|Array|Get(acopy) (Variables|/Script/Engine.ActorComponent|GetComponentTags (.HitComponent (Collision|BreakHitResult _array_element_2))) 1)))))
                                  (LocalVariables|SetIs_Inside_Of_Shape (not (> (Math|Vector|DotProduct (Math|Vector|SelectVector (Math|Vector|GetUnitDirection(Vector) _world_current_3 _world_next_3) (Math|Vector|GetUnitDirection(Vector) _world_current_3 _world_previous_2) (Math|Float|NearlyEqual(Float) (Math|Vector|DotProduct (Math|Vector|GetUnitDirection(Vector) _world_current_2 _world_next_2) (Math|Vector|GetUnitDirection(Vector) _world_current_3 _world_next_3)) 0.0 0.1)) (Math|Vector|CrossProduct (Math|Vector|GetUnitDirection(Vector) _world_current_2 _world_next_2) "0, 0,1.000000")) 0)))
                                  (Development|PrintString (Utilities|String|Append "ISInside   =   " (Utilities|String|ToString(Boolean) (not (> (Math|Vector|DotProduct (Math|Vector|SelectVector (Math|Vector|GetUnitDirection(Vector) _world_current_3 _world_next_3) (Math|Vector|GetUnitDirection(Vector) _world_current_3 _world_previous_2) (Math|Float|NearlyEqual(Float) (Math|Vector|DotProduct (Math|Vector|GetUnitDirection(Vector) _world_current_2 _world_next_2) (Math|Vector|GetUnitDirection(Vector) _world_current_3 _world_next_3)) 0.0 0.1)) (Math|Vector|CrossProduct (Math|Vector|GetUnitDirection(Vector) _world_current_2 _world_next_2) "0, 0,1.000000")) 0)))) false false)))
                              (if (or (and (== (Utilities|Array|Length _outhits) 1) (not (LocalVariables|GetIs_Inside_Of_Shape))) (select (LocalVariables|GetIs_Inside_Of_Shape) (and (> (Math|Vector|Distance(Vector) (.Location (Collision|BreakHitResult _array_element_2)) _world_current_2) _returnvalue_2) (> (Math|Vector|Distance(Vector) (.Location (Collision|BreakHitResult _array_element_2)) _world_next_2) _returnvalue_2)) true))
                                ;[1 0.009362 0] This might cause a bug needs to find key and and store key instead of index
                                (bind _keys_3 (Utilities|Map|Keys (Variables|GetConnectedSplineList_Input)))
                                (Utilities|Array|Add (LocalVariables|GetIntersectingPoints) (Utilities|Struct|MakeSplineMergingStruct (Utilities|String|StringToInteger (Utilities|String|ToString(Name) (Utilities|Array|Get(acopy) (Variables|/Script/Engine.ActorComponent|GetComponentTags (.HitComponent (Collision|BreakHitResult _array_element_2)))))) (Utilities|String|StringToInteger (Utilities|String|ToString(Name) (Utilities|Array|Get(acopy) (Variables|/Script/Engine.ActorComponent|GetComponentTags (.HitComponent (Collision|BreakHitResult _array_element_2))) 1))) _index_2))
                                (Development|PrintString (Utilities|String|Append "Index_ " (Utilities|String|ToString(Integer) _index_2)) false false)
                                (Utilities|Array|AddUnique (LocalVariables|GetKeyToRemoveFromMap) (Utilities|String|StringToInteger (Utilities|String|ToString(Name) (Utilities|Array|Get(acopy) (Variables|/Script/Engine.ActorComponent|GetComponentTags (.HitComponent (Collision|BreakHitResult _array_element_2)))))))))))))))))
            (:then_2
              (for _array_element_3 _splinecornercollision
                (Components|DestroyComponent _array_element_3))
              (Utilities|Array|Clear _splinecornercollision))
            (:then_3
              (Utilities|FlowControl|Sequence
                (:then_0
                  ;[1 1 1] Sorting
                  (Utilities|FlowControl|Sequence
                    (:then_0
                      (if false
                        (bind _array_element_4 (Utilities|Array|ForEachLoopwithBreak (LocalVariables|GetIntersectingPoints))
                          (:LoopBody
                            (Development|PrintString (Utilities|String|Append (Utilities|String|Append (Utilities|String|Append (Utilities|String|Append (Utilities|String|Append "PreSort_" (Utilities|String|ToString(Integer) (Utilities|Struct|BreakSplineMergingStruct _array_element_4))) "_") (Utilities|String|ToString(Integer) (.PointIndex_4_E631B2FE4EFCE4D254CA73ABEE160AA6 (Utilities|Struct|BreakSplineMergingStruct _array_element_4)))) "_") (Utilities|String|ToString(Integer) (.PointIndexPreviousToCurrent_8_C0E1CA234B1E806BE8B807BE84267C66 (Utilities|Struct|BreakSplineMergingStruct _array_element_4)))))))))
                    (:then_1
                      (if (LocalVariables|GetLastIsInverse)
                        (Utilities|Array|Insert _intersectingpoints (Utilities|Array|Get(aref) _intersectingpoints _returnvalue) 1)
                        (Utilities|Array|RemoveIndex _intersectingpoints _returnvalue)))))
                (:then_1
                  ;[1 1 1] remove unneeded Intersections
                  (Utilities|FlowControl|Sequence
                    (:then_0
                      (for (_array_element_5 _array_index_2) _intersectingpoints_2
                        (if (> _array_index_2 0)
                          (if (>= (- (Utilities|Array|Length _intersectingpoints_2) 1) (+ _array_index_2 1))
                            (if (and (== (Utilities|Struct|BreakSplineMergingStruct _array_element_5) (Utilities|Struct|BreakSplineMergingStruct (Utilities|Array|Get(acopy) _intersectingpoints_2 (+ _array_index_2 1)))) (== (Utilities|Struct|BreakSplineMergingStruct _array_element_5) (Utilities|Struct|BreakSplineMergingStruct (Utilities|Array|Get(acopy) _intersectingpoints_2 (+ _array_index_2 -1)))))
                              (Utilities|Array|Add (LocalVariables|GetIndex%20to%20remove) _array_index_2))))))
                    (:then_1
                      (Development|PrintString (Utilities|String|ToString(Integer) (Utilities|Array|Length (LocalVariables|GetIntersectingPoints))) false false)
                      (for (_array_element_6 _array_index_3) (LocalVariables|GetIntersectingPoints)
                        (if (not (>= (Utilities|Array|FindItem (LocalVariables|GetIndex%20to%20remove) _array_index_3) 0))
                          (Utilities|Array|Add (LocalVariables|GetTempList) _array_element_6)))
                      (LocalVariables|SetIntersectingPoints _templist)
                      (Utilities|Array|Clear _templist)
                      (Development|PrintString (Utilities|String|ToString(Integer) (Utilities|Array|Length (LocalVariables|GetIntersectingPoints))) false false))))
                (:then_2
                  ;[1 1 1] Is Merge Happening
                  (LocalVariables|SetCurrentIsMerge (> (Utilities|Array|Length (LocalVariables|GetIntersectingPoints)) 0)))
                (:then_3
                  (if (LocalVariables|GetCurrentIsMerge)
                    (Utilities|FlowControl|Sequence
                      (:then_0
                        ;[1 1 1] Debug
                        (Utilities|FlowControl|Sequence
                          (:then_0
                            (if false
                              (bind _array_element_7 (Utilities|Array|ForEachLoopwithBreak (LocalVariables|GetIntersectingPoints))
                                (:LoopBody
                                  (Development|PrintString (Utilities|String|Append (Utilities|String|Append (Utilities|String|Append (Utilities|String|Append (Utilities|String|ToString(Integer) (Utilities|Struct|BreakSplineMergingStruct _array_element_7)) "_") (Utilities|String|ToString(Integer) (.PointIndex_4_E631B2FE4EFCE4D254CA73ABEE160AA6 (Utilities|Struct|BreakSplineMergingStruct _array_element_7)))) "_") (Utilities|String|ToString(Integer) (.PointIndexPreviousToCurrent_8_C0E1CA234B1E806BE8B807BE84267C66 (Utilities|Struct|BreakSplineMergingStruct _array_element_7)))))))))
                          (:then_1
                            (Development|PrintString (Utilities|String|Append "CurrentSplineLength = " (Utilities|String|ToString(Integer) _returnvalue_3)) false false)
                            (for _index_3 (range _returnvalue_3)
                              (if (>= _index_3 (LocalVariables|GetFirstIndexInverse))
                                (if (not (or (and (> _index_3 (LocalVariables|GetLastIndexInverse)) (LocalVariables|GetLastIsInverse)) (and (not (LocalVariables|GetLastIsInverse)) (<= _index_3 (LocalVariables|GetLastIndexInverse)))))
                                  (Utilities|Array|AddUnique (LocalVariables|GetSplineLocations) (Spline|GetLocationatSplinePoint _currentspline _index_3 "World"))
                                  (Development|PrintString (Utilities|String|Append "AddingPoint" (Utilities|String|ToString(Integer) _index_3)) false false))
                                (Utilities|FlowControl|Sequence
                                  (:then_1
                                    (bind (_array_element_8 _array_index_4) (Utilities|Array|ForEachLoopwithBreak _intersectingpoints_3)
                                      (:LoopBody
                                        (Utilities|FlowControl|Sequence
                                          (:then_1
                                            (if (> _array_index_4 0)
                                              (if (!= (Utilities|Struct|BreakSplineMergingStruct _array_element_8) (Utilities|Struct|BreakSplineMergingStruct (Utilities|Array|Get(aref) _intersectingpoints_3 (Math|Integer|SelectInt (+ _array_index_4 1) (- _array_index_4 1) (and _lastisinverse (== _index_3 _lastindexinverse))))))
                                                _))
                                            (if true
                                              (if (== (.PointIndexPreviousToCurrent_8_C0E1CA234B1E806BE8B807BE84267C66 (Utilities|Struct|BreakSplineMergingStruct _array_element_8)) _index_3)
                                                (Utilities|FlowControl|Sequence
                                                  (:then_0
                                                    _)
                                                  (:then_1
                                                    (if (LocalVariables|GetLastIsInverse)
                                                      (Utilities|FlowControl|DoOnce false
                                                        (:Completed
                                                          (Utilities|FlowControl|Sequence
                                                            (:then_0
                                                              _)
                                                            (:then_1
                                                              (bind _output_get_2 (LocalVariables|SetFirstIndexInverse (.PointIndexPreviousToCurrent_8_C0E1CA234B1E806BE8B807BE84267C66 (Utilities|Struct|BreakSplineMergingStruct _array_element_8))))
                                                              (bind _output_get_3 (LocalVariables|SetLastIndexInverse (.PointIndexPreviousToCurrent_8_C0E1CA234B1E806BE8B807BE84267C66 (Utilities|Struct|BreakSplineMergingStruct (Utilities|Array|Get(aref) _intersectingpoints_3 (Math|Integer|SelectInt (- _array_index_4 1) (+ _array_index_4 1) (and _lastisinverse (== _index_3 _lastindexinverse))))))))
                                                              (LocalVariables|SetLastIsInverse true)
                                                              (Development|PrintString (Utilities|String|Append (Utilities|String|Append (Utilities|String|Append "First_" (Utilities|String|ToString(Integer) _output_get_2)) "  Last_") (Utilities|String|ToString(Integer) _output_get_3)) false false)))))))
                                                  (:then_2
                                                    _))
                                                (Utilities|FlowControl|Gate false
                                                  (:Exit
                                                    (bind _keys_4 (Utilities|Map|Keys _connectedsplinelist_input))
                                                    (bind _values_3 (Utilities|Map|Values _connectedsplinelist_input))
                                                    (Utilities|FlowControl|Sequence
                                                      (:then_0
                                                        (Development|PrintString (Utilities|String|Append (Utilities|String|Append (Utilities|String|Append (Utilities|String|Append "NumberofPointsToAdd " (Utilities|String|ToString(Integer) (- (Math|Integer|SelectInt (Math|Integer|SelectInt (+ (- (.PointIndex_4_E631B2FE4EFCE4D254CA73ABEE160AA6 (Utilities|Struct|BreakSplineMergingStruct _array_element_8)) (.PointIndex_4_E631B2FE4EFCE4D254CA73ABEE160AA6 (Utilities|Struct|BreakSplineMergingStruct (Utilities|Array|Get(aref) _intersectingpoints_3 (Math|Integer|SelectInt (- _array_index_4 1) (+ _array_index_4 1) (and _lastisinverse (== _index_3 _lastindexinverse))))))) 1) (+ (+ (- (Spline|GetNumberOfSplinePoints (Utilities|Array|Get(acopy) _values_3 (Utilities|Array|FindItem _keys_4 (Utilities|Struct|BreakSplineMergingStruct _array_element_8)))) (.PointIndex_4_E631B2FE4EFCE4D254CA73ABEE160AA6 (Utilities|Struct|BreakSplineMergingStruct (Utilities|Array|Get(aref) _intersectingpoints_3 (Math|Integer|SelectInt (- _array_index_4 1) (+ _array_index_4 1) (and _lastisinverse (== _index_3 _lastindexinverse))))))) (.PointIndex_4_E631B2FE4EFCE4D254CA73ABEE160AA6 (Utilities|Struct|BreakSplineMergingStruct _array_element_8))) 1) (> (.PointIndex_4_E631B2FE4EFCE4D254CA73ABEE160AA6 (Utilities|Struct|BreakSplineMergingStruct _array_element_8)) (.PointIndex_4_E631B2FE4EFCE4D254CA73ABEE160AA6 (Utilities|Struct|BreakSplineMergingStruct (Utilities|Array|Get(aref) _intersectingpoints_3 (Math|Integer|SelectInt (- _array_index_4 1) (+ _array_index_4 1) (and _lastisinverse (== _index_3 _lastindexinverse)))))))) (Math|Integer|SelectInt (+ (+ (.PointIndex_4_E631B2FE4EFCE4D254CA73ABEE160AA6 (Utilities|Struct|BreakSplineMergingStruct (Utilities|Array|Get(aref) _intersectingpoints_3 (Math|Integer|SelectInt (- _array_index_4 1) (+ _array_index_4 1) (and _lastisinverse (== _index_3 _lastindexinverse)))))) (- (Spline|GetNumberOfSplinePoints (Utilities|Array|Get(acopy) _values_3 (Utilities|Array|FindItem _keys_4 (Utilities|Struct|BreakSplineMergingStruct _array_element_8)))) (.PointIndex_4_E631B2FE4EFCE4D254CA73ABEE160AA6 (Utilities|Struct|BreakSplineMergingStruct _array_element_8)))) 1) (+ (- (.PointIndex_4_E631B2FE4EFCE4D254CA73ABEE160AA6 (Utilities|Struct|BreakSplineMergingStruct (Utilities|Array|Get(aref) _intersectingpoints_3 (Math|Integer|SelectInt (- _array_index_4 1) (+ _array_index_4 1) (and _lastisinverse (== _index_3 _lastindexinverse)))))) (.PointIndex_4_E631B2FE4EFCE4D254CA73ABEE160AA6 (Utilities|Struct|BreakSplineMergingStruct _array_element_8))) 1) (> (.PointIndex_4_E631B2FE4EFCE4D254CA73ABEE160AA6 (Utilities|Struct|BreakSplineMergingStruct _array_element_8)) (.PointIndex_4_E631B2FE4EFCE4D254CA73ABEE160AA6 (Utilities|Struct|BreakSplineMergingStruct (Utilities|Array|Get(aref) _intersectingpoints_3 (Math|Integer|SelectInt (- _array_index_4 1) (+ _array_index_4 1) (and _lastisinverse (== _index_3 _lastindexinverse)))))))) _returnvalue_4) 0))) (Utilities|String|ToString(Boolean) (> (.PointIndex_4_E631B2FE4EFCE4D254CA73ABEE160AA6 (Utilities|Struct|BreakSplineMergingStruct _array_element_8)) (.PointIndex_4_E631B2FE4EFCE4D254CA73ABEE160AA6 (Utilities|Struct|BreakSplineMergingStruct (Utilities|Array|Get(aref) _intersectingpoints_3 (Math|Integer|SelectInt (- _array_index_4 1) (+ _array_index_4 1) (and _lastisinverse (== _index_3 _lastindexinverse))))))))) (Utilities|String|ToString(Integer) (Math|Integer|SelectInt (- (Spline|GetNumberOfSplinePoints (Utilities|Array|Get(acopy) _values_3 (Utilities|Array|FindItem _keys_4 (Utilities|Struct|BreakSplineMergingStruct _array_element_8)))) (.PointIndex_4_E631B2FE4EFCE4D254CA73ABEE160AA6 (Utilities|Struct|BreakSplineMergingStruct _array_element_8))) (.PointIndex_4_E631B2FE4EFCE4D254CA73ABEE160AA6 (Utilities|Struct|BreakSplineMergingStruct _array_element_8)) _returnvalue_4))) (Utilities|String|ToString(Boolean) _returnvalue_4)) false false)
                                                        (for _index_4 (range (- (Math|Integer|SelectInt (Math|Integer|SelectInt (+ (- (.PointIndex_4_E631B2FE4EFCE4D254CA73ABEE160AA6 (Utilities|Struct|BreakSplineMergingStruct _array_element_8)) (.PointIndex_4_E631B2FE4EFCE4D254CA73ABEE160AA6 (Utilities|Struct|BreakSplineMergingStruct (Utilities|Array|Get(aref) _intersectingpoints_3 (Math|Integer|SelectInt (- _array_index_4 1) (+ _array_index_4 1) (and _lastisinverse (== _index_3 _lastindexinverse))))))) 1) (+ (+ (- (Spline|GetNumberOfSplinePoints (Utilities|Array|Get(acopy) _values_3 (Utilities|Array|FindItem _keys_4 (Utilities|Struct|BreakSplineMergingStruct _array_element_8)))) (.PointIndex_4_E631B2FE4EFCE4D254CA73ABEE160AA6 (Utilities|Struct|BreakSplineMergingStruct (Utilities|Array|Get(aref) _intersectingpoints_3 (Math|Integer|SelectInt (- _array_index_4 1) (+ _array_index_4 1) (and _lastisinverse (== _index_3 _lastindexinverse))))))) (.PointIndex_4_E631B2FE4EFCE4D254CA73ABEE160AA6 (Utilities|Struct|BreakSplineMergingStruct _array_element_8))) 1) (> (.PointIndex_4_E631B2FE4EFCE4D254CA73ABEE160AA6 (Utilities|Struct|BreakSplineMergingStruct _array_element_8)) (.PointIndex_4_E631B2FE4EFCE4D254CA73ABEE160AA6 (Utilities|Struct|BreakSplineMergingStruct (Utilities|Array|Get(aref) _intersectingpoints_3 (Math|Integer|SelectInt (- _array_index_4 1) (+ _array_index_4 1) (and _lastisinverse (== _index_3 _lastindexinverse)))))))) (Math|Integer|SelectInt (+ (+ (.PointIndex_4_E631B2FE4EFCE4D254CA73ABEE160AA6 (Utilities|Struct|BreakSplineMergingStruct (Utilities|Array|Get(aref) _intersectingpoints_3 (Math|Integer|SelectInt (- _array_index_4 1) (+ _array_index_4 1) (and _lastisinverse (== _index_3 _lastindexinverse)))))) (- (Spline|GetNumberOfSplinePoints (Utilities|Array|Get(acopy) _values_3 (Utilities|Array|FindItem _keys_4 (Utilities|Struct|BreakSplineMergingStruct _array_element_8)))) (.PointIndex_4_E631B2FE4EFCE4D254CA73ABEE160AA6 (Utilities|Struct|BreakSplineMergingStruct _array_element_8)))) 1) (+ (- (.PointIndex_4_E631B2FE4EFCE4D254CA73ABEE160AA6 (Utilities|Struct|BreakSplineMergingStruct (Utilities|Array|Get(aref) _intersectingpoints_3 (Math|Integer|SelectInt (- _array_index_4 1) (+ _array_index_4 1) (and _lastisinverse (== _index_3 _lastindexinverse)))))) (.PointIndex_4_E631B2FE4EFCE4D254CA73ABEE160AA6 (Utilities|Struct|BreakSplineMergingStruct _array_element_8))) 1) (> (.PointIndex_4_E631B2FE4EFCE4D254CA73ABEE160AA6 (Utilities|Struct|BreakSplineMergingStruct _array_element_8)) (.PointIndex_4_E631B2FE4EFCE4D254CA73ABEE160AA6 (Utilities|Struct|BreakSplineMergingStruct (Utilities|Array|Get(aref) _intersectingpoints_3 (Math|Integer|SelectInt (- _array_index_4 1) (+ _array_index_4 1) (and _lastisinverse (== _index_3 _lastindexinverse)))))))) _returnvalue_4) 0))
                                                          (Development|PrintString (Utilities|String|Append "Iteration AddingPoint" (Utilities|String|ToString(Integer) (Math|Integer|SelectInt (- (Spline|GetNumberOfSplinePoints (Utilities|Array|Get(acopy) _values_3 (Utilities|Array|FindItem _keys_4 (Utilities|Struct|BreakSplineMergingStruct _array_element_8)))) (Math|Integer|SelectInt (- (+ _index_4 (Math|Integer|SelectInt (- (Spline|GetNumberOfSplinePoints (Utilities|Array|Get(acopy) _values_3 (Utilities|Array|FindItem _keys_4 (Utilities|Struct|BreakSplineMergingStruct _array_element_8)))) (.PointIndex_4_E631B2FE4EFCE4D254CA73ABEE160AA6 (Utilities|Struct|BreakSplineMergingStruct _array_element_8))) (.PointIndex_4_E631B2FE4EFCE4D254CA73ABEE160AA6 (Utilities|Struct|BreakSplineMergingStruct _array_element_8)) _returnvalue_4)) (Spline|GetNumberOfSplinePoints (Utilities|Array|Get(acopy) _values_3 (Utilities|Array|FindItem _keys_4 (Utilities|Struct|BreakSplineMergingStruct _array_element_8))))) (+ _index_4 (Math|Integer|SelectInt (- (Spline|GetNumberOfSplinePoints (Utilities|Array|Get(acopy) _values_3 (Utilities|Array|FindItem _keys_4 (Utilities|Struct|BreakSplineMergingStruct _array_element_8)))) (.PointIndex_4_E631B2FE4EFCE4D254CA73ABEE160AA6 (Utilities|Struct|BreakSplineMergingStruct _array_element_8))) (.PointIndex_4_E631B2FE4EFCE4D254CA73ABEE160AA6 (Utilities|Struct|BreakSplineMergingStruct _array_element_8)) _returnvalue_4)) (> (+ _index_4 (Math|Integer|SelectInt (- (Spline|GetNumberOfSplinePoints (Utilities|Array|Get(acopy) _values_3 (Utilities|Array|FindItem _keys_4 (Utilities|Struct|BreakSplineMergingStruct _array_element_8)))) (.PointIndex_4_E631B2FE4EFCE4D254CA73ABEE160AA6 (Utilities|Struct|BreakSplineMergingStruct _array_element_8))) (.PointIndex_4_E631B2FE4EFCE4D254CA73ABEE160AA6 (Utilities|Struct|BreakSplineMergingStruct _array_element_8)) _returnvalue_4)) (Spline|GetNumberOfSplinePoints (Utilities|Array|Get(acopy) _values_3 (Utilities|Array|FindItem _keys_4 (Utilities|Struct|BreakSplineMergingStruct _array_element_8))))))) (Math|Integer|SelectInt (- (+ _index_4 (Math|Integer|SelectInt (- (Spline|GetNumberOfSplinePoints (Utilities|Array|Get(acopy) _values_3 (Utilities|Array|FindItem _keys_4 (Utilities|Struct|BreakSplineMergingStruct _array_element_8)))) (.PointIndex_4_E631B2FE4EFCE4D254CA73ABEE160AA6 (Utilities|Struct|BreakSplineMergingStruct _array_element_8))) (.PointIndex_4_E631B2FE4EFCE4D254CA73ABEE160AA6 (Utilities|Struct|BreakSplineMergingStruct _array_element_8)) _returnvalue_4)) (Spline|GetNumberOfSplinePoints (Utilities|Array|Get(acopy) _values_3 (Utilities|Array|FindItem _keys_4 (Utilities|Struct|BreakSplineMergingStruct _array_element_8))))) (+ _index_4 (Math|Integer|SelectInt (- (Spline|GetNumberOfSplinePoints (Utilities|Array|Get(acopy) _values_3 (Utilities|Array|FindItem _keys_4 (Utilities|Struct|BreakSplineMergingStruct _array_element_8)))) (.PointIndex_4_E631B2FE4EFCE4D254CA73ABEE160AA6 (Utilities|Struct|BreakSplineMergingStruct _array_element_8))) (.PointIndex_4_E631B2FE4EFCE4D254CA73ABEE160AA6 (Utilities|Struct|BreakSplineMergingStruct _array_element_8)) _returnvalue_4)) (> (+ _index_4 (Math|Integer|SelectInt (- (Spline|GetNumberOfSplinePoints (Utilities|Array|Get(acopy) _values_3 (Utilities|Array|FindItem _keys_4 (Utilities|Struct|BreakSplineMergingStruct _array_element_8)))) (.PointIndex_4_E631B2FE4EFCE4D254CA73ABEE160AA6 (Utilities|Struct|BreakSplineMergingStruct _array_element_8))) (.PointIndex_4_E631B2FE4EFCE4D254CA73ABEE160AA6 (Utilities|Struct|BreakSplineMergingStruct _array_element_8)) _returnvalue_4)) (Spline|GetNumberOfSplinePoints (Utilities|Array|Get(acopy) _values_3 (Utilities|Array|FindItem _keys_4 (Utilities|Struct|BreakSplineMergingStruct _array_element_8)))))) _returnvalue_4))) false false)
                                                          (Utilities|Array|AddUnique (LocalVariables|GetSplineLocations) (Spline|GetLocationatSplinePoint (Utilities|Array|Get(acopy) _values_3 (Utilities|Array|FindItem _keys_4 (Utilities|Struct|BreakSplineMergingStruct _array_element_8))) (Math|Integer|SelectInt (- (Spline|GetNumberOfSplinePoints (Utilities|Array|Get(acopy) _values_3 (Utilities|Array|FindItem _keys_4 (Utilities|Struct|BreakSplineMergingStruct _array_element_8)))) (Math|Integer|SelectInt (- (+ _index_4 (Math|Integer|SelectInt (- (Spline|GetNumberOfSplinePoints (Utilities|Array|Get(acopy) _values_3 (Utilities|Array|FindItem _keys_4 (Utilities|Struct|BreakSplineMergingStruct _array_element_8)))) (.PointIndex_4_E631B2FE4EFCE4D254CA73ABEE160AA6 (Utilities|Struct|BreakSplineMergingStruct _array_element_8))) (.PointIndex_4_E631B2FE4EFCE4D254CA73ABEE160AA6 (Utilities|Struct|BreakSplineMergingStruct _array_element_8)) _returnvalue_4)) (Spline|GetNumberOfSplinePoints (Utilities|Array|Get(acopy) _values_3 (Utilities|Array|FindItem _keys_4 (Utilities|Struct|BreakSplineMergingStruct _array_element_8))))) (+ _index_4 (Math|Integer|SelectInt (- (Spline|GetNumberOfSplinePoints (Utilities|Array|Get(acopy) _values_3 (Utilities|Array|FindItem _keys_4 (Utilities|Struct|BreakSplineMergingStruct _array_element_8)))) (.PointIndex_4_E631B2FE4EFCE4D254CA73ABEE160AA6 (Utilities|Struct|BreakSplineMergingStruct _array_element_8))) (.PointIndex_4_E631B2FE4EFCE4D254CA73ABEE160AA6 (Utilities|Struct|BreakSplineMergingStruct _array_element_8)) _returnvalue_4)) (> (+ _index_4 (Math|Integer|SelectInt (- (Spline|GetNumberOfSplinePoints (Utilities|Array|Get(acopy) _values_3 (Utilities|Array|FindItem _keys_4 (Utilities|Struct|BreakSplineMergingStruct _array_element_8)))) (.PointIndex_4_E631B2FE4EFCE4D254CA73ABEE160AA6 (Utilities|Struct|BreakSplineMergingStruct _array_element_8))) (.PointIndex_4_E631B2FE4EFCE4D254CA73ABEE160AA6 (Utilities|Struct|BreakSplineMergingStruct _array_element_8)) _returnvalue_4)) (Spline|GetNumberOfSplinePoints (Utilities|Array|Get(acopy) _values_3 (Utilities|Array|FindItem _keys_4 (Utilities|Struct|BreakSplineMergingStruct _array_element_8))))))) (Math|Integer|SelectInt (- (+ _index_4 (Math|Integer|SelectInt (- (Spline|GetNumberOfSplinePoints (Utilities|Array|Get(acopy) _values_3 (Utilities|Array|FindItem _keys_4 (Utilities|Struct|BreakSplineMergingStruct _array_element_8)))) (.PointIndex_4_E631B2FE4EFCE4D254CA73ABEE160AA6 (Utilities|Struct|BreakSplineMergingStruct _array_element_8))) (.PointIndex_4_E631B2FE4EFCE4D254CA73ABEE160AA6 (Utilities|Struct|BreakSplineMergingStruct _array_element_8)) _returnvalue_4)) (Spline|GetNumberOfSplinePoints (Utilities|Array|Get(acopy) _values_3 (Utilities|Array|FindItem _keys_4 (Utilities|Struct|BreakSplineMergingStruct _array_element_8))))) (+ _index_4 (Math|Integer|SelectInt (- (Spline|GetNumberOfSplinePoints (Utilities|Array|Get(acopy) _values_3 (Utilities|Array|FindItem _keys_4 (Utilities|Struct|BreakSplineMergingStruct _array_element_8)))) (.PointIndex_4_E631B2FE4EFCE4D254CA73ABEE160AA6 (Utilities|Struct|BreakSplineMergingStruct _array_element_8))) (.PointIndex_4_E631B2FE4EFCE4D254CA73ABEE160AA6 (Utilities|Struct|BreakSplineMergingStruct _array_element_8)) _returnvalue_4)) (> (+ _index_4 (Math|Integer|SelectInt (- (Spline|GetNumberOfSplinePoints (Utilities|Array|Get(acopy) _values_3 (Utilities|Array|FindItem _keys_4 (Utilities|Struct|BreakSplineMergingStruct _array_element_8)))) (.PointIndex_4_E631B2FE4EFCE4D254CA73ABEE160AA6 (Utilities|Struct|BreakSplineMergingStruct _array_element_8))) (.PointIndex_4_E631B2FE4EFCE4D254CA73ABEE160AA6 (Utilities|Struct|BreakSplineMergingStruct _array_element_8)) _returnvalue_4)) (Spline|GetNumberOfSplinePoints (Utilities|Array|Get(acopy) _values_3 (Utilities|Array|FindItem _keys_4 (Utilities|Struct|BreakSplineMergingStruct _array_element_8)))))) _returnvalue_4) "World"))))
                                                      (:then_1
                                                        (bind _output_get_4 (LocalVariables|SetFirstIndexInverse (.PointIndexPreviousToCurrent_8_C0E1CA234B1E806BE8B807BE84267C66 (Utilities|Struct|BreakSplineMergingStruct _array_element_8))))
                                                        (bind _output_get_5 (LocalVariables|SetLastIndexInverse (.PointIndexPreviousToCurrent_8_C0E1CA234B1E806BE8B807BE84267C66 (Utilities|Struct|BreakSplineMergingStruct (Utilities|Array|Get(aref) _intersectingpoints_3 (Math|Integer|SelectInt (- _array_index_4 1) (+ _array_index_4 1) (and _lastisinverse (== _index_3 _lastindexinverse))))))))
                                                        (Development|PrintString (Utilities|String|Append (Utilities|String|Append (Utilities|String|Append "First_" (Utilities|String|ToString(Integer) _output_get_4)) "  Last_ ") (Utilities|String|ToString(Integer) _output_get_5)) false false))))))))))))))))))
                      (:then_2
                        ;[1 1 1] Make Spline
                        (bind _returnvalue_7 (AddComponent|NODE_AddSplineComponent-10 :RelativeTransform (Math|Transform|MakeTransform "0, 0, 0")))
                        (Spline|SetSplinePoints _returnvalue_7 (LocalVariables|GetSplineLocations) "World" true)
                        (LocalVariables|SetCurrentSpline _returnvalue_7)
                        (LocalVariables|SetMergeHappened true)
                        (Utilities|Array|Clear (LocalVariables|GetSplineLocations))
                        (Utilities|Array|Clear (LocalVariables|GetIntersectingPoints))
                        (LocalVariables|SetLastIsInverse false)
                        (LocalVariables|SetFirstIndexInverse -1)
                        (LocalVariables|SetLastIndexInverse -1)
                        (LocalVariables|SetCurrentIsMerge false)
                        (Utilities|FlowControl|Sequence
                          (:then_0
                            _)
                          (:then_1
                            (LocalVariables|SetIs_Inside_Of_Shape true)
                            (for _array_element_9 _tempsplinelistmerged
                              (Components|DestroyComponent _array_element_9))
                            (Utilities|Array|Add _tempsplinelistmerged _returnvalue_7)))))))))))))
    (:then_1
      (Utilities|FlowControl|Sequence
        (:then_0
          (if false
            (SplineOperations|OffsetSpline (LocalVariables|GetCurrentSpline) 1024.0)))
        (:then_1
          (for _array_element_10 (LocalVariables|GetTempSplineListMerged)
            (if (!= _array_element_10 (LocalVariables|GetCurrentSpline))
              (Components|DestroyComponent _array_element_10))))
        (:then_2
          (return (LocalVariables|GetCurrentSpline) (LocalVariables|GetMergeHappened) (LocalVariables|GetKeyToRemoveFromMap)))))))

(fn OffsetSpline
  (params (Spline :type Spline%20Component%20Object%20Reference) (ZOffset :type Float%20%28double-precision%29))
  (for _index (range (Spline|GetNumberOfSplinePoints Spline))
    (Spline|SetLocationatSplinePoint Spline _index (+ (Spline|GetLocationatSplinePoint Spline _index "Local") (Math|Vector|MakeVector 0.0 0.0 ZOffset)) "Local" true)
    (Spline|SetSplinePointType Spline _index "Linear" true))
  (return Spline))

(fn PaintBasement
  (bind _paint_layer_basement (Variables|PaintLayer|GetPaint%20Layer%20Basement))
  (Utilities|FlowControl|Sequence
    (:then_0
      (LandscapeDeformation|FillArea _paint_layer_basement)
      (for _array_element (Variables|C.%29ConnectingComplexPOIs|GetConnectedActors)
        (LandscapeDeformation|FillArea _array_element _paint_layer_basement)))
    (:then_1
      (Development|ExecuteConsoleCommand "r.VT.Flush"))))

(fn PaintVertexColorFloor
  (params (StaticMeshComponent :type Static%20Mesh%20Component%20Object%20Reference) (Color :type Color%20Structure))
  (VertexPaint|PaintVerticesSingleColor StaticMeshComponent (Math|Conversions|ToLinearColor(Color) Color)))

(fn RemoveSelfandDupsFromConnected
  (bind _connectedactors (Variables|C.%29ConnectingComplexPOIs|GetConnectedActors))
  (Utilities|FlowControl|Sequence
    (:then_0
      (for (_array_element _array_index) _connectedactors
        (if (== _array_element self)
          (Utilities|Array|RemoveIndex _connectedactors _array_index))))
    (:then_1
      (for (_array_element_2 _array_index_2) (Variables|C.%29ConnectingComplexPOIs|GetConnectedActors)
        (for (_array_element_3 _array_index_3) (Variables|C.%29ConnectingComplexPOIs|GetConnectedActors)
          (if (and (!= _array_index_3 _array_index_2) (== _array_element_3 _array_element_2))
            (Utilities|Array|Insert (LocalVariables|GetListToRemove) _array_index_3)))
        (for _array_element_4 (LocalVariables|GetListToRemove)
          (Utilities|Array|RemoveIndex (Variables|C.%29ConnectingComplexPOIs|GetConnectedActors) _array_element_4))))))

(fn Reset&ClearAllConnections
  (Utilities|Array|Clear (Variables|C.%29ConnectingComplexPOIs|GetConnectedActors))
  (Tools|UpdateAllComponents)
  (Variables|X.%29DoNOTEdit|SetConnectedBasementsUp 0))

(fn ResetActorPivot
  (bind _returnvalue (Transformation|GetActorLocation))
  (bind _spline (Variables|Default|GetSpline))
  (Utilities|FlowControl|Sequence
    (:then_1
      (LocalVariables|SetAverage (Math|Vector|MakeVector (.x (Fortnite|SnaptoFortBuildingGrid (Collision|GetActorBounds false))) (.y (Fortnite|SnaptoFortBuildingGrid (Collision|GetActorBounds false))) (.z _returnvalue)))
      (LocalVariables|SetOffset (- (LocalVariables|GetAverage) _returnvalue)))
    (:then_2
      (Transformation|SetActorLocation self (LocalVariables|GetAverage) false false)
      (for _index (range (+ (Spline|GetNumberOfSplineSegments _spline) 1))
        (Spline|SetLocationatSplinePoint _spline _index (- (Spline|GetLocationatSplinePoint _spline _index "World") (LocalVariables|GetOffset)) "World" true)))
    (:then_3
      (Tools|UpdateAllComponents)
      (Tools|UpdateAllConnectedUP))))

(fn ReverseSpline
  (bind _splinepointlistcheck (Variables|X.%29DoNOTEdit|GetSplinePointListCheck))
  (CallFunction|MakeSplinePointChecklist)
  (Utilities|Array|Reverse _splinepointlistcheck)
  (Spline|SetSplinePoints (Variables|Default|GetSpline) _splinepointlistcheck "Local" true)
  (Tools|UpdateAllComponents))

(fn SetConnectedBasementsUPonChilds
  (bind _self self)
  (Utilities|FlowControl|Sequence
    (:then_0
      (for _array_element (Variables|C.%29ConnectingComplexPOIs|GetConnectedActors)
        (Variables|X.%29DoNOTEdit|SetConnectedBasementsUp _self _array_element)
        (Development|PrintString (Utilities|String|Append (Utilities|String|Append (Utilities|GetDisplayName _self) "  Updating  --->") (Utilities|GetDisplayName _array_element)) false false)))
    (:then_1
      (for _array_element_2 (Variables|D.%29Overrides|GetOverride%20Actor%20List)
        (Variables|/Game/ArtTools/Procedural/Foundation/Proc_POI_Override_Actor.Proc_POI_Override_Actor_C|SetConnectedActor _self _array_element_2)))))

(fn SetEditorGirdPlane
  (bind (_origin _boxextent) (Collision|GetActorBounds false))
  (bind _editoronlygridmesh (Variables|Default|GetEditorOnlyGridMesh))
  (Rendering|SetVisibility (Variables|Default|GetEditorOnlyGridMesh) (Variables|Default|GetEnableGridOverlay) false)
  (Transformation|SetWorldLocation _editoronlygridmesh _origin false false)
  (Transformation|SetWorldRotation _editoronlygridmesh "0, 0, 0" false false)
  (Transformation|SetRelativeScale3D _editoronlygridmesh (+ (Math|Vector|MakeVector (* (.x _boxextent) 2.0) (* (.y _boxextent) 2.0) 1.0) "1024.000000,1024.000000, 0"))
  (Transformation|AddRelativeLocation _editoronlygridmesh "0, 0,5.000000" false false))

(fn SetMaxOptimizationIteration
  (switch Utilities|FlowControl|Switch|SwitchonProc_Foundation_OptimizationMode (Variables|Default|GetOptimization)
    (:NewEnumerator0
      (Variables|Default|SetMaxOptimizationIterations 0))
    (:NewEnumerator1
      (Variables|Default|SetMaxOptimizationIterations 10))
    (:NewEnumerator2
      (Variables|Default|SetMaxOptimizationIterations (Spline|GetNumberOfSplinePoints (Variables|Default|GetSpline))))))

(fn SetPaintlayerFromPreset
  (switch Utilities|FlowControl|Switch|SwitchonProc_Foundation_Paintlayer (Variables|B.%29LandscapeEdtiting|GetPaintLayerPreset)
    (:NewEnumerator0
      (Variables|PaintLayer|SetPaint%20Layer%20Basement "/Game/Athena/Artemis/Environments/Landscape/LayerInfo/Gravel_LayerInfo.Gravel_LayerInfo")
      (Variables|PaintLayer|SetPaint%20Layer%20Grass "/Game/Athena/Artemis/Environments/Landscape/LayerInfo/Gravel_LayerInfo.Gravel_LayerInfo")
      (Variables|PaintLayer|SetEdit%20Layer%20Name (Utilities|Array|MakeArray "Splines")))
    (:NewEnumerator1
      (Variables|PaintLayer|SetPaint%20Layer%20Basement "/Game/Athena/Apollo/Environments/Landscape/LayerInfo/Mud_LayerInfo.Mud_LayerInfo")
      (Variables|PaintLayer|SetPaint%20Layer%20Grass "/Game/Athena/Apollo/Environments/Landscape/LayerInfo/Mud_LayerInfo.Mud_LayerInfo")
      (Variables|PaintLayer|SetEdit%20Layer%20Name (Utilities|Array|MakeArray "Layer")))))

(fn SetProperitesBasedOnMode
  (switch Utilities|FlowControl|Switch|SwitchonProc_FoundationSelection (Variables|Default|GetMode)
    (:NewEnumerator6
      (Variables|Default|SetisClosed false)
      (Spline|SetClosedLoop (Variables|Default|GetSpline) (Variables|Default|GetisClosed) true))))

(fn SetStaticmeshFromAssetData
  (params (StaticMeshComponent :type Static%20Mesh%20Component%20Object%20Reference) (Proc%20Asset%20List :type Proc%20Asset%20List%20Object%20Reference) (Asset%20Index :type Integer))
  (bind _assetlist (Variables|/Game/ArtTools/Procedural/Foundation/Proc_AssetList.Proc_AssetList_C|GetAssetList Proc%20Asset%20List))
  (bind _output (Utilities|Array|Get(acopy) _assetlist Asset%20Index))
  (bind (_staticmesh_2_d8df7d7b4556a6c1fde39bb633b7e288 _type_6_b4cd9ff0498f54172801239a22e7a5ca _prepivot_29_094082c34c836f4f1ab3deb660eef580 _forwarddirection_32_09c0a6674706473836365795dce83b27 _sizex_11_dfccfeb9407ac738b8ec5aaf3a48e13c _sizey_14_2f84ffcf464b2148efd89e96265f5278 _sizez_15_c53bdd28478b1f2003457e995bf7fdc6 _materialoverride_19_677b61fa4867d6ca16880f8725102c88 _allowscaling_21_81980cba407388bd2d44e98b35473aad _overridevertexcolor_23_687b993945a8361cb9f9a58ad2b938b9 _vertexcolor_26_5bb87c274324ed6f063cbaadb782e4b9) (Utilities|Struct|BreakProcAssetInfo _output))
  (Components|StaticMesh|SetStaticMesh StaticMeshComponent _staticmesh_2_d8df7d7b4556a6c1fde39bb633b7e288)
  (Transformation|AddRelativeLocation StaticMeshComponent (Math|Vector|RotateVector _prepivot_29_094082c34c836f4f1ab3deb660eef580 (Variables|/Script/Engine.SceneComponent|GetRelativeRotation StaticMeshComponent)) false false)
  (Utilities|FlowControl|Sequence
    (:then_0
      (if (> (Utilities|Array|Length _materialoverride_19_677b61fa4867d6ca16880f8725102c88) 0)
        (for (_array_element _array_index) _materialoverride_19_677b61fa4867d6ca16880f8725102c88
          (Rendering|Material|SetMaterial StaticMeshComponent _array_index _array_element))))
    (:then_1
      (if _overridevertexcolor_23_687b993945a8361cb9f9a58ad2b938b9
        (Tools|PaintVertexColorFloor StaticMeshComponent _vertexcolor_26_5bb87c274324ed6f063cbaadb782e4b9)))))

(fn SnapAllToGrid
  (bind _spline (Variables|Default|GetSpline))
  ;[1 1 1] Snap Spline To Grid
  (Spline|SetClosedLoop _spline (Variables|Default|GetisClosed) true)
  (for _index (range (Spline|GetNumberOfSplinePoints _spline))
    (Spline|SetLocationatSplinePoint _spline _index (Math|Vector|MakeVector (.x (Fortnite|SnaptoFortBuildingGrid (Spline|GetLocationatSplinePoint _spline _index "Local"))) (.y (Fortnite|SnaptoFortBuildingGrid (Spline|GetLocationatSplinePoint _spline _index "Local")))) "Local" true)
    (Spline|SetSplinePointType _spline _index "Linear" true)))

(fn SortIntArray
  (params (Array :type Array%20of%20Integers))
  (bind _temparray (LocalVariables|GetTempArray))
  (bind _temparray_2 (LocalVariables|GetTempArray))
  (Utilities|FlowControl|Sequence
    (:then_0
      (for (_array_element _array_index) Array
        (if (== _array_index 0)
          (Utilities|Array|Add (LocalVariables|GetTempArray) _array_element)
          (else
            (bind (_array_element_2 _array_index_2) (Utilities|Array|ForEachLoopwithBreak _temparray)
              (:LoopBody
                (if (<= _array_element_2 _array_element)
                  (Utilities|Array|Insert _temparray_2 _array_element _array_index_2))
                (if (== _array_index_2 (Utilities|Array|LastIndex _temparray))
                  (Utilities|Array|Add _temparray_2 _array_element))))))))
    (:then_1
      (return (LocalVariables|GetTempArray)))))

(fn TranformFromData
  (params (InTransform :type Transform%20%28by%20ref%29) (Proc%20Mesh%20Data :type Proc%20Asset%20Info%20Structure))
  (bind (_staticmesh_2_d8df7d7b4556a6c1fde39bb633b7e288 _type_6_b4cd9ff0498f54172801239a22e7a5ca _prepivot_29_094082c34c836f4f1ab3deb660eef580 _forwarddirection_32_09c0a6674706473836365795dce83b27) (Utilities|Struct|BreakProcAssetInfo Proc%20Mesh%20Data))
  (return (Math|Transform|MakeTransform (.location InTransform) (Math|Rotator|CombineRotators (.rotation InTransform) (Utilities|Select :Index _forwarddirection_32_09c0a6674706473836365795dce83b27 :NewEnumerator0 "0, 0, 0" :NewEnumerator1 "0,-90.000000, 0" :NewEnumerator2 "0, 0, 0" :NewEnumerator3 "0,180.000000, 0" :NewEnumerator4 "0,90.000000, 0" :NewEnumerator5 "0, 0, 0")) (.scale InTransform))))

(fn UndoActorConvert
  (Utilities|FlowControl|Sequence
    (:then_0
      (if (not (Variables|Default|GetEnabled))
        (Utilities|FlowControl|Sequence
          (:then_0
            (bind _outactors (Actor|GetAllActorswithTag (Variables|X.%29DoNOTEdit|GetRandomID)))
            (for _array_element _outactors
              (Actor|DestroyActor _array_element)))
          (:then_1
            (Variables|Default|SetEnabled true)
            (Tools|UpdateAllComponents)))))))

(fn UpdateAllComponents
  (if (Variables|Default|GetLiveUpdate)
    (Development|PrintString "UpdateAllComponents" false false)
    (Cleanup|DestroyAllComponents)
    (CreateComponents|CreateAllComponents)))

(fn UpdateAllConnectedDown
  (for _array_element (Variables|C.%29ConnectingComplexPOIs|GetConnectedActors)
    (Tools|UpdateAllComponents _array_element)
    (Variables|X.%29DoNOTEdit|SetConnectedBasementsUp self _array_element)))

(fn UpdateAllConnectedUP
  (bind _connectedbasementsup (Variables|X.%29DoNOTEdit|GetConnectedBasementsUp))
  (Utilities|IsValid _connectedbasementsup
    (:Is%20Valid
      (Tools|UpdateAllComponents _connectedbasementsup)
      (Development|PrintString (Utilities|String|Append (Utilities|String|Append (Utilities|GetDisplayName self) "  Updating  --->") (Utilities|GetDisplayName _connectedbasementsup)) false false))))

(fn ConstructionScript
  (bind _returnvalue (Math|DateTime|Now))
  (bind (_year _month _day _hour _minute _second _millisecond) (Math|DateTime|BreakDateTime _returnvalue))
  (bind _returnvalue_2 1e+02)
  (bind _returnvalue_3 (/ _millisecond _returnvalue_2))
  (bind _returnvalue_4 (+ _returnvalue_3 _second))
  (CallFunction|CheckifDuplicated)
  (CallFunction|SetPaintlayerfromPreset)
  (Development|PrintString (Utilities|String|Append (Utilities|String|Append (Utilities|String|Append "StartConstruction LIVEUPDATE = " (Utilities|String|ToString(Boolean) (Variables|Default|GetLiveUpdate))) "   ___") (Utilities|GetDisplayName self)) false false)
  (Tools|RemoveSelfandDupsfromConnected)
  (if (not (Variables|Default|GetLiveUpdate))
    (Tools|CheckifSplinewasEdited))
  (Variables|Default|SetDebugTime _returnvalue_4)
  (Development|PrintString (Utilities|String|Append "AfterCheck LIVEUPDATE = " (Utilities|String|ToString(Boolean) (Variables|Default|GetLiveUpdate))) false false)
  (Utilities|FlowControl|Sequence
    (:then_0
      (Variables|Default|SetTerrainBlendDistance (* (Math|Float|Round (/ (Variables|Default|GetTerrainBlendDistance) 64.0)) 64.0))
      (Transformation|SetActorLocation (Fortnite|SnaptoFortBuildingGrid (Transformation|GetActorLocation)))
      (Tools|SnapAlltoGrid)
      (if (Variables|Default|GetEnabled)
        (if (Variables|Default|GetLiveUpdate)
          (Variables|X.%29DoNOTEdit|SetCheckLocation (Transformation|GetActorLocation))
          (CallFunction|SetMaxOptimizationIteration)
          (CreateComponents|CreateAllComponents)
          (Tools|SetConnectedBasementsUPonChilds)
          (Tools|UpdateAllConnectedUP)
          (Tools|SetEditorGirdPlane)
          (else
            (CreateComponents|CreateComponentsfromSavedList)
            (CallFunction|SetProperitesBasedonMode)))))
    (:then_1
      ;[1 1 1] Rough Time Debug not very accurate
      (Development|PrintString (Utilities|String|Append "Blueprint Execution Time" (Utilities|String|ToString(Float) (- _returnvalue_4 (Variables|Default|GetDebugTime))))))))
